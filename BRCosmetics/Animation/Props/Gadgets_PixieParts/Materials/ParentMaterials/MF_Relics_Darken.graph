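# Material function that darkens and desaturates a material's attributes, apparently for Relics gadget props.
# It desaturates base color and emissive by a Desaturation input, multiplies both by Brightness, and writes them back into the attribute set.
# A scalar "OnOff 0-1" input blends between original and darkened attributes.
# The temporal-AA dither branch and static-bool switch do not feed the output, and OpacityMask is left unconnected.
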MaterialFunction /BRCosmetics/Animation/Props/Gadgets_PixieParts/Materials/ParentMaterials/MF_Relics_Darken
graph {nodes [
  n1 = GetMaterialAttributes(None=FunctionInput() {inputName=MA, inputType=FunctionInput_MaterialAttributes, sortPriority=-1}) {attributeGetTypes [#1, #2]}
  n2 = FunctionInput(Preview=Constant() {r=0.3}) {bUsePreviewValueAsDefault=true, inputName=Desaturation, inputType=FunctionInput_Scalar, sortPriority=2, previewValue=<0.3 0 0 1>}
  n3 = FunctionInput(Preview=Constant() {r=0.2}) {bUsePreviewValueAsDefault=true, inputName=Brightness, inputType=FunctionInput_Scalar, sortPriority=1, previewValue=<0.2 0 0 1>}
  n4 = SetMaterialAttributes!(
   MaterialAttributes=@n1.MaterialAttributes
   OpacityMask
   BaseColor=Multiply(A=Desaturation(None=@n1.BaseColor, Fraction=@n2), B=@n3)
   EmissiveColor=Multiply(A=Desaturation(None=@n1.EmissiveColor, Fraction=@n2), B=@n3)
  ) {attributeSetTypes [#3, #1, #2]}
  n5 = Constant()
  FunctionOutput(None=BlendMaterialAttributes(
    A=@n1.MaterialAttributes
    B=@n4
    Alpha=FunctionInput(Preview=@n5) {bUsePreviewValueAsDefault=true, inputName=OnOff 0-1, inputType=FunctionInput_Scalar}
   ))
  n6 = FunctionInput(Preview=Constant() {r=1}) {bUsePreviewValueAsDefault=true, inputName=Dithering, inputType=FunctionInput_Scalar, sortPriority=3, previewValue=<0.3 0 0 1>}
  Reroute(None=Multiply(A=MaterialFunctionCall() {materialFunction=&DitherTemporalAA}.Result, B=@n6))
  n7 = FunctionInput(Preview=StaticBool() {value=true}) {bUsePreviewValueAsDefault=true, inputName=OnOff, inputType=FunctionInput_StaticBool, previewValue=<1 0 0 1>}
  StaticSwitch(True=@n4, False=@n1.MaterialAttributes, Value=@n7)
 ]}
