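# Material function that replays baked bone animation from a texture as World Position Offset, for Pixie gadget sprite props.
# Vertex color R or UV1 selects each vertex's bone row; the VFX switch uses a bone-index parameter instead.
# Position, rotation and optional scale bands are sampled at two U values, blended by AnimTransition, then applied around the bind pose.
# It subtracts the top-row sample from position and also outputs quaternions and a rotated normal.
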MaterialFunction /BRCosmetics/Animation/Props/Gadgets_PixieParts/Materials/ParentMaterials/MF_Sprite_WPOTransform_subuvs
graph {nodes [
  n1 = Divide(B=ScalarParameter() {defaultValue=29, parameterName="Total number of bones (including index 0)"})
  n2 = Multiply(A=@n1, B=ScalarParameter() {defaultValue=3, parameterName="Single Bone Index (from 0 to n-1)"})
  Add(A=@n1, B=@n2) {desc=quick fix}
  n3 = FunctionInput(Preview=StaticBool()) {bUsePreviewValueAsDefault=true, inputName=VFX, inputType=FunctionInput_StaticBool}
  n4 = StaticSwitchParameter(True=ComponentMask(None=TextureCoordinate() {coordinateIndex=1}) {g=true}, False=VertexColor().R) {parameterName=UseWithNaniteEnabled}
  n5 = Multiply(
   A=FunctionInput() {inputName=vHeight, sortPriority=11}
   B=StaticSwitchParameter(True=Constant() {r=0.3333}, False=Constant() {r=0.5}) {parameterName=ConstainsScaleInfo}
  )
  n6 = Add(
   A=Add(A=Multiply(A=@n4, B=@n5) {constB=0.5}, B=FunctionInput() {inputName=vStart, sortPriority=10})
   B=Multiply(A=FunctionInput() {desc="height of a texel (1/yRez)", inputName=texelHeight, inputType=FunctionInput_Scalar, sortPriority=12}) {constB=0.5}
  ) {constB=-0.1}
  n7 = Multiply(A=@n2) {constB=0.5, desc="Constrain to top half of the texture (position)"}
  n8 = StaticSwitch(True=@n7, False=@n6, Value=@n3)
  n9 = FunctionInput() {desc=u value to sample, inputName=u, inputType=FunctionInput_Scalar}
  n10 = MaterialFunctionCall(X=@n9, Y=@n8) {materialFunction=&MakeFloat2}
  n11 = FunctionInput(Preview=TextureObject() {samplerType=SAMPLERTYPE_Data, texture=/Game/Global/Textures/Default/Blanks/T_BlackVectorDisplacement}) {bUsePreviewValueAsDefault=true, inputName=AnimationTexture, inputType=FunctionInput_Texture2D, sortPriority=12}
  n12 = TextureSample(UVs=AppendVector(A=ComponentMask(None=@n10.Result) {r=true}, B=Constant()), Tex=@n11) {samplerType=SAMPLERTYPE_Data, texture=&T_BlackVectorDisplacement}
  n13 = FunctionInput() {desc=u value to sample, inputName=u_New, inputType=FunctionInput_Scalar}
  n14 = MaterialFunctionCall(X=@n13, Y=@n8) {materialFunction=&MakeFloat2}
  n15 = FunctionInput(Preview=Constant()) {inputName=AnimTransition, inputType=FunctionInput_Scalar, sortPriority=1}
  n16 = Add(
   A=LinearInterpolate(
    A=TextureSample(UVs=@n10.Result, Tex=@n11) {samplerType=SAMPLERTYPE_Data, texture=&T_BlackVectorDisplacement}.RGB
    B=TextureSample(UVs=@n14.Result, Tex=@n11) {samplerType=SAMPLERTYPE_Data, texture=&T_BlackVectorDisplacement}.RGB
    Alpha=@n15
   )
   B=Multiply(A=@n12.RGB) {constB=-1}
  )
  n17 = Add(A=@n6, B=@n5) {constB=0.5}
  n18 = StaticSwitch(True=@n7, False=Add(A=@n17, B=@n5), Value=@n3)
  n19 = MaterialFunctionCall(X=@n9, Y=@n18) {materialFunction=&MakeFloat2}
  n20 = MaterialFunctionCall(X=@n13, Y=@n18) {materialFunction=&MakeFloat2}
  n21 = StaticSwitch(True=Constant(), False=PreSkinnedPosition(), Value=@n3)
  n22 = TextureSample(
   UVs=MaterialFunctionCall(
    X=FunctionInput() {desc=u value of bind pose, inputName=bind_u, inputType=FunctionInput_Scalar, sortPriority=1}
    Y=StaticSwitch(True=@n7, False=@n6, Value=@n3)
   ) {materialFunction=&MakeFloat2}.Result
   Tex=@n11
  ) {samplerType=SAMPLERTYPE_Data, texture=&T_BlackVectorDisplacement}
  n23 = Subtract(A=@n21, B=@n22.RGB) {constB=0}
  n24 = Multiply(
   A=LinearInterpolate(
    A=TextureSample(UVs=@n19.Result, Tex=@n11) {samplerType=SAMPLERTYPE_Data, texture=&T_BlackVectorDisplacement}.RGB
    B=TextureSample(UVs=@n20.Result, Tex=@n11) {samplerType=SAMPLERTYPE_Data, texture=&T_BlackVectorDisplacement}.RGB
    Alpha=@n15
   )
   B=@n23
  ) {constB=0.5}
  n25 = StaticSwitch(True=Add(A=@n7) {constB=0.5, desc="Constrain to bottom half of the texture (rotation)"}, False=@n17, Value=@n3)
  n26 = LinearInterpolate(
   A=TextureSample(UVs=MaterialFunctionCall(X=@n9, Y=@n25) {materialFunction=&MakeFloat2}.Result, Tex=@n11) {samplerType=SAMPLERTYPE_Data, texture=&T_BlackVectorDisplacement}.RGBA
   B=TextureSample(UVs=MaterialFunctionCall(X=@n13, Y=@n25) {materialFunction=&MakeFloat2}.Result, Tex=@n11) {samplerType=SAMPLERTYPE_Data, texture=&T_BlackVectorDisplacement}.RGBA
   Alpha=@n15
  )
  n27 = MaterialFunctionCall!(
   Vector=StaticSwitchParameter(True=StaticSwitchParameter(True=@n24, False=@n23) {group=Deformation, parameterName=ApplyScaleAnimation, sortPriority=10}, False=@n23) {group=BakedTransforms, parameterName=ConstainsScaleInfo, sortPriority=25}
   Quat=@n26
   Invert Quaternion
  ) {materialFunction=/Engine/Functions/Engine_MaterialFunctions03/Math/MultiplyVectorWithQuaternion}
  n28 = Subtract(A=Add(A=Add(A=@n16, B=@n27.Result) {constB=0}, B=@n22.RGB) {constB=0}, B=@n21)
  FunctionOutput(None=StaticSwitchParameter(True=Transform(None=@n28) {transformSourceType=TRANSFORMSOURCE_Local}, False=Constant3Vector()) {defaultValue=true, group=Deformation, parameterName=ApplyWPO, sortPriority=5}) {outputName=WPO}
  n29 = Transform(None=FunctionInput(Preview=VertexNormalWS()) {bUsePreviewValueAsDefault=true, inputName=WorldNormal, sortPriority=50}) {transformSourceType=TRANSFORMSOURCE_World, transformType=TRANSFORM_Instance}
  n30 = Add(A=@n12.RGBA, B=@n26)
  n31 = VertexInterpolator(VS=@n30)
  FunctionOutput(None=Transform(None=MaterialFunctionCall(Vector=@n29, Quat=@n31.PS) {materialFunction=&MultiplyVectorWithQuaternion}.Result) {transformSourceType=TRANSFORMSOURCE_Instance}) {outputName=Normal, sortPriority=-2}
  FunctionOutput(None=@n31.PS) {outputName=Quaternion}
  FunctionOutput(None=@n30) {outputName=QuaternionVS}
  LinearInterpolate(A=Constant4Vector() {constant=<rgba 0 0 0 1>}.RGBA, B=@n26, Alpha=ScalarParameter() {defaultValue=1, parameterName=QuaternionMix})
 ]}
