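# Material function that plays texture-baked vertex animation on sprite meshes, returning modified Material Attributes and a separate WPO.
# Playback time, offset per instance or by object position and scaled by the hologram collection's PlaybackSpeed, maps into four clip ranges.
# MF_Sprite_WPOTransform_subuvs samples the animation texture with an AnimTransition between two clips, producing offset and normal.
# Static switches toggle baked animation, normal blending and SplineThicken for ribbon LODs.
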MaterialFunction /BRCosmetics/Animation/Props/Gadgets_PixieParts/Materials/ParentMaterials/MF_Sprite_BakedAnimation
graph {nodes [
  Constant()
  Constant() {r=1}
  FunctionInput(Preview=StaticBool()) {bUsePreviewValueAsDefault=true, inputName=VFX, inputType=FunctionInput_StaticBool}
  n1 = GetMaterialAttributes(None=FunctionInput() {inputName=MaterialAttributes, inputType=FunctionInput_MaterialAttributes}) {attributeGetTypes [#1, #2]}
  n2 = Normalize(VectorInput=@n1.Normal)
  n3 = MaterialFunctionCall(BaseNormal=Normalize(VectorInput=NamedRerouteUsage()), AdditionalNormal=@n2) {materialFunction=&BlendAngleCorrectedNormals}
  n4 = NamedRerouteUsage()
  n5 = StaticSwitchParameter(
   True=SetMaterialAttributes!(
    MaterialAttributes=@n1.MaterialAttributes
    Normal=StaticSwitchParameter(True=@n3.Result, False=@n2) {defaultValue=true, parameterName=UseBakedAnimNormal}
    WorldPositionOffset=Add(A=@n1.WorldPositionOffset, B=@n4)
   ) {attributeSetTypes [#1, #2]}
   False=@n1.MaterialAttributes
  ) {defaultValue=true, parameterName=UseBakedAnim}
  FunctionOutput(None=@n5)
  n6 = ScalarParameter() {defaultValue=20, parameterName=Width}
  n7 = MaterialFunctionCall(WidthBase=@n6, WidthTip=@n6) {materialFunction=&SplineThicken}
  n8 = MaterialFunctionCall!(u, u_New, VFX, bind_u, AnimTransition, vStart, vHeight, texelHeight, AnimationTexture, WorldNormal) {materialFunction=&/MF_Sprite_WPOTransform_subuvs}
  n9 = Add(A=@n8.WPO, B=@n7.WorldPositionOffset)
  FunctionOutput(None=NamedRerouteDeclaration(None=StaticSwitchParameter(True=@n9, False=@n8.WPO) {parameterName=IsRibbonLOD}) {name=BakedAnimWPO, nodeColor=<rgba 1 0 0.6094 1>}) {outputName=WPO}
  n10 = Constant() {r=4}
  n11 = ScalarParameter() {defaultValue=2, parameterName=AnimIndex, sliderMax=4, sliderMin=1}
  n12 = Divide(A=@n11, B=@n10)
  n13 = Constant() {r=1}
  n14 = Constant() {r=1}
  n15 = ScalarParameter() {defaultValue=4, parameterName=NewAnimIndex, sliderMax=4, sliderMin=1}
  n16 = Divide(A=@n15, B=@n10)
  LinearInterpolate(
   A=HsvToRgb!(None=MaterialFunctionCall(R=@n12, G=@n13, B=@n14) {materialFunction=&AppendMany}.RGB)
   B=HsvToRgb(None=MaterialFunctionCall(R=@n16, G=@n13, B=@n14) {materialFunction=&AppendMany}.RGB)
   Alpha=ScalarParameter() {defaultValue=0.1, parameterName=AnimTransition, sliderMax=4, sliderMin=1}
  )
  n17 = ComponentMask(None=CollectionParameter() {collection=/Game/Environments/TechArt/Hologram/MPC_Hologram, parameterName=PlaybackSpeed}) {r=true}
  Multiply(A=@n17) {constB=4}
  n18 = TextureObjectParameter() {
   parameterName=AnimationTexture
   samplerType=SAMPLERTYPE_Data
   texture=/Sprites/Assets/Textures/BakedAnimation/Test1/T_Sprite_BakedAnim_Test1_ABCD_RGBA
  }
  n19 = MaterialFunctionCall(Float2=TextureProperty(Tex=@n18)) {materialFunction=&BreakOutFloat2Components}
  NamedRerouteDeclaration(None=Divide(B=NamedRerouteDeclaration(None=@n19.G) {name=textureHeight, nodeColor=<rgba 0.5469 0 1 1>})) {name=texelHeight, nodeColor=<rgba 0 1 0.2266 1>}
  n20 = Divide(B=NamedRerouteDeclaration(None=@n19.R) {name=textureWidth, nodeColor=<rgba 0.0078 0 1 1>})
  NamedRerouteDeclaration(None=Multiply(A=NamedRerouteDeclaration(None=@n20) {name=texelWidth, nodeColor=<rgba 0 1 0.2266 1>}) {constB=0.5}) {name=halfTexelWidth, nodeColor=<rgba 0.0312 0 1 1>}
  NamedRerouteDeclaration(None=@n18) {name=AnimationTexture, nodeColor=<rgba 0.6172 0 1 1>}
  n21 = MaterialFunctionCall(Float3=ObjectPositionWS()) {materialFunction=&BreakOutFloat3Components}
  n22 = Add(A=@n21.R, B=@n21.R)
  n23 = Frac(None=Divide(A=Add(A=@n22, B=@n21.B)) {constB=10})
  n24 = ScalarParameter() {defaultValue=33.3, parameterName=RandomTimeOffset}
  n25 = Time()
  n26 = ScalarParameter() {parameterName=TimeOffset}
  n27 = StaticSwitchParameter(True=Add(A=Multiply(A=@n23, B=@n24), B=@n25), False=Add(A=Add(A=@n25, B=@n26), B=PerInstanceRandom())) {parameterName=UseRandomOffsetPosition}
  n28 = Divide(B=NamedRerouteUsage()) {constA=30}
  n29 = Frac(None=Multiply(A=Multiply(A=@n27, B=@n28), B=@n17))
  n30 = NamedRerouteUsage()
  NamedRerouteDeclaration(None=Add(A=Multiply(A=@n29, B=@n30), B=NamedRerouteUsage())) {name=u, nodeColor=<rgba 0 1 0.9766 1>}
  NamedRerouteDeclaration(None=Add(A=NamedRerouteUsage(), B=NamedRerouteUsage())) {name=uStart, nodeColor=<rgba 1 0.3516 0 1>}
  n31 = NamedRerouteUsage()
  NamedRerouteDeclaration(None=Subtract(A=Subtract(B=@n31), B=@n31)) {name=uWidth, nodeColor=<rgba 0.9922 1 0 1>}
  NamedRerouteDeclaration(None=Transform(None=@n8.Normal) {transformSourceType=TRANSFORMSOURCE_World, transformType=TRANSFORM_Tangent}) {name=BakedAnimNormal, nodeColor=<rgba 1 0 0.6094 1>}
  n32 = NamedRerouteUsage()
  n33 = Constant()
  n34 = Constant() {r=1}
  n35 = ScalarParameter() {parameterName=Anim 1 Start}
  n36 = ScalarParameter() {defaultValue=0.25, parameterName=Anim 2 Start}
  n37 = ScalarParameter() {defaultValue=0.5, parameterName=Anim 3 Start}
  n38 = ScalarParameter() {defaultValue=0.75, parameterName=Anim 4 Start}
  n39 = Constant() {r=1}
  n40 = Constant() {r=4}
  n41 = Constant()
  n42 = Constant() {r=1}
  n43 = MaterialFunctionCall(Input=Round(None=@n11), Input Low=@n39, Input High=@n40, Target Low=@n41, Target High=@n42) {materialFunction=&RemapValueRange}
  n44 = ScalarParameter() {defaultValue=0.25, parameterName=Anim 1 End}
  n45 = ScalarParameter() {defaultValue=0.5, parameterName=Anim 2 End}
  n46 = ScalarParameter() {defaultValue=0.75, parameterName=Anim 3 End}
  n47 = ScalarParameter() {defaultValue=1, parameterName=Anim 4 End}
  NamedRerouteDeclaration(None=MaterialFunctionCall(
    Input=@n32
    Input Low=@n33
    Input High=@n34
    Target Low=MaterialFunctionCall!(1=@n35, 2=@n36, 3=@n37, 4=@n38, A=@n43.Result) {materialFunction=/Engine/Functions/Engine_MaterialFunctions02/Lerp_Multiple_Float}.Lerp 4 Inputs
    Target High=MaterialFunctionCall(1=@n44, 2=@n45, 3=@n46, 4=@n47, A=@n43.Result) {materialFunction=&Lerp_Multiple_Float}.Lerp 4 Inputs
   ) {materialFunction=&RemapValueRange}.Result) {name=u_MultiAnim, nodeColor=<rgba 0 1 0.9766 1>}
  n48 = MaterialFunctionCall(Input=Round(None=@n15), Input Low=@n39, Input High=@n40, Target Low=@n41, Target High=@n42) {materialFunction=&RemapValueRange}
  NamedRerouteDeclaration(None=MaterialFunctionCall(
    Input=@n32
    Input Low=@n33
    Input High=@n34
    Target Low=MaterialFunctionCall(1=@n35, 2=@n36, 3=@n37, 4=@n38, A=@n48.Result) {materialFunction=&Lerp_Multiple_Float}.Lerp 4 Inputs
    Target High=MaterialFunctionCall(1=@n44, 2=@n45, 3=@n46, 4=@n47, A=@n48.Result) {materialFunction=&Lerp_Multiple_Float}.Lerp 4 Inputs
   ) {materialFunction=&RemapValueRange}.Result) {name=u_MultiAnim_New, nodeColor=<rgba 0 1 0.9766 1>}
  NamedRerouteUsage()
  NamedRerouteUsage()
  n49 = TextureSampleParameter2D() {parameterName=Normals, samplerType=SAMPLERTYPE_Normal, texture=&FlatNormalBlue}
  Reroute(None=StaticSwitchParameter(
    True=@n7.Normal
    False=MaterialFunctionCall(BaseNormal=@n8.Normal, AdditionalNormal=@n49.RGB) {materialFunction=&BlendAngleCorrectedNormals}.Result
   ) {parameterName=IsRibbonLOD}) {desc=normal}
  Reroute(None=NamedRerouteUsage())
  Reroute(None=NamedRerouteUsage())
  n50 = Divide(B=NamedRerouteUsage()) {constA=30}
  n51 = ScalarParameter() {defaultValue=1, parameterName=Speed}
  n52 = ScalarParameter() {defaultValue=0.25, parameterName=End01}
  n53 = ScalarParameter() {parameterName=Start01}
  n54 = Multiply(A=Time(), B=Divide(A=Multiply(A=@n50, B=@n51), B=Subtract(A=@n52, B=@n53)))
  Reroute(None=MaterialFunctionCall(Input=Frac(None=@n54), Input Low=Constant(), Input High=Constant() {r=1}, Target Low=@n53, Target High=@n52) {materialFunction=&RemapValueRange}.Result)
  Reroute(None=NamedRerouteUsage())
  Reroute(None=@n18)
  ScalarParameter() {parameterName=AnimValue, sliderMax=4, sliderMin=1}
  n55 = Constant() {r=1}
  SetMaterialAttributes!(
   MaterialAttributes=@n5
   BaseColor=Constant3Vector() {constant=<rgba 1 0 0 0>}
   Metallic=Constant()
   Specular=@n55
   Roughness=@n55
   EmissiveColor=Constant() {r=0.05}
  ) {attributeSetTypes [
    #3
    #4
    #5
    #6
    #7
   ]}
 ]}
