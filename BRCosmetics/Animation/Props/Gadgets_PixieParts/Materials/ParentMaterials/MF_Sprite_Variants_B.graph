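# Material function for sprite variant B of the Pixie gadget props, adding an animated TwistedGoo look behind the Sprite_Variant_B static switch.
# World-aligned ComplexPan functions project panning Mastermind mask and normal textures using pre-skinned position and normals.
# The mask drives a two-color lerp and distortion, normals blend angle-corrected, and a third projection drives directional vertex offset.
# When enabled, BlendMaterialAttributes combines material attribute sets, with a VariantMask input available.
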MaterialFunction /BRCosmetics/Animation/Props/Gadgets_PixieParts/Materials/ParentMaterials/MF_Sprite_Variants_B
graph {nodes [
  Constant()
  n1 = NamedRerouteUsage()
  n2 = NamedRerouteUsage()
  n3 = NamedRerouteUsage()
  FunctionOutput(None=StaticSwitchParameter(True=BlendMaterialAttributes(A=@n1, B=@n2, Alpha=@n3), False=@n1) {group=TwistedGoo, parameterName=Sprite_Variant_B}) {outputName=SetMaterialAttributes}
  n4 = PreSkinnedPosition()
  NamedRerouteDeclaration(None=VertexInterpolator(VS=NamedRerouteDeclaration(None=@n4) {name=PreskinnedLocalPosVS, nodeColor=<rgba 1 0 0.0703 1>}).PS) {name=PreskinnedLocalPos, nodeColor=<rgba 1 0 0.0703 1>}
  n5 = VectorParameter() {group=TwistedGoo, parameterName=TG_PanDirection_XZ, sortPriority=11, defaultValue=<rgba 0.0762 0.1333 0 1>}
  n6 = Normalize(VectorInput=ComponentMask(None=@n5.RGB) {g=true, r=true})
  NamedRerouteDeclaration(None=Multiply(A=NamedRerouteUsage(), B=@n6)) {name=Pan_Mask_XZ, nodeColor=<rgba 0.7109 1 0 1>}
  n7 = VectorParameter() {group=TwistedGoo, parameterName=TG_PanDirection_XY, sortPriority=12, defaultValue=<rgba 0.0286 -0.05 0 1>}
  n8 = Normalize(VectorInput=ComponentMask(None=@n7.RGB) {g=true, r=true})
  NamedRerouteDeclaration(None=Multiply(A=NamedRerouteUsage(), B=@n8)) {name=Pan_Mask_XY, nodeColor=<rgba 0.7109 1 0 1>}
  NamedRerouteDeclaration(None=Multiply(A=@n6, B=NamedRerouteUsage())) {name=Pan_Normal_XZ, nodeColor=<rgba 0.7109 1 0 1>}
  NamedRerouteDeclaration(None=Multiply(A=@n8, B=NamedRerouteUsage())) {name=Pan_Normal_XY, nodeColor=<rgba 0.7109 1 0 1>}
  NamedRerouteDeclaration(None=Divide(B=ScalarParameter() {defaultValue=1, group=TwistedGoo, parameterName=TG_GlobalTiling, sortPriority=3})) {name=GlobalTiling, nodeColor=<rgba 0 0.4375 1 1>}
  n9 = NamedRerouteUsage()
  NamedRerouteDeclaration(None=Multiply(
    A=ScalarParameter() {defaultValue=6, group=TwistedGoo, parameterName=TG_NormalDistortionStrength, sortPriority=25}
    B=ComponentMask(None=@n9) {g=true}
   )) {name=Distortion, nodeColor=<rgba 0.0078 1 0 1>}
  n10 = TextureObject() {samplerType=SAMPLERTYPE_Masks, texture=/Game/Characters/Player/Male/Medium/Bodies/M_MED_Mastermind/Textures/T_M_MED_Mastermind_Twisted_M}
  NamedRerouteDeclaration(None=@n10) {name=MaskTexture, nodeColor=<rgba 0 0.9297 1 1>}
  NamedRerouteDeclaration(None=Multiply(
    A=ScalarParameter() {defaultValue=80, group=TwistedGoo, parameterName=TG_NormalTextureSize, primitiveDataIndex=10, sortPriority=22}
    B=NamedRerouteUsage()
   )) {name=NormalTextureSize, nodeColor=<rgba 0.6641 0 1 1>}
  n11 = PreSkinnedNormal()
  NamedRerouteDeclaration(None=VertexInterpolator(VS=NamedRerouteDeclaration(None=@n11) {name=PreskinnedLocalNormalVS, nodeColor=<rgba 1 0.9609 0 1>}).PS) {name=PreskinnedLocalNormal, nodeColor=<rgba 1 0.9609 0 1>}
  NamedRerouteDeclaration(None=Multiply(A=ScalarParameter() {defaultValue=35, group=TwistedGoo, parameterName=TG_DistortionTextureSize}, B=NamedRerouteUsage())) {name=DistortionTextureSize, nodeColor=<rgba 1 0 0.4219 1>}
  n12 = NamedRerouteUsage()
  n13 = NamedRerouteUsage()
  n14 = NamedRerouteUsage()
  n15 = NamedRerouteUsage()
  n16 = NamedRerouteUsage()
  n17 = NamedRerouteUsage()
  n18 = NamedRerouteUsage()
  n19 = LinearInterpolate(
   A=ScalarParameter() {defaultValue=-0.4238, group=TwistedGoo, parameterName=TG_VertexOffsetMin}
   B=ScalarParameter() {defaultValue=0.791, group=TwistedGoo, parameterName=TG_VertexOffsetMax}
   Alpha=ComponentMask(None=MaterialFunctionCall!(
     Export Float 4
     TextureObject=@n12
     TextureSize=@n13
     WorldPosition=@n14
     World Space Normal=@n15
     ProjectionTransitionContrast
     Pan_YZ=@n16
     Pan_XZ=@n17
     Pan_XY=@n18
    ) {materialFunction=/Game/Weapons/WeaponSkins/Wraps/MaterialFunctions/WorldAlignedTexture_ComplexPan}.XYZ Texture) {b=true}
  )
  n20 = VectorParameter() {group=TwistedGoo, parameterName=TG_VertexOffsetDirection, defaultValue=<rgba 2 -1 2 1>}
  NamedRerouteDeclaration(None=Multiply(A=Clamp(None=@n19) {clampMode=CMODE_ClampMin}, B=Transform(None=@n20.RGB) {transformSourceType=TRANSFORMSOURCE_Local})) {name=WPO, nodeColor=<rgba 0 1 0.8594 1>}
  n21 = NamedRerouteUsage()
  n22 = MaterialFunctionCall!(
   TextureObject=NamedRerouteUsage()
   TextureSize=NamedRerouteUsage()
   Normal=Normalize(VectorInput=Transform(None=@n21) {transformSourceType=TRANSFORMSOURCE_Local, transformType=TRANSFORM_Tangent})
   WorldSpace
   ProjectionTransitionContrast=ScalarParameter() {group=TwistedGoo, parameterName=TG_NormalProjectionContrast, sortPriority=24}
   WorldPosition=Add(A=NamedRerouteUsage(), B=NamedRerouteUsage())
   Use High Quality Normals=StaticBoolParameter() {defaultValue=true, group=TwistedGoo, parameterName=TG_UseHQNormals, sortPriority=26}
   Pan_YZ=NamedRerouteUsage()
   Pan_XZ=NamedRerouteUsage()
   Pan_XY=NamedRerouteUsage()
  ) {materialFunction=/Game/Weapons/WeaponSkins/Wraps/MaterialFunctions/WorldAlignedNormal_ComplexPan}
  n23 = ScalarParameter() {defaultValue=0.25, group=TwistedGoo, parameterName=TG_NormalFlatten, sortPriority=23}
  NamedRerouteDeclaration(None=MaterialFunctionCall(
    BaseNormal=NamedRerouteUsage()
    AdditionalNormal=MaterialFunctionCall(Normal=@n22.XYZ Texture, Flatness=@n23) {materialFunction=&FlattenNormal}.Result
   ) {materialFunction=&BlendAngleCorrectedNormals}.Result) {name=Normal, nodeColor=<rgba 0 1 0.5547 1>}
  n24 = VectorParameter() {group=TwistedGoo, parameterName=TG_Color_02, sortPriority=2, defaultValue=<rgba 0.0022 0.0017 0.005 1>}
  n25 = NamedRerouteUsage()
  NamedRerouteDeclaration(None=LinearInterpolate(
    A=VectorParameter() {group=TwistedGoo, parameterName=TG_Color_01, sortPriority=1, defaultValue=<rgba 0 0 0 1>}.RGB
    B=@n24.RGB
    Alpha=ComponentMask(None=@n25) {b=true}
   )) {name=BaseColor, nodeColor=<rgba 1 0.1641 0 1>}
  NamedRerouteDeclaration(None=MaterialFunctionCall(
    TextureObject=NamedRerouteUsage()
    TextureSize=NamedRerouteUsage()
    WorldPosition=NamedRerouteUsage()
    World Space Normal=NamedRerouteUsage()
    Pan_YZ=NamedRerouteUsage()
    Pan_XZ=NamedRerouteUsage()
    Pan_XY=NamedRerouteUsage()
   ) {materialFunction=&WorldAlignedTexture_ComplexPan}.XYZ Texture) {name=Mask, nodeColor=<rgba 1 0.8672 0 1>}
  n26 = GetMaterialAttributes(None=FunctionInput() {inputName=MaterialAttributes, inputType=FunctionInput_MaterialAttributes}) {attributeGetTypes [
    #1
    #2
    #3
    #4
    #5
   ]}
  NamedRerouteDeclaration(None=@n26.Normal) {name=IN_Normal, nodeColor=<rgba 0 1 0.4141 1>}
  n27 = SetMaterialAttributes!(
   MaterialAttributes
   BaseColor=NamedRerouteUsage()
   Specular=ScalarParameter() {defaultValue=0.3238, group=TwistedGoo, parameterName=TG_Specular}
   Roughness=ScalarParameter() {defaultValue=0.33, group=TwistedGoo, parameterName=TG_Roughness}
   Normal=NamedRerouteUsage()
   WorldPositionOffset=NamedRerouteUsage()
  ) {attributeSetTypes [
    #3
    #6
    #5
    #2
    #7
   ]}
  NamedRerouteDeclaration(None=@n27) {name=OUT_Material, nodeColor=<rgba 0 1 1 1>}
  n28 = TextureObject() {samplerType=SAMPLERTYPE_Normal, texture=/Game/Characters/Player/Male/Medium/Bodies/M_MED_Mastermind/Textures/T_M_MED_Mastermind_Twisted_N}
  NamedRerouteDeclaration(None=@n28) {name=NormalTexture, nodeColor=<rgba 1 0 0.2812 1>}
  NamedRerouteDeclaration(None=ScalarParameter() {defaultValue=0.04, group=TwistedGoo, parameterName=TG_DistortionPanSpeed}) {name=PanSpeedMask, nodeColor=<rgba 1 0 0.375 1>}
  NamedRerouteDeclaration(None=ScalarParameter() {defaultValue=0.06, group=TwistedGoo, parameterName=TG_NormalPanSpeed, sortPriority=20}) {name=PanSpeedNormal, nodeColor=<rgba 0 1 1 1>}
  NamedRerouteDeclaration(None=@n26.BaseColor) {name=OriginalBaseColor, nodeColor=<rgba 1 0.1406 0 1>}
  NamedRerouteDeclaration(None=@n26.MaterialAttributes) {name=OriginalMA, nodeColor=<rgba 1 0 0.1015 1>}
  NamedRerouteDeclaration(None=@n26.Metallic) {name=OriginalMetallic, nodeColor=<rgba 0 0.8594 1 1>}
  NamedRerouteDeclaration(None=@n26.Roughness) {name=OriginalRoughness, nodeColor=<rgba 1 0 0.3984 1>}
  n29 = PreSkinnedPosition()
  NamedRerouteDeclaration(None=ShaderStageSwitch(PixelShader=VertexInterpolator(VS=@n29).PS, VertexShader=@n29)) {name=PreskinnedPosition, nodeColor=<rgba 1 0 0.3281 1>}
  n30 = NamedRerouteUsage()
  NamedRerouteDeclaration(None=FunctionInput(Preview=Step(Y=@n30) {constX=0}) {inputName=VariantMask, inputType=FunctionInput_Scalar}) {name=Variant_Mask, nodeColor=<rgba 0.2422 1 0 1>}
  n31 = VectorParameter() {group=TwistedGoo, parameterName=TG_PanDirection_YZ, sortPriority=10, defaultValue=<rgba -0.0857 0.1333 0 1>}
  n32 = Normalize(VectorInput=ComponentMask(None=@n31.RGB) {g=true, r=true})
  NamedRerouteDeclaration(None=Multiply(A=NamedRerouteUsage(), B=@n32)) {name=Pan_Mask_YZ, nodeColor=<rgba 0.7109 1 0 1>}
  NamedRerouteDeclaration(None=Multiply(A=@n32, B=NamedRerouteUsage())) {name=Pan_Normal_YZ, nodeColor=<rgba 0.7109 1 0 1>}
  n33 = NamedRerouteUsage()
  n34 = SetMaterialAttributes!(MaterialAttributes=@n33, EmissiveColor, Normal, BaseColor, Metallic, Roughness) {attributeSetTypes [
    #1
    #2
    #3
    #4
    #5
   ]}
  StaticSwitchParameter(True=@n34, False=@n33) {group=SpriteVariants, parameterName=Sprite_Variant_B}
 ]}
