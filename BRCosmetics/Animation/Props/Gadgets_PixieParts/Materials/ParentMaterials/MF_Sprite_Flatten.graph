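# Material function for the Pixie Parts gadget prop that flattens a mesh along local Y through World Position Offset, apparently for a sprite look.
# Local Y position plus a plane offset scales a vector parameter transformed from local space, squashing vertices toward that plane by default.
# A clamped UseSpriteFlatten scalar blends the effect, which is added to the input material attributes' existing World Position Offset.
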MaterialFunction /BRCosmetics/Animation/Props/Gadgets_PixieParts/Materials/ParentMaterials/MF_Sprite_Flatten
graph {nodes [
  n1 = GetMaterialAttributes(None=FunctionInput() {inputName=Material, inputType=FunctionInput_MaterialAttributes}) {attributeGetTypes [#1]}
  n2 = ComponentMask(None=LocalPosition().XYZ) {g=true}
  n3 = ScalarParameter() {defaultValue=11, group=SpriteFlatten, parameterName=SpriteFlatten_YPlane}
  n4 = VectorParameter() {group=SpriteFlatten, parameterName=SpriteFlatten_WPOAmount, defaultValue=<rgba 0 -0.7 0 1>}
  n5 = Multiply(
   A=Multiply(A=Add(A=@n2, B=@n3), B=Transform(None=@n4.RGB) {transformSourceType=TRANSFORMSOURCE_Local})
   B=Clamp(None=ScalarParameter() {group=SpriteFlatten, parameterName=UseSpriteFlatten})
  )
  FunctionOutput(None=SetMaterialAttributes!(MaterialAttributes=@n1.MaterialAttributes, WorldPositionOffset=Add(A=@n1.WorldPositionOffset, B=@n5), EmissiveColor) {attributeSetTypes [#1, #2]})
 ]}
