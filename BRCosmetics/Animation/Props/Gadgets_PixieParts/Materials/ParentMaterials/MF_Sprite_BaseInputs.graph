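# Material function that supplies base material attributes for the Sprite gadget prop materials in BRCosmetics.
# Base Color, SRM and Normal textures feed the attributes, with SRM blue driving metallic, specular and roughness alike.
# Emissive multiplies base color by a time-of-day tint and intensity; day uses a zero constant, leaving TOD_Day unconnected.
# It outputs the attributes, the normal map, and a PaintMask texture as FX Mask.
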MaterialFunction /BRCosmetics/Animation/Props/Gadgets_PixieParts/Materials/ParentMaterials/MF_Sprite_BaseInputs
graph {nodes [
  n1 = TextureSampleParameter2D() {parameterName=Base Color, texture=/Game/Characters/Fortress_Character/Textures/T_All_White}
  n2 = TextureSampleParameter2D() {parameterName=SRM, samplerType=SAMPLERTYPE_Masks, texture=/Game/Packages/Fortress_SharedMaterials/Textures/Black_MASK}
  n3 = MaterialFunctionCall(
   Color_Morning=VectorParameter() {group=Sprite_Base, parameterName=TOD_Morning, sortPriority=9, defaultValue=<rgba 0.8 0.4 0.5 1>}.RGB
   Color_Day=Constant()
   Color_Evening=VectorParameter() {group=Sprite_Base, parameterName=TOD_Evening, sortPriority=9, defaultValue=<rgba 0.8 0.5 0.4 1>}.RGB
   Color_Night=VectorParameter() {group=Sprite_Base, parameterName=TOD_Night, sortPriority=9, defaultValue=<rgba 0.35 0.4667 0.7 1>}.RGB
  ) {materialFunction=&Color_Time_of_Day}
  n4 = ScalarParameter() {defaultValue=0.5, parameterName=BaseEmissive Tex ToD Intensity}
  n5 = TextureSampleParameter2D() {parameterName=Normal, samplerType=SAMPLERTYPE_Normal, texture=&FlatNormalBlue}
  n6 = SetMaterialAttributes!(
   MaterialAttributes
   BaseColor=@n1.RGB
   Metallic=@n2.B
   Specular=@n2.B
   Roughness=@n2.B
   Anisotropy
   EmissiveColor=Multiply(A=@n1.RGB, B=Multiply(A=@n3.Result, B=@n4) {constB=0.5})
   ShadingModel=ShadingModel()
   OpacityMask
   Normal=@n5.RGB
  ) {attributeSetTypes [
    #1
    #2
    #3
    #4
    #5
    #6
    #7
    #8
    #9
   ]}
  FunctionOutput(None=@n6) {outputName=MA}
  FunctionOutput(None=@n5.RGB) {outputName=Normal Map, sortPriority=2}
  FunctionOutput(None=TextureSampleParameter2D() {parameterName=PaintMask, samplerType=SAMPLERTYPE_Masks, texture=&Black_MASK}.RGB) {outputName=FX Mask, sortPriority=1}
  VectorParameter() {group=Sprite_Base, parameterName=TOD_Day, sortPriority=9, defaultValue=<rgba 1 1 1 1>}
 ]}
