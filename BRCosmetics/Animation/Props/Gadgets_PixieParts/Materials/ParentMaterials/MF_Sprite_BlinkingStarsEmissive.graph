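# Material function for Sprite cosmetic skins that drives emissive with blinking stars or pulsing cracks, gated by static switches.
# Stars use Emissive texture R/G as shape and B as blink phase, blend two tints over incoming emissive, and light by Scalar_Time_of_Day at night.
# Cracks replace emissive with masked, sine-pulsed texture, optionally bump-offset for depth using averaged pre-skinned scale.
# Outputs modified MaterialAttributes plus the star blink mask.
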MaterialFunction /BRCosmetics/Animation/Props/Gadgets_PixieParts/Materials/ParentMaterials/MF_Sprite_BlinkingStarsEmissive
graph {nodes [
  n1 = Multiply(A=Time(), B=ScalarParameter() {defaultValue=0.2, parameterName=Cracks_BlinkSpeed})
  n2 = TextureSampleParameter2D() {
   parameterName=Cracks Tex
   samplerType=SAMPLERTYPE_Masks
   texture=/BRCosmetics/Animation/Props/Gadgets_PixieParts/Textures/Sprite_Skins/Sprite_ZeroPoint/T_Sprite_ZeroPoint_Creature_FX3
  }
  n3 = Multiply(A=ScalarParameter() {defaultValue=-1, parameterName=Cracks_PanSpeed}, B=@n2.B)
  n4 = Add(A=Multiply(A=@n1) {constB=6.23}, B=@n3)
  n5 = Add(A=@n1, B=@n3)
  n6 = Add(A=LinearInterpolate(A=Sine(None=@n4), B=Sine(None=@n5)))
  n7 = Multiply(A=Power(Base=Multiply(A=@n6) {constB=0.5}), B=ScalarParameter() {defaultValue=3, parameterName=Cracks_BlinkIntensity})
  n8 = OneMinus(None=@n2.R)
  n9 = Saturate(None=OneMinus(None=Power(Base=@n8) {constExponent=5}))
  n10 = ScalarParameter() {defaultValue=1, parameterName=Cracks_PanDepth}
  n11 = FunctionInput() {bUsePreviewValueAsDefault=true, inputName=Pre-Skinned Scale, sortPriority=100, previewValue=<1 1 1 1>}
  n12 = Multiply(A=DotProduct(A=@n11, B=Convert!() {convertOutputs [{type=Vector3, defaultValue=<rgba 1 1 1 1>}]})) {constB=0.3333, desc="Scalar scale (avg)"}
  n13 = ScalarParameter() {
   defaultValue=1
   group=Sprite_Cosmetics | Blinking Stars
   parameterName=Blinking Stars - Bump Offset Use Pre-Skinned Scale
   sliderMax=1
   sortPriority=10
  }
  n14 = Multiply(
   A=Multiply(A=@n2.R, B=Multiply(A=@n10) {constB=-1})
   B=Multiply(
    A=LinearInterpolate(B=@n12, Alpha=@n13) {constA=1}
    B=ScalarParameter() {defaultValue=1, group=Sprite_Cosmetics | Blinking Stars, parameterName=Blinking Stars - Bump Offset Mult, sortPriority=20}
   )
  )
  n15 = TextureSampleParameter2D(UVs=StaticSwitchParameter(True=BumpOffset(Height=@n14), False=TextureCoordinate()) {defaultValue=true, group=BlinkingStars, parameterName=Cracks Bump}) {
   group=BlinkingStars
   parameterName=Emissive
   samplerType=SAMPLERTYPE_LinearColor
   texture=/BRCosmetics/Animation/Props/Gadgets_PixieParts/Textures/Sprite_Skins/Sprite_Sleepy/T_SleepySprite_Creature_E6
  }
  n16 = TextureSampleParameter2D() {
   group=BlinkingStars
   parameterName=Emissive
   samplerType=SAMPLERTYPE_LinearColor
   texture=&T_SleepySprite_Creature_E6
  }
  n17 = Max(A=Multiply(A=@n9, B=@n15.RGB), B=Multiply(A=OneMinus(None=@n9), B=Multiply(A=@n16.RGB) {constB=0.3}))
  n18 = Constant()
  n19 = MaterialFunctionCall(Value_Morning=@n18, Value_Day=@n18, Value_Evening=@n18, Value_Night=Constant() {r=1}) {materialFunction=&Scalar_Time_of_Day}
  n20 = Multiply(A=@n19.Result, B=VectorParameter() {group=BlinkingStars, parameterName=EmissiveTextureColor2, defaultValue=<rgba 2 0.1667 0 1>}.RGB)
  n21 = ScalarParameter() {defaultValue=1, group=BlinkingStars, parameterName=BlinkIntensityMin}
  n22 = ScalarParameter() {defaultValue=5, group=BlinkingStars, parameterName=BlinkIntensityMax}
  n23 = Time()
  n24 = ScalarParameter() {defaultValue=0.3, group=BlinkingStars, parameterName=BlinkingSpeed}
  n25 = TextureSampleParameter2D() {
   group=BlinkingStars
   parameterName=Emissive
   samplerType=SAMPLERTYPE_LinearColor
   texture=&T_SleepySprite_Creature_E6
  }
  n26 = Add(A=Multiply(A=@n25.R), B=Power(Base=@n25.G) {constExponent=1.3})
  n27 = AppendVector(A=@n26, B=@n25.B)
  n28 = Frac(None=Add(A=ComponentMask(None=@n27) {g=true}, B=Multiply(A=@n23, B=@n24) {constB=0.4, desc=.2}))
  n29 = OneMinus(None=@n28)
  n30 = Sine(None=Divide(A=Power(Base=@n29) {constExponent=4}))
  n31 = Power(Base=ComponentMask(None=@n27) {r=true}, Exp=Constant() {r=1}) {constExponent=3}
  n32 = Saturate(None=Multiply(A=Multiply(A=@n31, B=@n30), B=Constant() {r=1}) {constB=5})
  n33 = Power(Base=@n28) {constExponent=4}
  n34 = Saturate(None=Multiply(A=Multiply(A=@n32, B=@n33)) {constB=50})
  n35 = Multiply(A=Multiply(A=@n26, B=LinearInterpolate(A=@n21, B=@n22, Alpha=@n34)), B=@n19.Result)
  n36 = Multiply(A=@n25.R) {constB=10}
  n37 = Step(Y=@n36) {constX=0.05, constY=0.01}
  n38 = Multiply(A=Saturate(None=OneMinus(None=@n35)), B=Multiply(A=Saturate(None=@n36), B=OneMinus(None=@n37)))
  n39 = OneMinus(None=Saturate(None=@n25.R))
  n40 = GetMaterialAttributes(None=FunctionInput() {inputName=MA, inputType=FunctionInput_MaterialAttributes}) {attributeGetTypes [#1]}
  n41 = VectorParameter() {group=BlinkingStars, parameterName=EmissiveTextureColor, defaultValue=<rgba 1 0.6 0.2 1>}
  n42 = LinearInterpolate(A=@n40.EmissiveColor, B=@n41.RGB, Alpha=Saturate(None=Multiply(A=@n35) {constB=2}))
  n43 = Multiply(A=@n35, B=@n41.RGB)
  n44 = Add(A=Multiply(A=@n43) {constB=10})
  n45 = LinearInterpolate(
   A=StaticSwitchParameter(True=Multiply(A=@n40.EmissiveColor, B=@n44), False=@n42) {defaultValue=true, group=BlinkingStars, parameterName=AlternativeBlending}
   B=Add(A=@n40.EmissiveColor, B=@n43)
   Alpha=Saturate(None=@n35)
  )
  n46 = LinearInterpolate(A=@n45, B=@n42, Alpha=OneMinus(None=Power(Base=@n39)))
  n47 = StaticSwitchParameter(
   True=Add(A=Multiply(A=@n7, B=@n17), B=@n17)
   False=StaticSwitchParameter(True=LinearInterpolate(A=@n46, B=@n20, Alpha=@n38), False=@n25.RGB) {group=BlinkingStars, parameterName=Stars}
  ) {group=BlinkingStars, parameterName=Cracks}
  FunctionOutput(None=SetMaterialAttributes!(
    MaterialAttributes=@n40.MaterialAttributes
    EmissiveColor=StaticSwitchParameter(True=@n47, False=@n40.EmissiveColor) {group=BlinkingStars, parameterName=EmissiveTex}
   ) {attributeSetTypes [#1]}) {outputName=MA}
  FunctionOutput(None=StaticSwitchParameter(True=@n35, False=Constant()) {group=BlinkingStars, parameterName=EmissiveTex}) {outputName=Mask_BlinkingEmissive, sortPriority=1}
  OneMinus(None=@n2.R)
  Reroute(None=@n15.RGB)
  Reroute(None=@n45)
  Reroute(None=@n35)
  Reroute(None=@n26)
 ]}
