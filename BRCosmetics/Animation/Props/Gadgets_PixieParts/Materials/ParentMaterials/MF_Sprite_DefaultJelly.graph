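# Material function that builds the jelly sprite look for Pixie gadget props, blending it with incoming PBR attributes through vertex-preserving attribute blending.
# FXMask channels and generated fresnel, radial glow and bump-offset logo masks drive layered colors, with time-of-day tints and atmospheric-light terms.
# Object-anchored screen UVs correct for aspect and scale, alongside scrolling distortion and sine-based motion.
# Also outputs sprite WPO, masks, pre-skinned bounds and object position.
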MaterialFunction /BRCosmetics/Animation/Props/Gadgets_PixieParts/Materials/ParentMaterials/MF_Sprite_DefaultJelly
graph {expressions {
  MaterialExpressionAbs_0 {class=MaterialExpressionAbs, inputs [None {from=MaterialExpressionComponentMask_7, output=""}], properties {}}
  MaterialExpressionAbs_1 {class=MaterialExpressionAbs, inputs [None {from=MaterialExpressionDistance_1, output=""}], properties {}}
  MaterialExpressionAbs_2 {class=MaterialExpressionAbs, inputs [None {from=MaterialExpressionDistance_0, output=""}], properties {}}
  MaterialExpressionAdd_0 {
   class=MaterialExpressionAdd
   inputs [
    A {from=MaterialExpressionNamedRerouteUsage_12}
    B {from=MaterialExpressionMultiply_16}
   ]
  }
  MaterialExpressionAdd_1 {
   class=MaterialExpressionAdd
   inputs [
    A {from=MaterialExpressionDivide_7}
    B {from=MaterialExpressionReroute_16}
   ]
  }
  MaterialExpressionAdd_10 {
   class=MaterialExpressionAdd
   inputs [
    A {from=MaterialExpressionMultiply_27}
    B {from=MaterialExpressionVertexNormalWS_2}
   ]
  }
  MaterialExpressionAdd_11 {
   class=MaterialExpressionAdd
   inputs [
    A {from=MaterialExpressionMaterialFunctionCall_17, output=UVs}
    B {from=MaterialExpressionAppendVector_5}
   ]
  }
  MaterialExpressionAdd_12 {
   class=MaterialExpressionAdd
   inputs [
    A {from=MaterialExpressionMultiply_35}
    B {from=MaterialExpressionMultiply_33}
   ]
  }
  MaterialExpressionAdd_13 {
   class=MaterialExpressionAdd
   inputs [
    A {from=MaterialExpressionMultiply_31}
    B {from=MaterialExpressionFrac_0}
   ]
  }
  MaterialExpressionAdd_14 {
   class=MaterialExpressionAdd
   inputs [
    A {from=MaterialExpressionReroute_15}
    B {from=MaterialExpressionLinearInterpolate_16}
   ]
  }
  MaterialExpressionAdd_15 {
   class=MaterialExpressionAdd
   inputs [
    A {from=MaterialExpressionDivide_3}
    B {from=MaterialExpressionFrac_1}
   ]
  }
  MaterialExpressionAdd_16 {
   class=MaterialExpressionAdd
   inputs [
    A {from=MaterialExpressionNamedRerouteUsage_60}
    B {from=MaterialExpressionNamedRerouteUsage_59}
   ]
  }
  MaterialExpressionAdd_17 {
   class=MaterialExpressionAdd
   inputs [
    A {from=MaterialExpressionAdd_13}
    B {from=MaterialExpressionDivide_5}
   ]
  }
  MaterialExpressionAdd_2 {
   class=MaterialExpressionAdd
   inputs [
    A {from=MaterialExpressionNamedRerouteUsage_18}
    B {from=MaterialExpressionDivide_1}
   ]
  }
  MaterialExpressionAdd_3 {
   class=MaterialExpressionAdd
   inputs [
    A {from=MaterialExpressionReroute_2}
    B {from=MaterialExpressionMultiply_2}
   ]
  }
  MaterialExpressionAdd_4 {
   class=MaterialExpressionAdd
   inputs [
    A {from=MaterialExpressionAppendVector_1}
    B {from=MaterialExpressionVertexNormalWS_1}
   ]
  }
  MaterialExpressionAdd_45 {
   class=MaterialExpressionAdd
   inputs [
    A {from=MaterialExpressionConstant2Vector_7}
    B {from=MaterialExpressionScreenPosition_1, output=ViewportUV}
   ]
  }
  MaterialExpressionAdd_5 {
   class=MaterialExpressionAdd
   inputs [
    A {from=MaterialExpressionReroute_13}
    B {from=MaterialExpressionSubtract_2}
   ]
  }
  MaterialExpressionAdd_6 {
   class=MaterialExpressionAdd
   inputs [
    A {from=MaterialExpressionMaterialFunctionCall_30, output=R}
    B {from=MaterialExpressionMaterialFunctionCall_30, output=R}
   ]
  }
  MaterialExpressionAdd_7 {
   class=MaterialExpressionAdd
   inputs [
    A {from=MaterialExpressionMultiply_92}
    B {from=MaterialExpressionNamedRerouteUsage_14}
   ]
  }
  MaterialExpressionAdd_8 {
   class=MaterialExpressionAdd
   inputs [
    A {from=MaterialExpressionAdd_6}
    B {from=MaterialExpressionMaterialFunctionCall_30, output=B}
   ]
  }
  MaterialExpressionAdd_9 {
   class=MaterialExpressionAdd
   inputs [
    A {from=MaterialExpressionMultiply_26}
    B {from=""}
   ]
   properties {constB=-1}
  }
  MaterialExpressionAppendVector_0 {
   class=MaterialExpressionAppendVector
   inputs [
    A {from=MaterialExpressionScalarParameter_26, output=""}
    B {from=MaterialExpressionScalarParameter_6, output=""}
   ]
  }
  MaterialExpressionAppendVector_1 {
   class=MaterialExpressionAppendVector
   inputs [
    A {from=MaterialExpressionNamedRerouteUsage_5, output=""}
    B {from=MaterialExpressionConstant_15, output=""}
   ]
  }
  MaterialExpressionAppendVector_2 {
   class=MaterialExpressionAppendVector
   inputs [
    A {from=MaterialExpressionScalarParameter_19, output=""}
    B {from=MaterialExpressionScalarParameter_21, output=""}
   ]
  }
  MaterialExpressionAppendVector_5 {
   class=MaterialExpressionAppendVector
   inputs [
    A {from=MaterialExpressionVectorParameter_9, output=R}
    B {from=MaterialExpressionVectorParameter_9, output=R}
   ]
  }
  MaterialExpressionAppendVector_6 {
   class=MaterialExpressionAppendVector
   inputs [
    A {from=MaterialExpressionFunctionOutput_3, output=""}
    B {from=MaterialExpressionSaturate_7, output=""}
   ]
  }
  MaterialExpressionAppendVector_7 {
   class=MaterialExpressionAppendVector
   inputs [
    A {from=MaterialExpressionConstant_21, output=""}
    B {from=MaterialExpressionDivide_21, output=""}
   ]
  }
  MaterialExpressionAtmosphericLightVector_0 {class=MaterialExpressionAtmosphericLightVector, inputs [], properties {}}
  MaterialExpressionAtmosphericLightVector_1 {class=MaterialExpressionAtmosphericLightVector, inputs [], properties {}}
  MaterialExpressionBlendMaterialAttributes_0 {
   class=MaterialExpressionBlendMaterialAttributes
   inputs [
    A {from=MaterialExpressionReroute_8, output=""}
    B {from=MaterialExpressionFunctionInput_1, output=""}
    Alpha {from=MaterialExpressionNamedRerouteUsage_4, output=""}
   ]
   properties {vertexAttributeBlendType=UseA}
  }
  MaterialExpressionBounds_11 {class=MaterialExpressionBounds, inputs [], properties {type=MEILB_PreSkinnedLocal}}
  MaterialExpressionBumpOffset_1 {
   class=MaterialExpressionBumpOffset
   inputs [
    Coordinate {from=MaterialExpressionAdd_11, output=""}
    Height {from=MaterialExpressionScalarParameter_22, output=""}
    HeightRatioInput {from="", output=""}
   ]
   properties {heightRatio=0.1}
  }
  MaterialExpressionCameraVectorWS_1 {class=MaterialExpressionCameraVectorWS, inputs [], properties {}}
  MaterialExpressionCeil_0 {class=MaterialExpressionCeil, inputs [None {from=MaterialExpressionOneMinus_2, output=""}], properties {}}
  MaterialExpressionComponentMask_0 {class=MaterialExpressionComponentMask, inputs [None {from=MaterialExpressionFunctionInput_0}], properties {r=true}}
  MaterialExpressionComponentMask_1 {class=MaterialExpressionComponentMask, inputs [None {from=MaterialExpressionMultiply_4}], properties {b=true, r=true}}
  MaterialExpressionComponentMask_10 {class=MaterialExpressionComponentMask, inputs [None {from=MaterialExpressionNamedRerouteUsage_21}], properties {r=true}}
  MaterialExpressionComponentMask_11 {class=MaterialExpressionComponentMask, inputs [None {from=MaterialExpressionNamedRerouteUsage_93}], properties {b=true}}
  MaterialExpressionComponentMask_12 {class=MaterialExpressionComponentMask, inputs [None {from=MaterialExpressionNamedRerouteUsage_98}], properties {b=true}}
  MaterialExpressionComponentMask_13 {class=MaterialExpressionComponentMask, inputs [None {from=MaterialExpressionTextureSample_10, output=RGB}], properties {g=true, r=true}}
  MaterialExpressionComponentMask_14 {class=MaterialExpressionComponentMask, inputs [None {from=MaterialExpressionBounds_11, output=Extents}], properties {b=true}}
  MaterialExpressionComponentMask_15 {class=MaterialExpressionComponentMask, inputs [None {from=MaterialExpressionTransformPosition_6}], properties {b=true}}
  MaterialExpressionComponentMask_16 {class=MaterialExpressionComponentMask, inputs [None {from=MaterialExpressionNamedRerouteUsage_99}], properties {r=true}}
  MaterialExpressionComponentMask_17 {class=MaterialExpressionComponentMask, inputs [None {from=MaterialExpressionNamedRerouteUsage_8}], properties {b=true}}
  MaterialExpressionComponentMask_2 {class=MaterialExpressionComponentMask, inputs [None {from=MaterialExpressionNamedRerouteUsage_92}], properties {b=true}}
  MaterialExpressionComponentMask_3 {class=MaterialExpressionComponentMask, inputs [None {from=MaterialExpressionFunctionInput_0}], properties {g=true}}
  MaterialExpressionComponentMask_4 {class=MaterialExpressionComponentMask, inputs [None {from=MaterialExpressionMultiply_23}], properties {b=true}}
  MaterialExpressionComponentMask_5 {class=MaterialExpressionComponentMask, inputs [None {from=MaterialExpressionMultiply_1}], properties {g=true, r=true}}
  MaterialExpressionComponentMask_6 {class=MaterialExpressionComponentMask, inputs [None {from=MaterialExpressionFunctionInput_0}], properties {b=true}}
  MaterialExpressionComponentMask_7 {class=MaterialExpressionComponentMask, inputs [None {from=MaterialExpressionMultiply_23}], properties {r=true}}
  MaterialExpressionComponentMask_8 {class=MaterialExpressionComponentMask, inputs [None {from=MaterialExpressionNamedRerouteUsage_23}], properties {r=true}}
  MaterialExpressionComponentMask_9 {class=MaterialExpressionComponentMask, inputs [None {from=MaterialExpressionNamedRerouteUsage_23}], properties {g=true}}
  MaterialExpressionConstant2Vector_11 {class=MaterialExpressionConstant2Vector, properties {g=-0.5, r=0.5}}
  MaterialExpressionConstant2Vector_2 {class=MaterialExpressionConstant2Vector, properties {g=1}}
  MaterialExpressionConstant2Vector_3 {class=MaterialExpressionConstant2Vector, properties {g=-1, r=1}}
  MaterialExpressionConstant2Vector_6 {class=MaterialExpressionConstant2Vector, properties {g=0.3, r=0.4}}
  MaterialExpressionConstant2Vector_7 {class=MaterialExpressionConstant2Vector}
  MaterialExpressionConstant3Vector_3 {class=MaterialExpressionConstant3Vector}
  MaterialExpressionConstant4Vector_0 {class=MaterialExpressionConstant4Vector, inputs [], properties {constant=<rgba 1.5 5 3.5 1>}}
  MaterialExpressionConstant4Vector_1 {class=MaterialExpressionConstant4Vector, inputs [], properties {constant=<rgba 0 7.5 22 1>}}
  MaterialExpressionConstant4Vector_2 {class=MaterialExpressionConstant4Vector, inputs [], properties {constant=<rgba 0.6 0 -1 1>}}
  MaterialExpressionConstant4Vector_3 {class=MaterialExpressionConstant4Vector, inputs [], properties {constant=<rgba 5 2.7247 0.5652 0.12>}}
  MaterialExpressionConstantBiasScale_0 {
   class=MaterialExpressionConstantBiasScale
   inputs [None {from=MaterialExpressionTextureSample_9, output=RGB}]
   properties {bias=-0.5, scale=2}
  }
  MaterialExpressionConstantBiasScale_1 {class=MaterialExpressionConstantBiasScale, inputs [None {from=MaterialExpressionMultiply_84}]}
  MaterialExpressionConstantBiasScale_2 {class=MaterialExpressionConstantBiasScale, inputs [None {from=MaterialExpressionAdd_45}], properties {bias=-0.5, scale=2}}
  MaterialExpressionConstantBiasScale_3 {class=MaterialExpressionConstantBiasScale, inputs [None {from=MaterialExpressionPower_0}], properties {bias=-0.5, scale=2}}
  MaterialExpressionConstantBiasScale_4 {class=MaterialExpressionConstantBiasScale, inputs [None {from=MaterialExpressionSine_4}]}
  MaterialExpressionConstant_0 {class=MaterialExpressionConstant}
  MaterialExpressionConstant_1 {class=MaterialExpressionConstant, properties {r=0.5}}
  MaterialExpressionConstant_10 {class=MaterialExpressionConstant, properties {r=2}}
  MaterialExpressionConstant_12 {class=MaterialExpressionConstant, properties {r=1}}
  MaterialExpressionConstant_13 {class=MaterialExpressionConstant, properties {r=0.5}}
  MaterialExpressionConstant_15 {class=MaterialExpressionConstant}
  MaterialExpressionConstant_16 {class=MaterialExpressionConstant}
  MaterialExpressionConstant_2 {class=MaterialExpressionConstant, properties {r=0.5}}
  MaterialExpressionConstant_21 {class=MaterialExpressionConstant, properties {r=1}}
  MaterialExpressionConstant_3 {class=MaterialExpressionConstant, properties {r=0.5}}
  MaterialExpressionConstant_4 {class=MaterialExpressionConstant}
  MaterialExpressionConstant_5 {class=MaterialExpressionConstant}
  MaterialExpressionConstant_6 {class=MaterialExpressionConstant}
  MaterialExpressionConstant_7 {class=MaterialExpressionConstant, properties {r=1}}
  MaterialExpressionConstant_8 {class=MaterialExpressionConstant, properties {r=1.35}}
  MaterialExpressionConstant_9 {class=MaterialExpressionConstant}
  MaterialExpressionConvert_0 {
   class=MaterialExpressionConvert
   inputs [
    None {from=MaterialExpressionScalarParameter_32}
    None {from=MaterialExpressionScalarParameter_31}
   ]
   properties {
    nodeName=Make Vector2
    convertInputs [
     {
      type=Scalar
      defaultValue=<rgba 0.2 0 0 1>
      expressionInput {
       expression=&self:MaterialExpressionScalarParameter_32
       inputName=None
       mask=0
       maskA=0
       maskB=0
       maskG=0
       maskR=0
       outputIndex=0
      }
     }
     {
      type=Scalar
      defaultValue=<rgba 0.2 0 0 1>
      expressionInput {
       expression=&self:MaterialExpressionScalarParameter_31
       inputName=None
       mask=0
       maskA=0
       maskB=0
       maskG=0
       maskR=0
       outputIndex=0
      }
     }
    ]
    convertOutputs [{type=Vector2}]
   }
  }
  MaterialExpressionConvert_1 {
   class=MaterialExpressionConvert
   inputs [
    None {from=MaterialExpressionScalarParameter_43}
    None {from=MaterialExpressionScalarParameter_42}
   ]
   properties {
    nodeName=Make Vector2
    convertInputs [
     {
      type=Scalar
      defaultValue=<rgba 0.2 0 0 1>
      expressionInput {
       expression=&self:MaterialExpressionScalarParameter_43
       inputName=None
       mask=0
       maskA=0
       maskB=0
       maskG=0
       maskR=0
       outputIndex=0
      }
     }
     {
      type=Scalar
      defaultValue=<rgba 0.2 0 0 1>
      expressionInput {
       expression=&self:MaterialExpressionScalarParameter_42
       inputName=None
       mask=0
       maskA=0
       maskB=0
       maskG=0
       maskR=0
       outputIndex=0
      }
     }
    ]
    convertOutputs [{type=Vector2}]
   }
  }
  MaterialExpressionCosine_0 {class=MaterialExpressionCosine, inputs [None {from=MaterialExpressionAdd_12, output=""}], properties {}}
  MaterialExpressionDistance_0 {
   class=MaterialExpressionDistance
   inputs [
    A {from=MaterialExpressionBumpOffset_1, output=""}
    B {from=MaterialExpressionConstant_1, output=""}
   ]
   properties {}
  }
  MaterialExpressionDistance_1 {
   class=MaterialExpressionDistance
   inputs [
    A {from=MaterialExpressionBumpOffset_1, output=""}
    B {from=MaterialExpressionConstant_3, output=""}
   ]
   properties {}
  }
  MaterialExpressionDistance_2 {
   class=MaterialExpressionDistance
   inputs [
    A {from=MaterialExpressionComponentMask_2, output=""}
    B {from=MaterialExpressionScalarParameter_52, output=""}
   ]
   properties {}
  }
  MaterialExpressionDivide_0 {
   class=MaterialExpressionDivide
   inputs [
    A {from=MaterialExpressionDistance_2}
    B {from=MaterialExpressionLinearInterpolate_7}
   ]
   properties {constB=35}
  }
  MaterialExpressionDivide_1 {
   class=MaterialExpressionDivide
   inputs [
    A {from=MaterialExpressionNamedRerouteUsage_6}
    B {from=MaterialExpressionConvert_1}
   ]
  }
  MaterialExpressionDivide_10 {
   class=MaterialExpressionDivide
   inputs [
    A {from=MaterialExpressionMultiply_11}
    B {from=MaterialExpressionComponentMask_15}
   ]
   properties {constA=100}
  }
  MaterialExpressionDivide_2 {
   class=MaterialExpressionDivide
   inputs [
    A {from=MaterialExpressionAdd_9}
    B {from=MaterialExpressionScalarParameter_11}
   ]
  }
  MaterialExpressionDivide_21 {
   class=MaterialExpressionDivide
   inputs [
    A {from=MaterialExpressionMaterialFunctionCall_56, output=G}
    B {from=MaterialExpressionMaterialFunctionCall_56, output=G}
   ]
  }
  MaterialExpressionDivide_3 {
   class=MaterialExpressionDivide
   inputs [
    A {from=MaterialExpressionMultiply_31}
    B {from=""}
   ]
   properties {constB=1}
  }
  MaterialExpressionDivide_4 {
   class=MaterialExpressionDivide
   inputs [
    A {from=MaterialExpressionComponentMask_11}
    B {from=MaterialExpressionComponentMask_12}
   ]
  }
  MaterialExpressionDivide_5 {
   class=MaterialExpressionDivide
   inputs [
    A {from=MaterialExpressionComponentMask_13}
    B {from=""}
   ]
   properties {constB=30}
  }
  MaterialExpressionDivide_6 {
   class=MaterialExpressionDivide
   inputs [
    A {from=MaterialExpressionAdd_5}
    B {from=""}
   ]
  }
  MaterialExpressionDivide_7 {
   class=MaterialExpressionDivide
   inputs [
    A {from=MaterialExpressionSubtract_3}
    B {from=MaterialExpressionNamedRerouteDeclaration_48}
   ]
  }
  MaterialExpressionDivide_8 {
   class=MaterialExpressionDivide
   inputs [
    A {from=MaterialExpressionNamedRerouteDeclaration_37}
    B {from=MaterialExpressionNamedRerouteUsage_61}
   ]
  }
  MaterialExpressionDotProduct_0 {
   class=MaterialExpressionDotProduct
   inputs [
    A {from=MaterialExpressionAtmosphericLightVector_0, output=""}
    B {from=MaterialExpressionMaterialFunctionCall_12, output=Result}
   ]
   properties {}
  }
  MaterialExpressionDotProduct_1 {
   class=MaterialExpressionDotProduct
   inputs [
    A {from=MaterialExpressionAtmosphericLightVector_1, output=""}
    B {from=MaterialExpressionPixelNormalWS_0, output=""}
   ]
   properties {}
  }
  MaterialExpressionDotProduct_2 {
   class=MaterialExpressionDotProduct
   inputs [
    A {from=MaterialExpressionTextureSample_8, output=RGB}
    B {from=MaterialExpressionTextureSample_10, output=RGB}
   ]
   properties {}
  }
  MaterialExpressionDotProduct_3 {
   class=MaterialExpressionDotProduct
   inputs [
    A {from=MaterialExpressionNormalize_9, output=""}
    B {from=MaterialExpressionMultiply_111, output=""}
   ]
   properties {}
  }
  MaterialExpressionDotProduct_4 {
   class=MaterialExpressionDotProduct
   inputs [
    A {from=MaterialExpressionCameraVectorWS_1, output=""}
    B {from=MaterialExpressionVertexNormalWS_3, output=""}
   ]
   properties {}
  }
  MaterialExpressionDotProduct_5 {
   class=MaterialExpressionDotProduct
   inputs [
    A {from=MaterialExpressionVertexNormalWS_4, output=""}
    B {from=MaterialExpressionTransform_3, output=""}
   ]
   properties {}
  }
  MaterialExpressionFrac_0 {class=MaterialExpressionFrac, inputs [None {from=MaterialExpressionMultiply_39, output=""}], properties {}}
  MaterialExpressionFrac_1 {class=MaterialExpressionFrac, inputs [None {from=MaterialExpressionMultiply_53, output=""}], properties {}}
  MaterialExpressionFrac_2 {class=MaterialExpressionFrac, inputs [None {from=MaterialExpressionMultiply_64, output=""}], properties {}}
  MaterialExpressionFunctionInput_0 {class=MaterialExpressionFunctionInput, inputs [Preview {from="", output=""}], properties {inputName=FXMask}}
  MaterialExpressionFunctionInput_1 {
   class=MaterialExpressionFunctionInput
   inputs [Preview {from="", output=""}]
   properties {inputName=PBR_BaseInputs, inputType=FunctionInput_MaterialAttributes}
  }
  MaterialExpressionFunctionOutput_0 {
   class=MaterialExpressionFunctionOutput
   inputs [None {from=MaterialExpressionNamedRerouteDeclaration_10, output=""}]
   properties {outputName=Mask_InnerRadialGlow, sortPriority=10}
  }
  MaterialExpressionFunctionOutput_1 {
   class=MaterialExpressionFunctionOutput
   inputs [None {from=MaterialExpressionBlendMaterialAttributes_0, output=""}]
   properties {outputName=MA, sortPriority=-1}
  }
  MaterialExpressionFunctionOutput_10 {
   class=MaterialExpressionFunctionOutput
   inputs [None {from=MaterialExpressionNamedRerouteUsage_44, output=""}]
   properties {outputName=Pre-Skinned Local Scale, sortPriority=140}
  }
  MaterialExpressionFunctionOutput_11 {
   class=MaterialExpressionFunctionOutput
   inputs [None {from=MaterialExpressionNamedRerouteUsage_45, output=""}]
   properties {outputName=Pre-Skinned Local Bounds Center, sortPriority=130}
  }
  MaterialExpressionFunctionOutput_2 {
   class=MaterialExpressionFunctionOutput
   inputs [None {from=MaterialExpressionStaticSwitchParameter_0, output=""}]
   properties {outputName=Sprite WPO}
  }
  MaterialExpressionFunctionOutput_3 {
   class=MaterialExpressionFunctionOutput
   inputs [None {from=MaterialExpressionMultiply_29, output=""}]
   properties {outputName=Mask_Logo, sortPriority=1}
  }
  MaterialExpressionFunctionOutput_4 {
   class=MaterialExpressionFunctionOutput
   inputs [None {from=MaterialExpressionNamedRerouteDeclaration_9, output=""}]
   properties {outputName=Mask_SpriteFresnel, sortPriority=10}
  }
  MaterialExpressionFunctionOutput_5 {
   class=MaterialExpressionFunctionOutput
   inputs [None {from=MaterialExpressionNamedRerouteDeclaration_2, output=""}]
   properties {outputName=ScreenUVs}
  }
  MaterialExpressionFunctionOutput_6 {
   class=MaterialExpressionFunctionOutput
   inputs [None {from=MaterialExpressionNamedRerouteUsage_34, output=""}]
   properties {outputName=Pre-Skinned Local Pos, sortPriority=100}
  }
  MaterialExpressionFunctionOutput_7 {
   class=MaterialExpressionFunctionOutput
   inputs [None {from=MaterialExpressionSubtract_4, output=""}]
   properties {outputName=Pre-Skinned Local Bounds Min, sortPriority=110}
  }
  MaterialExpressionFunctionOutput_8 {
   class=MaterialExpressionFunctionOutput
   inputs [None {from=MaterialExpressionAdd_0, output=""}]
   properties {outputName=Pre-Skinned Local Bounds Max, sortPriority=120}
  }
  MaterialExpressionFunctionOutput_9 {
   class=MaterialExpressionFunctionOutput
   inputs [None {from=MaterialExpressionNamedRerouteUsage_39, output=""}]
   properties {outputName=Object Position WS, sortPriority=150}
  }
  MaterialExpressionIf_1 {
   class=MaterialExpressionIf
   inputs [
    A {from=MaterialExpressionComponentMask_14, output=""}
    B {from="", output=""}
    "A > B" {from=MaterialExpressionConstant_7, output=""}
    "A == B" {from="", output=""}
    "A < B" {from=MaterialExpressionConstant_9, output=""}
   ]
   properties {constB=30}
  }
  MaterialExpressionLinearInterpolate_0 {
   class=MaterialExpressionLinearInterpolate
   inputs [
    A {from=MaterialExpressionLinearInterpolate_1}
    B {from=MaterialExpressionMaterialFunctionCall_3, output=Result}
    Alpha {from=MaterialExpressionNamedRerouteUsage_24}
   ]
  }
  MaterialExpressionLinearInterpolate_1 {
   class=MaterialExpressionLinearInterpolate
   inputs [
    A {from=MaterialExpressionVectorParameter_2, output=RGB}
    B {from=MaterialExpressionLinearInterpolate_2}
    Alpha {from=MaterialExpressionSaturate_0}
   ]
  }
  MaterialExpressionLinearInterpolate_10 {
   class=MaterialExpressionLinearInterpolate
   inputs [
    A {from=MaterialExpressionMultiply_38}
    B {from=MaterialExpressionMultiply_28}
    Alpha {from=MaterialExpressionReroute_5}
   ]
  }
  MaterialExpressionLinearInterpolate_11 {
   class=MaterialExpressionLinearInterpolate
   inputs [
    A {from=MaterialExpressionNamedRerouteUsage_91}
    B {from=""}
    Alpha {from=MaterialExpressionNamedRerouteUsage_33}
   ]
  }
  MaterialExpressionLinearInterpolate_12 {
   class=MaterialExpressionLinearInterpolate
   inputs [
    A {from=MaterialExpressionLinearInterpolate_15}
    B {from=MaterialExpressionAdd_3}
    Alpha {from=MaterialExpressionVectorParameter_3, output="0="}
   ]
  }
  MaterialExpressionLinearInterpolate_14 {
   class=MaterialExpressionLinearInterpolate
   inputs [
    A {from=MaterialExpressionReroute_25}
    B {from=MaterialExpressionMultiply_37}
    Alpha {from=MaterialExpressionNamedRerouteUsage_32}
   ]
  }
  MaterialExpressionLinearInterpolate_15 {
   class=MaterialExpressionLinearInterpolate
   inputs [
    A {from=MaterialExpressionReroute_2}
    B {from=MaterialExpressionMultiply_2}
    Alpha {from=MaterialExpressionNamedRerouteUsage_25}
   ]
  }
  MaterialExpressionLinearInterpolate_16 {
   class=MaterialExpressionLinearInterpolate
   inputs [
    A {from=MaterialExpressionMultiply_17}
    B {from=MaterialExpressionMultiply_12}
    Alpha {from=MaterialExpressionVectorParameter_18, output=A}
   ]
  }
  MaterialExpressionLinearInterpolate_17 {
   class=MaterialExpressionLinearInterpolate
   inputs [
    A {from=MaterialExpressionNamedRerouteUsage_91}
    B {from=MaterialExpressionLinearInterpolate_11}
    Alpha {from=MaterialExpressionNamedRerouteUsage_29}
   ]
  }
  MaterialExpressionLinearInterpolate_19 {
   class=MaterialExpressionLinearInterpolate
   inputs [
    A {from=MaterialExpressionReroute_12}
    B {from=MaterialExpressionMultiply_56}
    Alpha {from=MaterialExpressionMultiply_58}
   ]
  }
  MaterialExpressionLinearInterpolate_2 {
   class=MaterialExpressionLinearInterpolate
   inputs [
    A {from=MaterialExpressionLinearInterpolate_8}
    B {from=MaterialExpressionLinearInterpolate_9}
    Alpha {from=MaterialExpressionStep_1}
   ]
  }
  MaterialExpressionLinearInterpolate_20 {
   class=MaterialExpressionLinearInterpolate
   inputs [
    A {from=MaterialExpressionDotProduct_2}
    B {from=""}
    Alpha {from=MaterialExpressionNamedRerouteUsage_50}
   ]
   properties {constB=-1}
  }
  MaterialExpressionLinearInterpolate_3 {
   class=MaterialExpressionLinearInterpolate
   inputs [
    A {from=""}
    B {from=MaterialExpressionMultiply_5}
    Alpha {from=MaterialExpressionNamedRerouteUsage_22}
   ]
  }
  MaterialExpressionLinearInterpolate_4 {
   class=MaterialExpressionLinearInterpolate
   inputs [
    A {from=MaterialExpressionAdd_7}
    B {from=""}
    Alpha {from=MaterialExpressionNamedRerouteUsage_20}
   ]
   properties {constB=0}
  }
  MaterialExpressionLinearInterpolate_5 {
   class=MaterialExpressionLinearInterpolate
   inputs [
    A {from=MaterialExpressionLinearInterpolate_0}
    B {from=MaterialExpressionVectorParameter_21, output=RGB}
    Alpha {from=MaterialExpressionNamedRerouteUsage_2}
   ]
  }
  MaterialExpressionLinearInterpolate_6 {
   class=MaterialExpressionLinearInterpolate
   inputs [
    A {from=MaterialExpressionOneMinus_5}
    B {from=""}
    Alpha {from=""}
   ]
   properties {constAlpha=0.2}
  }
  MaterialExpressionLinearInterpolate_7 {
   class=MaterialExpressionLinearInterpolate
   inputs [
    A {from=MaterialExpressionScalarParameter_51}
    B {from=MaterialExpressionScalarParameter_50}
    Alpha {from=MaterialExpressionStep_1}
   ]
  }
  MaterialExpressionLinearInterpolate_8 {
   class=MaterialExpressionLinearInterpolate
   inputs [
    A {from=MaterialExpressionVectorParameter_4, output=RGB}
    B {from=MaterialExpressionMultiply_19}
    Alpha {from=MaterialExpressionNamedRerouteUsage_16}
   ]
  }
  MaterialExpressionLinearInterpolate_9 {
   class=MaterialExpressionLinearInterpolate
   inputs [
    A {from=MaterialExpressionVectorParameter_5, output=RGB}
    B {from=MaterialExpressionMultiply_20}
    Alpha {from=MaterialExpressionNamedRerouteUsage_17}
   ]
  }
  MaterialExpressionLocalPosition_1 {class=MaterialExpressionLocalPosition, inputs [], properties {includedOffsets=ExcludeOffsets, localOrigin=InstancePreSkinning}}
  MaterialExpressionMaterialFunctionCall_0 {
   class=MaterialExpressionMaterialFunctionCall
   inputs [
    X {from=MaterialExpressionConstant_5}
    Y {from=MaterialExpressionConstant_5}
    Z {from=MaterialExpressionMultiply_22}
   ]
   properties {materialFunction=&MakeFloat3}
  }
  MaterialExpressionMaterialFunctionCall_1 {
   class=MaterialExpressionMaterialFunctionCall
   inputs []
   properties {materialFunction=&ObjectScale}
  }
  MaterialExpressionMaterialFunctionCall_10 {
   class=MaterialExpressionMaterialFunctionCall
   inputs []
   properties {materialFunction=&ObjectScale}
  }
  MaterialExpressionMaterialFunctionCall_11 {
   class=MaterialExpressionMaterialFunctionCall
   inputs [
    Value_Morning {from=MaterialExpressionConstant_12}
    Value_Day {from=MaterialExpressionConstant_12}
    Value_Evening {from=MaterialExpressionConstant_13}
    Value_Night {from=MaterialExpressionConstant_16}
   ]
   properties {materialFunction=&Scalar_Time_of_Day}
  }
  MaterialExpressionMaterialFunctionCall_12 {
   class=MaterialExpressionMaterialFunctionCall
   inputs []
   properties {materialFunction=&CameraDirectionVector}
  }
  MaterialExpressionMaterialFunctionCall_14 {
   class=MaterialExpressionMaterialFunctionCall
   inputs []
   properties {materialFunction=/Engine/Functions/Engine_MaterialFunctions02/Utility/FOV}
  }
  MaterialExpressionMaterialFunctionCall_16 {
   class=MaterialExpressionMaterialFunctionCall
   inputs [
    R {from=MaterialExpressionSine_0}
    G {from=MaterialExpressionSine_0}
    B {from=MaterialExpressionCosine_0}
    A {from=""}
   ]
   properties {materialFunction=&AppendMany}
  }
  MaterialExpressionMaterialFunctionCall_17 {
   class=MaterialExpressionMaterialFunctionCall
   inputs [
    UVs {from=MaterialExpressionTextureCoordinate_1}
    Texture Scale {from=MaterialExpressionScalarParameter_25}
   ]
   properties {materialFunction=&ScaleUVsByCenter}
  }
  MaterialExpressionMaterialFunctionCall_2 {
   class=MaterialExpressionMaterialFunctionCall
   inputs [
    Color_Morning {from=MaterialExpressionVectorParameter_7, output=RGB}
    Color_Day {from=MaterialExpressionVectorParameter_11, output=RGB}
    Color_Evening {from=MaterialExpressionVectorParameter_12, output=RGB}
    Color_Night {from=MaterialExpressionVectorParameter_13, output=RGB}
   ]
   properties {materialFunction=&Color_Time_of_Day}
  }
  MaterialExpressionMaterialFunctionCall_3 {
   class=MaterialExpressionMaterialFunctionCall
   inputs [
    Color_Morning {from=MaterialExpressionVectorParameter_17, output=RGB}
    Color_Day {from=MaterialExpressionVectorParameter_0, output=RGB}
    Color_Evening {from=MaterialExpressionVectorParameter_15, output=RGB}
    Color_Night {from=MaterialExpressionVectorParameter_16, output=RGB}
   ]
   properties {materialFunction=&Color_Time_of_Day}
  }
  MaterialExpressionMaterialFunctionCall_30 {
   class=MaterialExpressionMaterialFunctionCall
   inputs [Float3 {from=MaterialExpressionMultiply_62}]
   properties {materialFunction=&BreakOutFloat3Components}
  }
  MaterialExpressionMaterialFunctionCall_32 {
   class=MaterialExpressionMaterialFunctionCall
   inputs []
   properties {materialFunction=&CameraDirectionVector}
  }
  MaterialExpressionMaterialFunctionCall_4 {
   class=MaterialExpressionMaterialFunctionCall
   inputs [
    Alpha {from=MaterialExpressionPower_4}
    Min {from=""}
    Max {from=""}
   ]
   properties {materialFunction=&SmoothStep}
  }
  MaterialExpressionMaterialFunctionCall_56 {
   class=MaterialExpressionMaterialFunctionCall
   inputs [Float2 {from=MaterialExpressionViewSize_3}]
   properties {materialFunction=&BreakOutFloat2Components}
  }
  MaterialExpressionMaterialFunctionCall_57 {
   class=MaterialExpressionMaterialFunctionCall
   inputs [
    World Position {from=MaterialExpressionNamedRerouteUsage_96}
    Buffer UV Pixel Offset {from=""}
   ]
   properties {materialFunction=&TransformToClipSpace}
  }
  MaterialExpressionMaterialFunctionCall_58 {
   class=MaterialExpressionMaterialFunctionCall
   inputs [
    UVs {from=MaterialExpressionSubtract_10}
    Texture Scale {from=MaterialExpressionDivide_10}
   ]
   properties {materialFunction=&ScaleUVsByCenter}
  }
  MaterialExpressionMaterialFunctionCall_9 {
   class=MaterialExpressionMaterialFunctionCall
   inputs [
    Normal Vector {from=MaterialExpressionAdd_4}
    Camera Vector {from=""}
    Invert Fresnel {from=MaterialExpressionStaticBool_0}
    Power {from=MaterialExpressionScalarParameter_5}
    Use Cheap contrast {from=""}
    Cheap contrast dark {from=""}
    Cheap contrast bright {from=""}
    Clamp Fresnel Dot Product {from=""}
   ]
   properties {materialFunction=&Fresnel_Function}
  }
  MaterialExpressionMax_4 {
   class=MaterialExpressionMax
   inputs [
    A {from=MaterialExpressionDotProduct_3}
    B {from=""}
   ]
   properties {constB=0}
  }
  MaterialExpressionMultiply_0 {
   class=MaterialExpressionMultiply
   inputs [
    A {from=MaterialExpressionLinearInterpolate_6}
    B {from=MaterialExpressionReroute_21}
   ]
   properties {constA=1}
  }
  MaterialExpressionMultiply_1 {
   class=MaterialExpressionMultiply
   inputs [
    A {from=MaterialExpressionNamedRerouteUsage_90}
    B {from=MaterialExpressionScalarParameter_9}
   ]
   properties {constB=0.02}
  }
  MaterialExpressionMultiply_10 {
   class=MaterialExpressionMultiply
   inputs [
    A {from=MaterialExpressionConstantBiasScale_3}
    B {from=MaterialExpressionMultiply_25}
   ]
  }
  MaterialExpressionMultiply_109 {
   class=MaterialExpressionMultiply
   inputs [
    A {from=MaterialExpressionDotProduct_1}
    B {from=""}
   ]
   properties {constB=2}
  }
  MaterialExpressionMultiply_11 {
   class=MaterialExpressionMultiply
   inputs [
    A {from=MaterialExpressionScalarParameter_65}
    B {from=MaterialExpressionMultiply_13}
   ]
  }
  MaterialExpressionMultiply_110 {
   class=MaterialExpressionMultiply
   inputs [
    A {from=MaterialExpressionMultiply_109}
    B {from=MaterialExpressionPixelNormalWS_0}
   ]
  }
  MaterialExpressionMultiply_111 {
   class=MaterialExpressionMultiply
   inputs [
    A {from=MaterialExpressionMaterialFunctionCall_32, output=Result}
    B {from=""}
   ]
   properties {constB=-1}
  }
  MaterialExpressionMultiply_112 {
   class=MaterialExpressionMultiply
   inputs [
    A {from=MaterialExpressionPower_9}
    B {from=MaterialExpressionScalarParameter_76}
   ]
   properties {constB=0.4}
  }
  MaterialExpressionMultiply_12 {
   class=MaterialExpressionMultiply
   inputs [
    A {from=MaterialExpressionPower_12}
    B {from=MaterialExpressionVectorParameter_18, output=RGB}
   ]
  }
  MaterialExpressionMultiply_13 {
   class=MaterialExpressionMultiply
   inputs [
    A {from=MaterialExpressionMaterialFunctionCall_10, output=Scale X}
    B {from=MaterialExpressionComponentMask_16}
   ]
  }
  MaterialExpressionMultiply_14 {
   class=MaterialExpressionMultiply
   inputs [
    A {from=MaterialExpressionMaterialFunctionCall_58, output=UVs}
    B {from=MaterialExpressionConstant2Vector_6}
   ]
  }
  MaterialExpressionMultiply_15 {
   class=MaterialExpressionMultiply
   inputs [
    A {from=MaterialExpressionScalarParameter_10}
    B {from=MaterialExpressionNamedRerouteUsage_31}
   ]
  }
  MaterialExpressionMultiply_16 {
   class=MaterialExpressionMultiply
   inputs [
    A {from=MaterialExpressionNamedRerouteUsage_13}
    B {from=""}
   ]
   properties {constB=0.5}
  }
  MaterialExpressionMultiply_17 {
   class=MaterialExpressionMultiply
   inputs [
    A {from=MaterialExpressionReroute_15}
    B {from=MaterialExpressionMultiply_12}
   ]
  }
  MaterialExpressionMultiply_18 {
   class=MaterialExpressionMultiply
   inputs [
    A {from=MaterialExpressionScalarParameter_17}
    B {from=MaterialExpressionVectorParameter_22, output=RGB}
   ]
  }
  MaterialExpressionMultiply_19 {
   class=MaterialExpressionMultiply
   inputs [
    A {from=MaterialExpressionVectorParameter_1, output=RGB}
    B {from=MaterialExpressionNamedRerouteUsage_26}
   ]
  }
  MaterialExpressionMultiply_2 {
   class=MaterialExpressionMultiply
   inputs [
    A {from=MaterialExpressionNamedRerouteUsage_25}
    B {from=MaterialExpressionVectorParameter_3, output=RGB}
   ]
  }
  MaterialExpressionMultiply_20 {
   class=MaterialExpressionMultiply
   inputs [
    A {from=MaterialExpressionVectorParameter_8, output=RGB}
    B {from=MaterialExpressionNamedRerouteUsage_27}
   ]
  }
  MaterialExpressionMultiply_21 {
   class=MaterialExpressionMultiply
   inputs [
    A {from=MaterialExpressionScalarParameter_27}
    B {from=MaterialExpressionVectorParameter_25, output=RGB}
   ]
  }
  MaterialExpressionMultiply_22 {
   class=MaterialExpressionMultiply
   inputs [
    A {from=MaterialExpressionMultiply_10}
    B {from=MaterialExpressionNamedRerouteUsage_36}
   ]
  }
  MaterialExpressionMultiply_23 {
   class=MaterialExpressionMultiply
   inputs [
    A {from=MaterialExpressionNamedRerouteUsage_88}
    B {from=""}
   ]
   properties {constB=0.05}
  }
  MaterialExpressionMultiply_24 {
   class=MaterialExpressionMultiply
   inputs [
    A {from=MaterialExpressionSmoothStep_0}
    B {from=MaterialExpressionPower_2}
   ]
  }
  MaterialExpressionMultiply_25 {
   class=MaterialExpressionMultiply
   inputs [
    A {from=MaterialExpressionMultiply_9}
    B {from=MaterialExpressionMultiply_15}
   ]
  }
  MaterialExpressionMultiply_26 {
   class=MaterialExpressionMultiply
   inputs [
    A {from=MaterialExpressionAbs_2}
    B {from=""}
   ]
   properties {constB=2}
  }
  MaterialExpressionMultiply_27 {
   class=MaterialExpressionMultiply
   inputs [
    A {from=MaterialExpressionNamedRerouteUsage_41}
    B {from=MaterialExpressionPower_1}
   ]
  }
  MaterialExpressionMultiply_28 {
   class=MaterialExpressionMultiply
   inputs [
    A {from=MaterialExpressionReroute_11}
    B {from=MaterialExpressionVectorParameter_10, output=RGB}
   ]
   properties {constB=2}
  }
  MaterialExpressionMultiply_29 {
   class=MaterialExpressionMultiply
   inputs [
    A {from=MaterialExpressionSaturate_1}
    B {from=MaterialExpressionReroute_0}
   ]
  }
  MaterialExpressionMultiply_3 {
   class=MaterialExpressionMultiply
   inputs [
    A {from=MaterialExpressionReroute_20}
    B {from=MaterialExpressionStaticSwitchParameter_4}
   ]
  }
  MaterialExpressionMultiply_30 {
   class=MaterialExpressionMultiply
   inputs [
    A {from=MaterialExpressionAdd_16}
    B {from=""}
   ]
   properties {constB=0.5}
  }
  MaterialExpressionMultiply_31 {
   class=MaterialExpressionMultiply
   inputs [
    A {from=MaterialExpressionMultiply_52}
    B {from=MaterialExpressionScalarParameter_46}
   ]
  }
  MaterialExpressionMultiply_32 {
   class=MaterialExpressionMultiply
   inputs [
    A {from=MaterialExpressionAbs_1}
    B {from=""}
   ]
   properties {constB=2}
  }
  MaterialExpressionMultiply_33 {
   class=MaterialExpressionMultiply
   inputs [
    A {from=MaterialExpressionTime_0}
    B {from=""}
   ]
   properties {constB=0.2}
  }
  MaterialExpressionMultiply_34 {
   class=MaterialExpressionMultiply
   inputs [
    A {from=MaterialExpressionSubtract_0}
    B {from=MaterialExpressionConstant4Vector_0, output=RGB}
   ]
  }
  MaterialExpressionMultiply_35 {
   class=MaterialExpressionMultiply
   inputs [
    A {from=MaterialExpressionNamedRerouteUsage_35}
    B {from=""}
   ]
   properties {constB=0.5}
  }
  MaterialExpressionMultiply_36 {
   class=MaterialExpressionMultiply
   inputs [
    A {from=MaterialExpressionSine_1}
    B {from=""}
   ]
   properties {constB=0.5}
  }
  MaterialExpressionMultiply_37 {
   class=MaterialExpressionMultiply
   inputs [
    A {from=MaterialExpressionMaterialFunctionCall_16, output=RGB}
    B {from=MaterialExpressionScalarParameter_44}
   ]
   properties {constB=0.25}
  }
  MaterialExpressionMultiply_38 {
   class=MaterialExpressionMultiply
   inputs [
    A {from=MaterialExpressionReroute_11}
    B {from=MaterialExpressionOneMinus_4}
   ]
  }
  MaterialExpressionMultiply_39 {
   class=MaterialExpressionMultiply
   inputs [
    A {from=MaterialExpressionMultiply_46}
    B {from=MaterialExpressionConstant2Vector_2}
   ]
  }
  MaterialExpressionMultiply_4 {
   class=MaterialExpressionMultiply
   inputs [
    A {from=MaterialExpressionConstantBiasScale_0}
    B {from=MaterialExpressionScalarParameter_2}
   ]
  }
  MaterialExpressionMultiply_40 {
   class=MaterialExpressionMultiply
   inputs [
    A {from=MaterialExpressionScalarParameter_13}
    B {from=MaterialExpressionComponentMask_9}
   ]
  }
  MaterialExpressionMultiply_46 {
   class=MaterialExpressionMultiply
   inputs [
    A {from=MaterialExpressionTime_3}
    B {from=MaterialExpressionMultiply_49}
   ]
  }
  MaterialExpressionMultiply_49 {
   class=MaterialExpressionMultiply
   inputs [
    A {from=MaterialExpressionScalarParameter_47}
    B {from=""}
   ]
   properties {constB=-1}
  }
  MaterialExpressionMultiply_5 {
   class=MaterialExpressionMultiply
   inputs [
    A {from=MaterialExpressionNamedRerouteUsage_15}
    B {from=MaterialExpressionVectorParameter_6, output=RGB}
   ]
  }
  MaterialExpressionMultiply_52 {
   class=MaterialExpressionMultiply
   inputs [
    A {from=MaterialExpressionNamedRerouteUsage_47}
    B {from=MaterialExpressionConstant2Vector_3}
   ]
  }
  MaterialExpressionMultiply_53 {
   class=MaterialExpressionMultiply
   inputs [
    A {from=MaterialExpressionMultiply_39}
    B {from=""}
   ]
   properties {constB=0.2}
  }
  MaterialExpressionMultiply_55 {
   class=MaterialExpressionMultiply
   inputs [
    A {from=MaterialExpressionPower_3}
    B {from=MaterialExpressionScalarParameter_48}
   ]
  }
  MaterialExpressionMultiply_56 {
   class=MaterialExpressionMultiply
   inputs [
    A {from=MaterialExpressionReroute_12}
    B {from=MaterialExpressionReroute_10}
   ]
  }
  MaterialExpressionMultiply_58 {
   class=MaterialExpressionMultiply
   inputs [
    A {from=MaterialExpressionPower_3}
    B {from=MaterialExpressionPower_5}
   ]
  }
  MaterialExpressionMultiply_6 {
   class=MaterialExpressionMultiply
   inputs [
    A {from=MaterialExpressionMultiply_14}
    B {from=MaterialExpressionMaterialFunctionCall_14, output="FOV "}
   ]
  }
  MaterialExpressionMultiply_62 {
   class=MaterialExpressionMultiply
   inputs [
    A {from=MaterialExpressionStaticSwitchParameter_5}
    B {from=MaterialExpressionScalarParameter_0}
   ]
   properties {constB=0.02}
  }
  MaterialExpressionMultiply_63 {
   class=MaterialExpressionMultiply
   inputs [
    A {from=MaterialExpressionStaticSwitchParameter_1}
    B {from=MaterialExpressionNamedRerouteUsage_38}
   ]
  }
  MaterialExpressionMultiply_64 {
   class=MaterialExpressionMultiply
   inputs [
    A {from=MaterialExpressionSubtract_6}
    B {from=""}
   ]
   properties {constB=0.75}
  }
  MaterialExpressionMultiply_7 {
   class=MaterialExpressionMultiply
   inputs [
    A {from=MaterialExpressionSmoothStep_1}
    B {from=MaterialExpressionOneMinus_9}
   ]
  }
  MaterialExpressionMultiply_84 {
   class=MaterialExpressionMultiply
   inputs [
    A {from=MaterialExpressionConstantBiasScale_2}
    B {from=MaterialExpressionAppendVector_7}
   ]
  }
  MaterialExpressionMultiply_85 {
   class=MaterialExpressionMultiply
   inputs [
    A {from=MaterialExpressionConstant2Vector_11}
    B {from=MaterialExpressionMaterialFunctionCall_57, output=Clip Space XY}
   ]
   properties {constA=1}
  }
  MaterialExpressionMultiply_86 {
   class=MaterialExpressionMultiply
   inputs [
    A {from=MaterialExpressionAppendVector_7}
    B {from=MaterialExpressionMultiply_85}
   ]
  }
  MaterialExpressionMultiply_9 {
   class=MaterialExpressionMultiply
   inputs [
    A {from=MaterialExpressionMaterialFunctionCall_1, output=Scale Z}
    B {from=MaterialExpressionComponentMask_17}
   ]
  }
  MaterialExpressionMultiply_92 {
   class=MaterialExpressionMultiply
   inputs [
    A {from=MaterialExpressionMultiply_63}
    B {from=MaterialExpressionScalarParameter_55}
   ]
  }
  MaterialExpressionNamedRerouteDeclaration_0 {
   class=MaterialExpressionNamedRerouteDeclaration
   inputs [None {from=MaterialExpressionMultiply_30}]
   properties {name="Pre-Skinned Bounds Center Override (V3)", nodeColor=<rgba 1 0 0.2578 1>}
  }
  MaterialExpressionNamedRerouteDeclaration_1 {
   class=MaterialExpressionNamedRerouteDeclaration
   inputs [None {from=MaterialExpressionLinearInterpolate_3}]
   properties {name=OUT_SSS, nodeColor=<rgba 0 0.4141 1 1>}
  }
  MaterialExpressionNamedRerouteDeclaration_10 {
   class=MaterialExpressionNamedRerouteDeclaration
   inputs [None {from=MaterialExpressionPower_6}]
   properties {name=M_InnerGlow, nodeColor=<rgba 0.1719 0 1 1>}
  }
  MaterialExpressionNamedRerouteDeclaration_11 {
   class=MaterialExpressionNamedRerouteDeclaration
   inputs [None {from=MaterialExpressionSaturate_4}]
   properties {name=M_SunMask, nodeColor=<rgba 0.8047 0 1 1>}
  }
  MaterialExpressionNamedRerouteDeclaration_12 {
   class=MaterialExpressionNamedRerouteDeclaration
   inputs [None {from=MaterialExpressionMaterialFunctionCall_0, output=Result}]
   properties {name=CapWPO, nodeColor=<rgba 0.6406 1 0 1>}
  }
  MaterialExpressionNamedRerouteDeclaration_13 {
   class=MaterialExpressionNamedRerouteDeclaration
   inputs [None {from=MaterialExpressionNamedRerouteDeclaration_15}]
   properties {name=OUT_Normal, nodeColor=<rgba 0.8984 1 0 1>}
  }
  MaterialExpressionNamedRerouteDeclaration_14 {
   class=MaterialExpressionNamedRerouteDeclaration
   inputs [None {from=MaterialExpressionSaturate_5}]
   properties {name=FacialMask, nodeColor=<rgba 1 0 0.2109 1>}
  }
  MaterialExpressionNamedRerouteDeclaration_15 {
   class=MaterialExpressionNamedRerouteDeclaration
   inputs [None {from=MaterialExpressionTransform_0}]
   properties {name=FixedNormal, nodeColor=<rgba 1 0 0.8203 1>}
  }
  MaterialExpressionNamedRerouteDeclaration_16 {
   class=MaterialExpressionNamedRerouteDeclaration
   inputs [None {from=MaterialExpressionAppendVector_6}]
   properties {name=M_Logo, nodeColor=<rgba 0 1 0.6953 1>}
  }
  MaterialExpressionNamedRerouteDeclaration_17 {
   class=MaterialExpressionNamedRerouteDeclaration
   inputs [None {from=MaterialExpressionMaterialFunctionCall_11, output=Result}]
   properties {name=TOD_ScatteredColorMult, nodeColor=<rgba 1 0.375 0 1>}
  }
  MaterialExpressionNamedRerouteDeclaration_18 {
   class=MaterialExpressionNamedRerouteDeclaration
   inputs [None {from=MaterialExpressionComponentMask_0}]
   properties {name=M_PaintMask_R, nodeColor=<rgba 1 0 0 1>}
  }
  MaterialExpressionNamedRerouteDeclaration_19 {
   class=MaterialExpressionNamedRerouteDeclaration
   inputs [None {from=MaterialExpressionComponentMask_6}]
   properties {name=M_PaintMask_B, nodeColor=<rgba 0 0 1 1>}
  }
  MaterialExpressionNamedRerouteDeclaration_2 {
   class=MaterialExpressionNamedRerouteDeclaration
   inputs [None {from=MaterialExpressionMultiply_6}]
   properties {name=ScreenUV, nodeColor=<rgba 1 0 0.1172 1>}
  }
  MaterialExpressionNamedRerouteDeclaration_20 {
   class=MaterialExpressionNamedRerouteDeclaration
   inputs [None {from=MaterialExpressionMultiply_24}]
   properties {name=M_HeadCap, nodeColor=<rgba 0 1 0.6484 1>}
  }
  MaterialExpressionNamedRerouteDeclaration_21 {
   class=MaterialExpressionNamedRerouteDeclaration
   inputs [None {from=MaterialExpressionMultiply_36}]
   properties {name=Sine_localWaves, nodeColor=<rgba 1 0.3281 0 1>}
  }
  MaterialExpressionNamedRerouteDeclaration_22 {
   class=MaterialExpressionNamedRerouteDeclaration
   inputs [None {from=MaterialExpressionTextureObject_2}]
   properties {name=T_WPONoise, nodeColor=<rgba 0.8516 0 1 1>}
  }
  MaterialExpressionNamedRerouteDeclaration_23 {
   class=MaterialExpressionNamedRerouteDeclaration
   inputs [None {from=MaterialExpressionDivide_8}]
   properties {name="Pre-Skinned Scale Override (V3)", nodeColor=<rgba 1 0.5391 0 1>}
  }
  MaterialExpressionNamedRerouteDeclaration_24 {
   class=MaterialExpressionNamedRerouteDeclaration
   inputs [None {from=MaterialExpressionTextureObjectParameter_0}]
   properties {name=T_SpriteLogo, nodeColor=<rgba 1 0.1172 0 1>}
  }
  MaterialExpressionNamedRerouteDeclaration_25 {class=MaterialExpressionNamedRerouteDeclaration, inputs [None {from=""}], properties {name=Optimization Reminders, nodeColor=<rgba 1 0 0 1>}}
  MaterialExpressionNamedRerouteDeclaration_26 {
   class=MaterialExpressionNamedRerouteDeclaration
   inputs [None {from=MaterialExpressionTextureObjectParameter_2}]
   properties {name=T_InteriorNoise, nodeColor=<rgba 0 0.6016 1 1>}
  }
  MaterialExpressionNamedRerouteDeclaration_27 {
   class=MaterialExpressionNamedRerouteDeclaration
   inputs [None {from=MaterialExpressionIf_1}]
   properties {name="Is Backpack Sprite (S)", nodeColor=<rgba 0.3594 1 0 1>}
  }
  MaterialExpressionNamedRerouteDeclaration_28 {
   class=MaterialExpressionNamedRerouteDeclaration
   inputs [None {from=MaterialExpressionVectorParameter_19, output=RGB}]
   properties {name="Sprite Pre-Skinned Bounds Center (V3)", nodeColor=<rgba 1 1 1 1>}
  }
  MaterialExpressionNamedRerouteDeclaration_29 {
   class=MaterialExpressionNamedRerouteDeclaration
   inputs [None {from=MaterialExpressionScalarParameter_28}]
   properties {name="ColdTrophy - Cap WPO Amount Mult (S)", nodeColor=<rgba 0 1 0.4609 1>}
  }
  MaterialExpressionNamedRerouteDeclaration_3 {
   class=MaterialExpressionNamedRerouteDeclaration
   inputs [None {from=MaterialExpressionComponentMask_3}]
   properties {name=M_PaintMask_G, nodeColor=<rgba 0 1 0 1>}
  }
  MaterialExpressionNamedRerouteDeclaration_30 {
   class=MaterialExpressionNamedRerouteDeclaration
   inputs [None {from=MaterialExpressionSwitch_1}]
   properties {name="Cap WPO Amount Mult (S)", nodeColor=<rgba 0 1 0.4609 1>}
  }
  MaterialExpressionNamedRerouteDeclaration_31 {
   class=MaterialExpressionNamedRerouteDeclaration
   inputs [None {from=MaterialExpressionSaturate_3}]
   properties {name="Enable Pre-Skinned Data Override (S)", nodeColor=<rgba 1 0 0.1875 1>}
  }
  MaterialExpressionNamedRerouteDeclaration_32 {
   class=MaterialExpressionNamedRerouteDeclaration
   inputs [None {from=MaterialExpressionScalarParameter_29}]
   properties {name="Cap WPO Amount Mult Override (S)", nodeColor=<rgba 0 1 0.4609 1>}
  }
  MaterialExpressionNamedRerouteDeclaration_33 {
   class=MaterialExpressionNamedRerouteDeclaration
   inputs [None {from=MaterialExpressionSwitch_3}]
   properties {name="Input Data Mode (S)", nodeColor=<rgba 0.8281 1 0 1>}
  }
  MaterialExpressionNamedRerouteDeclaration_34 {
   class=MaterialExpressionNamedRerouteDeclaration
   inputs [None {from=MaterialExpressionVectorParameter_26, output=RGB}]
   properties {name="Pre-Skinned Bounds Min Override (S)", nodeColor=<rgba 0.7578 0 1 1>}
  }
  MaterialExpressionNamedRerouteDeclaration_35 {
   class=MaterialExpressionNamedRerouteDeclaration
   inputs [None {from=MaterialExpressionVectorParameter_27, output=RGB}]
   properties {name="Pre-Skinned Bounds Max Override(S)", nodeColor=<rgba 0 1 0.4375 1>}
  }
  MaterialExpressionNamedRerouteDeclaration_36 {
   class=MaterialExpressionNamedRerouteDeclaration
   inputs [None {from=MaterialExpressionVectorParameter_20, output=RGB}]
   properties {name="Pilot Pre-Skinned Bounds Center (V3)", nodeColor=<rgba 1 0 0.2578 1>}
  }
  MaterialExpressionNamedRerouteDeclaration_37 {
   class=MaterialExpressionNamedRerouteDeclaration
   inputs [None {from=MaterialExpressionSubtract_7}]
   properties {name="Pre-Skinned Bounds Extents Override (V3)", nodeColor=<rgba 1 0.1172 0 1>}
  }
  MaterialExpressionNamedRerouteDeclaration_38 {
   class=MaterialExpressionNamedRerouteDeclaration
   inputs [None {from=MaterialExpressionVectorParameter_23, output=RGB}]
   properties {name="Cold Trophy Pre-Skinned Bounds Center (V3)", nodeColor=<rgba 1 0 0.2578 1>}
  }
  MaterialExpressionNamedRerouteDeclaration_39 {
   class=MaterialExpressionNamedRerouteDeclaration
   inputs [None {from=MaterialExpressionVectorParameter_24, output=RGB}]
   properties {name="Sprite Pre-Skinned Bounds Extents (V3)", nodeColor=<rgba 1 1 1 1>}
  }
  MaterialExpressionNamedRerouteDeclaration_4 {
   class=MaterialExpressionNamedRerouteDeclaration
   inputs [None {from=MaterialExpressionComponentMask_1}]
   properties {name=M_InnerGlowNoise, nodeColor=<rgba 0.3125 0 1 1>}
  }
  MaterialExpressionNamedRerouteDeclaration_42 {
   class=MaterialExpressionNamedRerouteDeclaration
   inputs [None {from=MaterialExpressionShaderStageSwitch_0}]
   properties {name="Pre-Skinned Local Pos (src) (V3)", nodeColor=<rgba 0 0.1797 1 1>}
  }
  MaterialExpressionNamedRerouteDeclaration_43 {
   class=MaterialExpressionNamedRerouteDeclaration
   inputs [None {from=MaterialExpressionSwitch_0}]
   properties {name="Pre-Skinned Bounds Center (V3)", nodeColor=<rgba 1 0 0.2578 1>}
  }
  MaterialExpressionNamedRerouteDeclaration_46 {
   class=MaterialExpressionNamedRerouteDeclaration
   inputs [None {from=MaterialExpressionMultiply_18}]
   properties {name="Pilot Pre-Skinned Scale (V3)", nodeColor=<rgba 1 0.5391 0 1>}
  }
  MaterialExpressionNamedRerouteDeclaration_47 {
   class=MaterialExpressionNamedRerouteDeclaration
   inputs [None {from=MaterialExpressionMultiply_21}]
   properties {name="Cold Trophy Pre-Skinned Scale (V3)", nodeColor=<rgba 1 0.5391 0 1>}
  }
  MaterialExpressionNamedRerouteDeclaration_48 {
   class=MaterialExpressionNamedRerouteDeclaration
   inputs [None {from=MaterialExpressionSwitch_2}]
   properties {name="Pre-Skinned Scale (V3)", nodeColor=<rgba 1 0.5391 0 1>}
  }
  MaterialExpressionNamedRerouteDeclaration_5 {
   class=MaterialExpressionNamedRerouteDeclaration
   inputs [None {from=MaterialExpressionReroute_7}]
   properties {name=OUT_Emissive, nodeColor=<rgba 0.7344 1 0 1>}
  }
  MaterialExpressionNamedRerouteDeclaration_50 {
   class=MaterialExpressionNamedRerouteDeclaration
   inputs [None {from=MaterialExpressionAdd_1}]
   properties {name="Pre-Skinned Local Pos (adjusted) (V3)", nodeColor=<rgba 0 0.1797 1 1>}
  }
  MaterialExpressionNamedRerouteDeclaration_51 {
   class=MaterialExpressionNamedRerouteDeclaration
   inputs [None {from=MaterialExpressionTransformPosition_1}]
   properties {name="Object Position WS (V3)", nodeColor=<rgba 1 0.9375 0 1>}
  }
  MaterialExpressionNamedRerouteDeclaration_6 {
   class=MaterialExpressionNamedRerouteDeclaration
   inputs [None {from=MaterialExpressionAdd_7}]
   properties {name=WPO, nodeColor=<rgba 0.9453 0 1 1>}
  }
  MaterialExpressionNamedRerouteDeclaration_7 {
   class=MaterialExpressionNamedRerouteDeclaration
   inputs [None {from=MaterialExpressionMultiply_112}]
   properties {name=M_InnerScatter, nodeColor=<rgba 1 0 0.1406 1>}
  }
  MaterialExpressionNamedRerouteDeclaration_8 {
   class=MaterialExpressionNamedRerouteDeclaration
   inputs [None {from=MaterialExpressionScalarParameter_16}]
   properties {name="Pilot - Cap WPO Amount Mult (S)", nodeColor=<rgba 0 1 0.4609 1>}
  }
  MaterialExpressionNamedRerouteDeclaration_9 {
   class=MaterialExpressionNamedRerouteDeclaration
   inputs [None {from=MaterialExpressionDivide_6}]
   properties {name=M_OuterFresnel, nodeColor=<rgba 0 0.1797 1 1>}
  }
  MaterialExpressionNamedRerouteUsage_0 {class=MaterialExpressionNamedRerouteUsage}
  MaterialExpressionNamedRerouteUsage_1 {class=MaterialExpressionNamedRerouteUsage}
  MaterialExpressionNamedRerouteUsage_10 {class=MaterialExpressionNamedRerouteUsage}
  MaterialExpressionNamedRerouteUsage_11 {class=MaterialExpressionNamedRerouteUsage}
  MaterialExpressionNamedRerouteUsage_12 {class=MaterialExpressionNamedRerouteUsage}
  MaterialExpressionNamedRerouteUsage_13 {class=MaterialExpressionNamedRerouteUsage}
  MaterialExpressionNamedRerouteUsage_14 {class=MaterialExpressionNamedRerouteUsage}
  MaterialExpressionNamedRerouteUsage_15 {class=MaterialExpressionNamedRerouteUsage}
  MaterialExpressionNamedRerouteUsage_16 {class=MaterialExpressionNamedRerouteUsage}
  MaterialExpressionNamedRerouteUsage_17 {class=MaterialExpressionNamedRerouteUsage}
  MaterialExpressionNamedRerouteUsage_18 {class=MaterialExpressionNamedRerouteUsage}
  MaterialExpressionNamedRerouteUsage_19 {class=MaterialExpressionNamedRerouteUsage}
  MaterialExpressionNamedRerouteUsage_2 {class=MaterialExpressionNamedRerouteUsage}
  MaterialExpressionNamedRerouteUsage_20 {class=MaterialExpressionNamedRerouteUsage}
  MaterialExpressionNamedRerouteUsage_21 {class=MaterialExpressionNamedRerouteUsage}
  MaterialExpressionNamedRerouteUsage_22 {class=MaterialExpressionNamedRerouteUsage}
  MaterialExpressionNamedRerouteUsage_23 {class=MaterialExpressionNamedRerouteUsage}
  MaterialExpressionNamedRerouteUsage_24 {class=MaterialExpressionNamedRerouteUsage}
  MaterialExpressionNamedRerouteUsage_25 {class=MaterialExpressionNamedRerouteUsage}
  MaterialExpressionNamedRerouteUsage_26 {class=MaterialExpressionNamedRerouteUsage}
  MaterialExpressionNamedRerouteUsage_27 {class=MaterialExpressionNamedRerouteUsage}
  MaterialExpressionNamedRerouteUsage_28 {class=MaterialExpressionNamedRerouteUsage}
  MaterialExpressionNamedRerouteUsage_29 {class=MaterialExpressionNamedRerouteUsage}
  MaterialExpressionNamedRerouteUsage_3 {class=MaterialExpressionNamedRerouteUsage}
  MaterialExpressionNamedRerouteUsage_31 {class=MaterialExpressionNamedRerouteUsage}
  MaterialExpressionNamedRerouteUsage_32 {class=MaterialExpressionNamedRerouteUsage}
  MaterialExpressionNamedRerouteUsage_33 {class=MaterialExpressionNamedRerouteUsage}
  MaterialExpressionNamedRerouteUsage_34 {class=MaterialExpressionNamedRerouteUsage}
  MaterialExpressionNamedRerouteUsage_35 {class=MaterialExpressionNamedRerouteUsage}
  MaterialExpressionNamedRerouteUsage_36 {class=MaterialExpressionNamedRerouteUsage}
  MaterialExpressionNamedRerouteUsage_37 {class=MaterialExpressionNamedRerouteUsage}
  MaterialExpressionNamedRerouteUsage_38 {class=MaterialExpressionNamedRerouteUsage}
  MaterialExpressionNamedRerouteUsage_39 {class=MaterialExpressionNamedRerouteUsage}
  MaterialExpressionNamedRerouteUsage_4 {class=MaterialExpressionNamedRerouteUsage}
  MaterialExpressionNamedRerouteUsage_40 {class=MaterialExpressionNamedRerouteUsage}
  MaterialExpressionNamedRerouteUsage_41 {class=MaterialExpressionNamedRerouteUsage}
  MaterialExpressionNamedRerouteUsage_42 {class=MaterialExpressionNamedRerouteUsage}
  MaterialExpressionNamedRerouteUsage_43 {class=MaterialExpressionNamedRerouteUsage}
  MaterialExpressionNamedRerouteUsage_44 {class=MaterialExpressionNamedRerouteUsage}
  MaterialExpressionNamedRerouteUsage_45 {class=MaterialExpressionNamedRerouteUsage}
  MaterialExpressionNamedRerouteUsage_46 {class=MaterialExpressionNamedRerouteUsage}
  MaterialExpressionNamedRerouteUsage_47 {class=MaterialExpressionNamedRerouteUsage}
  MaterialExpressionNamedRerouteUsage_48 {class=MaterialExpressionNamedRerouteUsage}
  MaterialExpressionNamedRerouteUsage_49 {class=MaterialExpressionNamedRerouteUsage}
  MaterialExpressionNamedRerouteUsage_5 {class=MaterialExpressionNamedRerouteUsage}
  MaterialExpressionNamedRerouteUsage_50 {class=MaterialExpressionNamedRerouteUsage}
  MaterialExpressionNamedRerouteUsage_51 {class=MaterialExpressionNamedRerouteUsage}
  MaterialExpressionNamedRerouteUsage_52 {class=MaterialExpressionNamedRerouteUsage}
  MaterialExpressionNamedRerouteUsage_53 {class=MaterialExpressionNamedRerouteUsage}
  MaterialExpressionNamedRerouteUsage_54 {class=MaterialExpressionNamedRerouteUsage}
  MaterialExpressionNamedRerouteUsage_56 {class=MaterialExpressionNamedRerouteUsage}
  MaterialExpressionNamedRerouteUsage_57 {class=MaterialExpressionNamedRerouteUsage}
  MaterialExpressionNamedRerouteUsage_59 {class=MaterialExpressionNamedRerouteUsage}
  MaterialExpressionNamedRerouteUsage_6 {class=MaterialExpressionNamedRerouteUsage}
  MaterialExpressionNamedRerouteUsage_60 {class=MaterialExpressionNamedRerouteUsage}
  MaterialExpressionNamedRerouteUsage_61 {class=MaterialExpressionNamedRerouteUsage}
  MaterialExpressionNamedRerouteUsage_7 {class=MaterialExpressionNamedRerouteUsage}
  MaterialExpressionNamedRerouteUsage_73 {class=MaterialExpressionNamedRerouteUsage}
  MaterialExpressionNamedRerouteUsage_79 {class=MaterialExpressionNamedRerouteUsage}
  MaterialExpressionNamedRerouteUsage_8 {class=MaterialExpressionNamedRerouteUsage}
  MaterialExpressionNamedRerouteUsage_80 {class=MaterialExpressionNamedRerouteUsage}
  MaterialExpressionNamedRerouteUsage_88 {class=MaterialExpressionNamedRerouteUsage}
  MaterialExpressionNamedRerouteUsage_89 {class=MaterialExpressionNamedRerouteUsage}
  MaterialExpressionNamedRerouteUsage_9 {class=MaterialExpressionNamedRerouteUsage}
  MaterialExpressionNamedRerouteUsage_90 {class=MaterialExpressionNamedRerouteUsage}
  MaterialExpressionNamedRerouteUsage_91 {class=MaterialExpressionNamedRerouteUsage}
  MaterialExpressionNamedRerouteUsage_92 {class=MaterialExpressionNamedRerouteUsage}
  MaterialExpressionNamedRerouteUsage_93 {class=MaterialExpressionNamedRerouteUsage}
  MaterialExpressionNamedRerouteUsage_96 {class=MaterialExpressionNamedRerouteUsage}
  MaterialExpressionNamedRerouteUsage_97 {class=MaterialExpressionNamedRerouteUsage}
  MaterialExpressionNamedRerouteUsage_98 {class=MaterialExpressionNamedRerouteUsage}
  MaterialExpressionNamedRerouteUsage_99 {class=MaterialExpressionNamedRerouteUsage}
  MaterialExpressionNormalize_0 {class=MaterialExpressionNormalize, inputs [VectorInput {from=MaterialExpressionSubtract_1, output=""}]}
  MaterialExpressionNormalize_2 {class=MaterialExpressionNormalize, inputs [VectorInput {from=MaterialExpressionAdd_10, output=""}]}
  MaterialExpressionNormalize_9 {class=MaterialExpressionNormalize, inputs [VectorInput {from=MaterialExpressionSubtract_19, output=""}]}
  MaterialExpressionOneMinus_0 {class=MaterialExpressionOneMinus, inputs [None {from=MaterialExpressionAbs_0}]}
  MaterialExpressionOneMinus_1 {class=MaterialExpressionOneMinus, inputs [None {from=MaterialExpressionSaturate_8}]}
  MaterialExpressionOneMinus_10 {class=MaterialExpressionOneMinus, inputs [None {from=MaterialExpressionDotProduct_4}]}
  MaterialExpressionOneMinus_2 {class=MaterialExpressionOneMinus, inputs [None {from=MaterialExpressionMultiply_32}]}
  MaterialExpressionOneMinus_3 {class=MaterialExpressionOneMinus, inputs [None {from=MaterialExpressionNamedRerouteUsage_10}]}
  MaterialExpressionOneMinus_4 {class=MaterialExpressionOneMinus, inputs [None {from=MaterialExpressionMultiply_40}]}
  MaterialExpressionOneMinus_5 {class=MaterialExpressionOneMinus, inputs [None {from=MaterialExpressionPower_7}]}
  MaterialExpressionOneMinus_9 {class=MaterialExpressionOneMinus, inputs [None {from=MaterialExpressionConstantBiasScale_4}]}
  MaterialExpressionPanner_0 {
   class=MaterialExpressionPanner
   inputs [
    Coordinate {from=MaterialExpressionNamedRerouteUsage_1}
    Time {from=""}
    Speed {from=MaterialExpressionAppendVector_0}
   ]
   properties {speedY=0.2}
  }
  MaterialExpressionPanner_3 {
   class=MaterialExpressionPanner
   inputs [
    Coordinate {from=MaterialExpressionComponentMask_5}
    Time {from=""}
    Speed {from=MaterialExpressionAppendVector_2}
   ]
   properties {speedX=0.1, speedY=0.01}
  }
  MaterialExpressionPixelNormalWS_0 {class=MaterialExpressionPixelNormalWS, inputs [], properties {}}
  MaterialExpressionPixelNormalWS_1 {class=MaterialExpressionPixelNormalWS, inputs [], properties {}}
  MaterialExpressionPower_0 {
   class=MaterialExpressionPower
   inputs [
    Base {from=MaterialExpressionTextureSample_3, output=RGB}
    Exp {from=MaterialExpressionScalarParameter_20}
   ]
  }
  MaterialExpressionPower_1 {
   class=MaterialExpressionPower
   inputs [
    Base {from=MaterialExpressionNamedRerouteUsage_37}
    Exp {from=""}
   ]
  }
  MaterialExpressionPower_12 {
   class=MaterialExpressionPower
   inputs [
    Base {from=MaterialExpressionSaturate_2}
    Exp {from=MaterialExpressionScalarParameter_23}
   ]
  }
  MaterialExpressionPower_2 {
   class=MaterialExpressionPower
   inputs [
    Base {from=MaterialExpressionOneMinus_0}
    Exp {from=""}
   ]
   properties {constExponent=1}
  }
  MaterialExpressionPower_3 {
   class=MaterialExpressionPower
   inputs [
    Base {from=MaterialExpressionLinearInterpolate_20}
    Exp {from=""}
   ]
  }
  MaterialExpressionPower_4 {
   class=MaterialExpressionPower
   inputs [
    Base {from=MaterialExpressionOneMinus_1}
    Exp {from=MaterialExpressionScalarParameter_12}
   ]
  }
  MaterialExpressionPower_5 {
   class=MaterialExpressionPower
   inputs [
    Base {from=MaterialExpressionDivide_4}
    Exp {from=""}
   ]
   properties {constExponent=2.5}
  }
  MaterialExpressionPower_6 {
   class=MaterialExpressionPower
   inputs [
    Base {from=MaterialExpressionSphereMask_0}
    Exp {from=MaterialExpressionScalarParameter_1}
   ]
  }
  MaterialExpressionPower_7 {
   class=MaterialExpressionPower
   inputs [
    Base {from=MaterialExpressionOneMinus_3}
    Exp {from=""}
   ]
   properties {constExponent=10}
  }
  MaterialExpressionPower_8 {
   class=MaterialExpressionPower
   inputs [
    Base {from=MaterialExpressionOneMinus_10}
    Exp {from=MaterialExpressionScalarParameter_14}
   ]
  }
  MaterialExpressionPower_9 {
   class=MaterialExpressionPower
   inputs [
    Base {from=MaterialExpressionMax_4}
    Exp {from=MaterialExpressionScalarParameter_75}
   ]
  }
  MaterialExpressionReroute_0 {class=MaterialExpressionReroute, inputs [None {from=MaterialExpressionTextureSample_4, output=R}]}
  MaterialExpressionReroute_1 {class=MaterialExpressionReroute, inputs [None {from=MaterialExpressionReroute_18}]}
  MaterialExpressionReroute_10 {class=MaterialExpressionReroute, inputs [None {from=MaterialExpressionMultiply_55}]}
  MaterialExpressionReroute_11 {class=MaterialExpressionReroute, inputs [None {from=MaterialExpressionReroute_6}]}
  MaterialExpressionReroute_12 {class=MaterialExpressionReroute, inputs [None {from=MaterialExpressionReroute_18}]}
  MaterialExpressionReroute_13 {class=MaterialExpressionReroute, inputs [None {from=MaterialExpressionMultiply_7}]}
  MaterialExpressionReroute_15 {class=MaterialExpressionReroute, inputs [None {from=MaterialExpressionStaticSwitchParameter_3}]}
  MaterialExpressionReroute_16 {class=MaterialExpressionReroute, inputs [None {from=MaterialExpressionNamedRerouteUsage_49}]}
  MaterialExpressionReroute_17 {class=MaterialExpressionReroute, inputs [None {from=MaterialExpressionReroute_3}]}
  MaterialExpressionReroute_18 {class=MaterialExpressionReroute, inputs [None {from=MaterialExpressionReroute_17}]}
  MaterialExpressionReroute_19 {class=MaterialExpressionReroute, inputs [None {from=MaterialExpressionSetMaterialAttributes_0}]}
  MaterialExpressionReroute_2 {class=MaterialExpressionReroute, inputs [None {from=MaterialExpressionLinearInterpolate_5}]}
  MaterialExpressionReroute_20 {class=MaterialExpressionReroute, inputs [None {from=MaterialExpressionLinearInterpolate_10}]}
  MaterialExpressionReroute_21 {class=MaterialExpressionReroute, inputs [None {from=MaterialExpressionReroute_24}]}
  MaterialExpressionReroute_22 {class=MaterialExpressionReroute, inputs [None {from=MaterialExpressionReroute_21}]}
  MaterialExpressionReroute_23 {class=MaterialExpressionReroute, inputs [None {from=MaterialExpressionReroute_22}]}
  MaterialExpressionReroute_24 {class=MaterialExpressionReroute, inputs [None {from=MaterialExpressionMaterialFunctionCall_2, output=Result}]}
  MaterialExpressionReroute_25 {class=MaterialExpressionReroute, inputs [None {from=MaterialExpressionNormalize_0}]}
  MaterialExpressionReroute_3 {class=MaterialExpressionReroute, inputs [None {from=MaterialExpressionLinearInterpolate_12}]}
  MaterialExpressionReroute_4 {class=MaterialExpressionReroute, inputs [None {from=MaterialExpressionReroute_25}]}
  MaterialExpressionReroute_5 {class=MaterialExpressionReroute, inputs [None {from=MaterialExpressionComponentMask_8}]}
  MaterialExpressionReroute_6 {class=MaterialExpressionReroute, inputs [None {from=MaterialExpressionAdd_14}]}
  MaterialExpressionReroute_7 {class=MaterialExpressionReroute, inputs [None {from=MaterialExpressionMultiply_3}]}
  MaterialExpressionReroute_8 {class=MaterialExpressionReroute, inputs [None {from=MaterialExpressionReroute_19}]}
  MaterialExpressionReroute_9 {class=MaterialExpressionReroute, inputs [None {from=MaterialExpressionPower_8}]}
  MaterialExpressionSaturate_0 {class=MaterialExpressionSaturate, inputs [None {from=MaterialExpressionDivide_0}]}
  MaterialExpressionSaturate_1 {class=MaterialExpressionSaturate, inputs [None {from=MaterialExpressionCeil_0}]}
  MaterialExpressionSaturate_2 {class=MaterialExpressionSaturate, inputs [None {from=MaterialExpressionSubtract_5}]}
  MaterialExpressionSaturate_3 {class=MaterialExpressionSaturate, inputs [None {from=MaterialExpressionScalarParameter_15}]}
  MaterialExpressionSaturate_4 {class=MaterialExpressionSaturate, inputs [None {from=MaterialExpressionDotProduct_0}]}
  MaterialExpressionSaturate_5 {class=MaterialExpressionSaturate, inputs [None {from=MaterialExpressionSphereMask_1}]}
  MaterialExpressionSaturate_7 {class=MaterialExpressionSaturate, inputs [None {from=MaterialExpressionMaterialFunctionCall_4, output=Result}]}
  MaterialExpressionSaturate_8 {class=MaterialExpressionSaturate, inputs [None {from=MaterialExpressionDivide_2}]}
  MaterialExpressionScalarParameter_0 {
   class=MaterialExpressionScalarParameter
   properties {defaultValue=0.02, group=Sprite_Base, parameterName=WPO_Waves_Frequency, sortPriority=100}
  }
  MaterialExpressionScalarParameter_1 {class=MaterialExpressionScalarParameter, properties {defaultValue=1, group=Sprite_Base, parameterName=InnerGlow_Power, sortPriority=0}}
  MaterialExpressionScalarParameter_10 {class=MaterialExpressionScalarParameter, properties {defaultValue=6, group=Sprite_Base, parameterName=CapWPOAmount, sortPriority=100}}
  MaterialExpressionScalarParameter_11 {class=MaterialExpressionScalarParameter, properties {defaultValue=2, group=Sprite_Base, parameterName=LogoGradient_Size, sortPriority=50}}
  MaterialExpressionScalarParameter_12 {class=MaterialExpressionScalarParameter, properties {defaultValue=1.5, group=Sprite_Base, parameterName=LogoGradient_Power, sortPriority=50}}
  MaterialExpressionScalarParameter_13 {class=MaterialExpressionScalarParameter, properties {group=Sprite_Base, parameterName=LogoGradient_Intensity, sortPriority=50}}
  MaterialExpressionScalarParameter_14 {
   class=MaterialExpressionScalarParameter
   properties {
    defaultValue=1.3
    desc=originally 3 before last minute request for adding a sined fresnel
    group=Sprite_Base
    parameterName=OutGlowFresnelPow
    sortPriority=0
   }
  }
  MaterialExpressionScalarParameter_15 {
   class=MaterialExpressionScalarParameter
   properties {group=Sprite_Cosmetics | Setup, parameterName=Enable Input Data Override, sliderMax=1, sortPriority=400}
  }
  MaterialExpressionScalarParameter_16 {
   class=MaterialExpressionScalarParameter
   properties {defaultValue=1, group=Sprite_Cosmetics, parameterName=Pilot - Cap WPO Amount Mult, sortPriority=20}
  }
  MaterialExpressionScalarParameter_17 {
   class=MaterialExpressionScalarParameter
   properties {defaultValue=0.57, group=Sprite_Cosmetics | Setup, parameterName=Pilot Pre-Skinned Scale, sortPriority=220}
  }
  MaterialExpressionScalarParameter_18 {class=MaterialExpressionScalarParameter, properties {defaultValue=0.2, group=Sprite_Base, parameterName=Jelly_Emissive_RimLight_Bias}}
  MaterialExpressionScalarParameter_19 {class=MaterialExpressionScalarParameter, properties {defaultValue=0.1, group=Sprite_Base, parameterName=CapWPOSpeedX, sortPriority=100}}
  MaterialExpressionScalarParameter_2 {
   class=MaterialExpressionScalarParameter
   properties {defaultValue=0.05, group=Sprite_Base, parameterName=InnerGlowNoiseAmount, sortPriority=0}
  }
  MaterialExpressionScalarParameter_20 {class=MaterialExpressionScalarParameter, properties {defaultValue=0.2, group=Sprite_Base, parameterName=CapWPONoisePower, sortPriority=100}}
  MaterialExpressionScalarParameter_21 {class=MaterialExpressionScalarParameter, properties {defaultValue=0.06, group=Sprite_Base, parameterName=CapWPOSpeedY, sortPriority=100}}
  MaterialExpressionScalarParameter_22 {
   class=MaterialExpressionScalarParameter
   properties {defaultValue=-6.4038, group=Sprite_Base, parameterName=BumpOffsetLogoHeight, sortPriority=50}
  }
  MaterialExpressionScalarParameter_23 {class=MaterialExpressionScalarParameter, properties {defaultValue=16, group=Sprite_Base, parameterName=Sprite_Emissive_RimLight_Pow}}
  MaterialExpressionScalarParameter_24 {class=MaterialExpressionScalarParameter, properties {defaultValue=0.25, group=Sprite_Base, parameterName=Sprite_Roughness, sortPriority=999}}
  MaterialExpressionScalarParameter_25 {class=MaterialExpressionScalarParameter, properties {defaultValue=0.12, group=Sprite_Base, parameterName=Logo Scale, sortPriority=50}}
  MaterialExpressionScalarParameter_26 {class=MaterialExpressionScalarParameter, properties {group=Sprite_Base, parameterName=InnerGlowNoiseSpeed_X, sortPriority=0}}
  MaterialExpressionScalarParameter_27 {
   class=MaterialExpressionScalarParameter
   properties {defaultValue=0.455, group=Sprite_Cosmetics | Setup, parameterName=ColdTrophy Pre-Skinned Scale, sortPriority=320}
  }
  MaterialExpressionScalarParameter_28 {
   class=MaterialExpressionScalarParameter
   properties {defaultValue=1, group=Sprite_Cosmetics, parameterName=ColdTrophy - Cap WPO Amount Mult, sortPriority=30}
  }
  MaterialExpressionScalarParameter_29 {
   class=MaterialExpressionScalarParameter
   properties {
    defaultValue=1
    desc="\r\nUsed when \"Enable Input Data Override\" is ON\r\n"
    group=Sprite_Cosmetics
    parameterName=Cap WPO Amount Mult Override
    sortPriority=40
   }
  }
  MaterialExpressionScalarParameter_3 {class=MaterialExpressionScalarParameter, properties {defaultValue=0.3, group=Sprite_Base, parameterName=InnerGlow_Radius, sortPriority=0}}
  MaterialExpressionScalarParameter_31 {class=MaterialExpressionScalarParameter, properties {defaultValue=0.1, group=Sprite_Base, parameterName=InnerGlow_PosY, sortPriority=0}}
  MaterialExpressionScalarParameter_32 {class=MaterialExpressionScalarParameter, properties {defaultValue=0.2, group=Sprite_Base, parameterName=InnerGlow_PosX, sortPriority=0}}
  MaterialExpressionScalarParameter_33 {
   class=MaterialExpressionScalarParameter
   properties {defaultValue=31, group=Sprite_Base, parameterName=WPO_FacialMask_Radius, sortPriority=100}
  }
  MaterialExpressionScalarParameter_34 {
   class=MaterialExpressionScalarParameter
   properties {defaultValue=0.9, group=Sprite_Base, parameterName=WPO_FacialMask_Hardness, sortPriority=100}
  }
  MaterialExpressionScalarParameter_35 {class=MaterialExpressionScalarParameter, properties {defaultValue=1, group=Sprite_Base, parameterName=Sprite_Specular, sortPriority=999}}
  MaterialExpressionScalarParameter_36 {class=MaterialExpressionScalarParameter, properties {group=Sprite_Base, parameterName="=====COLORS====", sortPriority=0}}
  MaterialExpressionScalarParameter_37 {class=MaterialExpressionScalarParameter, properties {group=Sprite_Base, parameterName="=====WPO====", sortPriority=99}}
  MaterialExpressionScalarParameter_38 {class=MaterialExpressionScalarParameter, properties {group=Sprite_Base, parameterName="=====Logo====", sortPriority=50}}
  MaterialExpressionScalarParameter_39 {class=MaterialExpressionScalarParameter, properties {group=Sprite_Base, parameterName="=====MA====", sortPriority=999}}
  MaterialExpressionScalarParameter_4 {class=MaterialExpressionScalarParameter, properties {defaultValue=-0.5, group=Sprite_Base, parameterName=InnerGlow_Hardness, sortPriority=0}}
  MaterialExpressionScalarParameter_40 {class=MaterialExpressionScalarParameter, properties {group=Sprite_Base, parameterName=Sprite_Metallic, sortPriority=999}}
  MaterialExpressionScalarParameter_41 {class=MaterialExpressionScalarParameter, properties {group=Sprite_Base, parameterName="=====ScreenSpace====", sortPriority=30}}
  MaterialExpressionScalarParameter_42 {class=MaterialExpressionScalarParameter, properties {defaultValue=1, group=Sprite_Base, parameterName=InnerGlow_ScaleY, sortPriority=0}}
  MaterialExpressionScalarParameter_43 {class=MaterialExpressionScalarParameter, properties {defaultValue=1, group=Sprite_Base, parameterName=InnerGlow_ScaleX, sortPriority=0}}
  MaterialExpressionScalarParameter_44 {
   class=MaterialExpressionScalarParameter
   properties {defaultValue=0.25, group=Sprite_Base, parameterName=WPO_Chunks_Intensity, sortPriority=100}
  }
  MaterialExpressionScalarParameter_46 {class=MaterialExpressionScalarParameter, properties {defaultValue=1, group=Sprite_Base, parameterName=TinyBubbles_Tiling, sortPriority=49}}
  MaterialExpressionScalarParameter_47 {class=MaterialExpressionScalarParameter, properties {defaultValue=0.15, group=Sprite_Base, parameterName=TinyBubbles_Speed, sortPriority=49}}
  MaterialExpressionScalarParameter_48 {
   class=MaterialExpressionScalarParameter
   properties {defaultValue=40, group=Sprite_Base, parameterName=TinyBubbles_Intensity, sortPriority=49}
  }
  MaterialExpressionScalarParameter_5 {class=MaterialExpressionScalarParameter, properties {defaultValue=3, group=Sprite_Base, parameterName=OutGlowFresnelPow, sortPriority=0}}
  MaterialExpressionScalarParameter_50 {
   class=MaterialExpressionScalarParameter
   properties {defaultValue=35, group=Sprite_Base, parameterName=BaseColorB_FalloffDown, sortPriority=0}
  }
  MaterialExpressionScalarParameter_51 {class=MaterialExpressionScalarParameter, properties {defaultValue=35, group=Sprite_Base, parameterName=BaseColorB_FalloffUp, sortPriority=0}}
  MaterialExpressionScalarParameter_52 {class=MaterialExpressionScalarParameter, properties {defaultValue=12.3, group=Sprite_Base, parameterName=BaseColorB_Pos, sortPriority=0}}
  MaterialExpressionScalarParameter_55 {class=MaterialExpressionScalarParameter, properties {defaultValue=0.5, group=Sprite_Base, parameterName=WPO_Waves_Mult, sortPriority=100}}
  MaterialExpressionScalarParameter_6 {
   class=MaterialExpressionScalarParameter
   properties {defaultValue=0.3, group=Sprite_Base, parameterName=InnerGlowNoiseSpeed_Y, sortPriority=0}
  }
  MaterialExpressionScalarParameter_65 {class=MaterialExpressionScalarParameter, properties {defaultValue=6, group=Sprite_Base, parameterName=Screen Space Scale, sortPriority=30}}
  MaterialExpressionScalarParameter_75 {class=MaterialExpressionScalarParameter, properties {defaultValue=2, group=Sprite_Base, parameterName=InnerScatteringPower, sortPriority=0}}
  MaterialExpressionScalarParameter_76 {
   class=MaterialExpressionScalarParameter
   properties {defaultValue=0.7, group=Sprite_Base, parameterName=InnerScatteringScale, sortPriority=0}
  }
  MaterialExpressionScalarParameter_8 {class=MaterialExpressionScalarParameter, properties {group=Sprite_Base, parameterName=WPO_random_seed, sortPriority=100}}
  MaterialExpressionScalarParameter_9 {
   class=MaterialExpressionScalarParameter
   properties {defaultValue=0.02, group=Sprite_Base, parameterName=LocalWaveWPOFrequency, sortPriority=100}
  }
  MaterialExpressionScreenPosition_1 {class=MaterialExpressionScreenPosition, inputs [], properties {}}
  MaterialExpressionSetMaterialAttributes_0 {
   class=MaterialExpressionSetMaterialAttributes
   inputs [
    MaterialAttributes {from="", output=""}
    BaseColor {from=MaterialExpressionConstant_0, output=""}
    Metallic {from=MaterialExpressionScalarParameter_40, output=""}
    Specular {from=MaterialExpressionScalarParameter_35, output=""}
    Roughness {from=MaterialExpressionScalarParameter_24, output=""}
    Normal {from=MaterialExpressionNamedRerouteUsage_11, output=""}
    EmissiveColor {from=MaterialExpressionNamedRerouteUsage_7, output=""}
    Opacity {from="", output=""}
    SubsurfaceColor {from=MaterialExpressionNamedRerouteUsage_0, output=""}
    WorldPositionOffset {from=MaterialExpressionStaticSwitchParameter_0, output=""}
    ShadingModel {from=MaterialExpressionShadingModel_0, output=""}
   ]
   properties {attributeSetTypes [
     #1
     #2
     #3
     #4
     #5
     #6
     #7
     #8
     #9
     #10
    ]}
  }
  MaterialExpressionShaderStageSwitch_0 {
   class=MaterialExpressionShaderStageSwitch
   inputs [
    PixelShader {from=MaterialExpressionVertexInterpolator_1, output=PS}
    VertexShader {from=MaterialExpressionLocalPosition_1, output=XYZ}
   ]
   properties {}
  }
  MaterialExpressionShadingModel_0 {class=MaterialExpressionShadingModel, inputs [], properties {shadingModel=MSM_Subsurface}}
  MaterialExpressionSine_0 {class=MaterialExpressionSine, inputs [None {from=MaterialExpressionAdd_12}]}
  MaterialExpressionSine_1 {class=MaterialExpressionSine, inputs [None {from=MaterialExpressionSubtract_18}]}
  MaterialExpressionSine_4 {class=MaterialExpressionSine, inputs [None {from=MaterialExpressionReroute_9}]}
  MaterialExpressionSmoothStep_0 {
   class=MaterialExpressionSmoothStep
   inputs [
    Min {from=MaterialExpressionConstant_8, output=""}
    Max {from="", output=""}
    Value {from=MaterialExpressionComponentMask_4, output=""}
   ]
   properties {constMax=2, constMin=1}
  }
  MaterialExpressionSmoothStep_1 {
   class=MaterialExpressionSmoothStep
   inputs [
    Min {from="", output=""}
    Max {from="", output=""}
    Value {from=MaterialExpressionReroute_9, output=""}
   ]
   properties {constMax=0.1}
  }
  MaterialExpressionSphereMask_0 {
   class=MaterialExpressionSphereMask
   inputs [
    A {from=MaterialExpressionAdd_2, output=""}
    B {from=MaterialExpressionConvert_0, output=""}
    Radius {from=MaterialExpressionScalarParameter_3, output=""}
    Hardness {from=MaterialExpressionScalarParameter_4, output=""}
   ]
   properties {attenuationRadius=12, hardnessPercent=55}
  }
  MaterialExpressionSphereMask_1 {
   class=MaterialExpressionSphereMask
   inputs [
    A {from=MaterialExpressionMultiply_34, output=""}
    B {from=MaterialExpressionConstant3Vector_3, output=""}
    Radius {from=MaterialExpressionScalarParameter_33, output=""}
    Hardness {from=MaterialExpressionScalarParameter_34, output=""}
   ]
   properties {attenuationRadius=13, hardnessPercent=15}
  }
  MaterialExpressionStaticBool_0 {class=MaterialExpressionStaticBool}
  MaterialExpressionStaticSwitchParameter_0 {
   class=MaterialExpressionStaticSwitchParameter
   inputs [
    True {from=MaterialExpressionNamedRerouteUsage_9, output=""}
    False {from=MaterialExpressionConstant_4, output=""}
   ]
   properties {group=Sprite_Base, parameterName=Use WPO, sortPriority=99}
  }
  MaterialExpressionStaticSwitchParameter_1 {
   class=MaterialExpressionStaticSwitchParameter
   inputs [
    True {from=MaterialExpressionLinearInterpolate_14, output=""}
    False {from=MaterialExpressionReroute_4, output=""}
   ]
   properties {desc=Fixes broken WPO on UV seams on pillow etc, group=Sprite_Base, parameterName=WPO_Chunks, sortPriority=100}
  }
  MaterialExpressionStaticSwitchParameter_3 {
   class=MaterialExpressionStaticSwitchParameter
   inputs [
    True {from=MaterialExpressionLinearInterpolate_19, output=""}
    False {from=MaterialExpressionReroute_1, output=""}
   ]
   properties {parameterName=TinyBubbles}
  }
  MaterialExpressionStaticSwitchParameter_4 {
   class=MaterialExpressionStaticSwitchParameter
   inputs [
    True {from=MaterialExpressionMultiply_0, output=""}
    False {from=MaterialExpressionReroute_23, output=""}
   ]
   properties {parameterName=JW AO}
  }
  MaterialExpressionStaticSwitchParameter_5 {
   class=MaterialExpressionStaticSwitchParameter
   inputs [
    True {from=MaterialExpressionLinearInterpolate_11, output=""}
    False {from=MaterialExpressionNamedRerouteUsage_91, output=""}
   ]
   properties {desc=Fixes broken WPO on UV seams on pillow etc, group=Sprite_Base, parameterName=WPO_Chunks, sortPriority=100}
  }
  MaterialExpressionStep_1 {
   class=MaterialExpressionStep
   inputs [
    Y {from=MaterialExpressionComponentMask_2, output=""}
    X {from=MaterialExpressionScalarParameter_52, output=""}
   ]
   properties {constX=0.4}
  }
  MaterialExpressionSubtract_0 {
   class=MaterialExpressionSubtract
   inputs [
    A {from=MaterialExpressionNamedRerouteUsage_89}
    B {from=MaterialExpressionConstant4Vector_1, output=RGB}
   ]
  }
  MaterialExpressionSubtract_1 {
   class=MaterialExpressionSubtract
   inputs [
    A {from=MaterialExpressionWorldPosition_0, output=XYZ}
    B {from=MaterialExpressionNamedRerouteUsage_97}
   ]
  }
  MaterialExpressionSubtract_10 {
   class=MaterialExpressionSubtract
   inputs [
    A {from=MaterialExpressionConstantBiasScale_1}
    B {from=MaterialExpressionMultiply_86}
   ]
  }
  MaterialExpressionSubtract_17 {
   class=MaterialExpressionSubtract
   inputs [
    A {from=MaterialExpressionAdd_8}
    B {from=""}
   ]
   properties {constB=2}
  }
  MaterialExpressionSubtract_18 {
   class=MaterialExpressionSubtract
   inputs [
    A {from=MaterialExpressionSubtract_17}
    B {from=MaterialExpressionFrac_2}
   ]
  }
  MaterialExpressionSubtract_19 {
   class=MaterialExpressionSubtract
   inputs [
    A {from=MaterialExpressionMultiply_110}
    B {from=MaterialExpressionAtmosphericLightVector_1}
   ]
  }
  MaterialExpressionSubtract_2 {
   class=MaterialExpressionSubtract
   inputs [
    A {from=MaterialExpressionReroute_13}
    B {from=MaterialExpressionComponentMask_10}
   ]
  }
  MaterialExpressionSubtract_3 {
   class=MaterialExpressionSubtract
   inputs [
    A {from=MaterialExpressionNamedRerouteDeclaration_42}
    B {from=MaterialExpressionNamedRerouteUsage_73}
   ]
  }
  MaterialExpressionSubtract_4 {
   class=MaterialExpressionSubtract
   inputs [
    A {from=MaterialExpressionNamedRerouteUsage_12}
    B {from=MaterialExpressionMultiply_16}
   ]
  }
  MaterialExpressionSubtract_5 {
   class=MaterialExpressionSubtract
   inputs [
    A {from=MaterialExpressionDotProduct_5}
    B {from=MaterialExpressionScalarParameter_18}
   ]
  }
  MaterialExpressionSubtract_6 {
   class=MaterialExpressionSubtract
   inputs [
    A {from=MaterialExpressionTime_2}
    B {from=MaterialExpressionScalarParameter_8}
   ]
  }
  MaterialExpressionSubtract_7 {
   class=MaterialExpressionSubtract
   inputs [
    A {from=MaterialExpressionNamedRerouteUsage_60}
    B {from=MaterialExpressionNamedRerouteUsage_59}
   ]
  }
  MaterialExpressionSwitch_0 {
   class=MaterialExpressionSwitch
   inputs [
    SwitchValue {from=MaterialExpressionNamedRerouteUsage_57, output=""}
    Default {from="", output=""}
    pilot {from=MaterialExpressionNamedRerouteUsage_51, output=""}
    backpack {from=MaterialExpressionNamedRerouteUsage_52, output=""}
    override {from=MaterialExpressionNamedRerouteUsage_46, output=""}
   ]
   properties {inputs [
     {
      inputName=pilot
      input {
       expression=&self:MaterialExpressionNamedRerouteUsage_51
       inputName=None
       mask=0
       maskA=0
       maskB=0
       maskG=0
       maskR=0
       outputIndex=0
      }
     }
     {
      inputName=backpack
      input {
       expression=&self:MaterialExpressionNamedRerouteUsage_52
       inputName=None
       mask=0
       maskA=0
       maskB=0
       maskG=0
       maskR=0
       outputIndex=0
      }
     }
     {
      inputName=override
      input {
       expression=&self:MaterialExpressionNamedRerouteUsage_46
       inputName=None
       mask=0
       maskA=0
       maskB=0
       maskG=0
       maskR=0
       outputIndex=0
      }
     }
    ]}
  }
  MaterialExpressionSwitch_1 {
   class=MaterialExpressionSwitch
   inputs [
    SwitchValue {from=MaterialExpressionNamedRerouteDeclaration_33, output=""}
    Default {from="", output=""}
    pilot {from=MaterialExpressionNamedRerouteUsage_19, output=""}
    backpack {from=MaterialExpressionNamedRerouteUsage_28, output=""}
    override {from=MaterialExpressionNamedRerouteUsage_54, output=""}
   ]
   properties {inputs [
     {
      inputName=pilot
      input {
       expression=&self:MaterialExpressionNamedRerouteUsage_19
       inputName=None
       mask=0
       maskA=0
       maskB=0
       maskG=0
       maskR=0
       outputIndex=0
      }
     }
     {
      inputName=backpack
      input {
       expression=&self:MaterialExpressionNamedRerouteUsage_28
       inputName=None
       mask=0
       maskA=0
       maskB=0
       maskG=0
       maskR=0
       outputIndex=0
      }
     }
     {
      inputName=override
      input {
       expression=&self:MaterialExpressionNamedRerouteUsage_54
       inputName=None
       mask=0
       maskA=0
       maskB=0
       maskG=0
       maskR=0
       outputIndex=0
      }
     }
    ]}
  }
  MaterialExpressionSwitch_2 {
   class=MaterialExpressionSwitch
   inputs [
    SwitchValue {from=MaterialExpressionNamedRerouteUsage_56, output=""}
    Default {from="", output=""}
    pilot {from=MaterialExpressionNamedRerouteUsage_79, output=""}
    backpack {from=MaterialExpressionNamedRerouteUsage_80, output=""}
    override {from=MaterialExpressionNamedRerouteUsage_53, output=""}
   ]
   properties {inputs [
     {
      inputName=pilot
      input {
       expression=&self:MaterialExpressionNamedRerouteUsage_79
       inputName=None
       mask=0
       maskA=0
       maskB=0
       maskG=0
       maskR=0
       outputIndex=0
      }
     }
     {
      inputName=backpack
      input {
       expression=&self:MaterialExpressionNamedRerouteUsage_80
       inputName=None
       mask=0
       maskA=0
       maskB=0
       maskG=0
       maskR=0
       outputIndex=0
      }
     }
     {
      inputName=override
      input {
       expression=&self:MaterialExpressionNamedRerouteUsage_53
       inputName=None
       mask=0
       maskA=0
       maskB=0
       maskG=0
       maskR=0
       outputIndex=0
      }
     }
    ]}
  }
  MaterialExpressionSwitch_3 {
   class=MaterialExpressionSwitch
   inputs [
    SwitchValue {from=MaterialExpressionNamedRerouteUsage_3, output=""}
    Default {from="", output=""}
    Auto {from=MaterialExpressionNamedRerouteDeclaration_27, output=""}
    Override {from=MaterialExpressionConstant_10, output=""}
   ]
   properties {inputs [
     {
      inputName=Auto
      input {
       expression=&self:MaterialExpressionNamedRerouteDeclaration_27
       inputName=None
       mask=0
       maskA=0
       maskB=0
       maskG=0
       maskR=0
       outputIndex=0
      }
     }
     {
      inputName=Override
      input {
       expression=&self:MaterialExpressionConstant_10
       inputName=None
       mask=0
       maskA=0
       maskB=0
       maskG=0
       maskR=0
       outputIndex=0
      }
     }
    ]}
  }
  MaterialExpressionTextureCoordinate_1 {class=MaterialExpressionTextureCoordinate}
  MaterialExpressionTextureObjectParameter_0 {
   class=MaterialExpressionTextureObjectParameter
   inputs [
    UVs {from="", output=""}
    Apply View MipBias {from="", output=""}
   ]
   properties {
    group=Sprite_Base
    parameterName=Sprite Logo
    samplerType=SAMPLERTYPE_Masks
    sortPriority=50
    texture=/BRCosmetics/Animation/Props/Gadgets_PixieParts/Textures/Sprite_Skins/Sprite_Water/T_Sprite_Water
   }
  }
  MaterialExpressionTextureObjectParameter_2 {
   class=MaterialExpressionTextureObjectParameter
   inputs [
    UVs {from="", output=""}
    Apply View MipBias {from="", output=""}
   ]
   properties {
    group=Sprite_Base
    parameterName=InteriorNoise
    sortPriority=30
    texture=/BRCosmetics/Animation/Props/Gadgets_PixieParts/Textures/TA/T_GoldenFleeceStatuette_CellNoise
   }
  }
  MaterialExpressionTextureObject_2 {
   class=MaterialExpressionTextureObject
   inputs []
   properties {texture=/BRCosmetics/Animation/Props/Gadgets_PixieParts/Textures/TA/Cell_Noise_T1}
  }
  MaterialExpressionTextureSample_10 {
   class=MaterialExpressionTextureSample
   inputs [
    UVs {from=MaterialExpressionAdd_15}
    Tex {from=""}
    Apply View MipBias {from=""}
   ]
   properties {samplerType=SAMPLERTYPE_LinearColor, texture=&HighContrastNoise}
  }
  MaterialExpressionTextureSample_3 {
   class=MaterialExpressionTextureSample
   inputs [
    UVs {from=MaterialExpressionPanner_3}
    Tex {from=MaterialExpressionNamedRerouteUsage_42}
    Apply View MipBias {from=""}
   ]
   properties {samplerType=SAMPLERTYPE_LinearGrayscale}
  }
  MaterialExpressionTextureSample_4 {
   class=MaterialExpressionTextureSample
   inputs [
    UVs {from=MaterialExpressionBumpOffset_1}
    Tex {from=MaterialExpressionNamedRerouteUsage_43}
    Apply View MipBias {from=""}
   ]
   properties {samplerSource=SSM_Clamp_WorldGroupSettings}
  }
  MaterialExpressionTextureSample_8 {
   class=MaterialExpressionTextureSample
   inputs [
    UVs {from=MaterialExpressionAdd_17}
    Tex {from=""}
    Apply View MipBias {from=""}
   ]
   properties {texture=/Game/Effects/Fort_Effects/Textures/Tile/T_Dots_Small}
  }
  MaterialExpressionTextureSample_9 {
   class=MaterialExpressionTextureSample
   inputs [
    UVs {from=MaterialExpressionPanner_0}
    Tex {from=MaterialExpressionNamedRerouteUsage_48}
    Apply View MipBias {from=""}
   ]
  }
  MaterialExpressionTime_0 {class=MaterialExpressionTime}
  MaterialExpressionTime_2 {class=MaterialExpressionTime}
  MaterialExpressionTime_3 {class=MaterialExpressionTime}
  MaterialExpressionTransformPosition_1 {
   class=MaterialExpressionTransformPosition
   inputs [None {from=MaterialExpressionNamedRerouteDeclaration_43, output=""}]
   properties {transformType=TRANSFORMPOSSOURCE_World}
  }
  MaterialExpressionTransformPosition_6 {
   class=MaterialExpressionTransformPosition
   inputs [None {from=MaterialExpressionNamedRerouteUsage_96, output=""}]
   properties {transformSourceType=TRANSFORMPOSSOURCE_World, transformType=TRANSFORMPOSSOURCE_View}
  }
  MaterialExpressionTransform_0 {
   class=MaterialExpressionTransform
   inputs [None {from=MaterialExpressionNormalize_2}]
   properties {transformSourceType=TRANSFORMSOURCE_World, transformType=TRANSFORM_Tangent}
  }
  MaterialExpressionTransform_3 {
   class=MaterialExpressionTransform
   inputs [None {from=MaterialExpressionVectorParameter_14, output=RGB}]
   properties {transformSourceType=TRANSFORMSOURCE_Camera}
  }
  MaterialExpressionVectorParameter_0 {
   class=MaterialExpressionVectorParameter
   properties {group=Sprite_Base, parameterName=InnerGlowColor, sortPriority=0, defaultValue=<rgba 0 0.52 0.6979 1>}
  }
  MaterialExpressionVectorParameter_1 {
   class=MaterialExpressionVectorParameter
   properties {group=Sprite_Base, parameterName=BaseColorA_Scattered, sortPriority=0, defaultValue=<rgba 0.1302 2 3 1>}
  }
  MaterialExpressionVectorParameter_10 {
   class=MaterialExpressionVectorParameter
   properties {group=Sprite_Base, parameterName=Logo Color, sortPriority=50, defaultValue=<rgba 5 5 5 1>}
  }
  MaterialExpressionVectorParameter_11 {class=MaterialExpressionVectorParameter, properties {group=Sprite_Base, parameterName=TOD_Day, sortPriority=9, defaultValue=<rgba 1 1 1 1>}}
  MaterialExpressionVectorParameter_12 {
   class=MaterialExpressionVectorParameter
   properties {group=Sprite_Base, parameterName=TOD_Evening, sortPriority=9, defaultValue=<rgba 0.8 0.5 0.4 1>}
  }
  MaterialExpressionVectorParameter_13 {
   class=MaterialExpressionVectorParameter
   properties {group=Sprite_Base, parameterName=TOD_Night, sortPriority=9, defaultValue=<rgba 0.35 0.4667 0.7 1>}
  }
  MaterialExpressionVectorParameter_14 {
   class=MaterialExpressionVectorParameter
   properties {
    group=Sprite_Base
    parameterName=Jelly_Emissive_RimLight_CameraVector
    channelNames {a="", b=depht, g=up, r=sides}
    defaultValue=<rgba 0.6 0 -1 1>
   }
  }
  MaterialExpressionVectorParameter_15 {
   class=MaterialExpressionVectorParameter
   properties {group=Sprite_Base, parameterName=InnerGlowColor_Evening, sortPriority=0, defaultValue=<rgba 0 0.52 0.6979 1>}
  }
  MaterialExpressionVectorParameter_16 {
   class=MaterialExpressionVectorParameter
   properties {group=Sprite_Base, parameterName=InnerGlowColor_Night, sortPriority=0, defaultValue=<rgba 0 0.52 0.6979 1>}
  }
  MaterialExpressionVectorParameter_17 {
   class=MaterialExpressionVectorParameter
   properties {group=Sprite_Base, parameterName=InnerGlowColor_Morning, sortPriority=0, defaultValue=<rgba 0 0.52 0.6979 1>}
  }
  MaterialExpressionVectorParameter_18 {
   class=MaterialExpressionVectorParameter
   properties {group=Sprite_Base, parameterName=Sprite_Emissive_RimLight_Color, defaultValue=<rgba 5 2.7247 0.5652 0.12>}
  }
  MaterialExpressionVectorParameter_19 {
   class=MaterialExpressionVectorParameter
   properties {group=Sprite_Cosmetics | Setup, parameterName=Sprite Pre-Skinned Bounds Center, sortPriority=120, defaultValue=<rgba 0 0.09 17.93 0>}
  }
  MaterialExpressionVectorParameter_2 {
   class=MaterialExpressionVectorParameter
   properties {group=Sprite_Base, parameterName=BaseColorB, sortPriority=0, defaultValue=<rgba 0 0.1674 0.3438 1>}
  }
  MaterialExpressionVectorParameter_20 {
   class=MaterialExpressionVectorParameter
   properties {group=Sprite_Cosmetics | Setup, parameterName=Pilot Pre-Skinned Bounds Center, sortPriority=210, defaultValue=<rgba 0.075 0 10.85 0>}
  }
  MaterialExpressionVectorParameter_21 {
   class=MaterialExpressionVectorParameter
   properties {group=Sprite_Base, parameterName=InnerScatteringColor, sortPriority=0, defaultValue=<rgba 0.023 0.057 0.276 1>}
  }
  MaterialExpressionVectorParameter_22 {
   class=MaterialExpressionVectorParameter
   properties {
    group=Sprite_Cosmetics | Setup
    parameterName=PIlot Pre-Skinned Scale Ratio
    sortPriority=230
    channelNames {a=n/a, b="", g="", r=""}
    defaultValue=<rgba 1 1.1 1 0>
   }
  }
  MaterialExpressionVectorParameter_23 {
   class=MaterialExpressionVectorParameter
   properties {group=Sprite_Cosmetics | Setup, parameterName=ColdTrophy Pre-Skinned Bounds Center, sortPriority=310, defaultValue=<rgba -0.03 -11.1 -6.6 0>}
  }
  MaterialExpressionVectorParameter_24 {
   class=MaterialExpressionVectorParameter
   properties {group=Sprite_Cosmetics | Setup, parameterName=Sprite Pre-Skinned Bounds Extents, sortPriority=110, defaultValue=<rgba 31.27 17.28 35.87 0>}
  }
  MaterialExpressionVectorParameter_25 {
   class=MaterialExpressionVectorParameter
   properties {
    group=Sprite_Cosmetics | Setup
    parameterName=ColdTrophy Pre-Skinned Scale Ratio
    sortPriority=330
    channelNames {a=n/a, b="", g="", r=""}
    defaultValue=<rgba 1.01 0.95 1.05 0>
   }
  }
  MaterialExpressionVectorParameter_26 {
   class=MaterialExpressionVectorParameter
   properties {
    desc="\r\nUsed when \"Enable Input Data Override\" is ON\r\n"
    group=Sprite_Cosmetics | Setup
    parameterName=Pre-Skinned Bounds Min Override
    sortPriority=410
   }
  }
  MaterialExpressionVectorParameter_27 {
   class=MaterialExpressionVectorParameter
   properties {
    desc="\r\nUsed when \"Enable Input Data Override\" is ON\r\n"
    group=Sprite_Cosmetics | Setup
    parameterName=Pre-Skinned Bounds Max Override
    sortPriority=430
    channelNames {a=n/a, b="", g="", r=""}
    defaultValue=<rgba 1 1 1 0>
   }
  }
  MaterialExpressionVectorParameter_3 {
   class=MaterialExpressionVectorParameter
   properties {
    group=Sprite_Base
    parameterName=OutGlowFresnelColor
    sortPriority=0
    channelNames {a="0=", b="", g="", r=""}
    defaultValue=<rgba 0 0.9323 0.6506 1>
   }
  }
  MaterialExpressionVectorParameter_4 {
   class=MaterialExpressionVectorParameter
   properties {group=Sprite_Base, parameterName=BaseColorA, sortPriority=0, defaultValue=<rgba 0 0.1554 0.1927 1>}
  }
  MaterialExpressionVectorParameter_5 {
   class=MaterialExpressionVectorParameter
   properties {group=Sprite_Base, parameterName=BaseColorC, sortPriority=0, defaultValue=<rgba 0 1.5 0.6438 1>}
  }
  MaterialExpressionVectorParameter_6 {
   class=MaterialExpressionVectorParameter
   properties {group=Sprite_Base, parameterName=SubsurfaceColor, sortPriority=0, defaultValue=<rgba 0 0.974 0.974 1>}
  }
  MaterialExpressionVectorParameter_7 {
   class=MaterialExpressionVectorParameter
   properties {group=Sprite_Base, parameterName=TOD_Morning, sortPriority=9, defaultValue=<rgba 0.8 0.4 0.5 1>}
  }
  MaterialExpressionVectorParameter_8 {
   class=MaterialExpressionVectorParameter
   properties {group=Sprite_Base, parameterName=BaseColorC_Scattered, sortPriority=0, defaultValue=<rgba 0.0041 0.4323 0.2735 1>}
  }
  MaterialExpressionVectorParameter_9 {
   class=MaterialExpressionVectorParameter
   properties {group=Sprite_Base, parameterName=Offset Logo, sortPriority=50, defaultValue=<rgba 1.7962 0.1621 0 1>}
  }
  MaterialExpressionVertexInterpolator_1 {class=MaterialExpressionVertexInterpolator, inputs [VS {from=MaterialExpressionLocalPosition_1, output=XYZ}], properties {}}
  MaterialExpressionVertexNormalWS_1 {class=MaterialExpressionVertexNormalWS, inputs [], properties {}}
  MaterialExpressionVertexNormalWS_2 {class=MaterialExpressionVertexNormalWS, inputs [], properties {}}
  MaterialExpressionVertexNormalWS_3 {class=MaterialExpressionVertexNormalWS, inputs [], properties {}}
  MaterialExpressionVertexNormalWS_4 {class=MaterialExpressionVertexNormalWS, inputs [], properties {}}
  MaterialExpressionVertexNormalWS_5 {class=MaterialExpressionVertexNormalWS, inputs [], properties {}}
  MaterialExpressionViewSize_3 {class=MaterialExpressionViewSize, inputs [], properties {}}
  MaterialExpressionWorldPosition_0 {class=MaterialExpressionWorldPosition, inputs [], properties {}}
 }}
