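# Material function for sprite materials that multiplies OpacityMask by an optional dithered fresnel, dissolving grazing-angle edges.
# A mod-5 pixel grid and a procedural Bayer pattern, jittered by time and temporal sample index, are added to squared N·V and rounded.
# The optional HeadOnly switch biases the fresnel toward opaque along UV V and when the view aligns with local Z.
# A shared Bayer function call, SmoothStep branch and local-height gradient do not reach the output.
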MaterialFunction /BRCosmetics/Animation/Props/Gadgets_PixieParts/Materials/ParentMaterials/MF_Sprite_DitherFresnel
graph {nodes [
  n1 = GetMaterialAttributes(None=FunctionInput() {inputName=MA, inputType=FunctionInput_MaterialAttributes}) {attributeGetTypes [#1]}
  n2 = ViewProperty() {property=MEVP_TemporalSampleIndex}
  n3 = AppendVector(A=@n2.Property, B=@n2.Property)
  n4 = Frac(None=Multiply(A=Time()) {constA=100, constB=100})
  MaterialExpressionMultiply_3 = Multiply(A=Add(A=ScreenPosition().PixelPosition, B=Add(A=@n3, B=@n4)), B=Constant() {desc=Scale, r=1})
  n5 = Custom!(None=@MaterialExpressionMultiply_3) {
   code="Mod(((uint)(Input.x) + 2 * (uint)(Input.y)), 5);"
   description=Grid
   outputType=CMOT_Float1
   showCode=true
   inputs [{
     inputName=Input
     input {expression=&self:MaterialExpressionMultiply_3}
    }]
  }
  MaterialExpressionMax_0 = Max(A=Constant() {desc=Iterations, r=2})
  n6 = Custom!(UV=@MaterialExpressionMultiply_3, Iterations=@MaterialExpressionMax_0) {
   code="// Bayer matrix\r\n\r\nint2 p = int2(UV);\r\n\r\nint sum = 0;\r\n\r\nfor(int i = 0; i < int(Iterations); i++)\r\n{\r\n    int2 t = p && 1;\r\n    sum = sum * 4 | (t.x ^ t.y) * 2 | t.x;\r\n    p *= 0.5;\r\n}    \r\n\r\nreturn float(sum) / float(1 << (2 * int(Iterations)));"
   description=Bayer
   outputType=CMOT_Float1
   showCode=true
   inputs [
    {
     inputName=UV
     input {expression=&self:MaterialExpressionMultiply_3}
    }
    {
     inputName=Iterations
     input {expression=&self:MaterialExpressionMax_0}
    }
   ]
  }
  n7 = Multiply(A=Divide(A=Add(A=@n5, B=@n6) {desc=address banding issues}) {constB=6}) {constB=0.998}
  n8 = CameraVectorWS()
  n9 = Multiply(A=DotProduct(A=VertexNormalWS(), B=@n8)) {constB=2}
  n10 = Saturate(None=Multiply(A=@n9, B=@n9))
  n11 = TextureCoordinate()
  n12 = Saturate(None=LinearInterpolate(
    A=ScalarParameter() {defaultValue=-1.5, group=Sprite_Base, parameterName=DitherFresnel HeadOnly - V Min, sortPriority=1100}
    B=ScalarParameter() {defaultValue=4, group=Sprite_Base, parameterName=DitherFresnel HeadOnly - V Max, sortPriority=1110}
    Alpha=ComponentMask(None=@n11) {g=true}
   ) {constA=-1.5, constB=4})
  n13 = Constant3Vector() {constant=<rgba 0 0 1 0>}
  n14 = Abs(None=DotProduct(A=Transform(None=@n13) {transformSourceType=TRANSFORMSOURCE_Local}, B=@n8))
  n15 = StaticSwitchParameter(True=LinearInterpolate(A=LinearInterpolate(A=@n10, Alpha=@n12), Alpha=Power(Base=@n14) {constExponent=5}), False=@n10) {parameterName=DitherFresnel HeadOnly}
  n16 = Subtract(A=Add(A=Add(A=@n7) {constB=0.001}, B=@n15)) {constB=0.5}
  n17 = StaticSwitchParameter(True=Saturate(None=Round(None=@n16)), False=Constant() {r=1}) {parameterName=DitherFresnel}
  FunctionOutput(None=SetMaterialAttributes!(MaterialAttributes=@n1.MaterialAttributes, WorldPositionOffset, OpacityMask=Multiply(A=@n1.OpacityMask, B=@n17), Opacity) {attributeSetTypes [#2, #1, #3]}) {outputName=MA}
  MaterialFunctionCall!(In=@n15, Iterations, Scale) {materialFunction=/Game/Packages/Fortress_SharedMaterials/Base_Material_Functions/MF_DitherBayerNoTex}
  n18 = VertexNormalWS()
  n19 = Normalize(VectorInput=Subtract(A=WorldPosition().XYZ, B=ObjectPositionWS()))
  n20 = Multiply(A=DotProduct(A=LinearInterpolate(A=@n18, B=@n19) {constAlpha=0.2}, B=CameraVectorWS())) {constB=1.2}
  n21 = Power(Base=@n20) {constExponent=1.5}
  n22 = Saturate(None=Multiply(A=Add(A=@n20) {constB=-0.1}))
  MaterialFunctionCall(Alpha=Saturate(None=LinearInterpolate(A=@n20, B=@n21, Alpha=@n22))) {materialFunction=&SmoothStep}
  n23 = Add(A=VertexInterpolator(VS=LocalPosition() {localOrigin=InstancePreSkinning}.Z).PS) {constB=-15}
  OneMinus(None=Saturate(None=Divide(A=@n23) {constB=11}))
 ]}
