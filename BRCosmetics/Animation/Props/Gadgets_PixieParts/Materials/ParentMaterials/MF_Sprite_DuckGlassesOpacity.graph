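# Material function that adds a duck-glasses lens look to Sprite prop materials, gated by the DuckGlass static switch.
# Opacity blends between edge and front values by camera angle to the local Y axis, scaled by a variant parameter.
# Emissive gains a fake long-lat texture reflection along that axis plus pixel-normal fresnel rims, tinted by oversaturated time-of-day color.
# A second output supplies the glass variant color as an emissive multiplier.
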MaterialFunction /BRCosmetics/Animation/Props/Gadgets_PixieParts/Materials/ParentMaterials/MF_Sprite_DuckGlassesOpacity
graph {nodes [
  n1 = GetMaterialAttributes(None=FunctionInput() {inputName=MA, inputType=FunctionInput_MaterialAttributes}) {attributeGetTypes [#1, #2]}
  n2 = ScalarParameter() {defaultValue=1, parameterName=OpacityFresnel}
  n3 = ScalarParameter() {defaultValue=0.9, parameterName=OpacityFront}
  n4 = CameraVectorWS()
  n5 = Transform(None=Constant3Vector() {constant=<rgba 0 1 0 0>}) {transformSourceType=TRANSFORMSOURCE_Local}
  n6 = Power(Base=Abs(None=DotProduct(A=@n5, B=@n4)))
  n7 = Multiply(
   A=@n1.Opacity
   B=Multiply(A=LinearInterpolate(A=@n2, B=@n3, Alpha=@n6), B=ScalarParameter() {defaultValue=1, parameterName=Glass Variant Opacity})
  )
  n8 = MaterialFunctionCall!(Vector=ReflectionVectorWS(CustomWorldNormal=@n5)) {materialFunction=/Engine/Functions/Engine_MaterialFunctions01/Coordinates/LongLatToUV}
  n9 = PixelNormalWS()
  n10 = CameraVectorWS()
  n11 = OneMinus(None=Abs(None=DotProduct(A=@n9, B=@n10)))
  n12 = Multiply(
   A=Multiply(A=TextureSample(UVs=@n8.Result) {texture=/BRCosmetics/Animation/Props/Gadgets_PixieParts/Textures/TA/T_DuckGlassReflectionTest}.RGB) {constB=3}
   B=LinearInterpolate(Alpha=Power(Base=@n11)) {constA=0.1}
  )
  n13 = Multiply(A=Power(Base=@n11) {constExponent=10})
  n14 = MaterialFunctionCall(
   Color_Morning=VectorParameter() {group=Sprite_Base, parameterName=TOD_Morning, sortPriority=9, defaultValue=<rgba 0.8 0.4 0.5 1>}.RGB
   Color_Day=VectorParameter() {group=Sprite_Base, parameterName=TOD_Day, sortPriority=9, defaultValue=<rgba 1 1 1 1>}.RGB
   Color_Evening=VectorParameter() {group=Sprite_Base, parameterName=TOD_Evening, sortPriority=9, defaultValue=<rgba 0.8 0.5 0.4 1>}.RGB
   Color_Night=VectorParameter() {group=Sprite_Base, parameterName=TOD_Night, sortPriority=9, defaultValue=<rgba 0.35 0.4667 0.7 1>}.RGB
  ) {materialFunction=&Color_Time_of_Day}
  n15 = Desaturation(None=@n14.Result)
  n16 = Add(A=@n1.EmissiveColor, B=Multiply(A=Add(A=@n12, B=@n13), B=LinearInterpolate(A=@n15, B=@n14.Result) {constAlpha=1.5}))
  FunctionOutput(None=SetMaterialAttributes!(
    MaterialAttributes=@n1.MaterialAttributes
    Opacity=StaticSwitchParameter(True=@n7, False=@n1.Opacity) {parameterName=DuckGlass}
    EmissiveColor=StaticSwitchParameter(True=@n16, False=@n1.EmissiveColor) {parameterName=DuckGlass}
   ) {attributeSetTypes [#1, #2]}) {outputName=MA}
  FunctionOutput(None=StaticSwitchParameter(True=VectorParameter() {parameterName=Glass Variant Color, defaultValue=<rgba 1 1 1 1>}.RGB, False=Constant() {r=1}) {parameterName=DuckGlass}) {outputName="Glass Emissive multiplier (after variant)", sortPriority=1}
  VertexNormalWS()
 ]}
