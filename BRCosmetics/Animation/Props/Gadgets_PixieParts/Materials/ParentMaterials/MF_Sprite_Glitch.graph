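# Material function that adds an optional visor glitch overlay to incoming material attributes for Pixie gadget sprite materials.
# A glitch texture scaled by an intensity parameter is added to Emissive and blended into Base Color, using its red channel as mask.
# The VisorGlitch static switch toggles both changes.
# Two generic visor glitch function calls are present but do not feed the output, so the texture samples at default UVs.
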MaterialFunction /BRCosmetics/Animation/Props/Gadgets_PixieParts/Materials/ParentMaterials/MF_Sprite_Glitch
graph {nodes [
  n1 = TextureSampleParameter2D() {
   parameterName=VisorGlitch Tex
   texture=/BRCosmetics/Animation/Props/Gadgets_PixieParts/Textures/Sprite_Skins/Sprite_Seven/T_Sprite_Seven_Creature_E
  }
  n2 = Multiply(A=ScalarParameter() {defaultValue=1, parameterName=VisorGlitch Intensity}, B=@n1.RGB)
  n3 = GetMaterialAttributes(None=FunctionInput() {inputName=MA, inputType=FunctionInput_MaterialAttributes}) {attributeGetTypes [#1, #2]}
  n4 = Add(A=@n3.EmissiveColor, B=@n2)
  n5 = LinearInterpolate(A=@n3.BaseColor, B=@n2, Alpha=@n1.R)
  FunctionOutput(None=SetMaterialAttributes!(
    MaterialAttributes=@n3.MaterialAttributes
    EmissiveColor=StaticSwitchParameter(True=@n4, False=@n3.EmissiveColor) {parameterName=VisorGlitch}
    BaseColor=StaticSwitchParameter(True=@n5, False=@n3.BaseColor) {parameterName=VisorGlitch}
   ) {attributeSetTypes [#1, #2]})
  MaterialFunctionCall!(UV) {materialFunction=/Game/Global/Materials/Character_Parent_2023/Visor_Parent_2025/MF_GenericVisor_Glitch_Grid}
  MaterialFunctionCall!(Main_TextureUV) {materialFunction=/Game/Global/Materials/Character_Parent_2023/Visor_Parent_2025/MF_GenericVisor_Glitch}
 ]}
