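# Material function that adds a shader-driven spin to part of a mesh, apparently the peanut Sprite's spinning hat, by modifying incoming material attributes.
# Looping time scaled by SpinSpeed and masked by vertex color green drives RotateVertexAboutAxis around SpinAxis and SpinPivot on pre-skinned local positions.
# The resulting offset adds to World Position Offset and the rotated normal replaces the original, both gated by the Spin static switch.
# The FXMask input and a RotateAboutAxis branch are unused.
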MaterialFunction /BRCosmetics/Animation/Props/Gadgets_PixieParts/Materials/ParentMaterials/MF_Sprite_PeanutSpinningHat
graph {nodes [
  ComponentMask(None=FunctionInput() {inputName=FXMask, sortPriority=2}) {g=true}
  n1 = GetMaterialAttributes(None=FunctionInput() {inputName=MA, inputType=FunctionInput_MaterialAttributes}) {attributeGetTypes [#1, #2]}
  n2 = VertexColor()
  n3 = Frac(None=Multiply(A=Time(), B=ScalarParameter() {defaultValue=1, parameterName=SpinSpeed}))
  n4 = Normalize(VectorInput=VectorParameter() {parameterName=SpinAxis, defaultValue=<rgba 0 -0.3 1 1>}.RGB)
  n5 = VectorParameter() {parameterName=SpinPivot, defaultValue=<rgba -3 0 37 1>}
  n6 = LocalPosition() {includedOffsets=ExcludeOffsets, localOrigin=InstancePreSkinning}
  n7 = MaterialFunctionCall!(
   RotationAngle=Multiply(A=@n3, B=@n2.G)
   NormalizedRotationAxis=@n4
   PivotPoint=@n5.RGB
   Position=@n6.XYZ
   Normal=Transform(None=@n1.Normal) {transformType=TRANSFORM_Local}
   TangentU
   TangentV
  ) {materialFunction=/BRCosmetics/Characters/Player/Male/Medium/Bodies/M_MED_Inferno/FX/Mat/RotateVertexAboutAxis}
  n8 = Add(A=@n1.WorldPositionOffset, B=Transform(None=@n7.WPO) {transformSourceType=TRANSFORMSOURCE_Local})
  n9 = Transform(None=@n7.Normal) {transformSourceType=TRANSFORMSOURCE_Local, transformType=TRANSFORM_Tangent}
  FunctionOutput(None=SetMaterialAttributes!(
    MaterialAttributes=@n1.MaterialAttributes
    WorldPositionOffset=StaticSwitchParameter(True=@n8, False=@n1.WorldPositionOffset) {parameterName=Spin}
    Normal=StaticSwitchParameter(True=@n9, False=@n1.Normal) {parameterName=Spin}
   ) {attributeSetTypes [#1, #2]}) {outputName=MA}
  n10 = RotateAboutAxis(NormalizedRotationAxis=@n4, RotationAngle=@n3, PivotPoint=@n5.RGB, Position=@n6.XYZ)
  Reroute(None=Multiply(A=Transform(None=@n10) {transformSourceType=TRANSFORMSOURCE_Local}, B=@n2.G)) {desc=WPO}
  TransformPosition(None=@n5.RGB) {transformType=TRANSFORMPOSSOURCE_World}
  TransformPosition(None=@n6.XYZ) {transformType=TRANSFORMPOSSOURCE_World}
  Transform(None=@n4) {transformSourceType=TRANSFORMSOURCE_Local}
  WorldPosition()
 ]}
