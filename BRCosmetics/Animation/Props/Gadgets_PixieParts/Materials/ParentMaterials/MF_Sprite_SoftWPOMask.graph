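# Material function for sprite materials in the Pixie Parts gadget props that softly masks World Position Offset using an FX mask.
# It takes material attributes and an FXMask input, smoothsteps the mask's green channel and inverts the result.
# That factor multiplies the incoming offset, fading displacement out as the mask rises, and the attributes are output with the adjusted offset.
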MaterialFunction /BRCosmetics/Animation/Props/Gadgets_PixieParts/Materials/ParentMaterials/MF_Sprite_SoftWPOMask
graph {nodes [
  n1 = GetMaterialAttributes(None=FunctionInput() {inputName=MA, inputType=FunctionInput_MaterialAttributes}) {attributeGetTypes [#1]}
  FunctionOutput(None=SetMaterialAttributes!(MaterialAttributes=@n1.MaterialAttributes, WorldPositionOffset) {attributeSetTypes [#1]}) {outputName=MA}
  n2 = ComponentMask(None=FunctionInput() {inputName=FXMask, sortPriority=1}) {g=true}
  Multiply(
   A=@n1.WorldPositionOffset
   B=OneMinus(None=MaterialFunctionCall(Alpha=@n2) {materialFunction=&SmoothStep}.Result)
  )
 ]}
