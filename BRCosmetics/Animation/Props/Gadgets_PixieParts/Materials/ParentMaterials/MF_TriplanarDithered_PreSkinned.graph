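# Material function that projects triplanar UVs from pre-skinned local position and normal, so the mapping follows skinned meshes.
# Normal-derived blend weights, softened by Triplanar Fade, feed a temporal screen-space dither plus noise that picks one projection plane per pixel, avoiding three texture samples.
# It outputs those UVs, a one-hot axis Mask, and DDX/DDY taken from scaled mesh UV0 rather than the discontinuous dithered UVs.
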MaterialFunction /BRCosmetics/Animation/Props/Gadgets_PixieParts/Materials/ParentMaterials/MF_TriplanarDithered_PreSkinned
graph {expressions {
  MaterialExpressionAbs_1 {class=MaterialExpressionAbs, inputs [None {from=MaterialExpressionFunctionInput_7, output=""}], properties {}}
  MaterialExpressionAdd_12 {
   class=MaterialExpressionAdd
   inputs [
    A {from=MaterialExpressionScreenPosition_0, output=PixelPosition}
    B {from=MaterialExpressionAppendVector_2}
   ]
  }
  MaterialExpressionAdd_13 {
   class=MaterialExpressionAdd
   inputs [
    A {from=MaterialExpressionCustom_0}
    B {from=MaterialExpressionTextureSample_0, output=R}
   ]
   properties {constB=0}
  }
  MaterialExpressionAdd_14 {
   class=MaterialExpressionAdd
   inputs [
    A {from=MaterialExpressionMultiply_3}
    B {from=MaterialExpressionMultiply_8}
   ]
  }
  MaterialExpressionAdd_15 {
   class=MaterialExpressionAdd
   inputs [
    A {from=MaterialExpressionAdd_14}
    B {from=MaterialExpressionMultiply_10}
   ]
  }
  MaterialExpressionAdd_16 {
   class=MaterialExpressionAdd
   inputs [
    A {from=MaterialExpressionComponentMask_11}
    B {from=MaterialExpressionAppendVector_5}
   ]
  }
  MaterialExpressionAdd_17 {
   class=MaterialExpressionAdd
   inputs [
    A {from=MaterialExpressionConvert_4}
    B {from=MaterialExpressionDivide_6}
   ]
   properties {constB=0.2}
  }
  MaterialExpressionAdd_18 {
   class=MaterialExpressionAdd
   inputs [
    A {from=MaterialExpressionConvert_5}
    B {from=MaterialExpressionConvert_5}
   ]
  }
  MaterialExpressionAdd_19 {
   class=MaterialExpressionAdd
   inputs [
    A {from=MaterialExpressionAdd_18}
    B {from=MaterialExpressionConvert_5}
   ]
  }
  MaterialExpressionAppendVector_2 {
   class=MaterialExpressionAppendVector
   inputs [
    A {from=MaterialExpressionViewProperty_1, output=Property}
    B {from=MaterialExpressionViewProperty_1, output=Property}
   ]
  }
  MaterialExpressionAppendVector_5 {
   class=MaterialExpressionAppendVector
   inputs [
    A {from=MaterialExpressionMax_6, output=""}
    B {from=MaterialExpressionOneMinus_0, output=""}
   ]
  }
  MaterialExpressionAppendVector_6 {
   class=MaterialExpressionAppendVector
   inputs [
    A {from=MaterialExpressionComponentMask_5, output=""}
    B {from=MaterialExpressionComponentMask_8, output=""}
   ]
  }
  MaterialExpressionAppendVector_7 {
   class=MaterialExpressionAppendVector
   inputs [
    A {from=MaterialExpressionAppendVector_6, output=""}
    B {from=MaterialExpressionOneMinus_2, output=""}
   ]
  }
  MaterialExpressionComponentMask_0 {class=MaterialExpressionComponentMask, inputs [None {from=MaterialExpressionDivide_4}], properties {b=true, g=true}}
  MaterialExpressionComponentMask_11 {class=MaterialExpressionComponentMask, inputs [None {from=MaterialExpressionDivide_5}], properties {g=true, r=true}}
  MaterialExpressionComponentMask_5 {class=MaterialExpressionComponentMask, inputs [None {from=MaterialExpressionFloor_1}], properties {r=true}}
  MaterialExpressionComponentMask_7 {class=MaterialExpressionComponentMask, inputs [None {from=MaterialExpressionDivide_4}], properties {b=true, r=true}}
  MaterialExpressionComponentMask_8 {class=MaterialExpressionComponentMask, inputs [None {from=MaterialExpressionFloor_1}], properties {g=true}}
  MaterialExpressionComponentMask_9 {class=MaterialExpressionComponentMask, inputs [None {from=MaterialExpressionDivide_4}], properties {g=true, r=true}}
  MaterialExpressionConstant2Vector_1 {class=MaterialExpressionConstant2Vector, properties {g=64, r=64}}
  MaterialExpressionConstant3Vector_1 {class=MaterialExpressionConstant3Vector, properties {constant=<rgba 1 1 -1 0>}}
  MaterialExpressionConstant_0 {class=MaterialExpressionConstant, properties {r=0.2}}
  MaterialExpressionConvert_3 {
   class=MaterialExpressionConvert
   inputs [None {from=MaterialExpressionAbs_1}]
   properties {
    nodeName=Break Float3
    convertInputs [{
      type=Vector3
      expressionInput {
       expression=&self:MaterialExpressionAbs_1
       inputName=None
       mask=0
       maskA=0
       maskB=0
       maskG=0
       maskR=0
       outputIndex=0
      }
     }]
    convertOutputs [{type=Scalar}, {type=Scalar}, {type=Scalar}]
   }
  }
  MaterialExpressionConvert_4 {
   class=MaterialExpressionConvert
   inputs [
    None {from=MaterialExpressionSubtract_3}
    None {from=MaterialExpressionSubtract_4}
    None {from=MaterialExpressionSubtract_5}
   ]
   properties {
    nodeName=Make Vector3
    convertInputs [
     {
      type=Scalar
      expressionInput {
       expression=&self:MaterialExpressionSubtract_3
       inputName=None
       mask=0
       maskA=0
       maskB=0
       maskG=0
       maskR=0
       outputIndex=0
      }
     }
     {
      type=Scalar
      expressionInput {
       expression=&self:MaterialExpressionSubtract_4
       inputName=None
       mask=0
       maskA=0
       maskB=0
       maskG=0
       maskR=0
       outputIndex=0
      }
     }
     {
      type=Scalar
      expressionInput {
       expression=&self:MaterialExpressionSubtract_5
       inputName=None
       mask=0
       maskA=0
       maskB=0
       maskG=0
       maskR=0
       outputIndex=0
      }
     }
    ]
    convertOutputs [{type=Vector3}]
   }
  }
  MaterialExpressionConvert_5 {
   class=MaterialExpressionConvert
   inputs [None {from=MaterialExpressionMultiply_12}]
   properties {
    nodeName=Break Float3
    convertInputs [{
      type=Vector3
      expressionInput {
       expression=&self:MaterialExpressionMultiply_12
       inputName=None
       mask=0
       maskA=0
       maskB=0
       maskG=0
       maskR=0
       outputIndex=0
      }
     }]
    convertOutputs [{type=Scalar}, {type=Scalar}, {type=Scalar}]
   }
  }
  MaterialExpressionCustom_0 {
   class=MaterialExpressionCustom
   inputs [p {from=MaterialExpressionAdd_12, output=""}]
   properties {
    code="Mod( ((uint)(p.x) + 2 * (uint)(p.y)) , 5 )"
    outputType=CMOT_Float1
    inputs [{
      inputName=p
      input {
       expression=&self:MaterialExpressionAdd_12
       inputName=None
       mask=0
       maskA=0
       maskB=0
       maskG=0
       maskR=0
       outputIndex=0
      }
     }]
   }
  }
  MaterialExpressionDDX_0 {class=MaterialExpressionDDX, inputs [Value {from=MaterialExpressionMultiply_9, output=""}], properties {}}
  MaterialExpressionDDY_0 {class=MaterialExpressionDDY, inputs [Value {from=MaterialExpressionMultiply_9, output=""}], properties {}}
  MaterialExpressionDivide_4 {
   class=MaterialExpressionDivide
   inputs [
    A {from=MaterialExpressionMultiply_11}
    B {from=MaterialExpressionFunctionInput_5}
   ]
   properties {constB=120}
  }
  MaterialExpressionDivide_5 {
   class=MaterialExpressionDivide
   inputs [
    A {from=MaterialExpressionMultiply_12}
    B {from=MaterialExpressionAdd_19}
   ]
  }
  MaterialExpressionDivide_6 {
   class=MaterialExpressionDivide
   inputs [
    A {from=""}
    B {from=MaterialExpressionFunctionInput_6}
   ]
  }
  MaterialExpressionFloor_1 {class=MaterialExpressionFloor, inputs [None {from=MaterialExpressionAdd_16, output=""}], properties {}}
  MaterialExpressionFunctionInput_2 {
   class=MaterialExpressionFunctionInput
   inputs [Preview {from="", output=""}]
   properties {bUsePreviewValueAsDefault=true, inputName=Derivatives Scale, inputType=FunctionInput_Scalar, sortPriority=2, previewValue=<1 1 1 1>}
  }
  MaterialExpressionFunctionInput_5 {
   class=MaterialExpressionFunctionInput
   inputs [Preview {from="", output=""}]
   properties {bUsePreviewValueAsDefault=true, inputName=Texture Scale, inputType=FunctionInput_Scalar, previewValue=<120 120 120 120>}
  }
  MaterialExpressionFunctionInput_6 {
   class=MaterialExpressionFunctionInput
   inputs [Preview {from=MaterialExpressionConstant_0, output=""}]
   properties {bUsePreviewValueAsDefault=true, inputName=Triplanar Fade, inputType=FunctionInput_Scalar, sortPriority=1, previewValue=<10 10 10 1>}
  }
  MaterialExpressionFunctionInput_7 {
   class=MaterialExpressionFunctionInput
   inputs [Preview {from=MaterialExpressionVertexInterpolator_3, output=PS}]
   properties {bUsePreviewValueAsDefault=true, description=VertexNormalWS as Default, inputName=Normal}
  }
  MaterialExpressionFunctionInput_8 {
   class=MaterialExpressionFunctionInput
   inputs [Preview {from=MaterialExpressionVertexInterpolator_0, output=PS}]
   properties {bUsePreviewValueAsDefault=true, inputName=InPosition, previewValue=<120 120 120 120>}
  }
  MaterialExpressionFunctionOutput_0 {class=MaterialExpressionFunctionOutput, inputs [None {from=MaterialExpressionAdd_15, output=""}], properties {outputName=UVs}}
  MaterialExpressionFunctionOutput_1 {class=MaterialExpressionFunctionOutput, inputs [None {from=MaterialExpressionDDX_0, output=""}], properties {outputName=DDX, sortPriority=1}}
  MaterialExpressionFunctionOutput_2 {class=MaterialExpressionFunctionOutput, inputs [None {from=MaterialExpressionDDY_0, output=""}], properties {outputName=DDY, sortPriority=2}}
  MaterialExpressionFunctionOutput_3 {
   class=MaterialExpressionFunctionOutput
   inputs [None {from=MaterialExpressionAppendVector_7, output=""}]
   properties {outputName=Mask, sortPriority=3}
  }
  MaterialExpressionLocalPosition_1 {class=MaterialExpressionLocalPosition, inputs [], properties {localOrigin=InstancePreSkinning}}
  MaterialExpressionMaterialFunctionCall_6 {
   class=MaterialExpressionMaterialFunctionCall
   inputs [TextureResolution {from=MaterialExpressionConstant2Vector_1}]
   properties {materialFunction=/Engine/Functions/Engine_MaterialFunctions02/Texturing/ScreenAlignedPixelToPixelUVs}
  }
  MaterialExpressionMax_5 {
   class=MaterialExpressionMax
   inputs [
    A {from=MaterialExpressionComponentMask_5}
    B {from=MaterialExpressionComponentMask_8}
   ]
  }
  MaterialExpressionMax_6 {
   class=MaterialExpressionMax
   inputs [
    A {from=MaterialExpressionMultiply_4}
    B {from=""}
   ]
   properties {constB=0.001}
  }
  MaterialExpressionMax_7 {
   class=MaterialExpressionMax
   inputs [
    A {from=MaterialExpressionConvert_3}
    B {from=MaterialExpressionConvert_3}
   ]
  }
  MaterialExpressionMax_8 {
   class=MaterialExpressionMax
   inputs [
    A {from=MaterialExpressionConvert_3}
    B {from=MaterialExpressionConvert_3}
   ]
  }
  MaterialExpressionMax_9 {
   class=MaterialExpressionMax
   inputs [
    A {from=MaterialExpressionConvert_3}
    B {from=MaterialExpressionConvert_3}
   ]
  }
  MaterialExpressionMultiply_10 {
   class=MaterialExpressionMultiply
   inputs [
    A {from=MaterialExpressionComponentMask_9}
    B {from=MaterialExpressionOneMinus_2}
   ]
  }
  MaterialExpressionMultiply_11 {
   class=MaterialExpressionMultiply
   inputs [
    A {from=MaterialExpressionFunctionInput_8}
    B {from=MaterialExpressionConstant3Vector_1}
   ]
  }
  MaterialExpressionMultiply_12 {
   class=MaterialExpressionMultiply
   inputs [
    A {from=MaterialExpressionSaturate_3}
    B {from=MaterialExpressionSaturate_3}
   ]
  }
  MaterialExpressionMultiply_3 {
   class=MaterialExpressionMultiply
   inputs [
    A {from=MaterialExpressionComponentMask_0}
    B {from=MaterialExpressionComponentMask_5}
   ]
  }
  MaterialExpressionMultiply_4 {
   class=MaterialExpressionMultiply
   inputs [
    A {from=MaterialExpressionAdd_13}
    B {from=""}
   ]
   properties {constB=0.1667}
  }
  MaterialExpressionMultiply_8 {
   class=MaterialExpressionMultiply
   inputs [
    A {from=MaterialExpressionComponentMask_7}
    B {from=MaterialExpressionComponentMask_8}
   ]
  }
  MaterialExpressionMultiply_9 {
   class=MaterialExpressionMultiply
   inputs [
    A {from=MaterialExpressionTextureCoordinate_0}
    B {from=MaterialExpressionFunctionInput_2}
   ]
  }
  MaterialExpressionOneMinus_0 {class=MaterialExpressionOneMinus, inputs [None {from=MaterialExpressionMax_6}]}
  MaterialExpressionOneMinus_2 {class=MaterialExpressionOneMinus, inputs [None {from=MaterialExpressionMax_5}]}
  MaterialExpressionPreSkinnedNormal_1 {class=MaterialExpressionPreSkinnedNormal, inputs [], properties {}}
  MaterialExpressionSaturate_3 {class=MaterialExpressionSaturate, inputs [None {from=MaterialExpressionAdd_17}]}
  MaterialExpressionScreenPosition_0 {class=MaterialExpressionScreenPosition, inputs [], properties {}}
  MaterialExpressionSubtract_3 {
   class=MaterialExpressionSubtract
   inputs [
    A {from=MaterialExpressionConvert_3}
    B {from=MaterialExpressionMax_7}
   ]
  }
  MaterialExpressionSubtract_4 {
   class=MaterialExpressionSubtract
   inputs [
    A {from=MaterialExpressionConvert_3}
    B {from=MaterialExpressionMax_8}
   ]
  }
  MaterialExpressionSubtract_5 {
   class=MaterialExpressionSubtract
   inputs [
    A {from=MaterialExpressionConvert_3}
    B {from=MaterialExpressionMax_9}
   ]
  }
  MaterialExpressionTextureCoordinate_0 {class=MaterialExpressionTextureCoordinate}
  MaterialExpressionTextureSample_0 {
   class=MaterialExpressionTextureSample
   inputs [
    UVs {from=MaterialExpressionMaterialFunctionCall_6, output=Result}
    Tex {from=""}
    Apply View MipBias {from=""}
   ]
   properties {samplerType=SAMPLERTYPE_LinearGrayscale, texture=/Engine/EngineMaterials/Good64x64TilingNoiseHighFreq}
  }
  MaterialExpressionVertexInterpolator_0 {class=MaterialExpressionVertexInterpolator, inputs [VS {from=MaterialExpressionLocalPosition_1, output=XYZ}], properties {}}
  MaterialExpressionVertexInterpolator_3 {class=MaterialExpressionVertexInterpolator, inputs [VS {from=MaterialExpressionPreSkinnedNormal_1, output=""}], properties {}}
  MaterialExpressionViewProperty_1 {class=MaterialExpressionViewProperty, inputs [], properties {property=MEVP_TemporalSampleIndex}}
 }}
