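# Material function that layers a candy variant look onto incoming material attributes for sprite props, selectable through a static switch.
# Input masks target variant, logo, glow, fresnel and light regions for recoloring and roughness/metallic changes.
# Pre-skinned triplanar sampling and noise build sugar masks, with DDX/DDY deriving the sugar normal.
# Sun and camera vectors feed glitter and fake back-scatter shading, alongside time-of-day colors, distance blending and a bounds-based Z gradient.
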MaterialFunction /BRCosmetics/Animation/Props/Gadgets_PixieParts/Materials/ParentMaterials/MF_Sprite_Variants_Candy
graph {expressions {
  MaterialExpressionAbs_2 {class=MaterialExpressionAbs, inputs [None {from=MaterialExpressionSine_2, output=""}], properties {}}
  MaterialExpressionAdd_33 {
   class=MaterialExpressionAdd
   inputs [
    A {from=MaterialExpressionMultiply_89}
    B {from=MaterialExpressionScalarParameter_102}
   ]
  }
  MaterialExpressionAdd_34 {
   class=MaterialExpressionAdd
   inputs [
    A {from=MaterialExpressionMultiply_90}
    B {from=MaterialExpressionScalarParameter_103}
   ]
  }
  MaterialExpressionAdd_35 {
   class=MaterialExpressionAdd
   inputs [
    A {from=MaterialExpressionPower_14}
    B {from=MaterialExpressionScalarParameter_106}
   ]
   properties {constB=-0.5}
  }
  MaterialExpressionAdd_36 {
   class=MaterialExpressionAdd
   inputs [
    A {from=MaterialExpressionAdd_34}
    B {from=MaterialExpressionScalarParameter_135}
   ]
   properties {constB=0.15}
  }
  MaterialExpressionAdd_37 {
   class=MaterialExpressionAdd
   inputs [
    A {from=MaterialExpressionMultiply_90}
    B {from=""}
   ]
   properties {constB=0.15}
  }
  MaterialExpressionAdd_38 {
   class=MaterialExpressionAdd
   inputs [
    A {from=MaterialExpressionReroute_52}
    B {from=MaterialExpressionMultiply_95}
   ]
  }
  MaterialExpressionAdd_39 {
   class=MaterialExpressionAdd
   inputs [
    A {from=MaterialExpressionNamedRerouteUsage_131}
    B {from=MaterialExpressionConstant_19}
   ]
  }
  MaterialExpressionAdd_40 {
   class=MaterialExpressionAdd
   inputs [
    A {from=MaterialExpressionNamedRerouteUsage_133}
    B {from=MaterialExpressionConstant_21}
   ]
  }
  MaterialExpressionAdd_41 {
   class=MaterialExpressionAdd
   inputs [
    A {from=MaterialExpressionDivide_15}
    B {from=MaterialExpressionConstant2Vector_2}
   ]
  }
  MaterialExpressionAdd_42 {
   class=MaterialExpressionAdd
   inputs [
    A {from=MaterialExpressionMultiply_116}
    B {from=""}
   ]
   properties {constB=-1}
  }
  MaterialExpressionAdd_43 {
   class=MaterialExpressionAdd
   inputs [
    A {from=MaterialExpressionNamedRerouteUsage_140}
    B {from=MaterialExpressionVertexNormalWS_6}
   ]
  }
  MaterialExpressionAdd_44 {
   class=MaterialExpressionAdd
   inputs [
    A {from=MaterialExpressionCustom_5}
    B {from=MaterialExpressionTransform_7}
   ]
  }
  MaterialExpressionAdd_45 {
   class=MaterialExpressionAdd
   inputs [
    A {from=MaterialExpressionMultiply_130}
    B {from=""}
   ]
  }
  MaterialExpressionAdd_46 {
   class=MaterialExpressionAdd
   inputs [
    A {from=MaterialExpressionMultiply_127}
    B {from=""}
   ]
  }
  MaterialExpressionBounds_0 {class=MaterialExpressionBounds, inputs [], properties {type=MEILB_PreSkinnedLocal}}
  MaterialExpressionCameraVectorWS_3 {class=MaterialExpressionCameraVectorWS, inputs [], properties {}}
  MaterialExpressionCameraVectorWS_4 {class=MaterialExpressionCameraVectorWS, inputs [], properties {desc=ViewDir}}
  MaterialExpressionCollectionParameter_1 {
   class=MaterialExpressionCollectionParameter
   inputs []
   properties {
    collection=&FortniteMaterialParameters
    parameterName=SunlightDirectionalVector
   }
  }
  MaterialExpressionComponentMask_16 {class=MaterialExpressionComponentMask, inputs [None {from=MaterialExpressionDivide_28}], properties {b=true}}
  MaterialExpressionComponentMask_22 {class=MaterialExpressionComponentMask, inputs [None {from=MaterialExpressionCollectionParameter_1}], properties {b=true, g=true, r=true}}
  MaterialExpressionComponentMask_24 {class=MaterialExpressionComponentMask, inputs [None {from=MaterialExpressionNamedRerouteUsage_123}], properties {b=true, r=true}}
  MaterialExpressionComponentMask_27 {class=MaterialExpressionComponentMask, inputs [None {from=MaterialExpressionNamedRerouteUsage_128}], properties {r=true}}
  MaterialExpressionComponentMask_28 {class=MaterialExpressionComponentMask, inputs [None {from=MaterialExpressionLinearInterpolate_82}], properties {b=true, g=true, r=true}}
  MaterialExpressionConstant2Vector_2 {class=MaterialExpressionConstant2Vector, properties {r=0.15}}
  MaterialExpressionConstant3Vector_10 {class=MaterialExpressionConstant3Vector, properties {constant=<rgba 1 1 1 0>}}
  MaterialExpressionConstant3Vector_11 {class=MaterialExpressionConstant3Vector, properties {constant=<rgba 0 0 1 0>}}
  MaterialExpressionConstant3Vector_12 {class=MaterialExpressionConstant3Vector, properties {constant=<rgba 0 0 1 0>}}
  MaterialExpressionConstant3Vector_13 {class=MaterialExpressionConstant3Vector, properties {constant=<rgba 0 0 -1 0>}}
  MaterialExpressionConstant3Vector_14 {class=MaterialExpressionConstant3Vector, properties {constant=<rgba 0 0 1 0>}}
  MaterialExpressionConstant4Vector_6 {class=MaterialExpressionConstant4Vector, inputs [], properties {desc=1.1.1. 0.7, constant=<rgba 0 -1 6 1>}}
  MaterialExpressionConstant4Vector_7 {class=MaterialExpressionConstant4Vector, inputs [], properties {desc=1.1.1. 0.7, constant=<rgba 3 2 1 1>}}
  MaterialExpressionConstant4Vector_8 {class=MaterialExpressionConstant4Vector, inputs [], properties {constant=<rgba 0.7917 0.1113 0.6083 1>}}
  MaterialExpressionConstant_17 {class=MaterialExpressionConstant}
  MaterialExpressionConstant_18 {class=MaterialExpressionConstant, properties {r=10}}
  MaterialExpressionConstant_19 {class=MaterialExpressionConstant, properties {r=0.4}}
  MaterialExpressionConstant_20 {class=MaterialExpressionConstant, properties {r=50}}
  MaterialExpressionConstant_21 {class=MaterialExpressionConstant, properties {r=0.1}}
  MaterialExpressionConstant_22 {class=MaterialExpressionConstant, properties {r=1}}
  MaterialExpressionConstant_23 {class=MaterialExpressionConstant, properties {r=0.1}}
  MaterialExpressionConvert_2 {
   class=MaterialExpressionConvert
   inputs [
    None {from=MaterialExpressionDDX_2}
    None {from=MaterialExpressionDDY_2}
    None {from=""}
   ]
   properties {
    nodeName=Make Vector3
    convertInputs [
     {
      type=Scalar
      expressionInput {
       expression=&self:MaterialExpressionDDX_2
       inputName=None
       mask=0
       maskA=0
       maskB=0
       maskG=0
       maskR=0
       outputIndex=0
      }
     }
     {
      type=Scalar
      expressionInput {
       expression=&self:MaterialExpressionDDY_2
       inputName=None
       mask=0
       maskA=0
       maskB=0
       maskG=0
       maskR=0
       outputIndex=0
      }
     }
     {
      type=Scalar
      defaultValue=<rgba 1 0 0 1>
      expressionInput {expression=null, inputName=None, mask=0, maskA=0, maskB=0, maskG=0, maskR=0, outputIndex=0}
     }
    ]
    convertOutputs [{type=Vector3}]
   }
  }
  MaterialExpressionConvert_5 {class=MaterialExpressionConvert, inputs [], properties {convertOutputs [{type=Vector3, defaultValue=<rgba 1 1 1 1>}]}}
  MaterialExpressionCustom_4 {
   class=MaterialExpressionCustom
   inputs []
   properties {code="return View.DirectionalLightColor;", description=DirectionalLightColor, inputs []}
  }
  MaterialExpressionCustom_5 {
   class=MaterialExpressionCustom
   inputs []
   properties {code="return View.DirectionalLightDirection;", description=DirectionalLightVector, inputs []}
  }
  MaterialExpressionDDX_2 {class=MaterialExpressionDDX, inputs [Value {from=MaterialExpressionMultiply_98, output=""}], properties {}}
  MaterialExpressionDDX_3 {class=MaterialExpressionDDX, inputs [Value {from=MaterialExpressionMultiply_163, output=""}], properties {}}
  MaterialExpressionDDY_2 {class=MaterialExpressionDDY, inputs [Value {from=MaterialExpressionMultiply_98, output=""}], properties {}}
  MaterialExpressionDDY_3 {class=MaterialExpressionDDY, inputs [Value {from=MaterialExpressionMultiply_163, output=""}], properties {}}
  MaterialExpressionDesaturation_2 {
   class=MaterialExpressionDesaturation
   inputs [
    None {from=MaterialExpressionReroute_53, output=""}
    Fraction {from=MaterialExpressionScalarParameter_115, output=""}
   ]
   properties {}
  }
  MaterialExpressionDivide_15 {
   class=MaterialExpressionDivide
   inputs [
    A {from=MaterialExpressionComponentMask_24}
    B {from=MaterialExpressionScalarParameter_119}
   ]
  }
  MaterialExpressionDivide_16 {
   class=MaterialExpressionDivide
   inputs [
    A {from=MaterialExpressionComponentMask_24}
    B {from=MaterialExpressionMultiply_100}
   ]
   properties {constB=7.237}
  }
  MaterialExpressionDivide_19 {
   class=MaterialExpressionDivide
   inputs [
    A {from=MaterialExpressionDotProduct_19}
    B {from=""}
   ]
   properties {constB=4}
  }
  MaterialExpressionDivide_28 {
   class=MaterialExpressionDivide
   inputs [
    A {from=MaterialExpressionSubtract_41}
    B {from=MaterialExpressionNamedRerouteUsage_155}
   ]
  }
  MaterialExpressionDivide_32 {
   class=MaterialExpressionDivide
   inputs [
    A {from=""}
    B {from=MaterialExpressionNamedRerouteUsage_183}
   ]
  }
  MaterialExpressionDotProduct_14 {
   class=MaterialExpressionDotProduct
   inputs [
    A {from=MaterialExpressionComponentMask_22, output=""}
    B {from=MaterialExpressionReroute_45, output=""}
   ]
   properties {}
  }
  MaterialExpressionDotProduct_15 {
   class=MaterialExpressionDotProduct
   inputs [
    A {from=MaterialExpressionNormalize_4, output=""}
    B {from=MaterialExpressionMultiply_86, output=""}
   ]
   properties {}
  }
  MaterialExpressionDotProduct_16 {
   class=MaterialExpressionDotProduct
   inputs [
    A {from=MaterialExpressionNamedRerouteUsage_114, output=""}
    B {from=MaterialExpressionVectorParameter_40, output=RGB}
   ]
   properties {}
  }
  MaterialExpressionDotProduct_17 {
   class=MaterialExpressionDotProduct
   inputs [
    A {from=MaterialExpressionCameraVectorWS_3, output=""}
    B {from=MaterialExpressionVertexNormalWS_4, output=""}
   ]
   properties {}
  }
  MaterialExpressionDotProduct_18 {
   class=MaterialExpressionDotProduct
   inputs [
    A {from=MaterialExpressionNormalize_7, output=""}
    B {from=MaterialExpressionNormalize_6, output=""}
   ]
   properties {}
  }
  MaterialExpressionDotProduct_19 {
   class=MaterialExpressionDotProduct
   inputs [
    A {from=MaterialExpressionTransform_8, output=""}
    B {from=MaterialExpressionMaterialFunctionCall_32, output=Result}
   ]
   properties {}
  }
  MaterialExpressionDotProduct_20 {
   class=MaterialExpressionDotProduct
   inputs [
    A {from=MaterialExpressionReflectionVectorWS_5, output=""}
    B {from=MaterialExpressionMaterialFunctionCall_33, output=Result}
   ]
   properties {}
  }
  MaterialExpressionDotProduct_21 {
   class=MaterialExpressionDotProduct
   inputs [
    A {from=MaterialExpressionConvert_5, output=""}
    B {from=MaterialExpressionNamedRerouteDeclaration_129, output=""}
   ]
   properties {}
  }
  MaterialExpressionFunctionInput_0 {
   class=MaterialExpressionFunctionInput
   inputs [Preview {from="", output=""}]
   properties {inputName=MaterialAttributes, inputType=FunctionInput_MaterialAttributes, sortPriority=-1}
  }
  MaterialExpressionFunctionInput_1 {
   class=MaterialExpressionFunctionInput
   inputs [Preview {from="", output=""}]
   properties {inputName=Mask_BlinkingLights, inputType=FunctionInput_Scalar, sortPriority=1}
  }
  MaterialExpressionFunctionInput_10 {
   class=MaterialExpressionFunctionInput
   inputs [Preview {from="", output=""}]
   properties {inputName=WPO, inputType=FunctionInput_Scalar, sortPriority=2}
  }
  MaterialExpressionFunctionInput_12 {
   class=MaterialExpressionFunctionInput
   inputs [Preview {from="", output=""}]
   properties {bUsePreviewValueAsDefault=true, inputName=Pre-Skinned Scale, sortPriority=130, previewValue=<1 1 1 1>}
  }
  MaterialExpressionFunctionInput_2 {
   class=MaterialExpressionFunctionInput
   inputs [Preview {from="", output=""}]
   properties {inputName=Mask_DiffuseChannelMask, inputType=FunctionInput_Scalar, sortPriority=1}
  }
  MaterialExpressionFunctionInput_3 {
   class=MaterialExpressionFunctionInput
   inputs [Preview {from=MaterialExpressionShaderStageSwitch_6, output=""}]
   properties {bUsePreviewValueAsDefault=true, inputName=Pre-Skinned Local Pos, sortPriority=100}
  }
  MaterialExpressionFunctionInput_4 {
   class=MaterialExpressionFunctionInput
   inputs [Preview {from=MaterialExpressionBounds_0, output=Max}]
   properties {bUsePreviewValueAsDefault=true, inputName=Pre-Skinned Bounds Max, sortPriority=120}
  }
  MaterialExpressionFunctionInput_5 {
   class=MaterialExpressionFunctionInput
   inputs [Preview {from=MaterialExpressionBounds_0, output=Min}]
   properties {bUsePreviewValueAsDefault=true, inputName=Pre-Skinned Bounds Min, sortPriority=110}
  }
  MaterialExpressionFunctionInput_6 {
   class=MaterialExpressionFunctionInput
   inputs [Preview {from="", output=""}]
   properties {inputName=Mask Variant, inputType=FunctionInput_Scalar, sortPriority=1}
  }
  MaterialExpressionFunctionInput_7 {
   class=MaterialExpressionFunctionInput
   inputs [Preview {from="", output=""}]
   properties {inputName=Mask_InnerRadialGlow, inputType=FunctionInput_Scalar, sortPriority=1}
  }
  MaterialExpressionFunctionInput_8 {
   class=MaterialExpressionFunctionInput
   inputs [Preview {from="", output=""}]
   properties {inputName=Mask_Logo, inputType=FunctionInput_Scalar, sortPriority=1}
  }
  MaterialExpressionFunctionInput_9 {
   class=MaterialExpressionFunctionInput
   inputs [Preview {from="", output=""}]
   properties {inputName=Mask_SpriteFresnel, inputType=FunctionInput_Scalar, sortPriority=1}
  }
  MaterialExpressionFunctionOutput_1 {
   class=MaterialExpressionFunctionOutput
   inputs [None {from=MaterialExpressionStaticSwitchParameter_0, output=""}]
   properties {outputName=SetMaterialAttributes}
  }
  MaterialExpressionGetMaterialAttributes_2 {
   class=MaterialExpressionGetMaterialAttributes
   inputs [None {from=MaterialExpressionFunctionInput_0, output=""}]
   properties {attributeGetTypes [
     #1
     #2
     #3
     #4
     #5
    ]}
  }
  MaterialExpressionLinearInterpolate_58 {
   class=MaterialExpressionLinearInterpolate
   inputs [
    A {from=MaterialExpressionReroute_60}
    B {from=MaterialExpressionMultiply_103}
    Alpha {from=MaterialExpressionNamedRerouteUsage_101}
   ]
  }
  MaterialExpressionLinearInterpolate_59 {
   class=MaterialExpressionLinearInterpolate
   inputs [
    A {from=MaterialExpressionVectorParameter_38, output=RGB}
    B {from=MaterialExpressionVectorParameter_37, output=RGB}
    Alpha {from=MaterialExpressionNamedRerouteUsage_102}
   ]
  }
  MaterialExpressionLinearInterpolate_60 {
   class=MaterialExpressionLinearInterpolate
   inputs [
    A {from=MaterialExpressionLinearInterpolate_67}
    B {from=MaterialExpressionNamedRerouteUsage_120}
    Alpha {from=MaterialExpressionScalarParameter_107}
   ]
  }
  MaterialExpressionLinearInterpolate_61 {
   class=MaterialExpressionLinearInterpolate
   inputs [
    A {from=MaterialExpressionConstant3Vector_10}
    B {from=MaterialExpressionLinearInterpolate_77}
    Alpha {from=MaterialExpressionNamedRerouteUsage_104}
   ]
  }
  MaterialExpressionLinearInterpolate_62 {
   class=MaterialExpressionLinearInterpolate
   inputs [
    A {from=MaterialExpressionScalarParameter_111}
    B {from=MaterialExpressionScalarParameter_112}
    Alpha {from=MaterialExpressionNamedRerouteDeclaration_113}
   ]
  }
  MaterialExpressionLinearInterpolate_63 {
   class=MaterialExpressionLinearInterpolate
   inputs [
    A {from=MaterialExpressionScalarParameter_113}
    B {from=MaterialExpressionScalarParameter_114}
    Alpha {from=MaterialExpressionNamedRerouteUsage_137}
   ]
  }
  MaterialExpressionLinearInterpolate_64 {
   class=MaterialExpressionLinearInterpolate
   inputs [
    A {from=MaterialExpressionReroute_49}
    B {from=MaterialExpressionLinearInterpolate_65}
    Alpha {from=MaterialExpressionNamedRerouteUsage_112}
   ]
  }
  MaterialExpressionLinearInterpolate_65 {
   class=MaterialExpressionLinearInterpolate
   inputs [
    A {from=MaterialExpressionReroute_48}
    B {from=MaterialExpressionMultiply_92}
    Alpha {from=MaterialExpressionSaturate_25}
   ]
  }
  MaterialExpressionLinearInterpolate_66 {
   class=MaterialExpressionLinearInterpolate
   inputs [
    A {from=MaterialExpressionReroute_51}
    B {from=MaterialExpressionLinearInterpolate_70}
    Alpha {from=MaterialExpressionNamedRerouteUsage_113}
   ]
  }
  MaterialExpressionLinearInterpolate_67 {
   class=MaterialExpressionLinearInterpolate
   inputs [
    A {from=MaterialExpressionReroute_53}
    B {from=MaterialExpressionMultiply_96}
    Alpha {from=MaterialExpressionNamedRerouteUsage_103}
   ]
  }
  MaterialExpressionLinearInterpolate_68 {
   class=MaterialExpressionLinearInterpolate
   inputs [
    A {from=MaterialExpressionConstant3Vector_12}
    B {from=MaterialExpressionNamedRerouteDeclaration_115}
    Alpha {from=MaterialExpressionMultiply_102}
   ]
  }
  MaterialExpressionLinearInterpolate_69 {
   class=MaterialExpressionLinearInterpolate
   inputs [
    A {from=MaterialExpressionReroute_55}
    B {from=MaterialExpressionNamedRerouteUsage_110}
    Alpha {from=MaterialExpressionScalarParameter_116}
   ]
  }
  MaterialExpressionLinearInterpolate_70 {
   class=MaterialExpressionLinearInterpolate
   inputs [
    A {from=MaterialExpressionReroute_54}
    B {from=MaterialExpressionMultiply_97}
    Alpha {from=MaterialExpressionLinearInterpolate_71}
   ]
  }
  MaterialExpressionLinearInterpolate_71 {
   class=MaterialExpressionLinearInterpolate
   inputs [
    A {from=MaterialExpressionSaturate_27}
    B {from=MaterialExpressionOneMinus_20}
    Alpha {from=MaterialExpressionSaturate_26}
   ]
  }
  MaterialExpressionLinearInterpolate_72 {
   class=MaterialExpressionLinearInterpolate
   inputs [
    A {from=MaterialExpressionScalarParameter_120}
    B {from=MaterialExpressionScalarParameter_121}
    Alpha {from=MaterialExpressionTextureSample_7, output=R}
   ]
  }
  MaterialExpressionLinearInterpolate_73 {
   class=MaterialExpressionLinearInterpolate
   inputs [
    A {from=MaterialExpressionLinearInterpolate_68}
    B {from=MaterialExpressionReroute_55}
    Alpha {from=MaterialExpressionNamedRerouteUsage_124}
   ]
  }
  MaterialExpressionLinearInterpolate_74 {
   class=MaterialExpressionLinearInterpolate
   inputs [
    A {from=MaterialExpressionReroute_59}
    B {from=MaterialExpressionMultiply_113}
    Alpha {from=MaterialExpressionNamedRerouteUsage_127}
   ]
   properties {constB=0.5}
  }
  MaterialExpressionLinearInterpolate_75 {
   class=MaterialExpressionLinearInterpolate
   inputs [
    A {from=MaterialExpressionReroute_58}
    B {from=MaterialExpressionLinearInterpolate_78}
    Alpha {from=MaterialExpressionNamedRerouteDeclaration_109}
   ]
  }
  MaterialExpressionLinearInterpolate_76 {
   class=MaterialExpressionLinearInterpolate
   inputs [
    A {from=MaterialExpressionTextureSample_6, output=R}
    B {from=MaterialExpressionTextureSample_6, output=R}
    Alpha {from=MaterialExpressionMaterialFunctionCall_26, output=Result}
   ]
  }
  MaterialExpressionLinearInterpolate_77 {
   class=MaterialExpressionLinearInterpolate
   inputs [
    A {from=MaterialExpressionVectorParameter_44, output=RGB}
    B {from=MaterialExpressionVectorParameter_39, output=RGB}
    Alpha {from=MaterialExpressionNamedRerouteUsage_132}
   ]
  }
  MaterialExpressionLinearInterpolate_78 {
   class=MaterialExpressionLinearInterpolate
   inputs [
    A {from=MaterialExpressionReroute_58}
    B {from=""}
    Alpha {from=MaterialExpressionScalarParameter_124}
   ]
  }
  MaterialExpressionLinearInterpolate_79 {
   class=MaterialExpressionLinearInterpolate
   inputs [
    A {from=MaterialExpressionScalarParameter_125}
    B {from=MaterialExpressionScalarParameter_126}
    Alpha {from=MaterialExpressionNamedRerouteUsage_138}
   ]
  }
  MaterialExpressionLinearInterpolate_80 {
   class=MaterialExpressionLinearInterpolate
   inputs [
    A {from=MaterialExpressionNamedRerouteUsage_118}
    B {from=MaterialExpressionMultiply_108}
    Alpha {from=""}
   ]
   properties {constAlpha=0.8}
  }
  MaterialExpressionLinearInterpolate_81 {
   class=MaterialExpressionLinearInterpolate
   inputs [
    A {from=MaterialExpressionMultiply_114}
    B {from=MaterialExpressionNamedRerouteUsage_139}
    Alpha {from=MaterialExpressionConstant_23}
   ]
   properties {constAlpha=0.3}
  }
  MaterialExpressionLinearInterpolate_82 {
   class=MaterialExpressionLinearInterpolate
   inputs [
    A {from=MaterialExpressionTextureSampleParameter2D_5, output=RGBA}
    B {from=MaterialExpressionTextureSampleParameter2D_4, output=RGBA}
    Alpha {from=MaterialExpressionNamedRerouteDeclaration_119}
   ]
  }
  MaterialExpressionLinearInterpolate_83 {
   class=MaterialExpressionLinearInterpolate
   inputs [
    A {from=MaterialExpressionReroute_62}
    B {from=MaterialExpressionReroute_61}
    Alpha {from=MaterialExpressionSmoothStep_15}
   ]
   properties {constA=1}
  }
  MaterialExpressionLinearInterpolate_84 {
   class=MaterialExpressionLinearInterpolate
   inputs [
    A {from=""}
    B {from=""}
    Alpha {from=MaterialExpressionOneMinus_25}
   ]
   properties {constA=0.4}
  }
  MaterialExpressionLinearInterpolate_85 {
   class=MaterialExpressionLinearInterpolate
   inputs [
    A {from=MaterialExpressionScalarParameter_142}
    B {from=MaterialExpressionVectorParameter_53, output=RGB}
    Alpha {from=MaterialExpressionPower_17}
   ]
  }
  MaterialExpressionLinearInterpolate_86 {
   class=MaterialExpressionLinearInterpolate
   inputs [
    A {from=MaterialExpressionScalarParameter_137}
    B {from=MaterialExpressionScalarParameter_144}
    Alpha {from=MaterialExpressionAbs_2}
   ]
  }
  MaterialExpressionLinearInterpolate_87 {
   class=MaterialExpressionLinearInterpolate
   inputs [
    A {from=""}
    B {from=MaterialExpressionDivide_32}
    Alpha {from=MaterialExpressionScalarParameter_147}
   ]
   properties {constA=1}
  }
  MaterialExpressionLocalPosition_5 {class=MaterialExpressionLocalPosition, inputs [], properties {includedOffsets=ExcludeOffsets, localOrigin=InstancePreSkinning}}
  MaterialExpressionMaterialFunctionCall_24 {
   class=MaterialExpressionMaterialFunctionCall
   inputs [
    Texture Scale {from=MaterialExpressionReroute_56}
    Normal {from=""}
    InPosition {from=MaterialExpressionNamedRerouteUsage_160}
    Triplanar Fade {from=MaterialExpressionScalarParameter_109}
    Derivatives Scale {from=""}
   ]
   properties {materialFunction=&/MF_TriplanarDithered_PreSkinned}
  }
  MaterialExpressionMaterialFunctionCall_25 {
   class=MaterialExpressionMaterialFunctionCall
   inputs [
    BaseNormal {from=MaterialExpressionLinearInterpolate_69}
    AdditionalNormal {from=MaterialExpressionLinearInterpolate_73}
   ]
   properties {materialFunction=&BlendAngleCorrectedNormals}
  }
  MaterialExpressionMaterialFunctionCall_26 {
   class=MaterialExpressionMaterialFunctionCall
   inputs [
    In {from=MaterialExpressionNamedRerouteUsage_125}
    Contrast {from=MaterialExpressionConstant_18}
   ]
   properties {materialFunction=&CheapContrast}
  }
  MaterialExpressionMaterialFunctionCall_27 {
   class=MaterialExpressionMaterialFunctionCall
   inputs [
    In {from=MaterialExpressionAdd_39}
    Contrast {from=MaterialExpressionConstant_20}
   ]
   properties {materialFunction=&CheapContrast}
  }
  MaterialExpressionMaterialFunctionCall_28 {
   class=MaterialExpressionMaterialFunctionCall
   inputs [
    In {from=MaterialExpressionAdd_40}
    Contrast {from=MaterialExpressionConstant_22}
   ]
   properties {materialFunction=&CheapContrast}
  }
  MaterialExpressionMaterialFunctionCall_29 {
   class=MaterialExpressionMaterialFunctionCall
   inputs [
    Blend Range {from=MaterialExpressionScalarParameter_132}
    Start Offset {from=MaterialExpressionScalarParameter_133}
   ]
   properties {materialFunction=/Engine/Functions/Engine_MaterialFunctions03/Particles/Distance_Blend}
  }
  MaterialExpressionMaterialFunctionCall_30 {
   class=MaterialExpressionMaterialFunctionCall
   inputs [
    Color_Morning {from=MaterialExpressionVectorParameter_45, output=RGB}
    Color_Day {from=MaterialExpressionVectorParameter_46, output=RGB}
    Color_Evening {from=MaterialExpressionVectorParameter_47, output=RGB}
    Color_Night {from=MaterialExpressionVectorParameter_48, output=RGB}
   ]
   properties {materialFunction=&Color_Time_of_Day}
  }
  MaterialExpressionMaterialFunctionCall_31 {
   class=MaterialExpressionMaterialFunctionCall
   inputs [
    Color_Morning {from=MaterialExpressionVectorParameter_49, output=RGB}
    Color_Day {from=MaterialExpressionVectorParameter_50, output=RGB}
    Color_Evening {from=MaterialExpressionVectorParameter_51, output=RGB}
    Color_Night {from=MaterialExpressionVectorParameter_52, output=RGB}
   ]
   properties {materialFunction=&Color_Time_of_Day}
  }
  MaterialExpressionMaterialFunctionCall_32 {
   class=MaterialExpressionMaterialFunctionCall
   inputs []
   properties {materialFunction=&CameraDirectionVector}
  }
  MaterialExpressionMaterialFunctionCall_33 {
   class=MaterialExpressionMaterialFunctionCall
   inputs []
   properties {materialFunction=&CameraDirectionVector}
  }
  MaterialExpressionMaterialFunctionCall_34 {
   class=MaterialExpressionMaterialFunctionCall
   inputs [
    In {from=MaterialExpressionNamedRerouteDeclaration_85}
    Contrast {from=MaterialExpressionScalarParameter_145}
   ]
   properties {materialFunction=&CheapContrast}
  }
  MaterialExpressionMax_2 {
   class=MaterialExpressionMax
   inputs [
    A {from=MaterialExpressionDotProduct_15}
    B {from=""}
   ]
   properties {constB=0}
  }
  MaterialExpressionMax_3 {
   class=MaterialExpressionMax
   inputs [
    A {from=MaterialExpressionAdd_33}
    B {from=""}
   ]
   properties {constB=0}
  }
  MaterialExpressionMultiply_100 {
   class=MaterialExpressionMultiply
   inputs [
    A {from=MaterialExpressionScalarParameter_108}
    B {from=""}
   ]
   properties {constB=1.2171}
  }
  MaterialExpressionMultiply_101 {
   class=MaterialExpressionMultiply
   inputs [
    A {from=MaterialExpressionNamedRerouteUsage_130}
    B {from=MaterialExpressionNamedRerouteUsage_130}
   ]
  }
  MaterialExpressionMultiply_102 {
   class=MaterialExpressionMultiply
   inputs [
    A {from=MaterialExpressionScalarParameter_110}
    B {from=MaterialExpressionOneMinus_22}
   ]
  }
  MaterialExpressionMultiply_103 {
   class=MaterialExpressionMultiply
   inputs [
    A {from=MaterialExpressionReroute_60}
    B {from=MaterialExpressionVectorParameter_36, output=RGB}
   ]
  }
  MaterialExpressionMultiply_104 {
   class=MaterialExpressionMultiply
   inputs [
    A {from=MaterialExpressionSubtract_9}
    B {from=""}
   ]
   properties {constB=2}
  }
  MaterialExpressionMultiply_105 {
   class=MaterialExpressionMultiply
   inputs [
    A {from=MaterialExpressionMultiply_106}
    B {from=MaterialExpressionMaterialFunctionCall_27, output=Result}
   ]
   properties {constB=3}
  }
  MaterialExpressionMultiply_106 {
   class=MaterialExpressionMultiply
   inputs [
    A {from=MaterialExpressionReroute_57}
    B {from=MaterialExpressionScalarParameter_123}
   ]
  }
  MaterialExpressionMultiply_107 {
   class=MaterialExpressionMultiply
   inputs [
    A {from=MaterialExpressionMultiply_105}
    B {from=MaterialExpressionOneMinus_23}
   ]
  }
  MaterialExpressionMultiply_108 {
   class=MaterialExpressionMultiply
   inputs [
    A {from=MaterialExpressionNamedRerouteUsage_118}
    B {from=MaterialExpressionNamedRerouteUsage_136}
   ]
  }
  MaterialExpressionMultiply_113 {
   class=MaterialExpressionMultiply
   inputs [
    A {from=MaterialExpressionLinearInterpolate_81}
    B {from=MaterialExpressionScalarParameter_127}
   ]
  }
  MaterialExpressionMultiply_114 {
   class=MaterialExpressionMultiply
   inputs [
    A {from=MaterialExpressionReroute_59}
    B {from=MaterialExpressionNamedRerouteUsage_139}
   ]
  }
  MaterialExpressionMultiply_115 {
   class=MaterialExpressionMultiply
   inputs [
    A {from=MaterialExpressionMultiply_118}
    B {from=MaterialExpressionSmoothStep_14}
   ]
  }
  MaterialExpressionMultiply_116 {
   class=MaterialExpressionMultiply
   inputs [
    A {from=MaterialExpressionComponentMask_28}
    B {from=""}
   ]
   properties {constB=2}
  }
  MaterialExpressionMultiply_117 {
   class=MaterialExpressionMultiply
   inputs [
    A {from=MaterialExpressionAdd_42}
    B {from=MaterialExpressionScalarParameter_128}
   ]
   properties {constB=0.5}
  }
  MaterialExpressionMultiply_118 {
   class=MaterialExpressionMultiply
   inputs [
    A {from=MaterialExpressionCustom_4}
    B {from=MaterialExpressionScalarParameter_130}
   ]
   properties {constB=3}
  }
  MaterialExpressionMultiply_119 {
   class=MaterialExpressionMultiply
   inputs [
    A {from=MaterialExpressionTextureCoordinate_5}
    B {from=MaterialExpressionScalarParameter_131}
   ]
  }
  MaterialExpressionMultiply_120 {
   class=MaterialExpressionMultiply
   inputs [
    A {from=MaterialExpressionTextureCoordinate_4}
    B {from=MaterialExpressionScalarParameter_134}
   ]
   properties {constB=3}
  }
  MaterialExpressionMultiply_121 {
   class=MaterialExpressionMultiply
   inputs [
    A {from=MaterialExpressionMultiply_126}
    B {from=MaterialExpressionLinearInterpolate_83}
   ]
  }
  MaterialExpressionMultiply_122 {
   class=MaterialExpressionMultiply
   inputs [
    A {from=MaterialExpressionAdd_38}
    B {from=MaterialExpressionAdd_45}
   ]
  }
  MaterialExpressionMultiply_123 {
   class=MaterialExpressionMultiply
   inputs [
    A {from=MaterialExpressionReroute_47}
    B {from=""}
   ]
   properties {constB=0.5, desc=So emisive won't kill SMR}
  }
  MaterialExpressionMultiply_124 {
   class=MaterialExpressionMultiply
   inputs [
    A {from=MaterialExpressionLinearInterpolate_66}
    B {from=MaterialExpressionLinearInterpolate_84}
   ]
  }
  MaterialExpressionMultiply_125 {
   class=MaterialExpressionMultiply
   inputs [
    A {from=MaterialExpressionOneMinus_24}
    B {from=MaterialExpressionOneMinus_24}
   ]
  }
  MaterialExpressionMultiply_126 {
   class=MaterialExpressionMultiply
   inputs [
    A {from=MaterialExpressionMultiply_122}
    B {from=MaterialExpressionAdd_46}
   ]
  }
  MaterialExpressionMultiply_127 {
   class=MaterialExpressionMultiply
   inputs [
    A {from=MaterialExpressionNamedRerouteUsage_145}
    B {from=MaterialExpressionScalarParameter_136}
   ]
  }
  MaterialExpressionMultiply_128 {
   class=MaterialExpressionMultiply
   inputs [
    A {from=MaterialExpressionTextureSample_6, output=R}
    B {from=""}
   ]
  }
  MaterialExpressionMultiply_129 {
   class=MaterialExpressionMultiply
   inputs [
    A {from=MaterialExpressionReroute_66}
    B {from=MaterialExpressionLinearInterpolate_85}
   ]
  }
  MaterialExpressionMultiply_130 {
   class=MaterialExpressionMultiply
   inputs [
    A {from=MaterialExpressionNamedRerouteUsage_143}
    B {from=""}
   ]
   properties {constB=1.5}
  }
  MaterialExpressionMultiply_147 {
   class=MaterialExpressionMultiply
   inputs [
    A {from=MaterialExpressionDotProduct_21}
    B {from=""}
   ]
   properties {constB=0.3333}
  }
  MaterialExpressionMultiply_162 {
   class=MaterialExpressionMultiply
   inputs [
    A {from=MaterialExpressionLinearInterpolate_87}
    B {from=MaterialExpressionScalarParameter_150}
   ]
  }
  MaterialExpressionMultiply_163 {
   class=MaterialExpressionMultiply
   inputs [
    A {from=MaterialExpressionMultiply_162}
    B {from=MaterialExpressionTextureCoordinate_6}
   ]
  }
  MaterialExpressionMultiply_84 {
   class=MaterialExpressionMultiply
   inputs [
    A {from=MaterialExpressionDotProduct_14}
    B {from=""}
   ]
   properties {constB=2}
  }
  MaterialExpressionMultiply_85 {
   class=MaterialExpressionMultiply
   inputs [
    A {from=MaterialExpressionMultiply_84}
    B {from=MaterialExpressionReroute_45}
   ]
  }
  MaterialExpressionMultiply_86 {
   class=MaterialExpressionMultiply
   inputs [
    A {from=MaterialExpressionTransform_6}
    B {from=""}
   ]
   properties {constB=-1}
  }
  MaterialExpressionMultiply_87 {
   class=MaterialExpressionMultiply
   inputs [
    A {from=MaterialExpressionPower_11}
    B {from=MaterialExpressionScalarParameter_99}
   ]
   properties {constB=0.4}
  }
  MaterialExpressionMultiply_88 {
   class=MaterialExpressionMultiply
   inputs [
    A {from=MaterialExpressionPower_12}
    B {from=MaterialExpressionScalarParameter_100}
   ]
   properties {constB=0.55}
  }
  MaterialExpressionMultiply_89 {
   class=MaterialExpressionMultiply
   inputs [
    A {from=MaterialExpressionDotProduct_15}
    B {from=""}
   ]
   properties {constB=-1}
  }
  MaterialExpressionMultiply_90 {
   class=MaterialExpressionMultiply
   inputs [
    A {from=MaterialExpressionSubtract_8}
    B {from=""}
   ]
   properties {constB=0.5}
  }
  MaterialExpressionMultiply_91 {
   class=MaterialExpressionMultiply
   inputs [
    A {from=MaterialExpressionLinearInterpolate_58}
    B {from=MaterialExpressionLinearInterpolate_61}
   ]
  }
  MaterialExpressionMultiply_92 {
   class=MaterialExpressionMultiply
   inputs [
    A {from=MaterialExpressionReroute_48}
    B {from=MaterialExpressionNamedRerouteUsage_147}
   ]
  }
  MaterialExpressionMultiply_93 {
   class=MaterialExpressionMultiply
   inputs [
    A {from=MaterialExpressionNamedRerouteUsage_115}
    B {from=MaterialExpressionVectorParameter_41, output=RGB}
   ]
  }
  MaterialExpressionMultiply_94 {
   class=MaterialExpressionMultiply
   inputs [
    A {from=MaterialExpressionMultiply_93}
    B {from=MaterialExpressionOneMinus_19}
   ]
  }
  MaterialExpressionMultiply_95 {
   class=MaterialExpressionMultiply
   inputs [
    A {from=MaterialExpressionReroute_52}
    B {from=MaterialExpressionMultiply_94}
   ]
  }
  MaterialExpressionMultiply_96 {
   class=MaterialExpressionMultiply
   inputs [
    A {from=MaterialExpressionDesaturation_2}
    B {from=MaterialExpressionVectorParameter_42, output=RGB}
   ]
  }
  MaterialExpressionMultiply_97 {
   class=MaterialExpressionMultiply
   inputs [
    A {from=MaterialExpressionReroute_50}
    B {from=MaterialExpressionVectorParameter_43, output=RGB}
   ]
  }
  MaterialExpressionMultiply_98 {
   class=MaterialExpressionMultiply
   inputs [
    A {from=MaterialExpressionComponentMask_27}
    B {from=""}
   ]
   properties {constB=-1}
  }
  MaterialExpressionMultiply_99 {
   class=MaterialExpressionMultiply
   inputs [
    A {from=MaterialExpressionSaturate_31}
    B {from=MaterialExpressionNamedRerouteUsage_129}
   ]
  }
  MaterialExpressionNamedRerouteDeclaration_100 {
   class=MaterialExpressionNamedRerouteDeclaration
   inputs [None {from=MaterialExpressionFunctionInput_8}]
   properties {name=M_Logo, nodeColor=<rgba 1 0.9844 0 1>}
  }
  MaterialExpressionNamedRerouteDeclaration_101 {
   class=MaterialExpressionNamedRerouteDeclaration
   inputs [None {from=MaterialExpressionFunctionInput_10}]
   properties {name=IN_WPO, nodeColor=<rgba 0 0.9766 1 1>}
  }
  MaterialExpressionNamedRerouteDeclaration_102 {
   class=MaterialExpressionNamedRerouteDeclaration
   inputs [None {from=MaterialExpressionTextureObject_5}]
   properties {name=T_SugarMask, nodeColor=<rgba 0.7344 1 0 1>}
  }
  MaterialExpressionNamedRerouteDeclaration_103 {
   class=MaterialExpressionNamedRerouteDeclaration
   inputs [None {from=MaterialExpressionSaturate_28}]
   properties {name=M_Sugar, nodeColor=<rgba 0 1 0.625 1>}
  }
  MaterialExpressionNamedRerouteDeclaration_104 {
   class=MaterialExpressionNamedRerouteDeclaration
   inputs [None {from=MaterialExpressionShaderStageSwitch_5}]
   properties {name=PreSkinnedNormal_VS-PS, nodeColor=<rgba 0.3587 0.1657 1 1>}
  }
  MaterialExpressionNamedRerouteDeclaration_105 {
   class=MaterialExpressionNamedRerouteDeclaration
   inputs [None {from=MaterialExpressionTextureObject_4}]
   properties {name=T_Noise, nodeColor=<rgba 0 0.0391 1 1>}
  }
  MaterialExpressionNamedRerouteDeclaration_106 {
   class=MaterialExpressionNamedRerouteDeclaration
   inputs [None {from=MaterialExpressionSaturate_30}]
   properties {name=M_Noise_SugarColorBias, nodeColor=<rgba 1 0 0 1>}
  }
  MaterialExpressionNamedRerouteDeclaration_107 {
   class=MaterialExpressionNamedRerouteDeclaration
   inputs [None {from=MaterialExpressionTextureSample_8, output=G}]
   properties {name=M_PerlinNoise_SugarTexBlend, nodeColor=<rgba 0 1 0 1>}
  }
  MaterialExpressionNamedRerouteDeclaration_109 {
   class=MaterialExpressionNamedRerouteDeclaration
   inputs [None {from=MaterialExpressionMultiply_99}]
   properties {name=M_Sugar_Color, nodeColor=<rgba 0 1 0.4844 1>}
  }
  MaterialExpressionNamedRerouteDeclaration_110 {
   class=MaterialExpressionNamedRerouteDeclaration
   inputs [None {from=MaterialExpressionOneMinus_21}]
   properties {name=M_Fresnel, nodeColor=<rgba 1 0.9844 0 1>}
  }
  MaterialExpressionNamedRerouteDeclaration_111 {
   class=MaterialExpressionNamedRerouteDeclaration
   inputs [None {from=MaterialExpressionMaterialFunctionCall_34, output=Result}]
   properties {name=M_BackSSS/ShadowCore, nodeColor=<rgba 1 0 0.5625 1>}
  }
  MaterialExpressionNamedRerouteDeclaration_112 {
   class=MaterialExpressionNamedRerouteDeclaration
   inputs [None {from=MaterialExpressionLinearInterpolate_79}]
   properties {name=OUT_Metallic, nodeColor=<rgba 0 0.8828 1 1>}
  }
  MaterialExpressionNamedRerouteDeclaration_113 {
   class=MaterialExpressionNamedRerouteDeclaration
   inputs [None {from=MaterialExpressionLinearInterpolate_80}]
   properties {name=M_SMR, nodeColor=<rgba 0 0.1094 1 1>}
  }
  MaterialExpressionNamedRerouteDeclaration_114 {
   class=MaterialExpressionNamedRerouteDeclaration
   inputs [None {from=MaterialExpressionMultiply_115}]
   properties {name=Glitter, nodeColor=<rgba 0 0.2687 1 1>}
  }
  MaterialExpressionNamedRerouteDeclaration_115 {
   class=MaterialExpressionNamedRerouteDeclaration
   inputs [None {from=MaterialExpressionConvert_2}]
   properties {name=Sugar_Normal, nodeColor=<rgba 1 0.0469 0 1>}
  }
  MaterialExpressionNamedRerouteDeclaration_116 {
   class=MaterialExpressionNamedRerouteDeclaration
   inputs [None {from=MaterialExpressionMultiply_117}]
   properties {name=Random Normal Dir_1, nodeColor=<rgba 0 0.3501 1 1>}
  }
  MaterialExpressionNamedRerouteDeclaration_117 {
   class=MaterialExpressionNamedRerouteDeclaration
   inputs [None {from=MaterialExpressionFunctionInput_1}]
   properties {name=M_BlinkingLights, nodeColor=<rgba 1 0.9844 0 1>}
  }
  MaterialExpressionNamedRerouteDeclaration_118 {
   class=MaterialExpressionNamedRerouteDeclaration
   inputs [None {from=MaterialExpressionComponentMask_16}]
   properties {name=M_ZGradient, nodeColor=<rgba 0.1719 1 0 1>}
  }
  MaterialExpressionNamedRerouteDeclaration_119 {
   class=MaterialExpressionNamedRerouteDeclaration
   inputs [None {from=MaterialExpressionMaterialFunctionCall_29, output=Result}]
   properties {name=M_DistanceBlend, nodeColor=<rgba 1 0 0.3984 1>}
  }
  MaterialExpressionNamedRerouteDeclaration_120 {
   class=MaterialExpressionNamedRerouteDeclaration
   inputs [None {from=MaterialExpressionFunctionInput_2}]
   properties {name=M_DiffuseMask, nodeColor=<rgba 0.875 1 0 1>}
  }
  MaterialExpressionNamedRerouteDeclaration_122 {
   class=MaterialExpressionNamedRerouteDeclaration
   inputs [None {from=MaterialExpressionFunctionInput_5}]
   properties {name="Pre-Skinned Bounds Min (V3)", nodeColor=<rgba 1 1 1 1>}
  }
  MaterialExpressionNamedRerouteDeclaration_123 {
   class=MaterialExpressionNamedRerouteDeclaration
   inputs [None {from=MaterialExpressionFunctionInput_4}]
   properties {name="Pre-Skinned Bounds Max (V3)", nodeColor=<rgba 1 1 1 1>}
  }
  MaterialExpressionNamedRerouteDeclaration_124 {
   class=MaterialExpressionNamedRerouteDeclaration
   inputs [None {from=MaterialExpressionSubtract_40}]
   properties {name="Pre-Skinned Bounds Extents (V3)", nodeColor=<rgba 1 1 1 1>}
  }
  MaterialExpressionNamedRerouteDeclaration_127 {
   class=MaterialExpressionNamedRerouteDeclaration
   inputs [None {from=MaterialExpressionMultiply_147}]
   properties {name="Pre-Skinned Scale (S)", nodeColor=<rgba 1 0.5391 0 1>}
  }
  MaterialExpressionNamedRerouteDeclaration_129 {
   class=MaterialExpressionNamedRerouteDeclaration
   inputs [None {from=MaterialExpressionFunctionInput_12}]
   properties {name="Pre-Skinned Scale (V3)", nodeColor=<rgba 1 0.5391 0 1>}
  }
  MaterialExpressionNamedRerouteDeclaration_60 {
   class=MaterialExpressionNamedRerouteDeclaration
   inputs [None {from=MaterialExpressionDDX_3}]
   properties {name="Scaled UV DDX (V2)", nodeColor=<rgba 0.0547 1 0 1>}
  }
  MaterialExpressionNamedRerouteDeclaration_62 {
   class=MaterialExpressionNamedRerouteDeclaration
   inputs [None {from=MaterialExpressionDDY_3}]
   properties {name="Scaled UV DDY (V2)", nodeColor=<rgba 0.0547 1 0 1>}
  }
  MaterialExpressionNamedRerouteDeclaration_80 {
   class=MaterialExpressionNamedRerouteDeclaration
   inputs [None {from=MaterialExpressionGetMaterialAttributes_2, output=MaterialAttributes}]
   properties {name=IN_MA, nodeColor=<rgba 1 0 0.1015 1>}
  }
  MaterialExpressionNamedRerouteDeclaration_81 {
   class=MaterialExpressionNamedRerouteDeclaration
   inputs [None {from=MaterialExpressionFunctionInput_3}]
   properties {name="Pre-Skinned Local Pos (V3)", nodeColor=<rgba 0 0.1797 1 1>}
  }
  MaterialExpressionNamedRerouteDeclaration_82 {
   class=MaterialExpressionNamedRerouteDeclaration
   inputs [None {from=MaterialExpressionSaturate_22}]
   properties {name=M_SSS, nodeColor=<rgba 0 0.1797 1 1>}
  }
  MaterialExpressionNamedRerouteDeclaration_83 {
   class=MaterialExpressionNamedRerouteDeclaration
   inputs [None {from=MaterialExpressionOneMinus_18}]
   properties {name=Mask_Colors, nodeColor=<rgba 0.6406 0 1 1>}
  }
  MaterialExpressionNamedRerouteDeclaration_84 {
   class=MaterialExpressionNamedRerouteDeclaration
   inputs [None {from=MaterialExpressionSaturate_24}]
   properties {name=M_ColorTransition, nodeColor=<rgba 0 1 0.5547 1>}
  }
  MaterialExpressionNamedRerouteDeclaration_85 {
   class=MaterialExpressionNamedRerouteDeclaration
   inputs [None {from=MaterialExpressionSmoothStep_17}]
   properties {name=M_Terminator, nodeColor=<rgba 0 0.7656 1 1>}
  }
  MaterialExpressionNamedRerouteDeclaration_86 {
   class=MaterialExpressionNamedRerouteDeclaration
   inputs [None {from=MaterialExpressionMultiply_121}]
   properties {name=OUT_Emissive, nodeColor=<rgba 1 0.1172 0 1>}
  }
  MaterialExpressionNamedRerouteDeclaration_87 {
   class=MaterialExpressionNamedRerouteDeclaration
   inputs [None {from=MaterialExpressionLinearInterpolate_63}]
   properties {name=OUT_Specular, nodeColor=<rgba 0 0.2031 1 1>}
  }
  MaterialExpressionNamedRerouteDeclaration_88 {
   class=MaterialExpressionNamedRerouteDeclaration
   inputs [None {from=MaterialExpressionLinearInterpolate_75}]
   properties {name=OUT_BaseColor, nodeColor=<rgba 0.5234 1 0 1>}
  }
  MaterialExpressionNamedRerouteDeclaration_89 {
   class=MaterialExpressionNamedRerouteDeclaration
   inputs [None {from=MaterialExpressionLinearInterpolate_62}]
   properties {name=OUT_Roughness, nodeColor=<rgba 1 0 0.8438 1>}
  }
  MaterialExpressionNamedRerouteDeclaration_90 {
   class=MaterialExpressionNamedRerouteDeclaration
   inputs [None {from=MaterialExpressionMaterialFunctionCall_25, output=Result}]
   properties {name=OUT_Normals, nodeColor=<rgba 0.9219 0 1 1>}
  }
  MaterialExpressionNamedRerouteDeclaration_91 {
   class=MaterialExpressionNamedRerouteDeclaration
   inputs [None {from=MaterialExpressionGetMaterialAttributes_2, output=Normal}]
   properties {name=IN_Normals, nodeColor=<rgba 0 1 0.6016 1>}
  }
  MaterialExpressionNamedRerouteDeclaration_92 {
   class=MaterialExpressionNamedRerouteDeclaration
   inputs [None {from=MaterialExpressionMaterialFunctionCall_24, output=UVs}]
   properties {name=Sugar_UV, nodeColor=<rgba 1 0.2812 0 1>}
  }
  MaterialExpressionNamedRerouteDeclaration_93 {
   class=MaterialExpressionNamedRerouteDeclaration
   inputs [None {from=MaterialExpressionGetMaterialAttributes_2, output=BaseColor}]
   properties {name=IN_BaseColor, nodeColor=<rgba 0.0781 0 1 1>}
  }
  MaterialExpressionNamedRerouteDeclaration_94 {
   class=MaterialExpressionNamedRerouteDeclaration
   inputs [None {from=MaterialExpressionGetMaterialAttributes_2, output=Metallic}]
   properties {name=In_Metallic, nodeColor=<rgba 0.2422 1 0 1>}
  }
  MaterialExpressionNamedRerouteDeclaration_95 {
   class=MaterialExpressionNamedRerouteDeclaration
   inputs [None {from=MaterialExpressionGetMaterialAttributes_2, output=Roughness}]
   properties {name=In_Roughness, nodeColor=<rgba 1 0 0.5625 1>}
  }
  MaterialExpressionNamedRerouteDeclaration_96 {
   class=MaterialExpressionNamedRerouteDeclaration
   inputs [None {from=MaterialExpressionGetMaterialAttributes_2, output=Specular}]
   properties {name=IN_Specular, nodeColor=<rgba 1 0.4453 0 1>}
  }
  MaterialExpressionNamedRerouteDeclaration_97 {
   class=MaterialExpressionNamedRerouteDeclaration
   inputs [None {from=MaterialExpressionFunctionInput_6}]
   properties {name=M_VariantMask, nodeColor=<rgba 0 1 0.8828 1>}
  }
  MaterialExpressionNamedRerouteDeclaration_98 {
   class=MaterialExpressionNamedRerouteDeclaration
   inputs [None {from=MaterialExpressionFunctionInput_7}]
   properties {name=M_InnerRadialGlow, nodeColor=<rgba 1 0.9844 0 1>}
  }
  MaterialExpressionNamedRerouteDeclaration_99 {
   class=MaterialExpressionNamedRerouteDeclaration
   inputs [None {from=MaterialExpressionFunctionInput_9}]
   properties {name=M_SpriteFresnel, nodeColor=<rgba 1 0.9844 0 1>}
  }
  MaterialExpressionNamedRerouteUsage_100 {class=MaterialExpressionNamedRerouteUsage}
  MaterialExpressionNamedRerouteUsage_101 {class=MaterialExpressionNamedRerouteUsage}
  MaterialExpressionNamedRerouteUsage_102 {class=MaterialExpressionNamedRerouteUsage}
  MaterialExpressionNamedRerouteUsage_103 {class=MaterialExpressionNamedRerouteUsage}
  MaterialExpressionNamedRerouteUsage_104 {class=MaterialExpressionNamedRerouteUsage}
  MaterialExpressionNamedRerouteUsage_105 {class=MaterialExpressionNamedRerouteUsage}
  MaterialExpressionNamedRerouteUsage_106 {class=MaterialExpressionNamedRerouteUsage}
  MaterialExpressionNamedRerouteUsage_107 {class=MaterialExpressionNamedRerouteUsage}
  MaterialExpressionNamedRerouteUsage_108 {class=MaterialExpressionNamedRerouteUsage}
  MaterialExpressionNamedRerouteUsage_109 {class=MaterialExpressionNamedRerouteUsage}
  MaterialExpressionNamedRerouteUsage_110 {class=MaterialExpressionNamedRerouteUsage}
  MaterialExpressionNamedRerouteUsage_112 {class=MaterialExpressionNamedRerouteUsage}
  MaterialExpressionNamedRerouteUsage_113 {class=MaterialExpressionNamedRerouteUsage}
  MaterialExpressionNamedRerouteUsage_114 {class=MaterialExpressionNamedRerouteUsage}
  MaterialExpressionNamedRerouteUsage_115 {class=MaterialExpressionNamedRerouteUsage}
  MaterialExpressionNamedRerouteUsage_116 {class=MaterialExpressionNamedRerouteUsage}
  MaterialExpressionNamedRerouteUsage_117 {class=MaterialExpressionNamedRerouteUsage}
  MaterialExpressionNamedRerouteUsage_118 {class=MaterialExpressionNamedRerouteUsage}
  MaterialExpressionNamedRerouteUsage_119 {class=MaterialExpressionNamedRerouteUsage}
  MaterialExpressionNamedRerouteUsage_120 {class=MaterialExpressionNamedRerouteUsage}
  MaterialExpressionNamedRerouteUsage_121 {class=MaterialExpressionNamedRerouteUsage}
  MaterialExpressionNamedRerouteUsage_122 {class=MaterialExpressionNamedRerouteUsage}
  MaterialExpressionNamedRerouteUsage_123 {class=MaterialExpressionNamedRerouteUsage}
  MaterialExpressionNamedRerouteUsage_124 {class=MaterialExpressionNamedRerouteUsage}
  MaterialExpressionNamedRerouteUsage_125 {class=MaterialExpressionNamedRerouteUsage}
  MaterialExpressionNamedRerouteUsage_127 {class=MaterialExpressionNamedRerouteUsage}
  MaterialExpressionNamedRerouteUsage_128 {class=MaterialExpressionNamedRerouteUsage}
  MaterialExpressionNamedRerouteUsage_129 {class=MaterialExpressionNamedRerouteUsage}
  MaterialExpressionNamedRerouteUsage_130 {class=MaterialExpressionNamedRerouteUsage}
  MaterialExpressionNamedRerouteUsage_131 {class=MaterialExpressionNamedRerouteUsage}
  MaterialExpressionNamedRerouteUsage_132 {class=MaterialExpressionNamedRerouteUsage}
  MaterialExpressionNamedRerouteUsage_133 {class=MaterialExpressionNamedRerouteUsage}
  MaterialExpressionNamedRerouteUsage_134 {class=MaterialExpressionNamedRerouteUsage}
  MaterialExpressionNamedRerouteUsage_135 {class=MaterialExpressionNamedRerouteUsage}
  MaterialExpressionNamedRerouteUsage_136 {class=MaterialExpressionNamedRerouteUsage}
  MaterialExpressionNamedRerouteUsage_137 {class=MaterialExpressionNamedRerouteUsage}
  MaterialExpressionNamedRerouteUsage_138 {class=MaterialExpressionNamedRerouteUsage}
  MaterialExpressionNamedRerouteUsage_139 {class=MaterialExpressionNamedRerouteUsage}
  MaterialExpressionNamedRerouteUsage_140 {class=MaterialExpressionNamedRerouteUsage}
  MaterialExpressionNamedRerouteUsage_141 {class=MaterialExpressionNamedRerouteUsage}
  MaterialExpressionNamedRerouteUsage_142 {class=MaterialExpressionNamedRerouteUsage}
  MaterialExpressionNamedRerouteUsage_143 {class=MaterialExpressionNamedRerouteUsage}
  MaterialExpressionNamedRerouteUsage_144 {class=MaterialExpressionNamedRerouteUsage}
  MaterialExpressionNamedRerouteUsage_145 {class=MaterialExpressionNamedRerouteUsage}
  MaterialExpressionNamedRerouteUsage_146 {class=MaterialExpressionNamedRerouteUsage}
  MaterialExpressionNamedRerouteUsage_147 {class=MaterialExpressionNamedRerouteUsage}
  MaterialExpressionNamedRerouteUsage_148 {class=MaterialExpressionNamedRerouteUsage}
  MaterialExpressionNamedRerouteUsage_151 {class=MaterialExpressionNamedRerouteUsage}
  MaterialExpressionNamedRerouteUsage_153 {class=MaterialExpressionNamedRerouteUsage}
  MaterialExpressionNamedRerouteUsage_154 {class=MaterialExpressionNamedRerouteUsage}
  MaterialExpressionNamedRerouteUsage_155 {class=MaterialExpressionNamedRerouteUsage}
  MaterialExpressionNamedRerouteUsage_157 {class=MaterialExpressionNamedRerouteUsage}
  MaterialExpressionNamedRerouteUsage_158 {class=MaterialExpressionNamedRerouteUsage}
  MaterialExpressionNamedRerouteUsage_159 {class=MaterialExpressionNamedRerouteUsage}
  MaterialExpressionNamedRerouteUsage_160 {class=MaterialExpressionNamedRerouteUsage}
  MaterialExpressionNamedRerouteUsage_161 {class=MaterialExpressionNamedRerouteUsage}
  MaterialExpressionNamedRerouteUsage_162 {class=MaterialExpressionNamedRerouteUsage}
  MaterialExpressionNamedRerouteUsage_183 {class=MaterialExpressionNamedRerouteUsage}
  MaterialExpressionNamedRerouteUsage_36 {class=MaterialExpressionNamedRerouteUsage}
  MaterialExpressionNormalize_4 {class=MaterialExpressionNormalize, inputs [VectorInput {from=MaterialExpressionSubtract_7, output=""}]}
  MaterialExpressionNormalize_6 {class=MaterialExpressionNormalize, inputs [VectorInput {from=MaterialExpressionAdd_43, output=""}], properties {desc=GlitterDirection}}
  MaterialExpressionNormalize_7 {class=MaterialExpressionNormalize, inputs [VectorInput {from=MaterialExpressionAdd_44, output=""}], properties {desc="L + V  average"}}
  MaterialExpressionOneMinus_18 {class=MaterialExpressionOneMinus, inputs [None {from=MaterialExpressionSaturate_23}]}
  MaterialExpressionOneMinus_19 {class=MaterialExpressionOneMinus, inputs [None {from=MaterialExpressionNamedRerouteUsage_116}]}
  MaterialExpressionOneMinus_20 {class=MaterialExpressionOneMinus, inputs [None {from=MaterialExpressionSaturate_27}]}
  MaterialExpressionOneMinus_21 {class=MaterialExpressionOneMinus, inputs [None {from=MaterialExpressionDotProduct_17}]}
  MaterialExpressionOneMinus_22 {class=MaterialExpressionOneMinus, inputs [None {from=MaterialExpressionMultiply_101}]}
  MaterialExpressionOneMinus_23 {class=MaterialExpressionOneMinus, inputs [None {from=MaterialExpressionMaterialFunctionCall_28, output=Result}]}
  MaterialExpressionOneMinus_24 {class=MaterialExpressionOneMinus, inputs [None {from=MaterialExpressionNamedRerouteUsage_144}]}
  MaterialExpressionOneMinus_25 {class=MaterialExpressionOneMinus, inputs [None {from=MaterialExpressionMultiply_125}]}
  MaterialExpressionOneMinus_26 {class=MaterialExpressionOneMinus, inputs [None {from=MaterialExpressionDotProduct_20}]}
  MaterialExpressionPixelNormalWS_2 {class=MaterialExpressionPixelNormalWS, inputs [], properties {}}
  MaterialExpressionPower_11 {
   class=MaterialExpressionPower
   inputs [
    Base {from=MaterialExpressionMax_2}
    Exp {from=MaterialExpressionScalarParameter_98}
   ]
  }
  MaterialExpressionPower_12 {
   class=MaterialExpressionPower
   inputs [
    Base {from=MaterialExpressionMax_3}
    Exp {from=MaterialExpressionScalarParameter_101}
   ]
  }
  MaterialExpressionPower_14 {
   class=MaterialExpressionPower
   inputs [
    Base {from=MaterialExpressionSphereMask_2}
    Exp {from=MaterialExpressionScalarParameter_105}
   ]
   properties {constExponent=4}
  }
  MaterialExpressionPower_15 {
   class=MaterialExpressionPower
   inputs [
    Base {from=MaterialExpressionSmoothStep_13}
    Exp {from=MaterialExpressionScalarParameter_122}
   ]
  }
  MaterialExpressionPower_16 {
   class=MaterialExpressionPower
   inputs [
    Base {from=MaterialExpressionNamedRerouteUsage_142}
    Exp {from=MaterialExpressionScalarParameter_139}
   ]
  }
  MaterialExpressionPower_17 {
   class=MaterialExpressionPower
   inputs [
    Base {from=MaterialExpressionSmoothStep_16}
    Exp {from=""}
   ]
   properties {constExponent=3}
  }
  MaterialExpressionPreSkinnedNormal_2 {class=MaterialExpressionPreSkinnedNormal, inputs [], properties {}}
  MaterialExpressionReflectionVectorWS_4 {class=MaterialExpressionReflectionVectorWS, inputs [CustomWorldNormal {from="", output=""}], properties {}}
  MaterialExpressionReflectionVectorWS_5 {class=MaterialExpressionReflectionVectorWS, inputs [CustomWorldNormal {from="", output=""}], properties {}}
  MaterialExpressionReroute_45 {class=MaterialExpressionReroute, inputs [None {from=MaterialExpressionVertexNormalWS_8}]}
  MaterialExpressionReroute_46 {class=MaterialExpressionReroute, inputs [None {from=MaterialExpressionNamedRerouteUsage_157}]}
  MaterialExpressionReroute_47 {class=MaterialExpressionReroute, inputs [None {from=MaterialExpressionReroute_65}]}
  MaterialExpressionReroute_48 {class=MaterialExpressionReroute, inputs [None {from=MaterialExpressionReroute_49}]}
  MaterialExpressionReroute_49 {class=MaterialExpressionReroute, inputs [None {from=MaterialExpressionLinearInterpolate_60}]}
  MaterialExpressionReroute_50 {class=MaterialExpressionReroute, inputs [None {from=MaterialExpressionReroute_51}]}
  MaterialExpressionReroute_51 {class=MaterialExpressionReroute, inputs [None {from=MaterialExpressionReroute_64}]}
  MaterialExpressionReroute_52 {class=MaterialExpressionReroute, inputs [None {from=MaterialExpressionLinearInterpolate_74}]}
  MaterialExpressionReroute_53 {class=MaterialExpressionReroute, inputs [None {from=MaterialExpressionLinearInterpolate_59}]}
  MaterialExpressionReroute_54 {class=MaterialExpressionReroute, inputs [None {from=MaterialExpressionReroute_50}]}
  MaterialExpressionReroute_55 {class=MaterialExpressionReroute, inputs [None {from=MaterialExpressionConstant3Vector_12}]}
  MaterialExpressionReroute_56 {class=MaterialExpressionReroute, inputs [None {from=MaterialExpressionScalarParameter_108}]}
  MaterialExpressionReroute_57 {class=MaterialExpressionReroute, inputs [None {from=MaterialExpressionMultiply_124}]}
  MaterialExpressionReroute_58 {class=MaterialExpressionReroute, inputs [None {from=MaterialExpressionLinearInterpolate_64}]}
  MaterialExpressionReroute_59 {class=MaterialExpressionReroute, inputs [None {from=MaterialExpressionReroute_57}]}
  MaterialExpressionReroute_60 {class=MaterialExpressionReroute, inputs [None {from=MaterialExpressionReroute_47}]}
  MaterialExpressionReroute_61 {class=MaterialExpressionReroute, inputs [None {from=MaterialExpressionReroute_63}]}
  MaterialExpressionReroute_62 {class=MaterialExpressionReroute, inputs [None {from=MaterialExpressionMaterialFunctionCall_31, output=Result}]}
  MaterialExpressionReroute_63 {class=MaterialExpressionReroute, inputs [None {from=MaterialExpressionMaterialFunctionCall_30, output=Result}]}
  MaterialExpressionReroute_64 {class=MaterialExpressionReroute, inputs [None {from=MaterialExpressionMultiply_91}]}
  MaterialExpressionReroute_65 {class=MaterialExpressionReroute, inputs [None {from=MaterialExpressionLinearInterpolate_67}]}
  MaterialExpressionReroute_66 {class=MaterialExpressionReroute, inputs [None {from=MaterialExpressionReroute_65}]}
  MaterialExpressionSaturate_21 {class=MaterialExpressionSaturate, inputs [None {from=MaterialExpressionMultiply_88}]}
  MaterialExpressionSaturate_22 {class=MaterialExpressionSaturate, inputs [None {from=MaterialExpressionMultiply_87}]}
  MaterialExpressionSaturate_23 {class=MaterialExpressionSaturate, inputs [None {from=MaterialExpressionSmoothStep_12}]}
  MaterialExpressionSaturate_24 {class=MaterialExpressionSaturate, inputs [None {from=MaterialExpressionAdd_35}]}
  MaterialExpressionSaturate_25 {class=MaterialExpressionSaturate, inputs [None {from=MaterialExpressionScalarParameter_146}]}
  MaterialExpressionSaturate_26 {class=MaterialExpressionSaturate, inputs [None {from=MaterialExpressionVectorParameter_40, output=InverseMask}]}
  MaterialExpressionSaturate_27 {class=MaterialExpressionSaturate, inputs [None {from=MaterialExpressionDotProduct_16}]}
  MaterialExpressionSaturate_28 {class=MaterialExpressionSaturate, inputs [None {from=MaterialExpressionLinearInterpolate_76}]}
  MaterialExpressionSaturate_30 {class=MaterialExpressionSaturate, inputs [None {from=MaterialExpressionLinearInterpolate_72}]}
  MaterialExpressionSaturate_31 {class=MaterialExpressionSaturate, inputs [None {from=MaterialExpressionPower_15}]}
  MaterialExpressionSaturate_32 {class=MaterialExpressionSaturate, inputs [None {from=MaterialExpressionMultiply_104}]}
  MaterialExpressionScalarParameter_100 {
   class=MaterialExpressionScalarParameter
   properties {defaultValue=0.4, group=SpriteVariants_Candy, parameterName=Slurp_InnerBackScatteringScale}
  }
  MaterialExpressionScalarParameter_101 {
   class=MaterialExpressionScalarParameter
   properties {defaultValue=1, group=SpriteVariants_Candy, parameterName=Slurp_InnerBackScatteringPower}
  }
  MaterialExpressionScalarParameter_102 {class=MaterialExpressionScalarParameter, properties {group=SpriteVariants_Candy, parameterName=Slurp_InnerBackScatteringBias}}
  MaterialExpressionScalarParameter_103 {class=MaterialExpressionScalarParameter, properties {defaultValue=0.1, group=SpriteVariants_Candy, parameterName=Sweet_BlendMin}}
  MaterialExpressionScalarParameter_104 {class=MaterialExpressionScalarParameter, properties {defaultValue=1.5, group=SpriteVariants_Candy, parameterName=Sweet_BlendMax}}
  MaterialExpressionScalarParameter_105 {class=MaterialExpressionScalarParameter, properties {defaultValue=2.4, group=SpriteVariants_Candy, parameterName=Sweet_MidColor_Power}}
  MaterialExpressionScalarParameter_106 {class=MaterialExpressionScalarParameter, properties {defaultValue=0.18, parameterName=Sweet_MidColor_ValueBias}}
  MaterialExpressionScalarParameter_107 {class=MaterialExpressionScalarParameter, properties {defaultValue=0.12, group=SpriteVariants_Candy, parameterName=Sweet_BaseColorIntensity}}
  MaterialExpressionScalarParameter_108 {class=MaterialExpressionScalarParameter, properties {defaultValue=7, group=SpriteVariants_Candy, parameterName=SpriteVar_Candy_Sugar_Scale}}
  MaterialExpressionScalarParameter_109 {
   class=MaterialExpressionScalarParameter
   properties {defaultValue=100, group=SpriteVariants_Candy, parameterName=SpriteVar_Candy_Sugar_TriplanarFade}
  }
  MaterialExpressionScalarParameter_110 {class=MaterialExpressionScalarParameter, properties {defaultValue=1, group=SpriteVariants, parameterName=SpriteVar_Candy_Normal_Sugar}}
  MaterialExpressionScalarParameter_111 {class=MaterialExpressionScalarParameter, properties {defaultValue=0.2, group=SpriteVariants, parameterName=SpriteVar_Candy_Roughness_Candy}}
  MaterialExpressionScalarParameter_112 {class=MaterialExpressionScalarParameter, properties {defaultValue=0.8, group=SpriteVariants, parameterName=SpriteVar_Candy_Roughness_Sugar}}
  MaterialExpressionScalarParameter_113 {
   class=MaterialExpressionScalarParameter
   properties {defaultValue=0.8, group=SpriteVariants, parameterName=SpriteVar_Candy_SpecularAdd_Candy}
  }
  MaterialExpressionScalarParameter_114 {
   class=MaterialExpressionScalarParameter
   properties {defaultValue=0.8, group=SpriteVariants, parameterName=SpriteVar_Candy_SpecularAdd_Sugar}
  }
  MaterialExpressionScalarParameter_115 {class=MaterialExpressionScalarParameter, properties {defaultValue=0.35, parameterName=SpriteVar_Candy_Transition_Desaturation}}
  MaterialExpressionScalarParameter_116 {class=MaterialExpressionScalarParameter, properties {defaultValue=0.4, group=SpriteVariants, parameterName=SpriteVar_Candy_Normal_Base}}
  MaterialExpressionScalarParameter_117 {
   class=MaterialExpressionScalarParameter
   properties {defaultValue=-0.2, group=SpriteVariants_Candy, parameterName=SpriteVar_Candy_Color_Sugar_MaskMin}
  }
  MaterialExpressionScalarParameter_118 {
   class=MaterialExpressionScalarParameter
   properties {defaultValue=1.6, group=SpriteVariants_Candy, parameterName=SpriteVar_Candy_Color_Sugar_MaskMax}
  }
  MaterialExpressionScalarParameter_119 {class=MaterialExpressionScalarParameter, properties {defaultValue=50, group=SpriteVariants_Candy, parameterName=SpriteVar_Candy_Noise_Scale}}
  MaterialExpressionScalarParameter_120 {class=MaterialExpressionScalarParameter, properties {defaultValue=-0.15, parameterName=SpreiteVar_Cady_Sugar_Noise_Min}}
  MaterialExpressionScalarParameter_121 {class=MaterialExpressionScalarParameter, properties {defaultValue=1.2, parameterName=SpreiteVar_Cady_Sugar_Noise_Max}}
  MaterialExpressionScalarParameter_122 {
   class=MaterialExpressionScalarParameter
   properties {defaultValue=5, group=SpriteVariants_Candy, parameterName=SpriteVar_Candy_Color_Sugar_MaskPower}
  }
  MaterialExpressionScalarParameter_123 {class=MaterialExpressionScalarParameter, properties {defaultValue=2, parameterName=SpriteVar_Candy_Sugar_EmissiveTerminator}}
  MaterialExpressionScalarParameter_124 {class=MaterialExpressionScalarParameter, properties {defaultValue=1.5, parameterName=Param_1}}
  MaterialExpressionScalarParameter_125 {class=MaterialExpressionScalarParameter, properties {defaultValue=0.5, group=SpriteVariants, parameterName=SpriteVar_Candy_Metallic_Candy}}
  MaterialExpressionScalarParameter_126 {class=MaterialExpressionScalarParameter, properties {group=SpriteVariants, parameterName=SpriteVar_Candy_Metallic_Sugar}}
  MaterialExpressionScalarParameter_127 {class=MaterialExpressionScalarParameter, properties {defaultValue=50, parameterName=SpriteVar_Candy_Sugar_EmissiveGliter}}
  MaterialExpressionScalarParameter_128 {
   class=MaterialExpressionScalarParameter
   properties {defaultValue=0.7, group=Glitter, parameterName=Normal Noise Intensity, sliderMax=2, sortPriority=6}
  }
  MaterialExpressionScalarParameter_129 {
   class=MaterialExpressionScalarParameter
   properties {defaultValue=0.95, group=Glitter, parameterName=Max Glitter, sliderMax=1, sortPriority=8}
  }
  MaterialExpressionScalarParameter_130 {
   class=MaterialExpressionScalarParameter
   properties {defaultValue=5, group=Glitter, parameterName=Boost Light, sliderMax=16, sortPriority=10}
  }
  MaterialExpressionScalarParameter_131 {
   class=MaterialExpressionScalarParameter
   properties {defaultValue=4, group=Glitter, parameterName=Tiling, sliderMax=16, sliderMin=1, sortPriority=3}
  }
  MaterialExpressionScalarParameter_132 {class=MaterialExpressionScalarParameter, properties {defaultValue=50, group=Glitter, parameterName=Blend Distance, sortPriority=4}}
  MaterialExpressionScalarParameter_133 {class=MaterialExpressionScalarParameter, properties {defaultValue=-150, group=Glitter, parameterName=Blend Offset, sortPriority=5}}
  MaterialExpressionScalarParameter_134 {
   class=MaterialExpressionScalarParameter
   properties {defaultValue=3.5, group=Glitter, parameterName=DetailTiling, sliderMax=64, sliderMin=8, sortPriority=2}
  }
  MaterialExpressionScalarParameter_135 {class=MaterialExpressionScalarParameter, properties {defaultValue=-0.35, parameterName=Sweet_MidColor_Height}}
  MaterialExpressionScalarParameter_136 {class=MaterialExpressionScalarParameter, properties {defaultValue=1.3, parameterName=SpriteVar_Candy_InnerGlow}}
  MaterialExpressionScalarParameter_137 {class=MaterialExpressionScalarParameter, properties {defaultValue=0.5, parameterName=Param_max_Side}}
  MaterialExpressionScalarParameter_138 {class=MaterialExpressionScalarParameter, properties {defaultValue=0.05, parameterName=Param_3min}}
  MaterialExpressionScalarParameter_139 {class=MaterialExpressionScalarParameter, properties {defaultValue=1.5}}
  MaterialExpressionScalarParameter_140 {class=MaterialExpressionScalarParameter, properties {defaultValue=0.4, parameterName=Param_3fff_Mina}}
  MaterialExpressionScalarParameter_141 {class=MaterialExpressionScalarParameter, properties {defaultValue=0.05, parameterName=Param_3fff_Maxa}}
  MaterialExpressionScalarParameter_142 {class=MaterialExpressionScalarParameter, properties {defaultValue=1, parameterName=Frenelcore Min}}
  MaterialExpressionScalarParameter_143 {class=MaterialExpressionScalarParameter, properties {defaultValue=0.1, parameterName=Frenelcore Max}}
  MaterialExpressionScalarParameter_144 {class=MaterialExpressionScalarParameter, properties {defaultValue=1.5, parameterName=Param_max_Top}}
  MaterialExpressionScalarParameter_145 {class=MaterialExpressionScalarParameter, properties {defaultValue=0.2, parameterName=Param_2}}
  MaterialExpressionScalarParameter_146 {class=MaterialExpressionScalarParameter, properties {defaultValue=0.4, group=SpriteVariants, parameterName=SpriteVar_Candy_Color_HueShift}}
  MaterialExpressionScalarParameter_147 {
   class=MaterialExpressionScalarParameter
   properties {defaultValue=1, group=Sprite_Cosmetics | Triplanar, parameterName=Derivatives Use Pre-Skinned Scale, sliderMax=1, sortPriority=30}
  }
  MaterialExpressionScalarParameter_150 {
   class=MaterialExpressionScalarParameter
   properties {defaultValue=2, group=Sprite_Cosmetics | Triplanar, parameterName=Derivatives Mult, sortPriority=40}
  }
  MaterialExpressionScalarParameter_98 {class=MaterialExpressionScalarParameter, properties {defaultValue=2, group=SpriteVariants_Candy, parameterName=Slurp_InnerScatteringPower}}
  MaterialExpressionScalarParameter_99 {class=MaterialExpressionScalarParameter, properties {defaultValue=0.5, group=SpriteVariants_Candy, parameterName=Slurp_InnerScatteringScale}}
  MaterialExpressionSetMaterialAttributes_2 {
   class=MaterialExpressionSetMaterialAttributes
   inputs [
    MaterialAttributes {from=MaterialExpressionNamedRerouteUsage_100, output=""}
    BaseColor {from=MaterialExpressionNamedRerouteUsage_107, output=""}
    EmissiveColor {from=MaterialExpressionNamedRerouteUsage_105, output=""}
    Metallic {from=MaterialExpressionNamedRerouteUsage_134, output=""}
    Specular {from=MaterialExpressionNamedRerouteUsage_106, output=""}
    Roughness {from=MaterialExpressionNamedRerouteUsage_108, output=""}
    Normal {from=MaterialExpressionNamedRerouteUsage_109, output=""}
   ]
   properties {attributeSetTypes [
     #2
     #6
     #3
     #5
     #4
     #1
    ]}
  }
  MaterialExpressionShaderStageSwitch_5 {
   class=MaterialExpressionShaderStageSwitch
   inputs [
    PixelShader {from=MaterialExpressionVertexInterpolator_5, output=PS}
    VertexShader {from=MaterialExpressionPreSkinnedNormal_2, output=""}
   ]
   properties {}
  }
  MaterialExpressionShaderStageSwitch_6 {
   class=MaterialExpressionShaderStageSwitch
   inputs [
    PixelShader {from=MaterialExpressionVertexInterpolator_6, output=PS}
    VertexShader {from=MaterialExpressionLocalPosition_5, output=XYZ}
   ]
   properties {}
  }
  MaterialExpressionSine_2 {class=MaterialExpressionSine, inputs [None {from=MaterialExpressionDivide_19}]}
  MaterialExpressionSmoothStep_12 {
   class=MaterialExpressionSmoothStep
   inputs [
    Min {from=MaterialExpressionScalarParameter_103, output=""}
    Max {from=MaterialExpressionScalarParameter_104, output=""}
    Value {from=MaterialExpressionReroute_46, output=""}
   ]
   properties {constMax=0.7, constMin=0.4}
  }
  MaterialExpressionSmoothStep_13 {
   class=MaterialExpressionSmoothStep
   inputs [
    Min {from=MaterialExpressionScalarParameter_117, output=""}
    Max {from=MaterialExpressionScalarParameter_118, output=""}
    Value {from=MaterialExpressionNamedRerouteUsage_121, output=""}
   ]
   properties {}
  }
  MaterialExpressionSmoothStep_14 {
   class=MaterialExpressionSmoothStep
   inputs [
    Min {from=MaterialExpressionScalarParameter_129, output=""}
    Max {from="", output=""}
    Value {from=MaterialExpressionDotProduct_18, output=""}
   ]
   properties {}
  }
  MaterialExpressionSmoothStep_15 {
   class=MaterialExpressionSmoothStep
   inputs [
    Min {from=MaterialExpressionScalarParameter_138, output=""}
    Max {from=MaterialExpressionLinearInterpolate_86, output=""}
    Value {from=MaterialExpressionPower_16, output=""}
   ]
   properties {}
  }
  MaterialExpressionSmoothStep_16 {
   class=MaterialExpressionSmoothStep
   inputs [
    Min {from=MaterialExpressionScalarParameter_140, output=""}
    Max {from=MaterialExpressionScalarParameter_141, output=""}
    Value {from=MaterialExpressionNamedRerouteUsage_146, output=""}
   ]
   properties {}
  }
  MaterialExpressionSmoothStep_17 {
   class=MaterialExpressionSmoothStep
   inputs [
    Min {from="", output=""}
    Max {from="", output=""}
    Value {from=MaterialExpressionSaturate_21, output=""}
   ]
   properties {constMax=0.5}
  }
  MaterialExpressionSphereMask_2 {
   class=MaterialExpressionSphereMask
   inputs [
    A {from=MaterialExpressionReroute_46, output=""}
    B {from=MaterialExpressionAdd_36, output=""}
    Radius {from=MaterialExpressionAdd_37, output=""}
    Hardness {from=MaterialExpressionConstant_17, output=""}
   ]
   properties {}
  }
  MaterialExpressionStaticSwitchParameter_0 {
   class=MaterialExpressionStaticSwitchParameter
   inputs [
    True {from=MaterialExpressionSetMaterialAttributes_2, output=""}
    False {from=MaterialExpressionNamedRerouteUsage_100, output=""}
   ]
   properties {group=SpriteVariants, parameterName=Sprite_Variant_Candy, sortPriority=-1}
  }
  MaterialExpressionSubtract_40 {
   class=MaterialExpressionSubtract
   inputs [
    A {from=MaterialExpressionNamedRerouteUsage_36}
    B {from=MaterialExpressionNamedRerouteUsage_151}
   ]
  }
  MaterialExpressionSubtract_41 {
   class=MaterialExpressionSubtract
   inputs [
    A {from=MaterialExpressionNamedRerouteUsage_153}
    B {from=MaterialExpressionNamedRerouteUsage_154}
   ]
  }
  MaterialExpressionSubtract_7 {
   class=MaterialExpressionSubtract
   inputs [
    A {from=MaterialExpressionMultiply_85}
    B {from=MaterialExpressionComponentMask_22}
   ]
  }
  MaterialExpressionSubtract_8 {
   class=MaterialExpressionSubtract
   inputs [
    A {from=MaterialExpressionScalarParameter_104}
    B {from=MaterialExpressionScalarParameter_103}
   ]
  }
  MaterialExpressionSubtract_9 {
   class=MaterialExpressionSubtract
   inputs [
    A {from=MaterialExpressionSaturate_21}
    B {from=""}
   ]
   properties {constB=0.1}
  }
  MaterialExpressionTextureCoordinate_4 {class=MaterialExpressionTextureCoordinate}
  MaterialExpressionTextureCoordinate_5 {class=MaterialExpressionTextureCoordinate}
  MaterialExpressionTextureCoordinate_6 {class=MaterialExpressionTextureCoordinate}
  MaterialExpressionTextureObject_4 {
   class=MaterialExpressionTextureObject
   inputs []
   properties {samplerType=SAMPLERTYPE_LinearColor, texture=&HighContrastNoise}
  }
  MaterialExpressionTextureObject_5 {
   class=MaterialExpressionTextureObject
   inputs []
   properties {samplerType=SAMPLERTYPE_LinearColor, texture=/BRCosmetics/Animation/Props/Gadgets_PixieParts/Textures/TA/T_SpritesVar_Candy_Sugar_Mask}
  }
  MaterialExpressionTextureSampleParameter2D_4 {
   class=MaterialExpressionTextureSampleParameter2D
   inputs [
    UVs {from=MaterialExpressionMultiply_119, output=""}
    Apply View MipBias {from="", output=""}
   ]
   properties {
    group=Glitter
    parameterName=Glitter Direction Mask
    samplerSource=SSM_Wrap_WorldGroupSettings
    samplerType=SAMPLERTYPE_Masks
    sortPriority=0
    texture=/Game/Weapons/WeaponSkins/Textures/T_GlitterDirectionMask
   }
  }
  MaterialExpressionTextureSampleParameter2D_5 {
   class=MaterialExpressionTextureSampleParameter2D
   inputs [
    UVs {from=MaterialExpressionMultiply_120, output=""}
    Apply View MipBias {from="", output=""}
   ]
   properties {
    group=Glitter
    parameterName=Glitter Direction Mask
    samplerSource=SSM_Wrap_WorldGroupSettings
    samplerType=SAMPLERTYPE_Masks
    sortPriority=0
    texture=&T_GlitterDirectionMask
   }
  }
  MaterialExpressionTextureSample_6 {
   class=MaterialExpressionTextureSample
   inputs [
    UVs {from=MaterialExpressionNamedRerouteDeclaration_92}
    Tex {from=MaterialExpressionNamedRerouteUsage_117}
    "DDX(UVs)" {from=MaterialExpressionNamedRerouteUsage_161}
    "DDY(UVs)" {from=MaterialExpressionNamedRerouteUsage_162}
    Apply View MipBias {from=""}
   ]
   properties {mipValueMode=TMVM_Derivative, samplerType=SAMPLERTYPE_LinearColor}
  }
  MaterialExpressionTextureSample_7 {
   class=MaterialExpressionTextureSample
   inputs [
    UVs {from=MaterialExpressionAdd_41}
    Tex {from=MaterialExpressionNamedRerouteUsage_122}
    "DDX(UVs)" {from=MaterialExpressionNamedRerouteUsage_158}
    "DDY(UVs)" {from=MaterialExpressionNamedRerouteUsage_159}
    Apply View MipBias {from=""}
   ]
   properties {mipValueMode=TMVM_Derivative, samplerType=SAMPLERTYPE_LinearColor}
  }
  MaterialExpressionTextureSample_8 {
   class=MaterialExpressionTextureSample
   inputs [
    UVs {from=MaterialExpressionDivide_16}
    Tex {from=MaterialExpressionNamedRerouteUsage_122}
    "DDX(UVs)" {from=MaterialExpressionNamedRerouteUsage_158}
    "DDY(UVs)" {from=MaterialExpressionNamedRerouteUsage_159}
    Apply View MipBias {from=""}
   ]
   properties {mipValueMode=TMVM_Derivative, samplerType=SAMPLERTYPE_LinearColor}
  }
  MaterialExpressionTransform_6 {
   class=MaterialExpressionTransform
   inputs [None {from=MaterialExpressionConstant3Vector_11}]
   properties {transformSourceType=TRANSFORMSOURCE_View}
  }
  MaterialExpressionTransform_7 {
   class=MaterialExpressionTransform
   inputs [None {from=MaterialExpressionConstant3Vector_13}]
   properties {desc=Camera Vector Forward, transformSourceType=TRANSFORMSOURCE_Camera}
  }
  MaterialExpressionTransform_8 {
   class=MaterialExpressionTransform
   inputs [None {from=MaterialExpressionConstant3Vector_14}]
   properties {transformSourceType=TRANSFORMSOURCE_Local}
  }
  MaterialExpressionVectorParameter_36 {
   class=MaterialExpressionVectorParameter
   properties {group=SpriteVariants, parameterName=SpriteVar_Candy_Emissive_InnerScatteringColor, defaultValue=<rgba 2 0.458 0.8066 1>}
  }
  MaterialExpressionVectorParameter_37 {
   class=MaterialExpressionVectorParameter
   properties {group=SpriteVariants, parameterName=SpriteVar_Candy_Emissive_ColorB, defaultValue=<rgba 0 0.3385 0.0423 1>}
  }
  MaterialExpressionVectorParameter_38 {
   class=MaterialExpressionVectorParameter
   properties {group=SpriteVariants, parameterName=SpriteVar_Candy_Emissive_ColorA, defaultValue=<rgba 0.8464 0.0254 0.1253 1>}
  }
  MaterialExpressionVectorParameter_39 {
   class=MaterialExpressionVectorParameter
   properties {group=SpriteVariants, parameterName="SpriteVar_Candy_Emissive_BackSSS(Shadow)", defaultValue=<rgba 2 1.2174 1.4216 1>}
  }
  MaterialExpressionVectorParameter_40 {
   class=MaterialExpressionVectorParameter
   properties {
    desc=1.1.1. 0.7
    parameterName=SpriteVar_Candy_Emissive_DiffuseBlend_Mask
    channelNames {a=InverseMask, b=DiffuseB Influence, g=DiffuseG Influence, r=DiffuseR Influence}
    defaultValue=<rgba 0.8571 0 6 0.9>
   }
  }
  MaterialExpressionVectorParameter_41 {
   class=MaterialExpressionVectorParameter
   properties {group=SpriteVariants, parameterName=SpriteVar_Candy_Emissive_Logo, defaultValue=<rgba 3 0.6261 1.4144 1>}
  }
  MaterialExpressionVectorParameter_42 {
   class=MaterialExpressionVectorParameter
   properties {parameterName=SpriteVar_Candy_Emissive_Transition_Mult, defaultValue=<rgba 1.513 0.136 0.3994 1>}
  }
  MaterialExpressionVectorParameter_43 {
   class=MaterialExpressionVectorParameter
   properties {parameterName=SpriteVar_Candy_Emissive_DiffuseBlend_Color, defaultValue=<rgba 0.7917 0.1113 0.6083 1>}
  }
  MaterialExpressionVectorParameter_44 {
   class=MaterialExpressionVectorParameter
   properties {group=SpriteVariants, parameterName=SpriteVar_Candy_Emissive_Terminator, defaultValue=<rgba 0.0626 0 0.1 1>}
  }
  MaterialExpressionVectorParameter_45 {
   class=MaterialExpressionVectorParameter
   properties {group=SpriteVariants, parameterName=SpriteVar_Candy_Emissive_TOD_Morning_Fresnel, defaultValue=<rgba 3.7837 4 2.087 1>}
  }
  MaterialExpressionVectorParameter_46 {
   class=MaterialExpressionVectorParameter
   properties {group=SpriteVariants, parameterName=SpriteVar_Candy_Emissive_TOD_Day_Fresnel, defaultValue=<rgba 4.4075 4.5 3.8739 1>}
  }
  MaterialExpressionVectorParameter_47 {
   class=MaterialExpressionVectorParameter
   properties {group=SpriteVariants, parameterName=SpriteVar_Candy_Emissive_TOD_Evening_Fresnel, defaultValue=<rgba 4.5 3.0964 1.5652 1>}
  }
  MaterialExpressionVectorParameter_48 {
   class=MaterialExpressionVectorParameter
   properties {group=SpriteVariants, parameterName=SpriteVar_Candy_Emissive_TOD_Night_Fresnel, defaultValue=<rgba 1.8087 5.2544 8 1>}
  }
  MaterialExpressionVectorParameter_49 {
   class=MaterialExpressionVectorParameter
   properties {group=SpriteVariants, parameterName=SpriteVar_Candy_Emissive_TOD_Morning, defaultValue=<rgba 1.3 1.0122 0.7574 1>}
  }
  MaterialExpressionVectorParameter_50 {
   class=MaterialExpressionVectorParameter
   properties {group=SpriteVariants, parameterName=SpriteVar_Candy_Emissive_TOD_Day, defaultValue=<rgba 1 1 1 1>}
  }
  MaterialExpressionVectorParameter_51 {
   class=MaterialExpressionVectorParameter
   properties {group=SpriteVariants, parameterName=SpriteVar_Candy_Emissive_TOD_Evening, defaultValue=<rgba 1 0.8926 0.7478 1>}
  }
  MaterialExpressionVectorParameter_52 {
   class=MaterialExpressionVectorParameter
   properties {group=SpriteVariants, parameterName=SpriteVar_Candy_Emissive_TOD_Night, defaultValue=<rgba 0.77 1.034 1.1 1>}
  }
  MaterialExpressionVectorParameter_53 {class=MaterialExpressionVectorParameter, properties {parameterName=SpriteVar_Candy_Emissive_Core_Mult, defaultValue=<rgba 0.3 0 0.1293 1>}}
  MaterialExpressionVertexInterpolator_5 {class=MaterialExpressionVertexInterpolator, inputs [VS {from=MaterialExpressionPreSkinnedNormal_2, output=""}], properties {}}
  MaterialExpressionVertexInterpolator_6 {class=MaterialExpressionVertexInterpolator, inputs [VS {from=MaterialExpressionLocalPosition_5, output=XYZ}], properties {}}
  MaterialExpressionVertexNormalWS_4 {class=MaterialExpressionVertexNormalWS, inputs [], properties {}}
  MaterialExpressionVertexNormalWS_6 {class=MaterialExpressionVertexNormalWS, inputs [], properties {}}
  MaterialExpressionVertexNormalWS_8 {class=MaterialExpressionVertexNormalWS, inputs [], properties {}}
 }}
