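# Material function that produces screen-space UVs stabilized to an object, using a position and radius that default to the actor position and object radius.
# Viewport UVs are offset by the position's clip-space projection, with a view-size factor that evaluates to 1 as wired.
# ScaleUVsByCenter scales them by radius over view depth times UVRadialScale, so the pattern tracks the object's apparent size.
# Outputs are the stabilized UVs and a 0-1 mask.
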MaterialFunction /BRCosmetics/Animation/Props/Gadgets_PixieParts/Materials/ParentMaterials/MF_ObjectStabilizedScreenSpaceUVs
graph {
 properties {description="Screen Space UV's stabilzied to Object "}
 nodes [
  n1 = MaterialFunctionCall(Float2=ViewSize()) {materialFunction=&BreakOutFloat2Components}
  n2 = AppendVector(A=Constant() {r=1}, B=Divide(A=@n1.G, B=@n1.G))
  n3 = Multiply(A=ConstantBiasScale(None=ScreenPosition().ViewportUV) {bias=-0.5, scale=2}, B=@n2)
  n4 = FunctionInput(Preview=ActorPositionWS()) {bUsePreviewValueAsDefault=true, inputName=Position}
  n5 = Multiply(
   A=Constant2Vector() {g=-0.5, r=0.5}
   B=MaterialFunctionCall(World Position=@n4) {materialFunction=&TransformToClipSpace}.Clip Space XY
  )
  n6 = FunctionInput(Preview=ObjectRadius()) {bUsePreviewValueAsDefault=true, inputName=ObjectRadius, sortPriority=1}
  n7 = ComponentMask(None=TransformPosition(None=@n4) {transformSourceType=TRANSFORMPOSSOURCE_World, transformType=TRANSFORMPOSSOURCE_View}) {b=true}
  n8 = Constant() {r=1}
  n9 = MaterialFunctionCall(
   UVs=Subtract(A=ConstantBiasScale(None=@n3), B=Multiply(A=@n2, B=@n5))
   Texture Scale=Multiply(
    A=Divide(A=@n6, B=@n7)
    B=FunctionInput(Preview=@n8) {bUsePreviewValueAsDefault=true, inputName=UVRadialScale, inputType=FunctionInput_Scalar}
   )
  ) {materialFunction=&ScaleUVsByCenter}
  FunctionOutput(None=@n9.UVs) {outputName=StabilizedUVs}
  FunctionOutput(None=@n9.0-1 mask) {outputName=0-1_MAsk}
 ]
}
