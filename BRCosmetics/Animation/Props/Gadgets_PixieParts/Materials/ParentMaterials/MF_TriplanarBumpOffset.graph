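# Material function for the PixieParts gadget prop materials that adds a view-dependent bump offset to triplanar UVs, faking depth.
# It transforms the camera vector, defaulting to CameraDirectionVector, from world to local space and builds XY, XZ and YZ offsets, negating Z in the latter two.
# The triplanar RGB mask blends these offsets, which are scaled by the scalar Bump Offset and the object's X scale, then added to the UVs.
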MaterialFunction /BRCosmetics/Animation/Props/Gadgets_PixieParts/Materials/ParentMaterials/MF_TriplanarBumpOffset
graph {nodes [
  n1 = FunctionInput() {desc=Abs RGB Mask, inputName=TriplanarMask, sortPriority=10}
  n2 = ComponentMask(None=@n1) {b=true}
  n3 = Transform(None=FunctionInput(Preview=MaterialFunctionCall() {materialFunction=&CameraDirectionVector}.Result) {bUsePreviewValueAsDefault=true, inputName=CameraVector, sortPriority=19}) {transformSourceType=TRANSFORMSOURCE_World, transformType=TRANSFORM_Local}
  n4 = Multiply(A=ComponentMask(None=@n3) {g=true, r=true})
  n5 = Constant2Vector() {g=-1, r=1}
  n6 = Multiply(A=ComponentMask(None=@n1) {r=true}, B=Multiply(A=ComponentMask(None=@n3) {b=true, g=true}, B=@n5))
  n7 = Multiply(A=Multiply(A=ComponentMask(None=@n3) {b=true, r=true}, B=@n5), B=ComponentMask(None=@n1) {g=true})
  n8 = Multiply(
   A=Add(A=Multiply(A=@n2, B=@n4), B=Add(A=@n6, B=@n7))
   B=FunctionInput() {inputName=Bump Offset, inputType=FunctionInput_Scalar, sortPriority=20}
  ) {constB=0.5}
  n9 = MaterialFunctionCall() {materialFunction=&ObjectScale}
  FunctionOutput(None=Add(A=FunctionInput() {inputName=UVs, inputType=FunctionInput_Vector2}, B=Multiply(A=@n8, B=@n9.Scale X))) {outputName=Add}
 ]}
