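# Material function for a cube sprite variant on Pixie Parts gadget props, taking material attributes and masks for variant, logo, fresnel, blinking lights and inner glow.
# Pre-skinned position and bounds inputs support object-space mapping, including dithered triplanar UVs with bump offset.
# Interior-cubemap UVs, height blending, camera-relative fresnel and distance fades, and time-of-day colors add layered detail.
# A static switch selects the output material attributes.
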MaterialFunction /BRCosmetics/Animation/Props/Gadgets_PixieParts/Materials/ParentMaterials/MF_Sprite_Variants_Cube
graph {expressions {
  MaterialExpressionAbs_0 {class=MaterialExpressionAbs, inputs [None {from=MaterialExpressionDotProduct_3, output=""}], properties {}}
  MaterialExpressionAbs_1 {class=MaterialExpressionAbs, inputs [None {from=MaterialExpressionNamedRerouteUsage_7, output=""}], properties {}}
  MaterialExpressionAbs_3 {class=MaterialExpressionAbs, inputs [None {from=MaterialExpressionSubtract_1, output=""}], properties {}}
  MaterialExpressionAbs_4 {class=MaterialExpressionAbs, inputs [None {from=MaterialExpressionMaterialFunctionCall_17, output=UVW}], properties {}}
  MaterialExpressionAdd_0 {
   class=MaterialExpressionAdd
   inputs [
    A {from=MaterialExpressionReroute_0}
    B {from=MaterialExpressionSubtract_7}
   ]
  }
  MaterialExpressionAdd_1 {
   class=MaterialExpressionAdd
   inputs [
    A {from=MaterialExpressionReroute_6}
    B {from=MaterialExpressionMultiply_1}
   ]
  }
  MaterialExpressionAdd_10 {
   class=MaterialExpressionAdd
   inputs [
    A {from=MaterialExpressionAdd_8}
    B {from=""}
   ]
  }
  MaterialExpressionAdd_101 {
   class=MaterialExpressionAdd
   inputs [
    A {from=MaterialExpressionNamedRerouteUsage_86}
    B {from=MaterialExpressionConstant_15}
   ]
  }
  MaterialExpressionAdd_13 {
   class=MaterialExpressionAdd
   inputs [
    A {from=MaterialExpressionReroute_78}
    B {from=""}
   ]
   properties {constB=0.01}
  }
  MaterialExpressionAdd_2 {
   class=MaterialExpressionAdd
   inputs [
    A {from=MaterialExpressionReroute_27}
    B {from=MaterialExpressionMultiply_9}
   ]
  }
  MaterialExpressionAdd_22 {
   class=MaterialExpressionAdd
   inputs [
    A {from=MaterialExpressionScalarParameter_14}
    B {from=MaterialExpressionMaterialFunctionCall_42, output=R}
   ]
  }
  MaterialExpressionAdd_23 {
   class=MaterialExpressionAdd
   inputs [
    A {from=MaterialExpressionMaterialFunctionCall_42, output=G}
    B {from=MaterialExpressionScalarParameter_19}
   ]
  }
  MaterialExpressionAdd_3 {
   class=MaterialExpressionAdd
   inputs [
    A {from=MaterialExpressionReroute_10}
    B {from=MaterialExpressionMultiply_14}
   ]
  }
  MaterialExpressionAdd_33 {
   class=MaterialExpressionAdd
   inputs [
    A {from=MaterialExpressionObjectPositionWS_1}
    B {from=MaterialExpressionObjectBounds_0}
   ]
  }
  MaterialExpressionAdd_34 {
   class=MaterialExpressionAdd
   inputs [
    A {from=MaterialExpressionMaterialFunctionCall_61, output=UVs}
    B {from=""}
   ]
   properties {constA=0.5, constB=0.5}
  }
  MaterialExpressionAdd_39 {
   class=MaterialExpressionAdd
   inputs [
    A {from=MaterialExpressionNamedRerouteUsage_14}
    B {from=MaterialExpressionScalarParameter_196}
   ]
  }
  MaterialExpressionAdd_4 {
   class=MaterialExpressionAdd
   inputs [
    A {from=MaterialExpressionPower_19}
    B {from=""}
   ]
   properties {constB=0.2}
  }
  MaterialExpressionAdd_40 {
   class=MaterialExpressionAdd
   inputs [
    A {from=MaterialExpressionNamedRerouteUsage_158}
    B {from=MaterialExpressionVectorParameter_37, output=RGB}
   ]
  }
  MaterialExpressionAdd_41 {
   class=MaterialExpressionAdd
   inputs [
    A {from=MaterialExpressionAdd_40}
    B {from=MaterialExpressionVectorParameter_38, output=RGB}
   ]
  }
  MaterialExpressionAdd_44 {
   class=MaterialExpressionAdd
   inputs [
    A {from=MaterialExpressionReroute_20}
    B {from=MaterialExpressionMultiply_56}
   ]
  }
  MaterialExpressionAdd_45 {
   class=MaterialExpressionAdd
   inputs [
    A {from=MaterialExpressionMultiply_63}
    B {from=MaterialExpressionReroute_64}
   ]
  }
  MaterialExpressionAdd_48 {
   class=MaterialExpressionAdd
   inputs [
    A {from=MaterialExpressionMultiply_100}
    B {from=MaterialExpressionMultiply_77}
   ]
  }
  MaterialExpressionAdd_5 {
   class=MaterialExpressionAdd
   inputs [
    A {from=MaterialExpressionReroute_25}
    B {from=MaterialExpressionSubtract_0}
   ]
  }
  MaterialExpressionAdd_51 {
   class=MaterialExpressionAdd
   inputs [
    A {from=MaterialExpressionMultiply_139}
    B {from=MaterialExpressionReroute_79}
   ]
  }
  MaterialExpressionAdd_52 {
   class=MaterialExpressionAdd
   inputs [
    A {from=MaterialExpressionLinearInterpolate_18}
    B {from=MaterialExpressionMultiply_140}
   ]
  }
  MaterialExpressionAdd_54 {
   class=MaterialExpressionAdd
   inputs [
    A {from=MaterialExpressionFresnel_1}
    B {from=MaterialExpressionScalarParameter_73}
   ]
  }
  MaterialExpressionAdd_55 {
   class=MaterialExpressionAdd
   inputs [
    A {from=MaterialExpressionDDX_0}
    B {from=MaterialExpressionDDX_1}
   ]
  }
  MaterialExpressionAdd_6 {
   class=MaterialExpressionAdd
   inputs [
    A {from=MaterialExpressionNamedRerouteUsage_39}
    B {from=""}
   ]
   properties {constB=0}
  }
  MaterialExpressionAdd_61 {
   class=MaterialExpressionAdd
   inputs [
    A {from=MaterialExpressionDivide_28}
    B {from=""}
   ]
   properties {constB=-0.5}
  }
  MaterialExpressionAdd_62 {
   class=MaterialExpressionAdd
   inputs [
    A {from=MaterialExpressionReroute_2}
    B {from=MaterialExpressionSubtract_6}
   ]
  }
  MaterialExpressionAdd_67 {
   class=MaterialExpressionAdd
   inputs [
    A {from=MaterialExpressionNamedRerouteUsage_32}
    B {from=MaterialExpressionConstant_22}
   ]
  }
  MaterialExpressionAdd_7 {
   class=MaterialExpressionAdd
   inputs [
    A {from=MaterialExpressionAdd_0}
    B {from=""}
   ]
   properties {constB=0.1}
  }
  MaterialExpressionAdd_75 {
   class=MaterialExpressionAdd
   inputs [
    A {from=MaterialExpressionMaterialFunctionCall_30, output=Lerp 4 Inputs}
    B {from=MaterialExpressionMultiply_153}
   ]
  }
  MaterialExpressionAdd_8 {
   class=MaterialExpressionAdd
   inputs [
    A {from=MaterialExpressionMultiply_6}
    B {from=MaterialExpressionAdd_9}
   ]
  }
  MaterialExpressionAdd_82 {
   class=MaterialExpressionAdd
   inputs [
    A {from=MaterialExpressionMultiply_219}
    B {from=MaterialExpressionVectorParameter_30, output=RGB}
   ]
  }
  MaterialExpressionAdd_9 {
   class=MaterialExpressionAdd
   inputs [
    A {from=MaterialExpressionMaterialFunctionCall_12, output=G}
    B {from=MaterialExpressionMaterialFunctionCall_12, output=G}
   ]
  }
  MaterialExpressionAppendVector_0 {
   class=MaterialExpressionAppendVector
   inputs [
    A {from=MaterialExpressionLinearInterpolate_2, output=""}
    B {from=MaterialExpressionConstant_1, output=""}
   ]
  }
  MaterialExpressionBounds_0 {class=MaterialExpressionBounds, inputs [], properties {type=MEILB_PreSkinnedLocal}}
  MaterialExpressionCameraPositionWS_0 {class=MaterialExpressionCameraPositionWS, inputs [], properties {}}
  MaterialExpressionCameraPositionWS_1 {class=MaterialExpressionCameraPositionWS, inputs [], properties {}}
  MaterialExpressionCameraPositionWS_14 {class=MaterialExpressionCameraPositionWS, inputs [], properties {}}
  MaterialExpressionCameraPositionWS_2 {class=MaterialExpressionCameraPositionWS, inputs [], properties {}}
  MaterialExpressionCameraVectorWS_1 {class=MaterialExpressionCameraVectorWS, inputs [], properties {}}
  MaterialExpressionCameraVectorWS_2 {class=MaterialExpressionCameraVectorWS, inputs [], properties {}}
  MaterialExpressionCameraVectorWS_3 {class=MaterialExpressionCameraVectorWS, inputs [], properties {}}
  MaterialExpressionCameraVectorWS_5 {class=MaterialExpressionCameraVectorWS, inputs [], properties {}}
  MaterialExpressionComponentMask_0 {class=MaterialExpressionComponentMask, inputs [None {from=MaterialExpressionNamedRerouteUsage_159}], properties {r=true}}
  MaterialExpressionComponentMask_1 {class=MaterialExpressionComponentMask, inputs [None {from=MaterialExpressionNamedRerouteUsage_154}], properties {r=true}}
  MaterialExpressionComponentMask_10 {class=MaterialExpressionComponentMask, inputs [None {from=MaterialExpressionVectorParameter_40, output=RGB}], properties {g=true, r=true}}
  MaterialExpressionComponentMask_16 {class=MaterialExpressionComponentMask, inputs [None {from=MaterialExpressionDivide_28}], properties {b=true}}
  MaterialExpressionComponentMask_17 {class=MaterialExpressionComponentMask, inputs [None {from=MaterialExpressionAbs_4}], properties {b=true}}
  MaterialExpressionComponentMask_18 {class=MaterialExpressionComponentMask, inputs [None {from=MaterialExpressionAbs_4}], properties {g=true}}
  MaterialExpressionComponentMask_9 {class=MaterialExpressionComponentMask, inputs [None {from=MaterialExpressionMaterialFunctionCall_17, output=UVW}], properties {r=true}}
  MaterialExpressionConstant2Vector_1 {class=MaterialExpressionConstant2Vector, properties {g=0.6, r=0.5}}
  MaterialExpressionConstant2Vector_3 {class=MaterialExpressionConstant2Vector, properties {g=0.5, r=0.5}}
  MaterialExpressionConstant2Vector_4 {class=MaterialExpressionConstant2Vector}
  MaterialExpressionConstant2Vector_6 {class=MaterialExpressionConstant2Vector, properties {g=0.5, r=0.5}}
  MaterialExpressionConstant3Vector_0 {class=MaterialExpressionConstant3Vector, properties {constant=<rgba 1.35 1.4 1.35 0>}}
  MaterialExpressionConstant3Vector_3 {class=MaterialExpressionConstant3Vector}
  MaterialExpressionConstant3Vector_4 {class=MaterialExpressionConstant3Vector, properties {constant=<rgba 0 0 0.03 0>}}
  MaterialExpressionConstant3Vector_5 {class=MaterialExpressionConstant3Vector, properties {constant=<rgba 0 0 1 0>}}
  MaterialExpressionConstant4Vector_0 {class=MaterialExpressionConstant4Vector, inputs [], properties {constant=<rgba 1.35 1.6 1.35 1>}}
  MaterialExpressionConstant4Vector_1 {class=MaterialExpressionConstant4Vector, inputs [], properties {constant=<rgba 0.0356 0.001 0.2 1>}}
  MaterialExpressionConstant4Vector_2 {class=MaterialExpressionConstant4Vector, inputs [], properties {constant=<rgba 0.1264 0 0.3 1>}}
  MaterialExpressionConstant4Vector_3 {class=MaterialExpressionConstant4Vector, inputs [], properties {constant=<rgba 0.3583 0 0.5 1>}}
  MaterialExpressionConstant4Vector_4 {class=MaterialExpressionConstant4Vector, inputs [], properties {constant=<rgba 0.012 0.0011 0.08 1>}}
  MaterialExpressionConstant4Vector_5 {class=MaterialExpressionConstant4Vector, inputs [], properties {constant=<rgba 0.6076 0 1 1>}}
  MaterialExpressionConstant4Vector_6 {class=MaterialExpressionConstant4Vector, inputs [], properties {constant=<rgba 3 2 3.5 1>}}
  MaterialExpressionConstant4Vector_7 {class=MaterialExpressionConstant4Vector, inputs [], properties {constant=<rgba 0 7 15.5 1>}}
  MaterialExpressionConstant4Vector_8 {class=MaterialExpressionConstant4Vector, inputs [], properties {constant=<rgba 1.1 1.4 1.7 1>}}
  MaterialExpressionConstantBiasScale_0 {class=MaterialExpressionConstantBiasScale, inputs [None {from=MaterialExpressionNamedRerouteUsage_3}]}
  MaterialExpressionConstant_1 {class=MaterialExpressionConstant, properties {r=1}}
  MaterialExpressionConstant_10 {class=MaterialExpressionConstant, properties {r=15}}
  MaterialExpressionConstant_12 {class=MaterialExpressionConstant}
  MaterialExpressionConstant_15 {class=MaterialExpressionConstant, properties {r=-0.05}}
  MaterialExpressionConstant_17 {class=MaterialExpressionConstant, properties {r=2}}
  MaterialExpressionConstant_18 {class=MaterialExpressionConstant, properties {r=7}}
  MaterialExpressionConstant_2 {class=MaterialExpressionConstant, properties {r=0.5}}
  MaterialExpressionConstant_21 {class=MaterialExpressionConstant, properties {desc=2, r=100}}
  MaterialExpressionConstant_22 {class=MaterialExpressionConstant, properties {r=0.15}}
  MaterialExpressionConstant_27 {class=MaterialExpressionConstant, properties {r=1}}
  MaterialExpressionConstant_28 {class=MaterialExpressionConstant}
  MaterialExpressionConstant_29 {class=MaterialExpressionConstant, properties {r=10}}
  MaterialExpressionConstant_3 {class=MaterialExpressionConstant, properties {desc=1.5, r=2.5}}
  MaterialExpressionConstant_30 {class=MaterialExpressionConstant, properties {r=1}}
  MaterialExpressionConstant_31 {class=MaterialExpressionConstant, properties {r=1}}
  MaterialExpressionConstant_32 {class=MaterialExpressionConstant, properties {r=12}}
  MaterialExpressionConstant_33 {class=MaterialExpressionConstant, properties {r=14}}
  MaterialExpressionConstant_34 {class=MaterialExpressionConstant}
  MaterialExpressionConstant_35 {class=MaterialExpressionConstant}
  MaterialExpressionConstant_42 {class=MaterialExpressionConstant, properties {r=8}}
  MaterialExpressionConstant_44 {class=MaterialExpressionConstant, properties {r=2}}
  MaterialExpressionConstant_45 {class=MaterialExpressionConstant}
  MaterialExpressionConstant_46 {class=MaterialExpressionConstant, properties {r=1200}}
  MaterialExpressionConstant_47 {class=MaterialExpressionConstant, properties {r=1400}}
  MaterialExpressionConstant_48 {class=MaterialExpressionConstant}
  MaterialExpressionConstant_49 {class=MaterialExpressionConstant}
  MaterialExpressionConstant_5 {class=MaterialExpressionConstant, properties {r=1.2}}
  MaterialExpressionConstant_54 {class=MaterialExpressionConstant, properties {r=2.75}}
  MaterialExpressionConstant_55 {class=MaterialExpressionConstant, properties {r=1.5}}
  MaterialExpressionConstant_56 {class=MaterialExpressionConstant, properties {r=3}}
  MaterialExpressionConstant_57 {class=MaterialExpressionConstant, properties {r=10}}
  MaterialExpressionConstant_58 {class=MaterialExpressionConstant, properties {r=0.6}}
  MaterialExpressionConstant_6 {class=MaterialExpressionConstant, properties {desc=2, r=100}}
  MaterialExpressionConstant_7 {class=MaterialExpressionConstant, properties {r=1.4}}
  MaterialExpressionConstant_72 {class=MaterialExpressionConstant, properties {r=1.5}}
  MaterialExpressionConstant_73 {class=MaterialExpressionConstant, properties {r=1}}
  MaterialExpressionConstant_74 {class=MaterialExpressionConstant, properties {r=-1}}
  MaterialExpressionConstant_8 {class=MaterialExpressionConstant, properties {r=0.05}}
  MaterialExpressionConstant_9 {class=MaterialExpressionConstant, properties {r=0.7}}
  MaterialExpressionConvert_0 {
   class=MaterialExpressionConvert
   inputs [
    None {from=""}
    None {from=""}
    None {from=MaterialExpressionObjectLocalBounds_0, output=Half Extents}
   ]
   properties {
    nodeName=Make Vector3
    convertInputs [
     {type=Scalar, expressionInput {expression=null, inputName=None, mask=0, maskA=0, maskB=0, maskG=0, maskR=0, outputIndex=0}}
     {type=Scalar, expressionInput {expression=null, inputName=None, mask=0, maskA=0, maskB=0, maskG=0, maskR=0, outputIndex=0}}
     {
      type=Scalar
      expressionInput {
       expression=&self:MaterialExpressionObjectLocalBounds_0
       inputName=None
       mask=1
       maskA=0
       maskB=1
       maskG=1
       maskR=1
       outputIndex=0
      }
     }
    ]
    convertOutputs [{type=Vector3}]
   }
  }
  MaterialExpressionConvert_1 {
   class=MaterialExpressionConvert
   inputs [None {from=MaterialExpressionNamedRerouteUsage_173}]
   properties {
    nodeName=Break Float3
    convertInputs [{
      type=Vector3
      expressionInput {
       expression=&self:MaterialExpressionNamedRerouteUsage_173
       inputName=None
       mask=0
       maskA=0
       maskB=0
       maskG=0
       maskR=0
       outputIndex=0
      }
     }]
    convertOutputs [{type=Scalar}, {type=Scalar}, {type=Scalar}]
   }
  }
  MaterialExpressionConvert_2 {
   class=MaterialExpressionConvert
   inputs [
    None {from=""}
    None {from=""}
    None {from=MaterialExpressionScalarParameter_24}
   ]
   properties {
    nodeName=Make Vector3
    convertInputs [
     {type=Scalar, expressionInput {expression=null, inputName=None, mask=0, maskA=0, maskB=0, maskG=0, maskR=0, outputIndex=0}}
     {
      type=Scalar
      defaultValue=<rgba 7 0 0 1>
      expressionInput {expression=null, inputName=None, mask=0, maskA=0, maskB=0, maskG=0, maskR=0, outputIndex=0}
     }
     {
      type=Scalar
      defaultValue=<rgba 1 0 0 1>
      expressionInput {
       expression=&self:MaterialExpressionScalarParameter_24
       inputName=None
       mask=0
       maskA=0
       maskB=0
       maskG=0
       maskR=0
       outputIndex=0
      }
     }
    ]
    convertOutputs [{type=Vector3}]
   }
  }
  MaterialExpressionConvert_3 {
   class=MaterialExpressionConvert
   inputs [None {from=MaterialExpressionAbs_1}]
   properties {
    nodeName=Break Float3
    convertInputs [{
      type=Vector3
      expressionInput {
       expression=&self:MaterialExpressionAbs_1
       inputName=None
       mask=0
       maskA=0
       maskB=0
       maskG=0
       maskR=0
       outputIndex=0
      }
     }]
    convertOutputs [{type=Scalar}, {type=Scalar}, {type=Scalar}]
   }
  }
  MaterialExpressionConvert_5 {class=MaterialExpressionConvert, inputs [], properties {convertOutputs [{type=Vector3, defaultValue=<rgba 1 1 1 1>}]}}
  MaterialExpressionDDX_0 {class=MaterialExpressionDDX, inputs [Value {from=MaterialExpressionReroute_1, output=""}], properties {}}
  MaterialExpressionDDX_1 {class=MaterialExpressionDDX, inputs [Value {from=MaterialExpressionReroute_1, output=""}], properties {}}
  MaterialExpressionDDX_2 {class=MaterialExpressionDDX, inputs [Value {from=MaterialExpressionMultiply_163, output=""}], properties {}}
  MaterialExpressionDDY_0 {class=MaterialExpressionDDY, inputs [Value {from=MaterialExpressionMultiply_163, output=""}], properties {}}
  MaterialExpressionDesaturation_0 {
   class=MaterialExpressionDesaturation
   inputs [
    None {from=MaterialExpressionNamedRerouteUsage_8, output=""}
    Fraction {from=MaterialExpressionScalarParameter_134, output=""}
   ]
   properties {}
  }
  MaterialExpressionDistance_0 {
   class=MaterialExpressionDistance
   inputs [
    A {from=MaterialExpressionDivide_8, output=""}
    B {from=MaterialExpressionConstant2Vector_4, output=""}
   ]
   properties {}
  }
  MaterialExpressionDistance_1 {
   class=MaterialExpressionDistance
   inputs [
    A {from=MaterialExpressionSubtract_12, output=""}
    B {from=MaterialExpressionConstant_48, output=""}
   ]
   properties {}
  }
  MaterialExpressionDistance_2 {
   class=MaterialExpressionDistance
   inputs [
    A {from=MaterialExpressionSubtract_8, output=""}
    B {from=MaterialExpressionConstant_49, output=""}
   ]
   properties {}
  }
  MaterialExpressionDistance_4 {
   class=MaterialExpressionDistance
   inputs [
    A {from=MaterialExpressionSubtract_39, output=""}
    B {from=MaterialExpressionConstant_12, output=""}
   ]
   properties {}
  }
  MaterialExpressionDivide_0 {
   class=MaterialExpressionDivide
   inputs [
    A {from=MaterialExpressionAdd_10}
    B {from=""}
   ]
  }
  MaterialExpressionDivide_1 {
   class=MaterialExpressionDivide
   inputs [
    A {from=MaterialExpressionScalarParameter_2}
    B {from=MaterialExpressionComponentMask_0}
   ]
  }
  MaterialExpressionDivide_15 {
   class=MaterialExpressionDivide
   inputs [
    A {from=MaterialExpressionMultiply_69}
    B {from=MaterialExpressionMultiply_68}
   ]
  }
  MaterialExpressionDivide_16 {
   class=MaterialExpressionDivide
   inputs [
    A {from=MaterialExpressionComponentMask_9}
    B {from=MaterialExpressionScalarParameter_52}
   ]
  }
  MaterialExpressionDivide_18 {
   class=MaterialExpressionDivide
   inputs [
    A {from=MaterialExpressionAdd_61}
    B {from=MaterialExpressionVectorParameter_0, output=RGB}
   ]
  }
  MaterialExpressionDivide_2 {
   class=MaterialExpressionDivide
   inputs [
    A {from=MaterialExpressionReroute_4}
    B {from=MaterialExpressionScalarParameter_20}
   ]
  }
  MaterialExpressionDivide_28 {
   class=MaterialExpressionDivide
   inputs [
    A {from=MaterialExpressionSubtract_41}
    B {from=MaterialExpressionNamedRerouteUsage_148}
   ]
  }
  MaterialExpressionDivide_3 {
   class=MaterialExpressionDivide
   inputs [
    A {from=MaterialExpressionMaterialFunctionCall_56, output=UVs}
    B {from=""}
   ]
   properties {constB=16}
  }
  MaterialExpressionDivide_32 {
   class=MaterialExpressionDivide
   inputs [
    A {from=""}
    B {from=MaterialExpressionNamedRerouteUsage_183}
   ]
  }
  MaterialExpressionDivide_4 {
   class=MaterialExpressionDivide
   inputs [
    A {from=MaterialExpressionComponentMask_1}
    B {from=""}
   ]
   properties {constB=50}
  }
  MaterialExpressionDivide_5 {
   class=MaterialExpressionDivide
   inputs [
    A {from=MaterialExpressionMultiply_86}
    B {from=MaterialExpressionMultiply_72}
   ]
  }
  MaterialExpressionDivide_6 {
   class=MaterialExpressionDivide
   inputs [
    A {from=MaterialExpressionTime_2}
    B {from=""}
   ]
  }
  MaterialExpressionDivide_7 {
   class=MaterialExpressionDivide
   inputs [
    A {from=MaterialExpressionReroute_24}
    B {from=""}
   ]
   properties {constB=2.5}
  }
  MaterialExpressionDivide_8 {
   class=MaterialExpressionDivide
   inputs [
    A {from=MaterialExpressionSubtract_9}
    B {from=MaterialExpressionMultiply_16}
   ]
  }
  MaterialExpressionDivide_9 {
   class=MaterialExpressionDivide
   inputs [
    A {from=MaterialExpressionNamedRerouteUsage_39}
    B {from=""}
   ]
   properties {constB=1.25}
  }
  MaterialExpressionDotProduct_0 {
   class=MaterialExpressionDotProduct
   inputs [
    A {from=MaterialExpressionSubtract_12, output=""}
    B {from=MaterialExpressionCameraVectorWS_2, output=""}
   ]
   properties {}
  }
  MaterialExpressionDotProduct_1 {
   class=MaterialExpressionDotProduct
   inputs [
    A {from=MaterialExpressionMultiply_93, output=""}
    B {from=MaterialExpressionCameraVectorWS_3, output=""}
   ]
   properties {}
  }
  MaterialExpressionDotProduct_2 {
   class=MaterialExpressionDotProduct
   inputs [
    A {from=MaterialExpressionConvert_5, output=""}
    B {from=MaterialExpressionNamedRerouteDeclaration_31, output=""}
   ]
   properties {}
  }
  MaterialExpressionDotProduct_23 {
   class=MaterialExpressionDotProduct
   inputs [
    A {from=MaterialExpressionSubtract_39, output=""}
    B {from=MaterialExpressionCameraVectorWS_1, output=""}
   ]
   properties {}
  }
  MaterialExpressionDotProduct_3 {
   class=MaterialExpressionDotProduct
   inputs [
    A {from=MaterialExpressionTransform_2, output=""}
    B {from=MaterialExpressionConstant3Vector_5, output=""}
   ]
   properties {}
  }
  MaterialExpressionDotProduct_6 {
   class=MaterialExpressionDotProduct
   inputs [
    A {from=MaterialExpressionCameraVectorWS_5, output=""}
    B {from=MaterialExpressionVertexNormalWS_7, output=""}
   ]
   properties {}
  }
  MaterialExpressionFresnel_1 {
   class=MaterialExpressionFresnel
   inputs [
    ExponentIn {from="", output=""}
    BaseReflectFractionIn {from="", output=""}
    Normal {from="", output=""}
   ]
   properties {baseReflectFraction=0}
  }
  MaterialExpressionFunctionInput_0 {
   class=MaterialExpressionFunctionInput
   inputs [Preview {from=MaterialExpressionObjectPositionWS_4, output=""}]
   properties {bUsePreviewValueAsDefault=true, inputName=Object Position WS, sortPriority=140}
  }
  MaterialExpressionFunctionInput_1 {
   class=MaterialExpressionFunctionInput
   inputs [Preview {from="", output=""}]
   properties {bUsePreviewValueAsDefault=true, inputName=Pre-Skinned Scale, sortPriority=130, previewValue=<1 1 1 1>}
  }
  MaterialExpressionFunctionInput_11 {
   class=MaterialExpressionFunctionInput
   inputs [Preview {from="", output=""}]
   properties {inputName=Mask_Logo, inputType=FunctionInput_Scalar, sortPriority=1}
  }
  MaterialExpressionFunctionInput_12 {
   class=MaterialExpressionFunctionInput
   inputs [Preview {from="", output=""}]
   properties {inputName=Mask_SpriteFresnel, inputType=FunctionInput_Scalar, sortPriority=1}
  }
  MaterialExpressionFunctionInput_13 {
   class=MaterialExpressionFunctionInput
   inputs [Preview {from="", output=""}]
   properties {inputName=WPO, inputType=FunctionInput_Scalar, sortPriority=2}
  }
  MaterialExpressionFunctionInput_14 {
   class=MaterialExpressionFunctionInput
   inputs [Preview {from="", output=""}]
   properties {inputName=Mask_BlinkingLights, inputType=FunctionInput_Scalar, sortPriority=1}
  }
  MaterialExpressionFunctionInput_15 {
   class=MaterialExpressionFunctionInput
   inputs [Preview {from="", output=""}]
   properties {inputName=Mask_DiffuseChannelMask, inputType=FunctionInput_Scalar, sortPriority=1}
  }
  MaterialExpressionFunctionInput_3 {
   class=MaterialExpressionFunctionInput
   inputs [Preview {from="", output=""}]
   properties {inputName=MaterialAttributes, inputType=FunctionInput_MaterialAttributes, sortPriority=-1}
  }
  MaterialExpressionFunctionInput_4 {
   class=MaterialExpressionFunctionInput
   inputs [Preview {from="", output=""}]
   properties {inputName=Mask Variant, inputType=FunctionInput_Scalar, sortPriority=1}
  }
  MaterialExpressionFunctionInput_5 {
   class=MaterialExpressionFunctionInput
   inputs [Preview {from="", output=""}]
   properties {inputName=Mask_InnerRadialGlow, inputType=FunctionInput_Scalar, sortPriority=1}
  }
  MaterialExpressionFunctionInput_7 {
   class=MaterialExpressionFunctionInput
   inputs [Preview {from=MaterialExpressionShaderStageSwitch_4, output=""}]
   properties {bUsePreviewValueAsDefault=true, inputName=Pre-Skinned Local Pos, sortPriority=100}
  }
  MaterialExpressionFunctionInput_8 {
   class=MaterialExpressionFunctionInput
   inputs [Preview {from=MaterialExpressionBounds_0, output=Max}]
   properties {bUsePreviewValueAsDefault=true, inputName=Pre-Skinned Bounds Max, sortPriority=120}
  }
  MaterialExpressionFunctionInput_9 {
   class=MaterialExpressionFunctionInput
   inputs [Preview {from=MaterialExpressionBounds_0, output=Min}]
   properties {bUsePreviewValueAsDefault=true, inputName=Pre-Skinned Bounds Min, sortPriority=110}
  }
  MaterialExpressionFunctionOutput_1 {
   class=MaterialExpressionFunctionOutput
   inputs [None {from=MaterialExpressionStaticSwitchParameter_1, output=""}]
   properties {outputName=SetMaterialAttributes}
  }
  MaterialExpressionGetMaterialAttributes_1 {
   class=MaterialExpressionGetMaterialAttributes
   inputs [None {from=MaterialExpressionFunctionInput_3, output=""}]
   properties {attributeGetTypes [
     #1
     #2
     #3
     #4
     #5
     #6
    ]}
  }
  MaterialExpressionLength_0 {class=MaterialExpressionLength, inputs [None {from=MaterialExpressionAdd_55, output=""}], properties {}}
  MaterialExpressionLinearInterpolate_0 {
   class=MaterialExpressionLinearInterpolate
   inputs [
    A {from=""}
    B {from=""}
    Alpha {from=MaterialExpressionNamedRerouteUsage_52}
   ]
   properties {constA=0.05}
  }
  MaterialExpressionLinearInterpolate_1 {
   class=MaterialExpressionLinearInterpolate
   inputs [
    A {from=""}
    B {from=""}
    Alpha {from=MaterialExpressionConstantBiasScale_0}
   ]
   properties {constA=0.3, constB=1.1}
  }
  MaterialExpressionLinearInterpolate_10 {
   class=MaterialExpressionLinearInterpolate
   inputs [
    A {from=MaterialExpressionScalarParameter_50}
    B {from=MaterialExpressionScalarParameter_55}
    Alpha {from=MaterialExpressionTextureSampleParameter2D_17, output=B}
   ]
  }
  MaterialExpressionLinearInterpolate_11 {
   class=MaterialExpressionLinearInterpolate
   inputs [
    A {from=MaterialExpressionScalarParameter_58}
    B {from=MaterialExpressionScalarParameter_57}
    Alpha {from=MaterialExpressionTextureSampleParameter2D_16, output=R}
   ]
  }
  MaterialExpressionLinearInterpolate_12 {
   class=MaterialExpressionLinearInterpolate
   inputs [
    A {from=MaterialExpressionScalarParameter_60}
    B {from=MaterialExpressionScalarParameter_41}
    Alpha {from=MaterialExpressionNamedRerouteUsage_18}
   ]
  }
  MaterialExpressionLinearInterpolate_13 {
   class=MaterialExpressionLinearInterpolate
   inputs [
    A {from=MaterialExpressionConstant_46}
    B {from=MaterialExpressionConstant_47}
    Alpha {from=MaterialExpressionNamedRerouteUsage_31}
   ]
  }
  MaterialExpressionLinearInterpolate_14 {
   class=MaterialExpressionLinearInterpolate
   inputs [
    A {from=MaterialExpressionScalarParameter_17}
    B {from=MaterialExpressionScalarParameter_21}
    Alpha {from=MaterialExpressionReroute_15}
   ]
  }
  MaterialExpressionLinearInterpolate_15 {
   class=MaterialExpressionLinearInterpolate
   inputs [
    A {from=MaterialExpressionMultiply_59}
    B {from=MaterialExpressionQualitySwitch_0}
    Alpha {from=MaterialExpressionScalarParameter_80}
   ]
  }
  MaterialExpressionLinearInterpolate_17 {
   class=MaterialExpressionLinearInterpolate
   inputs [
    A {from=""}
    B {from=""}
    Alpha {from=MaterialExpressionNamedRerouteUsage_40}
   ]
   properties {constA=0.5}
  }
  MaterialExpressionLinearInterpolate_18 {
   class=MaterialExpressionLinearInterpolate
   inputs [
    A {from=MaterialExpressionAdd_51}
    B {from=MaterialExpressionReroute_77}
    Alpha {from=MaterialExpressionScalarParameter_133}
   ]
  }
  MaterialExpressionLinearInterpolate_19 {
   class=MaterialExpressionLinearInterpolate
   inputs [
    A {from=MaterialExpressionConstant_5}
    B {from=MaterialExpressionConstant_7}
    Alpha {from=MaterialExpressionNamedRerouteUsage_22}
   ]
  }
  MaterialExpressionLinearInterpolate_2 {
   class=MaterialExpressionLinearInterpolate
   inputs [
    A {from=""}
    B {from=""}
    Alpha {from=MaterialExpressionOneMinus_0}
   ]
   properties {constA=1, constB=-1}
  }
  MaterialExpressionLinearInterpolate_20 {
   class=MaterialExpressionLinearInterpolate
   inputs [
    A {from=MaterialExpressionMultiply_154}
    B {from=MaterialExpressionMultiply_50}
    Alpha {from=MaterialExpressionScalarParameter_8}
   ]
   properties {constAlpha=1}
  }
  MaterialExpressionLinearInterpolate_21 {
   class=MaterialExpressionLinearInterpolate
   inputs [
    A {from=MaterialExpressionConstant_32}
    B {from=MaterialExpressionConstant_33}
    Alpha {from=MaterialExpressionNamedRerouteUsage_99}
   ]
  }
  MaterialExpressionLinearInterpolate_23 {
   class=MaterialExpressionLinearInterpolate
   inputs [
    A {from=""}
    B {from=""}
    Alpha {from=MaterialExpressionNamedRerouteUsage_52}
   ]
   properties {constA=1, constB=0.03}
  }
  MaterialExpressionLinearInterpolate_24 {
   class=MaterialExpressionLinearInterpolate
   inputs [
    A {from=""}
    B {from=""}
    Alpha {from=MaterialExpressionNamedRerouteUsage_54}
   ]
   properties {constA=0.5}
  }
  MaterialExpressionLinearInterpolate_25 {
   class=MaterialExpressionLinearInterpolate
   inputs [
    A {from=MaterialExpressionConstant2Vector_1}
    B {from=MaterialExpressionConstant2Vector_3}
    Alpha {from=MaterialExpressionSaturate_16}
   ]
  }
  MaterialExpressionLinearInterpolate_26 {
   class=MaterialExpressionLinearInterpolate
   inputs [
    A {from=MaterialExpressionReroute_26}
    B {from=MaterialExpressionMultiply_107}
    Alpha {from=MaterialExpressionConstant_9}
   ]
  }
  MaterialExpressionLinearInterpolate_3 {
   class=MaterialExpressionLinearInterpolate
   inputs [
    A {from=""}
    B {from=MaterialExpressionTextureSample_73, output=R}
    Alpha {from=MaterialExpressionNamedRerouteUsage_49}
   ]
   properties {constA=0.5}
  }
  MaterialExpressionLinearInterpolate_31 {
   class=MaterialExpressionLinearInterpolate
   inputs [
    A {from=MaterialExpressionConstant_54}
    B {from=MaterialExpressionConstant_55}
    Alpha {from=MaterialExpressionNamedRerouteUsage_30}
   ]
  }
  MaterialExpressionLinearInterpolate_32 {
   class=MaterialExpressionLinearInterpolate
   inputs [
    A {from=MaterialExpressionScalarParameter_12}
    B {from=MaterialExpressionScalarParameter_13}
    Alpha {from=MaterialExpressionReroute_29}
   ]
  }
  MaterialExpressionLinearInterpolate_4 {
   class=MaterialExpressionLinearInterpolate
   inputs [
    A {from=MaterialExpressionPower_17}
    B {from=MaterialExpressionNamedRerouteUsage_24}
    Alpha {from=""}
   ]
   properties {constAlpha=0.3}
  }
  MaterialExpressionLinearInterpolate_44 {
   class=MaterialExpressionLinearInterpolate
   inputs [
    A {from=MaterialExpressionMultiply_19}
    B {from=MaterialExpressionMultiply_48}
    Alpha {from=MaterialExpressionSaturate_2}
   ]
  }
  MaterialExpressionLinearInterpolate_47 {
   class=MaterialExpressionLinearInterpolate
   inputs [
    A {from=MaterialExpressionPower_16}
    B {from=MaterialExpressionPower_6}
    Alpha {from=MaterialExpressionSaturate_73}
   ]
  }
  MaterialExpressionLinearInterpolate_48 {
   class=MaterialExpressionLinearInterpolate
   inputs [
    A {from=MaterialExpressionSaturate_8}
    B {from=MaterialExpressionVectorParameter_36, output=RGB}
    Alpha {from=MaterialExpressionNamedRerouteUsage_125}
   ]
  }
  MaterialExpressionLinearInterpolate_5 {
   class=MaterialExpressionLinearInterpolate
   inputs [
    A {from=MaterialExpressionConstant_27}
    B {from=MaterialExpressionConstant_28}
    Alpha {from=MaterialExpressionTextureSampleParameter2D_8, output=G}
   ]
   properties {constA=-1}
  }
  MaterialExpressionLinearInterpolate_51 {
   class=MaterialExpressionLinearInterpolate
   inputs [
    A {from=MaterialExpressionMultiply_154}
    B {from=MaterialExpressionQualitySwitch_4}
    Alpha {from=MaterialExpressionScalarParameter_29}
   ]
  }
  MaterialExpressionLinearInterpolate_54 {
   class=MaterialExpressionLinearInterpolate
   inputs [
    A {from=""}
    B {from=MaterialExpressionScalarParameter_48}
    Alpha {from=MaterialExpressionPower_4}
   ]
   properties {constA=1}
  }
  MaterialExpressionLinearInterpolate_55 {
   class=MaterialExpressionLinearInterpolate
   inputs [
    A {from=MaterialExpressionConstant_18}
    B {from=MaterialExpressionConstant_42}
    Alpha {from=MaterialExpressionNamedRerouteUsage_93}
   ]
  }
  MaterialExpressionLinearInterpolate_57 {
   class=MaterialExpressionLinearInterpolate
   inputs [
    A {from=MaterialExpressionScalarParameter_61}
    B {from=""}
    Alpha {from=MaterialExpressionOneMinus_6}
   ]
   properties {constA=0.2}
  }
  MaterialExpressionLinearInterpolate_59 {
   class=MaterialExpressionLinearInterpolate
   inputs [
    A {from=MaterialExpressionScalarParameter_207}
    B {from=MaterialExpressionScalarParameter_206}
    Alpha {from=MaterialExpressionNamedRerouteUsage_15}
   ]
  }
  MaterialExpressionLinearInterpolate_6 {
   class=MaterialExpressionLinearInterpolate
   inputs [
    A {from=MaterialExpressionConstant_30}
    B {from=MaterialExpressionConstant_35}
    Alpha {from=MaterialExpressionTextureSampleParameter2D_10, output=G}
   ]
   properties {constA=1, constB=0}
  }
  MaterialExpressionLinearInterpolate_60 {
   class=MaterialExpressionLinearInterpolate
   inputs [
    A {from=MaterialExpressionScalarParameter_63}
    B {from=""}
    Alpha {from=MaterialExpressionPower_8}
   ]
  }
  MaterialExpressionLinearInterpolate_7 {
   class=MaterialExpressionLinearInterpolate
   inputs [
    A {from=MaterialExpressionNamedRerouteUsage_56}
    B {from=""}
    Alpha {from=MaterialExpressionNamedRerouteUsage_55}
   ]
   properties {constB=0}
  }
  MaterialExpressionLinearInterpolate_71 {
   class=MaterialExpressionLinearInterpolate
   inputs [
    A {from=""}
    B {from=MaterialExpressionDivide_32}
    Alpha {from=MaterialExpressionScalarParameter_142}
   ]
   properties {constA=1}
  }
  MaterialExpressionLinearInterpolate_72 {
   class=MaterialExpressionLinearInterpolate
   inputs [
    A {from=""}
    B {from=MaterialExpressionNamedRerouteUsage_210}
    Alpha {from=MaterialExpressionScalarParameter_143}
   ]
   properties {constA=1}
  }
  MaterialExpressionLinearInterpolate_9 {
   class=MaterialExpressionLinearInterpolate
   inputs [
    A {from=""}
    B {from=""}
    Alpha {from=MaterialExpressionReroute_9}
   ]
   properties {constA=0.5}
  }
  MaterialExpressionLocalPosition_2 {class=MaterialExpressionLocalPosition, inputs [], properties {}}
  MaterialExpressionLocalPosition_5 {class=MaterialExpressionLocalPosition, inputs [], properties {includedOffsets=ExcludeOffsets, localOrigin=InstancePreSkinning}}
  MaterialExpressionMaterialFunctionCall_1 {
   class=MaterialExpressionMaterialFunctionCall
   inputs [
    In {from=MaterialExpressionLength_0}
    Contrast {from=MaterialExpressionConstant_2}
   ]
   properties {materialFunction=&CheapContrast}
  }
  MaterialExpressionMaterialFunctionCall_12 {
   class=MaterialExpressionMaterialFunctionCall
   inputs [Color {from=MaterialExpressionNamedRerouteUsage_2}]
   properties {materialFunction=&SplitComponents}
  }
  MaterialExpressionMaterialFunctionCall_13 {
   class=MaterialExpressionMaterialFunctionCall
   inputs [
    Color_Morning {from=MaterialExpressionVectorParameter_31, output=RGB}
    Color_Day {from=MaterialExpressionVectorParameter_32, output=RGB}
    Color_Evening {from=MaterialExpressionVectorParameter_33, output=RGB}
    Color_Night {from=MaterialExpressionVectorParameter_34, output=RGB}
   ]
   properties {materialFunction=&Color_Time_of_Day}
  }
  MaterialExpressionMaterialFunctionCall_15 {
   class=MaterialExpressionMaterialFunctionCall
   inputs [
    Blend Range {from=MaterialExpressionScalarParameter_46}
    Start Offset {from=MaterialExpressionConstant_34}
   ]
   properties {materialFunction=/Engine/Functions/Engine_MaterialFunctions03/Particles/Distance_Blend}
  }
  MaterialExpressionMaterialFunctionCall_16 {
   class=MaterialExpressionMaterialFunctionCall
   inputs [
    Blend Range {from=MaterialExpressionScalarParameter_49}
    Start Offset {from=MaterialExpressionConstant_45}
   ]
   properties {materialFunction=&Distance_Blend}
  }
  MaterialExpressionMaterialFunctionCall_17 {
   class=MaterialExpressionMaterialFunctionCall
   inputs [
    UVs {from=MaterialExpressionAdd_34}
    View {from=""}
    RoomZscale {from=MaterialExpressionScalarParameter_53}
    Tiling {from=MaterialExpressionComponentMask_10}
    Randomize Rotation {from=""}
   ]
   properties {materialFunction=/Game/Global/MaterialFunctions/MF_InteriorCubemapUV}
  }
  MaterialExpressionMaterialFunctionCall_21 {
   class=MaterialExpressionMaterialFunctionCall
   inputs [
    Texture Scale {from=MaterialExpressionScalarParameter_77}
    Normal {from=""}
    InPosition {from=MaterialExpressionNamedRerouteUsage_5}
    Triplanar Fade {from=MaterialExpressionConstant_21}
    Derivatives Scale {from=""}
   ]
   properties {materialFunction=&/MF_TriplanarDithered_PreSkinned}
  }
  MaterialExpressionMaterialFunctionCall_24 {
   class=MaterialExpressionMaterialFunctionCall
   inputs [
    Value {from=MaterialExpressionMultiply_110}
    Period {from=""}
    -1 to 1 {from=""}
    Sine Phase {from=""}
   ]
   properties {materialFunction=&LinearSine}
  }
  MaterialExpressionMaterialFunctionCall_28 {
   class=MaterialExpressionMaterialFunctionCall
   inputs [
    A {from=""}
    B {from=""}
    Transition Phase {from=MaterialExpressionReroute_18}
    Height Texture {from=MaterialExpressionMultiply_156}
    Contrast {from=""}
   ]
   properties {materialFunction=&HeightLerp}
  }
  MaterialExpressionMaterialFunctionCall_30 {
   class=MaterialExpressionMaterialFunctionCall
   inputs [
    1 {from=MaterialExpressionVectorParameter_29, output=RGB}
    2 {from=MaterialExpressionVectorParameter_11, output=RGB}
    3 {from=MaterialExpressionVectorParameter_25, output=RGB}
    4 {from=MaterialExpressionMultiply_90}
    A {from=MaterialExpressionNamedRerouteUsage_34}
   ]
   properties {materialFunction=&Lerp_Multiple_Float3}
  }
  MaterialExpressionMaterialFunctionCall_31 {
   class=MaterialExpressionMaterialFunctionCall
   inputs [
    UVs {from=MaterialExpressionMaterialFunctionCall_21, output=UVs}
    TriplanarMask {from=MaterialExpressionMaterialFunctionCall_21, output=UVs}
    CameraVector {from=""}
    Bump Offset {from=MaterialExpressionMultiply_157}
   ]
   properties {materialFunction=&/MF_TriplanarBumpOffset}
  }
  MaterialExpressionMaterialFunctionCall_32 {
   class=MaterialExpressionMaterialFunctionCall
   inputs [Float2 {from=MaterialExpressionAbs_3}]
   properties {materialFunction=&BreakOutFloat2Components}
  }
  MaterialExpressionMaterialFunctionCall_34 {
   class=MaterialExpressionMaterialFunctionCall
   inputs [
    Texture Scale {from=MaterialExpressionScalarParameter_6}
    Normal {from=""}
    InPosition {from=MaterialExpressionNamedRerouteUsage_1}
    Triplanar Fade {from=MaterialExpressionConstant_6}
    Derivatives Scale {from=""}
   ]
   properties {materialFunction=&MF_TriplanarDithered_PreSkinned}
  }
  MaterialExpressionMaterialFunctionCall_35 {
   class=MaterialExpressionMaterialFunctionCall
   inputs [
    UVs {from=MaterialExpressionMaterialFunctionCall_34, output=UVs}
    TriplanarMask {from=MaterialExpressionMaterialFunctionCall_34, output=UVs}
    CameraVector {from=""}
    Bump Offset {from=MaterialExpressionMultiply_148}
   ]
   properties {materialFunction=&MF_TriplanarBumpOffset}
  }
  MaterialExpressionMaterialFunctionCall_37 {
   class=MaterialExpressionMaterialFunctionCall
   inputs [
    UVs {from=MaterialExpressionMaterialFunctionCall_34, output=UVs}
    TriplanarMask {from=MaterialExpressionMaterialFunctionCall_34, output=UVs}
    CameraVector {from=""}
    Bump Offset {from=MaterialExpressionMultiply_150}
   ]
   properties {materialFunction=&MF_TriplanarBumpOffset}
  }
  MaterialExpressionMaterialFunctionCall_40 {
   class=MaterialExpressionMaterialFunctionCall
   inputs [
    Texture Scale {from=MaterialExpressionReroute_4}
    Normal {from=""}
    InPosition {from=MaterialExpressionNamedRerouteUsage_6}
    Triplanar Fade {from=MaterialExpressionConstant_29}
    Derivatives Scale {from=""}
   ]
   properties {materialFunction=&MF_TriplanarDithered_PreSkinned}
  }
  MaterialExpressionMaterialFunctionCall_41 {
   class=MaterialExpressionMaterialFunctionCall
   inputs [
    UVs {from=MaterialExpressionReroute_14}
    TriplanarMask {from=MaterialExpressionMaterialFunctionCall_40, output=Mask}
    CameraVector {from=""}
    Bump Offset {from=MaterialExpressionMultiply_158}
   ]
   properties {materialFunction=&MF_TriplanarBumpOffset}
  }
  MaterialExpressionMaterialFunctionCall_42 {
   class=MaterialExpressionMaterialFunctionCall
   inputs [Float2 {from=MaterialExpressionMaterialFunctionCall_40, output=UVs}]
   properties {materialFunction=&BreakOutFloat2Components}
  }
  MaterialExpressionMaterialFunctionCall_43 {
   class=MaterialExpressionMaterialFunctionCall
   inputs [
    X {from=MaterialExpressionAdd_22}
    Y {from=MaterialExpressionAdd_23}
   ]
   properties {materialFunction=&MakeFloat2}
  }
  MaterialExpressionMaterialFunctionCall_45 {
   class=MaterialExpressionMaterialFunctionCall
   inputs [
    Texture Scale {from=MaterialExpressionDivide_2}
    Normal {from=""}
    InPosition {from=MaterialExpressionNamedRerouteUsage_11}
    Triplanar Fade {from=MaterialExpressionConstant_29}
    Derivatives Scale {from=""}
   ]
   properties {materialFunction=&MF_TriplanarDithered_PreSkinned}
  }
  MaterialExpressionMaterialFunctionCall_47 {
   class=MaterialExpressionMaterialFunctionCall
   inputs [
    UVs {from=MaterialExpressionMaterialFunctionCall_45, output=UVs}
    TriplanarMask {from=MaterialExpressionMaterialFunctionCall_45, output=UVs}
    CameraVector {from=""}
    Bump Offset {from=MaterialExpressionMultiply_159}
   ]
   properties {materialFunction=&MF_TriplanarBumpOffset}
  }
  MaterialExpressionMaterialFunctionCall_56 {
   class=MaterialExpressionMaterialFunctionCall
   inputs [
    Texture Scale {from=MaterialExpressionDivide_7}
    Normal {from=""}
    InPosition {from=MaterialExpressionNamedRerouteUsage_17}
    Triplanar Fade {from=MaterialExpressionConstant_44}
    Derivatives Scale {from=""}
   ]
   properties {materialFunction=&MF_TriplanarDithered_PreSkinned}
  }
  MaterialExpressionMaterialFunctionCall_57 {
   class=MaterialExpressionMaterialFunctionCall
   inputs [
    UVs {from=MaterialExpressionMaterialFunctionCall_56, output=UVs}
    TriplanarMask {from=MaterialExpressionMaterialFunctionCall_56, output=UVs}
    CameraVector {from=""}
    Bump Offset {from=MaterialExpressionMultiply_160}
   ]
   properties {materialFunction=&MF_TriplanarBumpOffset}
  }
  MaterialExpressionMaterialFunctionCall_59 {
   class=MaterialExpressionMaterialFunctionCall
   inputs [
    Position {from=MaterialExpressionNamedRerouteUsage_171}
    UVRadialScale {from=MaterialExpressionNamedRerouteUsage_176}
    ObjectRadius {from=MaterialExpressionNamedRerouteUsage_175}
   ]
   properties {materialFunction=&/MF_ObjectStabilizedScreenSpaceUVs}
  }
  MaterialExpressionMaterialFunctionCall_6 {
   class=MaterialExpressionMaterialFunctionCall
   inputs []
   properties {materialFunction=&CameraDirectionVector}
  }
  MaterialExpressionMaterialFunctionCall_61 {
   class=MaterialExpressionMaterialFunctionCall
   inputs [
    Texture Scale {from=MaterialExpressionNamedRerouteUsage_156}
    Normal {from=MaterialExpressionNamedRerouteUsage_41}
    InPosition {from=MaterialExpressionNamedRerouteUsage_21}
    Triplanar Fade {from=MaterialExpressionConstant_57}
    Derivatives Scale {from=""}
   ]
   properties {materialFunction=&MF_TriplanarDithered_PreSkinned}
  }
  MaterialExpressionMax_1 {
   class=MaterialExpressionMax
   inputs [
    A {from=MaterialExpressionMaterialFunctionCall_32, output=R}
    B {from=MaterialExpressionMaterialFunctionCall_32, output=R}
   ]
  }
  MaterialExpressionMax_10 {
   class=MaterialExpressionMax
   inputs [
    A {from=MaterialExpressionConvert_1}
    B {from=MaterialExpressionConvert_1}
   ]
  }
  MaterialExpressionMax_11 {
   class=MaterialExpressionMax
   inputs [
    A {from=MaterialExpressionMax_10}
    B {from=MaterialExpressionConvert_1}
   ]
  }
  MaterialExpressionMax_2 {
   class=MaterialExpressionMax
   inputs [
    A {from=MaterialExpressionSubtract_37}
    B {from=""}
   ]
   properties {constB=0}
  }
  MaterialExpressionMax_3 {
   class=MaterialExpressionMax
   inputs [
    A {from=MaterialExpressionSubtract_3}
    B {from=MaterialExpressionSubtract_2}
   ]
  }
  MaterialExpressionMax_4 {
   class=MaterialExpressionMax
   inputs [
    A {from=MaterialExpressionNamedRerouteUsage_50}
    B {from=MaterialExpressionNamedRerouteUsage_53}
   ]
  }
  MaterialExpressionMax_5 {
   class=MaterialExpressionMax
   inputs [
    A {from=MaterialExpressionMax_3}
    B {from=MaterialExpressionSubtract_5}
   ]
  }
  MaterialExpressionMax_6 {
   class=MaterialExpressionMax
   inputs [
    A {from=MaterialExpressionSubtract_10}
    B {from=""}
   ]
   properties {constB=0}
  }
  MaterialExpressionMax_7 {
   class=MaterialExpressionMax
   inputs [
    A {from=MaterialExpressionConvert_3}
    B {from=MaterialExpressionConvert_3}
   ]
  }
  MaterialExpressionMax_8 {
   class=MaterialExpressionMax
   inputs [
    A {from=MaterialExpressionConvert_3}
    B {from=MaterialExpressionConvert_3}
   ]
  }
  MaterialExpressionMax_9 {
   class=MaterialExpressionMax
   inputs [
    A {from=MaterialExpressionConvert_3}
    B {from=MaterialExpressionConvert_3}
   ]
  }
  MaterialExpressionMultiply_0 {
   class=MaterialExpressionMultiply
   inputs [
    A {from=MaterialExpressionNamedRerouteUsage_0}
    B {from=MaterialExpressionLinearInterpolate_1}
   ]
   properties {constB=2}
  }
  MaterialExpressionMultiply_1 {
   class=MaterialExpressionMultiply
   inputs [
    A {from=MaterialExpressionReroute_6}
    B {from=MaterialExpressionMultiply_25}
   ]
  }
  MaterialExpressionMultiply_10 {
   class=MaterialExpressionMultiply
   inputs [
    A {from=MaterialExpressionNamedRerouteUsage_16}
    B {from=MaterialExpressionNamedRerouteUsage_29}
   ]
  }
  MaterialExpressionMultiply_100 {
   class=MaterialExpressionMultiply
   inputs [
    A {from=MaterialExpressionReroute_3}
    B {from=MaterialExpressionOneMinus_10}
   ]
  }
  MaterialExpressionMultiply_107 {
   class=MaterialExpressionMultiply
   inputs [
    A {from=MaterialExpressionReroute_26}
    B {from=MaterialExpressionReroute_8}
   ]
  }
  MaterialExpressionMultiply_108 {
   class=MaterialExpressionMultiply
   inputs [
    A {from=MaterialExpressionReroute_21}
    B {from=MaterialExpressionSaturate_79}
   ]
  }
  MaterialExpressionMultiply_11 {
   class=MaterialExpressionMultiply
   inputs [
    A {from=MaterialExpressionAdd_75}
    B {from=MaterialExpressionScalarParameter_0}
   ]
  }
  MaterialExpressionMultiply_110 {
   class=MaterialExpressionMultiply
   inputs [
    A {from=MaterialExpressionTime_9}
    B {from=MaterialExpressionScalarParameter_205}
   ]
  }
  MaterialExpressionMultiply_117 {
   class=MaterialExpressionMultiply
   inputs [
    A {from=MaterialExpressionSaturate_79}
    B {from=MaterialExpressionNamedRerouteUsage_23}
   ]
  }
  MaterialExpressionMultiply_118 {
   class=MaterialExpressionMultiply
   inputs [
    A {from=MaterialExpressionMultiply_117}
    B {from=MaterialExpressionPower_22}
   ]
  }
  MaterialExpressionMultiply_12 {
   class=MaterialExpressionMultiply
   inputs [
    A {from=MaterialExpressionMax_11}
    B {from=""}
   ]
   properties {constB=1.4115, desc=a bit arbitrary...}
  }
  MaterialExpressionMultiply_13 {
   class=MaterialExpressionMultiply
   inputs [
    A {from=MaterialExpressionMultiply_10}
    B {from=MaterialExpressionLinearInterpolate_17}
   ]
   properties {constB=0.5}
  }
  MaterialExpressionMultiply_139 {
   class=MaterialExpressionMultiply
   inputs [
    A {from=MaterialExpressionNormalize_1}
    B {from=MaterialExpressionDesaturation_0}
   ]
  }
  MaterialExpressionMultiply_14 {
   class=MaterialExpressionMultiply
   inputs [
    A {from=MaterialExpressionNamedRerouteUsage_48}
    B {from=MaterialExpressionVectorParameter_2, output=RGB}
   ]
  }
  MaterialExpressionMultiply_140 {
   class=MaterialExpressionMultiply
   inputs [
    A {from=MaterialExpressionNamedRerouteUsage_9}
    B {from=MaterialExpressionScalarParameter_135}
   ]
  }
  MaterialExpressionMultiply_141 {
   class=MaterialExpressionMultiply
   inputs [
    A {from=MaterialExpressionNamedRerouteUsage_165}
    B {from=""}
   ]
   properties {constB=0.5}
  }
  MaterialExpressionMultiply_147 {
   class=MaterialExpressionMultiply
   inputs [
    A {from=MaterialExpressionDotProduct_2}
    B {from=""}
   ]
   properties {constB=0.3333}
  }
  MaterialExpressionMultiply_148 {
   class=MaterialExpressionMultiply
   inputs [
    A {from=MaterialExpressionScalarParameter_7}
    B {from=MaterialExpressionNamedRerouteUsage_212}
   ]
  }
  MaterialExpressionMultiply_149 {
   class=MaterialExpressionMultiply
   inputs [
    A {from=MaterialExpressionDivide_15}
    B {from=MaterialExpressionScalarParameter_192}
   ]
  }
  MaterialExpressionMultiply_15 {
   class=MaterialExpressionMultiply
   inputs [
    A {from=MaterialExpressionDivide_4}
    B {from=MaterialExpressionScalarParameter_5}
   ]
  }
  MaterialExpressionMultiply_150 {
   class=MaterialExpressionMultiply
   inputs [
    A {from=MaterialExpressionNamedRerouteUsage_212}
    B {from=MaterialExpressionScalarParameter_9}
   ]
  }
  MaterialExpressionMultiply_152 {
   class=MaterialExpressionMultiply
   inputs [
    A {from=MaterialExpressionNamedRerouteUsage_188}
    B {from=MaterialExpressionScalarParameter_43}
   ]
  }
  MaterialExpressionMultiply_153 {
   class=MaterialExpressionMultiply
   inputs [
    A {from=MaterialExpressionMultiply_90}
    B {from=MaterialExpressionNamedRerouteUsage_33}
   ]
   properties {desc=Core Glow}
  }
  MaterialExpressionMultiply_154 {
   class=MaterialExpressionMultiply
   inputs [
    A {from=MaterialExpressionPower_35}
    B {from=MaterialExpressionScalarParameter_101}
   ]
  }
  MaterialExpressionMultiply_155 {
   class=MaterialExpressionMultiply
   inputs [
    A {from=MaterialExpressionNamedRerouteUsage_189}
    B {from=MaterialExpressionScalarParameter_43}
   ]
  }
  MaterialExpressionMultiply_156 {
   class=MaterialExpressionMultiply
   inputs [
    A {from=MaterialExpressionQualitySwitch_5}
    B {from=MaterialExpressionNamedRerouteUsage_26}
   ]
  }
  MaterialExpressionMultiply_157 {
   class=MaterialExpressionMultiply
   inputs [
    A {from=MaterialExpressionScalarParameter_3}
    B {from=MaterialExpressionNamedRerouteUsage_213}
   ]
  }
  MaterialExpressionMultiply_158 {
   class=MaterialExpressionMultiply
   inputs [
    A {from=MaterialExpressionMultiply_36}
    B {from=MaterialExpressionNamedRerouteUsage_214}
   ]
  }
  MaterialExpressionMultiply_159 {
   class=MaterialExpressionMultiply
   inputs [
    A {from=MaterialExpressionMultiply_41}
    B {from=MaterialExpressionNamedRerouteUsage_215}
   ]
  }
  MaterialExpressionMultiply_16 {
   class=MaterialExpressionMultiply
   inputs [
    A {from=MaterialExpressionMultiply_15}
    B {from=MaterialExpressionLinearInterpolate_19}
   ]
  }
  MaterialExpressionMultiply_160 {
   class=MaterialExpressionMultiply
   inputs [
    A {from=MaterialExpressionLinearInterpolate_11}
    B {from=MaterialExpressionNamedRerouteUsage_216}
   ]
  }
  MaterialExpressionMultiply_161 {
   class=MaterialExpressionMultiply
   inputs [
    A {from=MaterialExpressionLinearInterpolate_72}
    B {from=MaterialExpressionScalarParameter_144}
   ]
  }
  MaterialExpressionMultiply_162 {
   class=MaterialExpressionMultiply
   inputs [
    A {from=MaterialExpressionLinearInterpolate_71}
    B {from=MaterialExpressionScalarParameter_145}
   ]
  }
  MaterialExpressionMultiply_163 {
   class=MaterialExpressionMultiply
   inputs [
    A {from=MaterialExpressionMultiply_162}
    B {from=MaterialExpressionTextureCoordinate_0}
   ]
  }
  MaterialExpressionMultiply_17 {
   class=MaterialExpressionMultiply
   inputs [
    A {from=MaterialExpressionNamedRerouteUsage_42}
    B {from=""}
   ]
  }
  MaterialExpressionMultiply_18 {
   class=MaterialExpressionMultiply
   inputs [
    A {from=MaterialExpressionMultiply_118}
    B {from=MaterialExpressionLinearInterpolate_23}
   ]
  }
  MaterialExpressionMultiply_19 {
   class=MaterialExpressionMultiply
   inputs [
    A {from=MaterialExpressionLinearInterpolate_47}
    B {from=MaterialExpressionScalarParameter_1}
   ]
  }
  MaterialExpressionMultiply_194 {
   class=MaterialExpressionMultiply
   inputs [
    A {from=MaterialExpressionPower_34}
    B {from=MaterialExpressionSaturate_0}
   ]
  }
  MaterialExpressionMultiply_21 {
   class=MaterialExpressionMultiply
   inputs [
    A {from=MaterialExpressionPower_11}
    B {from=MaterialExpressionLinearInterpolate_24}
   ]
  }
  MaterialExpressionMultiply_219 {
   class=MaterialExpressionMultiply
   inputs [
    A {from=MaterialExpressionNamedRerouteUsage_124}
    B {from=MaterialExpressionNamedRerouteUsage_88}
   ]
  }
  MaterialExpressionMultiply_22 {
   class=MaterialExpressionMultiply
   inputs [
    A {from=MaterialExpressionAbs_0}
    B {from=""}
   ]
   properties {constB=2}
  }
  MaterialExpressionMultiply_222 {
   class=MaterialExpressionMultiply
   inputs [
    A {from=MaterialExpressionNamedRerouteUsage_126}
    B {from=MaterialExpressionVectorParameter_35, output=RGB}
   ]
  }
  MaterialExpressionMultiply_24 {
   class=MaterialExpressionMultiply
   inputs [
    A {from=MaterialExpressionMultiply_94}
    B {from=MaterialExpressionMaterialFunctionCall_13, output=Result}
   ]
  }
  MaterialExpressionMultiply_25 {
   class=MaterialExpressionMultiply
   inputs [
    A {from=MaterialExpressionSubtract_19}
    B {from=MaterialExpressionScalarParameter_4}
   ]
  }
  MaterialExpressionMultiply_3 {
   class=MaterialExpressionMultiply
   inputs [
    A {from=MaterialExpressionMultiply_13}
    B {from=MaterialExpressionVectorParameter_1, output=RGB}
   ]
  }
  MaterialExpressionMultiply_34 {
   class=MaterialExpressionMultiply
   inputs [
    A {from=MaterialExpressionSubtract_18}
    B {from=MaterialExpressionConstant4Vector_6, output=RGB}
   ]
  }
  MaterialExpressionMultiply_36 {
   class=MaterialExpressionMultiply
   inputs [
    A {from=MaterialExpressionLinearInterpolate_5}
    B {from=MaterialExpressionScalarParameter_16}
   ]
  }
  MaterialExpressionMultiply_4 {
   class=MaterialExpressionMultiply
   inputs [
    A {from=MaterialExpressionLinearInterpolate_21}
    B {from=MaterialExpressionScalarParameter_2}
   ]
   properties {constB=2}
  }
  MaterialExpressionMultiply_41 {
   class=MaterialExpressionMultiply
   inputs [
    A {from=MaterialExpressionLinearInterpolate_6}
    B {from=MaterialExpressionScalarParameter_22}
   ]
  }
  MaterialExpressionMultiply_44 {
   class=MaterialExpressionMultiply
   inputs [
    A {from=MaterialExpressionMultiply_47}
    B {from=MaterialExpressionScalarParameter_47}
   ]
  }
  MaterialExpressionMultiply_45 {
   class=MaterialExpressionMultiply
   inputs [
    A {from=MaterialExpressionScalarParameter_56}
    B {from=MaterialExpressionTime_0}
   ]
  }
  MaterialExpressionMultiply_47 {
   class=MaterialExpressionMultiply
   inputs [
    A {from=MaterialExpressionReroute_5}
    B {from=MaterialExpressionReroute_16}
   ]
  }
  MaterialExpressionMultiply_48 {
   class=MaterialExpressionMultiply
   inputs [
    A {from=MaterialExpressionPower_9}
    B {from=MaterialExpressionScalarParameter_62}
   ]
  }
  MaterialExpressionMultiply_49 {
   class=MaterialExpressionMultiply
   inputs [
    A {from=MaterialExpressionMultiply_48}
    B {from=MaterialExpressionSaturate_2}
   ]
  }
  MaterialExpressionMultiply_5 {
   class=MaterialExpressionMultiply
   inputs [
    A {from=MaterialExpressionPower_14}
    B {from=MaterialExpressionNamedRerouteUsage_49}
   ]
  }
  MaterialExpressionMultiply_50 {
   class=MaterialExpressionMultiply
   inputs [
    A {from=MaterialExpressionMultiply_154}
    B {from=MaterialExpressionNamedRerouteUsage_80}
   ]
  }
  MaterialExpressionMultiply_52 {
   class=MaterialExpressionMultiply
   inputs [
    A {from=MaterialExpressionPower_37}
    B {from=MaterialExpressionSaturate_6}
   ]
  }
  MaterialExpressionMultiply_56 {
   class=MaterialExpressionMultiply
   inputs [
    A {from=MaterialExpressionMultiply_18}
    B {from=MaterialExpressionVectorParameter_39, output=RGB}
   ]
  }
  MaterialExpressionMultiply_58 {
   class=MaterialExpressionMultiply
   inputs [
    A {from=MaterialExpressionDivide_5}
    B {from=MaterialExpressionScalarParameter_78}
   ]
  }
  MaterialExpressionMultiply_59 {
   class=MaterialExpressionMultiply
   inputs [
    A {from=MaterialExpressionPower_12}
    B {from=MaterialExpressionScalarParameter_71}
   ]
  }
  MaterialExpressionMultiply_6 {
   class=MaterialExpressionMultiply
   inputs [
    A {from=MaterialExpressionMaterialFunctionCall_12, output=R}
    B {from=""}
   ]
   properties {constB=-1}
  }
  MaterialExpressionMultiply_60 {
   class=MaterialExpressionMultiply
   inputs [
    A {from=MaterialExpressionMultiply_59}
    B {from=""}
   ]
  }
  MaterialExpressionMultiply_61 {
   class=MaterialExpressionMultiply
   inputs [
    A {from=MaterialExpressionLinearInterpolate_13}
    B {from=MaterialExpressionScalarParameter_70}
   ]
  }
  MaterialExpressionMultiply_62 {
   class=MaterialExpressionMultiply
   inputs [
    A {from=MaterialExpressionReroute_63}
    B {from=MaterialExpressionScalarParameter_42}
   ]
  }
  MaterialExpressionMultiply_63 {
   class=MaterialExpressionMultiply
   inputs [
    A {from=MaterialExpressionMultiply_76}
    B {from=MaterialExpressionReroute_64}
   ]
  }
  MaterialExpressionMultiply_65 {
   class=MaterialExpressionMultiply
   inputs [
    A {from=MaterialExpressionReroute_19}
    B {from=MaterialExpressionReroute_19}
   ]
   properties {desc="r^2"}
  }
  MaterialExpressionMultiply_67 {
   class=MaterialExpressionMultiply
   inputs [
    A {from=MaterialExpressionReroute_66}
    B {from=MaterialExpressionReroute_66}
   ]
   properties {desc="r^2"}
  }
  MaterialExpressionMultiply_68 {
   class=MaterialExpressionMultiply
   inputs [
    A {from=MaterialExpressionReroute_66}
    B {from=""}
   ]
   properties {constB=2}
  }
  MaterialExpressionMultiply_69 {
   class=MaterialExpressionMultiply
   inputs [
    A {from=MaterialExpressionSquareRoot_12}
    B {from=""}
   ]
   properties {constB=2}
  }
  MaterialExpressionMultiply_7 {
   class=MaterialExpressionMultiply
   inputs [
    A {from=MaterialExpressionTextureSample_74, output=G}
    B {from=MaterialExpressionNamedRerouteUsage_49}
   ]
  }
  MaterialExpressionMultiply_70 {
   class=MaterialExpressionMultiply
   inputs [
    A {from=MaterialExpressionDotProduct_23}
    B {from=MaterialExpressionDotProduct_23}
   ]
  }
  MaterialExpressionMultiply_71 {
   class=MaterialExpressionMultiply
   inputs [
    A {from=MaterialExpressionDistance_4}
    B {from=MaterialExpressionDistance_4}
   ]
  }
  MaterialExpressionMultiply_72 {
   class=MaterialExpressionMultiply
   inputs [
    A {from=MaterialExpressionReroute_19}
    B {from=""}
   ]
   properties {constB=2}
  }
  MaterialExpressionMultiply_76 {
   class=MaterialExpressionMultiply
   inputs [
    A {from=MaterialExpressionNamedRerouteUsage_101}
    B {from=MaterialExpressionNamedRerouteUsage_100}
   ]
  }
  MaterialExpressionMultiply_77 {
   class=MaterialExpressionMultiply
   inputs [
    A {from=MaterialExpressionNamedRerouteDeclaration_35}
    B {from=MaterialExpressionLinearInterpolate_54}
   ]
  }
  MaterialExpressionMultiply_78 {
   class=MaterialExpressionMultiply
   inputs [
    A {from=MaterialExpressionTextureSample_37, output=B}
    B {from=MaterialExpressionTextureSample_39, output=B}
   ]
  }
  MaterialExpressionMultiply_80 {
   class=MaterialExpressionMultiply
   inputs [
    A {from=""}
    B {from=MaterialExpressionNamedRerouteUsage_119}
   ]
  }
  MaterialExpressionMultiply_83 {
   class=MaterialExpressionMultiply
   inputs [
    A {from=MaterialExpressionComponentMask_18}
    B {from=MaterialExpressionComponentMask_17}
   ]
  }
  MaterialExpressionMultiply_85 {
   class=MaterialExpressionMultiply
   inputs [
    A {from=MaterialExpressionLinearInterpolate_57}
    B {from=MaterialExpressionLinearInterpolate_60}
   ]
  }
  MaterialExpressionMultiply_86 {
   class=MaterialExpressionMultiply
   inputs [
    A {from=MaterialExpressionSquareRoot_0}
    B {from=""}
   ]
   properties {constB=2}
  }
  MaterialExpressionMultiply_89 {
   class=MaterialExpressionMultiply
   inputs [
    A {from=MaterialExpressionReroute_13}
    B {from=MaterialExpressionScalarParameter_165}
   ]
  }
  MaterialExpressionMultiply_9 {
   class=MaterialExpressionMultiply
   inputs [
    A {from=MaterialExpressionAdd_3}
    B {from=MaterialExpressionLinearInterpolate_9}
   ]
  }
  MaterialExpressionMultiply_90 {
   class=MaterialExpressionMultiply
   inputs [
    A {from=MaterialExpressionVectorParameter_27, output=RGB}
    B {from=MaterialExpressionVectorParameter_27, output=RGB}
   ]
  }
  MaterialExpressionMultiply_91 {
   class=MaterialExpressionMultiply
   inputs [
    A {from=MaterialExpressionDotProduct_0}
    B {from=MaterialExpressionDotProduct_0}
   ]
  }
  MaterialExpressionMultiply_92 {
   class=MaterialExpressionMultiply
   inputs [
    A {from=MaterialExpressionDistance_1}
    B {from=MaterialExpressionDistance_1}
   ]
  }
  MaterialExpressionMultiply_93 {
   class=MaterialExpressionMultiply
   inputs [
    A {from=MaterialExpressionDistance_2}
    B {from=MaterialExpressionDistance_2}
   ]
  }
  MaterialExpressionMultiply_94 {
   class=MaterialExpressionMultiply
   inputs [
    A {from=MaterialExpressionAdd_2}
    B {from=MaterialExpressionScalarParameter_66}
   ]
  }
  MaterialExpressionMultiply_96 {
   class=MaterialExpressionMultiply
   inputs [
    A {from=MaterialExpressionDotProduct_1}
    B {from=MaterialExpressionDotProduct_1}
   ]
  }
  MaterialExpressionMultiply_97 {
   class=MaterialExpressionMultiply
   inputs [
    A {from=MaterialExpressionSaturate_15}
    B {from=MaterialExpressionScalarParameter_72}
   ]
  }
  MaterialExpressionMultiply_98 {
   class=MaterialExpressionMultiply
   inputs [
    A {from=MaterialExpressionScalarParameter_69}
    B {from=MaterialExpressionScalarParameter_69}
   ]
  }
  MaterialExpressionNamedRerouteDeclaration_0 {
   class=MaterialExpressionNamedRerouteDeclaration
   inputs [None {from=MaterialExpressionMaterialFunctionCall_1, output=Result}]
   properties {name=Thickness, nodeColor=<rgba 1 0 0.6797 1>}
  }
  MaterialExpressionNamedRerouteDeclaration_11 {
   class=MaterialExpressionNamedRerouteDeclaration
   inputs [None {from=MaterialExpressionMultiply_147}]
   properties {name="Pre-Skinned Scale (S)", nodeColor=<rgba 1 0.5391 0 1>}
  }
  MaterialExpressionNamedRerouteDeclaration_12 {
   class=MaterialExpressionNamedRerouteDeclaration
   inputs [None {from=MaterialExpressionFunctionInput_7}]
   properties {name="Pre-Skinned Local Pos (V3)", nodeColor=<rgba 0 0.1797 1 1>}
  }
  MaterialExpressionNamedRerouteDeclaration_14 {
   class=MaterialExpressionNamedRerouteDeclaration
   inputs [None {from=MaterialExpressionSmoothStep_2}]
   properties {name=M_TriplanarSeam 70%, nodeColor=<rgba 1 0 0.9375 1>}
  }
  MaterialExpressionNamedRerouteDeclaration_15 {
   class=MaterialExpressionNamedRerouteDeclaration
   inputs [None {from=MaterialExpressionShaderStageSwitch_0}]
   properties {name=PreSkinnedNormal_VS-PS, nodeColor=<rgba 0.3587 0.1657 1 1>}
  }
  MaterialExpressionNamedRerouteDeclaration_17 {
   class=MaterialExpressionNamedRerouteDeclaration
   inputs [None {from=MaterialExpressionSmoothStep_8}]
   properties {name=M_TriplanarSeam 20%, nodeColor=<rgba 1 0 0.9375 1>}
  }
  MaterialExpressionNamedRerouteDeclaration_18 {
   class=MaterialExpressionNamedRerouteDeclaration
   inputs [None {from=MaterialExpressionMultiply_12}]
   properties {name=Object Radius, nodeColor=<rgba 1 0.1406 0 1>}
  }
  MaterialExpressionNamedRerouteDeclaration_19 {
   class=MaterialExpressionNamedRerouteDeclaration
   inputs [None {from=MaterialExpressionMultiply_194}]
   properties {name=M_Sparkles, nodeColor=<rgba 0 0.5078 1 1>}
  }
  MaterialExpressionNamedRerouteDeclaration_2 {
   class=MaterialExpressionNamedRerouteDeclaration
   inputs [None {from=MaterialExpressionFunctionInput_9}]
   properties {name="Pre-Skinned Bounds Min (V3)", nodeColor=<rgba 1 1 1 1>}
  }
  MaterialExpressionNamedRerouteDeclaration_20 {
   class=MaterialExpressionNamedRerouteDeclaration
   inputs [None {from=MaterialExpressionMax_5}]
   properties {name=M_TriplanarSeams, nodeColor=<rgba 0.4531 0 1 1>}
  }
  MaterialExpressionNamedRerouteDeclaration_21 {
   class=MaterialExpressionNamedRerouteDeclaration
   inputs [None {from=MaterialExpressionGetMaterialAttributes_1, output=MaterialAttributes}]
   properties {name=IN_MA, nodeColor=<rgba 1 0 0.1015 1>}
  }
  MaterialExpressionNamedRerouteDeclaration_22 {
   class=MaterialExpressionNamedRerouteDeclaration
   inputs [None {from=MaterialExpressionLinearInterpolate_44}]
   properties {name=M_Runes, nodeColor=<rgba 0.0547 0 1 1>}
  }
  MaterialExpressionNamedRerouteDeclaration_24 {
   class=MaterialExpressionNamedRerouteDeclaration
   inputs [None {from=MaterialExpressionTextureObjectParameter_0}]
   properties {name=CubeNoise_Texture, nodeColor=<rgba 0.7109 0 1 1>}
  }
  MaterialExpressionNamedRerouteDeclaration_25 {
   class=MaterialExpressionNamedRerouteDeclaration
   inputs [None {from=MaterialExpressionLinearInterpolate_20}]
   properties {name=M_Innerglow_Hotspot, nodeColor=<rgba 0.0781 1 0 1>}
  }
  MaterialExpressionNamedRerouteDeclaration_26 {
   class=MaterialExpressionNamedRerouteDeclaration
   inputs [None {from=MaterialExpressionMaterialFunctionCall_28, output=Lerp Alpha No Contrast}]
   properties {name=M_Inner Glow, nodeColor=<rgba 1 0 0.3984 1>}
  }
  MaterialExpressionNamedRerouteDeclaration_27 {
   class=MaterialExpressionNamedRerouteDeclaration
   inputs [None {from=MaterialExpressionOneMinus_14}]
   properties {name=M_SphereMaskEdges_Pow, nodeColor=<rgba 1 0 0 1>}
  }
  MaterialExpressionNamedRerouteDeclaration_28 {
   class=MaterialExpressionNamedRerouteDeclaration
   inputs [None {from=MaterialExpressionSaturate_11}]
   properties {name=M_FaceGradient, nodeColor=<rgba 1 0 0.2109 1>}
  }
  MaterialExpressionNamedRerouteDeclaration_29 {
   class=MaterialExpressionNamedRerouteDeclaration
   inputs [None {from=MaterialExpressionOneMinus_9}]
   properties {name=M_Fresnel, nodeColor=<rgba 1 0.9844 0 1>}
  }
  MaterialExpressionNamedRerouteDeclaration_3 {
   class=MaterialExpressionNamedRerouteDeclaration
   inputs [None {from=MaterialExpressionFunctionInput_8}]
   properties {name="Pre-Skinned Bounds Max (V3)", nodeColor=<rgba 1 1 1 1>}
  }
  MaterialExpressionNamedRerouteDeclaration_30 {
   class=MaterialExpressionNamedRerouteDeclaration
   inputs [None {from=MaterialExpressionGetMaterialAttributes_1, output=Normal}]
   properties {name=IN_Normals, nodeColor=<rgba 0 1 0.6016 1>}
  }
  MaterialExpressionNamedRerouteDeclaration_31 {
   class=MaterialExpressionNamedRerouteDeclaration
   inputs [None {from=MaterialExpressionFunctionInput_1}]
   properties {name="Pre-Skinned Scale (V3)", nodeColor=<rgba 1 0.5391 0 1>}
  }
  MaterialExpressionNamedRerouteDeclaration_32 {
   class=MaterialExpressionNamedRerouteDeclaration
   inputs [None {from=MaterialExpressionLinearInterpolate_3}]
   properties {name=CubeDefaultNoise, nodeColor=<rgba 0 1 0.7891 1>}
  }
  MaterialExpressionNamedRerouteDeclaration_33 {
   class=MaterialExpressionNamedRerouteDeclaration
   inputs [None {from=MaterialExpressionMultiply_5}]
   properties {name=M_BlurredShadow, nodeColor=<rgba 1 0 0.9844 1>}
  }
  MaterialExpressionNamedRerouteDeclaration_34 {
   class=MaterialExpressionNamedRerouteDeclaration
   inputs [None {from=MaterialExpressionMultiply_7}]
   properties {name=M_BlurredNoise, nodeColor=<rgba 0.4531 0 1 1>}
  }
  MaterialExpressionNamedRerouteDeclaration_35 {
   class=MaterialExpressionNamedRerouteDeclaration
   inputs [None {from=MaterialExpressionMultiply_222}]
   properties {name=SparklesColor, nodeColor=<rgba 1 0.4453 0 1>}
  }
  MaterialExpressionNamedRerouteDeclaration_36 {
   class=MaterialExpressionNamedRerouteDeclaration
   inputs [None {from=MaterialExpressionLinearInterpolate_14}]
   properties {name=OUT_Specular, nodeColor=<rgba 0.8281 0 1 1>}
  }
  MaterialExpressionNamedRerouteDeclaration_37 {
   class=MaterialExpressionNamedRerouteDeclaration
   inputs [None {from=MaterialExpressionLinearInterpolate_48}]
   properties {name=OUT_Base Color, nodeColor=<rgba 0 1 0.5078 1>}
  }
  MaterialExpressionNamedRerouteDeclaration_38 {
   class=MaterialExpressionNamedRerouteDeclaration
   inputs [None {from=MaterialExpressionLinearInterpolate_32}]
   properties {name=OUT_Roughness, nodeColor=<rgba 0.6641 0 1 1>}
  }
  MaterialExpressionNamedRerouteDeclaration_39 {
   class=MaterialExpressionNamedRerouteDeclaration
   inputs [None {from=MaterialExpressionMaterialFunctionCall_16, output=Result}]
   properties {name=DistanceLerp, nodeColor=<rgba 0 0.25 1 1>}
  }
  MaterialExpressionNamedRerouteDeclaration_40 {
   class=MaterialExpressionNamedRerouteDeclaration
   inputs [None {from=MaterialExpressionGetMaterialAttributes_1, output=BaseColor}]
   properties {name=IN_BaseColor, nodeColor=<rgba 0.0781 0 1 1>}
  }
  MaterialExpressionNamedRerouteDeclaration_41 {
   class=MaterialExpressionNamedRerouteDeclaration
   inputs [None {from=MaterialExpressionGetMaterialAttributes_1, output=Metallic}]
   properties {name=In_Metallic, nodeColor=<rgba 0.2422 1 0 1>}
  }
  MaterialExpressionNamedRerouteDeclaration_42 {
   class=MaterialExpressionNamedRerouteDeclaration
   inputs [None {from=MaterialExpressionMaterialFunctionCall_24, output="Rounded Linear Sine "}]
   properties {name=AnimatedSine, nodeColor=<rgba 0.1484 1 0 1>}
  }
  MaterialExpressionNamedRerouteDeclaration_43 {
   class=MaterialExpressionNamedRerouteDeclaration
   inputs [None {from=MaterialExpressionMultiply_49}]
   properties {name=M_ScrollingRunes_Sined, nodeColor=<rgba 0.7812 0 1 1>}
  }
  MaterialExpressionNamedRerouteDeclaration_44 {
   class=MaterialExpressionNamedRerouteDeclaration
   inputs [None {from=MaterialExpressionSmoothStep_3}]
   properties {name=SphereMaskEdges_InnerGlow, nodeColor=<rgba 1 0 0.9844 1>}
  }
  MaterialExpressionNamedRerouteDeclaration_45 {
   class=MaterialExpressionNamedRerouteDeclaration
   inputs [None {from=MaterialExpressionComponentMask_16}]
   properties {name=M_ZGradient, nodeColor=<rgba 0.1719 1 0 1>}
  }
  MaterialExpressionNamedRerouteDeclaration_47 {
   class=MaterialExpressionNamedRerouteDeclaration
   inputs [None {from=MaterialExpressionGetMaterialAttributes_1, output=Roughness}]
   properties {name=In_Roughness, nodeColor=<rgba 1 0 0.5625 1>}
  }
  MaterialExpressionNamedRerouteDeclaration_48 {
   class=MaterialExpressionNamedRerouteDeclaration
   inputs [None {from=MaterialExpressionGetMaterialAttributes_1, output=Specular}]
   properties {name=IN_Specular, nodeColor=<rgba 1 0.4453 0 1>}
  }
  MaterialExpressionNamedRerouteDeclaration_49 {
   class=MaterialExpressionNamedRerouteDeclaration
   inputs [None {from=MaterialExpressionFunctionInput_4}]
   properties {name=M_VariantMask, nodeColor=<rgba 0 1 0.8828 1>}
  }
  MaterialExpressionNamedRerouteDeclaration_5 {
   class=MaterialExpressionNamedRerouteDeclaration
   inputs [None {from=MaterialExpressionSubtract_40}]
   properties {name="Pre-Skinned Bounds Extents (V3)", nodeColor=<rgba 1 1 1 1>}
  }
  MaterialExpressionNamedRerouteDeclaration_50 {
   class=MaterialExpressionNamedRerouteDeclaration
   inputs [None {from=MaterialExpressionAdd_52}]
   properties {name=OUT_Emissive, nodeColor=<rgba 0 0.0859 1 1>}
  }
  MaterialExpressionNamedRerouteDeclaration_51 {
   class=MaterialExpressionNamedRerouteDeclaration
   inputs [None {from=MaterialExpressionFunctionInput_5}]
   properties {name=M_InnerRadialGlow, nodeColor=<rgba 1 0.9844 0 1>}
  }
  MaterialExpressionNamedRerouteDeclaration_52 {
   class=MaterialExpressionNamedRerouteDeclaration
   inputs [None {from=MaterialExpressionNamedRerouteDeclaration_44}]
   properties {name=M_SphereMaskEdges, nodeColor=<rgba 1 0 0 1>}
  }
  MaterialExpressionNamedRerouteDeclaration_53 {
   class=MaterialExpressionNamedRerouteDeclaration
   inputs [None {from=MaterialExpressionScalarParameter_75}]
   properties {name=OUT_Metallic, nodeColor=<rgba 0.7344 0 1 1>}
  }
  MaterialExpressionNamedRerouteDeclaration_54 {
   class=MaterialExpressionNamedRerouteDeclaration
   inputs [None {from=MaterialExpressionMultiply_85}]
   properties {name=M_InteriorDephtGrad, nodeColor=<rgba 1 0 0.3984 1>}
  }
  MaterialExpressionNamedRerouteDeclaration_55 {
   class=MaterialExpressionNamedRerouteDeclaration
   inputs [None {from=MaterialExpressionOneMinus_3}]
   properties {name=M_EdgesNoise, nodeColor=<rgba 1 0 0.4453 1>}
  }
  MaterialExpressionNamedRerouteDeclaration_56 {
   class=MaterialExpressionNamedRerouteDeclaration
   inputs [None {from=MaterialExpressionFunctionInput_12}]
   properties {name=M_SpriteFresnel, nodeColor=<rgba 1 0.9844 0 1>}
  }
  MaterialExpressionNamedRerouteDeclaration_57 {
   class=MaterialExpressionNamedRerouteDeclaration
   inputs [None {from=MaterialExpressionFunctionInput_11}]
   properties {name=M_Logo, nodeColor=<rgba 1 0.9844 0 1>}
  }
  MaterialExpressionNamedRerouteDeclaration_58 {
   class=MaterialExpressionNamedRerouteDeclaration
   inputs [None {from=MaterialExpressionFunctionInput_13}]
   properties {name=IN_WPO, nodeColor=<rgba 0 0.9766 1 1>}
  }
  MaterialExpressionNamedRerouteDeclaration_6 {
   class=MaterialExpressionNamedRerouteDeclaration
   inputs [None {from=MaterialExpressionMultiply_141}]
   properties {name="Pre-Skinned Bounds Half Extents (V3)", nodeColor=<rgba 1 1 1 1>}
  }
  MaterialExpressionNamedRerouteDeclaration_60 {
   class=MaterialExpressionNamedRerouteDeclaration
   inputs [None {from=MaterialExpressionDDX_2}]
   properties {name="Scaled UV DDX (V2)", nodeColor=<rgba 0.0547 1 0 1>}
  }
  MaterialExpressionNamedRerouteDeclaration_61 {
   class=MaterialExpressionNamedRerouteDeclaration
   inputs [None {from=MaterialExpressionFunctionInput_14}]
   properties {name=M_BlinkingLights, nodeColor=<rgba 1 0.9844 0 1>}
  }
  MaterialExpressionNamedRerouteDeclaration_62 {
   class=MaterialExpressionNamedRerouteDeclaration
   inputs [None {from=MaterialExpressionDDY_0}]
   properties {name="Scaled UV DDY (V2)", nodeColor=<rgba 0.0547 1 0 1>}
  }
  MaterialExpressionNamedRerouteDeclaration_63 {
   class=MaterialExpressionNamedRerouteDeclaration
   inputs [None {from=MaterialExpressionGetMaterialAttributes_1, output=EmissiveColor}]
   properties {name=IN_Emissive, nodeColor=<rgba 0.9688 1 0 1>}
  }
  MaterialExpressionNamedRerouteDeclaration_64 {
   class=MaterialExpressionNamedRerouteDeclaration
   inputs [None {from=MaterialExpressionLinearInterpolate_7}]
   properties {name=M_Logo+Variant, nodeColor=<rgba 0 1 0.1328 1>}
  }
  MaterialExpressionNamedRerouteDeclaration_65 {
   class=MaterialExpressionNamedRerouteDeclaration
   inputs [None {from=MaterialExpressionSaturate_5}]
   properties {name=IN_BaseColor+Emissive, nodeColor=<rgba 0 1 0.5078 1>}
  }
  MaterialExpressionNamedRerouteDeclaration_66 {
   class=MaterialExpressionNamedRerouteDeclaration
   inputs [None {from=MaterialExpressionFunctionInput_15}]
   properties {name=M_DiffuseMask, nodeColor=<rgba 0.875 1 0 1>}
  }
  MaterialExpressionNamedRerouteDeclaration_67 {
   class=MaterialExpressionNamedRerouteDeclaration
   inputs [None {from=MaterialExpressionMultiply_161}]
   properties {name="Bump Offset Mult (S)", nodeColor=<rgba 0.1484 0 1 1>}
  }
  MaterialExpressionNamedRerouteDeclaration_8 {
   class=MaterialExpressionNamedRerouteDeclaration
   inputs [None {from=MaterialExpressionFunctionInput_0}]
   properties {name="Object Position WS (V3)", nodeColor=<rgba 1 0.9375 0 1>}
  }
  MaterialExpressionNamedRerouteUsage_0 {class=MaterialExpressionNamedRerouteUsage}
  MaterialExpressionNamedRerouteUsage_1 {class=MaterialExpressionNamedRerouteUsage}
  MaterialExpressionNamedRerouteUsage_100 {class=MaterialExpressionNamedRerouteUsage}
  MaterialExpressionNamedRerouteUsage_101 {class=MaterialExpressionNamedRerouteUsage}
  MaterialExpressionNamedRerouteUsage_108 {class=MaterialExpressionNamedRerouteUsage}
  MaterialExpressionNamedRerouteUsage_11 {class=MaterialExpressionNamedRerouteUsage}
  MaterialExpressionNamedRerouteUsage_118 {class=MaterialExpressionNamedRerouteUsage}
  MaterialExpressionNamedRerouteUsage_119 {class=MaterialExpressionNamedRerouteUsage}
  MaterialExpressionNamedRerouteUsage_12 {class=MaterialExpressionNamedRerouteUsage}
  MaterialExpressionNamedRerouteUsage_121 {class=MaterialExpressionNamedRerouteUsage}
  MaterialExpressionNamedRerouteUsage_122 {class=MaterialExpressionNamedRerouteUsage}
  MaterialExpressionNamedRerouteUsage_123 {class=MaterialExpressionNamedRerouteUsage}
  MaterialExpressionNamedRerouteUsage_124 {class=MaterialExpressionNamedRerouteUsage}
  MaterialExpressionNamedRerouteUsage_125 {class=MaterialExpressionNamedRerouteUsage}
  MaterialExpressionNamedRerouteUsage_126 {class=MaterialExpressionNamedRerouteUsage}
  MaterialExpressionNamedRerouteUsage_13 {class=MaterialExpressionNamedRerouteUsage}
  MaterialExpressionNamedRerouteUsage_14 {class=MaterialExpressionNamedRerouteUsage}
  MaterialExpressionNamedRerouteUsage_148 {class=MaterialExpressionNamedRerouteUsage}
  MaterialExpressionNamedRerouteUsage_149 {class=MaterialExpressionNamedRerouteUsage}
  MaterialExpressionNamedRerouteUsage_15 {class=MaterialExpressionNamedRerouteUsage}
  MaterialExpressionNamedRerouteUsage_154 {class=MaterialExpressionNamedRerouteUsage}
  MaterialExpressionNamedRerouteUsage_156 {class=MaterialExpressionNamedRerouteUsage}
  MaterialExpressionNamedRerouteUsage_158 {class=MaterialExpressionNamedRerouteUsage}
  MaterialExpressionNamedRerouteUsage_159 {class=MaterialExpressionNamedRerouteUsage}
  MaterialExpressionNamedRerouteUsage_16 {class=MaterialExpressionNamedRerouteUsage}
  MaterialExpressionNamedRerouteUsage_165 {class=MaterialExpressionNamedRerouteUsage}
  MaterialExpressionNamedRerouteUsage_17 {class=MaterialExpressionNamedRerouteUsage}
  MaterialExpressionNamedRerouteUsage_171 {class=MaterialExpressionNamedRerouteUsage}
  MaterialExpressionNamedRerouteUsage_173 {class=MaterialExpressionNamedRerouteUsage}
  MaterialExpressionNamedRerouteUsage_175 {class=MaterialExpressionNamedRerouteUsage}
  MaterialExpressionNamedRerouteUsage_176 {class=MaterialExpressionNamedRerouteUsage}
  MaterialExpressionNamedRerouteUsage_18 {class=MaterialExpressionNamedRerouteUsage}
  MaterialExpressionNamedRerouteUsage_183 {class=MaterialExpressionNamedRerouteUsage}
  MaterialExpressionNamedRerouteUsage_188 {class=MaterialExpressionNamedRerouteUsage}
  MaterialExpressionNamedRerouteUsage_189 {class=MaterialExpressionNamedRerouteUsage}
  MaterialExpressionNamedRerouteUsage_190 {class=MaterialExpressionNamedRerouteUsage}
  MaterialExpressionNamedRerouteUsage_191 {class=MaterialExpressionNamedRerouteUsage}
  MaterialExpressionNamedRerouteUsage_192 {class=MaterialExpressionNamedRerouteUsage}
  MaterialExpressionNamedRerouteUsage_193 {class=MaterialExpressionNamedRerouteUsage}
  MaterialExpressionNamedRerouteUsage_194 {class=MaterialExpressionNamedRerouteUsage}
  MaterialExpressionNamedRerouteUsage_195 {class=MaterialExpressionNamedRerouteUsage}
  MaterialExpressionNamedRerouteUsage_196 {class=MaterialExpressionNamedRerouteUsage}
  MaterialExpressionNamedRerouteUsage_197 {class=MaterialExpressionNamedRerouteUsage}
  MaterialExpressionNamedRerouteUsage_198 {class=MaterialExpressionNamedRerouteUsage}
  MaterialExpressionNamedRerouteUsage_199 {class=MaterialExpressionNamedRerouteUsage}
  MaterialExpressionNamedRerouteUsage_2 {class=MaterialExpressionNamedRerouteUsage}
  MaterialExpressionNamedRerouteUsage_20 {class=MaterialExpressionNamedRerouteUsage}
  MaterialExpressionNamedRerouteUsage_200 {class=MaterialExpressionNamedRerouteUsage}
  MaterialExpressionNamedRerouteUsage_201 {class=MaterialExpressionNamedRerouteUsage}
  MaterialExpressionNamedRerouteUsage_202 {class=MaterialExpressionNamedRerouteUsage}
  MaterialExpressionNamedRerouteUsage_203 {class=MaterialExpressionNamedRerouteUsage}
  MaterialExpressionNamedRerouteUsage_204 {class=MaterialExpressionNamedRerouteUsage}
  MaterialExpressionNamedRerouteUsage_205 {class=MaterialExpressionNamedRerouteUsage}
  MaterialExpressionNamedRerouteUsage_21 {class=MaterialExpressionNamedRerouteUsage}
  MaterialExpressionNamedRerouteUsage_210 {class=MaterialExpressionNamedRerouteUsage}
  MaterialExpressionNamedRerouteUsage_212 {class=MaterialExpressionNamedRerouteUsage}
  MaterialExpressionNamedRerouteUsage_213 {class=MaterialExpressionNamedRerouteUsage}
  MaterialExpressionNamedRerouteUsage_214 {class=MaterialExpressionNamedRerouteUsage}
  MaterialExpressionNamedRerouteUsage_215 {class=MaterialExpressionNamedRerouteUsage}
  MaterialExpressionNamedRerouteUsage_216 {class=MaterialExpressionNamedRerouteUsage}
  MaterialExpressionNamedRerouteUsage_22 {class=MaterialExpressionNamedRerouteUsage}
  MaterialExpressionNamedRerouteUsage_23 {class=MaterialExpressionNamedRerouteUsage}
  MaterialExpressionNamedRerouteUsage_24 {class=MaterialExpressionNamedRerouteUsage}
  MaterialExpressionNamedRerouteUsage_25 {class=MaterialExpressionNamedRerouteUsage}
  MaterialExpressionNamedRerouteUsage_26 {class=MaterialExpressionNamedRerouteUsage}
  MaterialExpressionNamedRerouteUsage_27 {class=MaterialExpressionNamedRerouteUsage}
  MaterialExpressionNamedRerouteUsage_28 {class=MaterialExpressionNamedRerouteUsage}
  MaterialExpressionNamedRerouteUsage_29 {class=MaterialExpressionNamedRerouteUsage}
  MaterialExpressionNamedRerouteUsage_3 {class=MaterialExpressionNamedRerouteUsage}
  MaterialExpressionNamedRerouteUsage_30 {class=MaterialExpressionNamedRerouteUsage}
  MaterialExpressionNamedRerouteUsage_31 {class=MaterialExpressionNamedRerouteUsage}
  MaterialExpressionNamedRerouteUsage_32 {class=MaterialExpressionNamedRerouteUsage}
  MaterialExpressionNamedRerouteUsage_33 {class=MaterialExpressionNamedRerouteUsage}
  MaterialExpressionNamedRerouteUsage_34 {class=MaterialExpressionNamedRerouteUsage}
  MaterialExpressionNamedRerouteUsage_35 {class=MaterialExpressionNamedRerouteUsage}
  MaterialExpressionNamedRerouteUsage_36 {class=MaterialExpressionNamedRerouteUsage}
  MaterialExpressionNamedRerouteUsage_37 {class=MaterialExpressionNamedRerouteUsage}
  MaterialExpressionNamedRerouteUsage_38 {class=MaterialExpressionNamedRerouteUsage}
  MaterialExpressionNamedRerouteUsage_39 {class=MaterialExpressionNamedRerouteUsage}
  MaterialExpressionNamedRerouteUsage_4 {class=MaterialExpressionNamedRerouteUsage}
  MaterialExpressionNamedRerouteUsage_40 {class=MaterialExpressionNamedRerouteUsage}
  MaterialExpressionNamedRerouteUsage_41 {class=MaterialExpressionNamedRerouteUsage}
  MaterialExpressionNamedRerouteUsage_42 {class=MaterialExpressionNamedRerouteUsage}
  MaterialExpressionNamedRerouteUsage_43 {class=MaterialExpressionNamedRerouteUsage}
  MaterialExpressionNamedRerouteUsage_44 {class=MaterialExpressionNamedRerouteUsage}
  MaterialExpressionNamedRerouteUsage_45 {class=MaterialExpressionNamedRerouteUsage}
  MaterialExpressionNamedRerouteUsage_46 {class=MaterialExpressionNamedRerouteUsage}
  MaterialExpressionNamedRerouteUsage_47 {class=MaterialExpressionNamedRerouteUsage}
  MaterialExpressionNamedRerouteUsage_48 {class=MaterialExpressionNamedRerouteUsage}
  MaterialExpressionNamedRerouteUsage_49 {class=MaterialExpressionNamedRerouteUsage}
  MaterialExpressionNamedRerouteUsage_5 {class=MaterialExpressionNamedRerouteUsage}
  MaterialExpressionNamedRerouteUsage_50 {class=MaterialExpressionNamedRerouteUsage}
  MaterialExpressionNamedRerouteUsage_51 {class=MaterialExpressionNamedRerouteUsage}
  MaterialExpressionNamedRerouteUsage_52 {class=MaterialExpressionNamedRerouteUsage}
  MaterialExpressionNamedRerouteUsage_53 {class=MaterialExpressionNamedRerouteUsage}
  MaterialExpressionNamedRerouteUsage_54 {class=MaterialExpressionNamedRerouteUsage}
  MaterialExpressionNamedRerouteUsage_55 {class=MaterialExpressionNamedRerouteUsage}
  MaterialExpressionNamedRerouteUsage_56 {class=MaterialExpressionNamedRerouteUsage}
  MaterialExpressionNamedRerouteUsage_6 {class=MaterialExpressionNamedRerouteUsage}
  MaterialExpressionNamedRerouteUsage_7 {class=MaterialExpressionNamedRerouteUsage}
  MaterialExpressionNamedRerouteUsage_8 {class=MaterialExpressionNamedRerouteUsage}
  MaterialExpressionNamedRerouteUsage_80 {class=MaterialExpressionNamedRerouteUsage}
  MaterialExpressionNamedRerouteUsage_81 {class=MaterialExpressionNamedRerouteUsage}
  MaterialExpressionNamedRerouteUsage_86 {class=MaterialExpressionNamedRerouteUsage}
  MaterialExpressionNamedRerouteUsage_87 {class=MaterialExpressionNamedRerouteUsage}
  MaterialExpressionNamedRerouteUsage_88 {class=MaterialExpressionNamedRerouteUsage}
  MaterialExpressionNamedRerouteUsage_89 {class=MaterialExpressionNamedRerouteUsage}
  MaterialExpressionNamedRerouteUsage_9 {class=MaterialExpressionNamedRerouteUsage}
  MaterialExpressionNamedRerouteUsage_90 {class=MaterialExpressionNamedRerouteUsage}
  MaterialExpressionNamedRerouteUsage_91 {class=MaterialExpressionNamedRerouteUsage}
  MaterialExpressionNamedRerouteUsage_93 {class=MaterialExpressionNamedRerouteUsage}
  MaterialExpressionNamedRerouteUsage_99 {class=MaterialExpressionNamedRerouteUsage}
  MaterialExpressionNormalize_1 {class=MaterialExpressionNormalize, inputs [VectorInput {from=MaterialExpressionAdd_13, output=""}]}
  MaterialExpressionObjectBounds_0 {class=MaterialExpressionObjectBounds, inputs [], properties {}}
  MaterialExpressionObjectLocalBounds_0 {class=MaterialExpressionObjectLocalBounds, inputs [], properties {}}
  MaterialExpressionObjectLocalBounds_1 {class=MaterialExpressionObjectLocalBounds, inputs [], properties {}}
  MaterialExpressionObjectPositionWS_1 {class=MaterialExpressionObjectPositionWS, inputs [], properties {}}
  MaterialExpressionObjectPositionWS_2 {class=MaterialExpressionObjectPositionWS, inputs [], properties {}}
  MaterialExpressionObjectPositionWS_4 {class=MaterialExpressionObjectPositionWS, inputs [], properties {}}
  MaterialExpressionOneMinus_0 {class=MaterialExpressionOneMinus, inputs [None {from=MaterialExpressionDivide_0}]}
  MaterialExpressionOneMinus_10 {class=MaterialExpressionOneMinus, inputs [None {from=MaterialExpressionPower_1}]}
  MaterialExpressionOneMinus_11 {class=MaterialExpressionOneMinus, inputs [None {from=MaterialExpressionMultiply_83}]}
  MaterialExpressionOneMinus_12 {class=MaterialExpressionOneMinus, inputs [None {from=MaterialExpressionNamedRerouteDeclaration_52}]}
  MaterialExpressionOneMinus_13 {class=MaterialExpressionOneMinus, inputs [None {from=MaterialExpressionLength_0}]}
  MaterialExpressionOneMinus_14 {class=MaterialExpressionOneMinus, inputs [None {from=MaterialExpressionPower_2}]}
  MaterialExpressionOneMinus_15 {class=MaterialExpressionOneMinus, inputs [None {from=MaterialExpressionSaturate_10}]}
  MaterialExpressionOneMinus_3 {class=MaterialExpressionOneMinus, inputs [None {from=MaterialExpressionSaturate_17}]}
  MaterialExpressionOneMinus_5 {class=MaterialExpressionOneMinus, inputs [None {from=MaterialExpressionSaturate_9}]}
  MaterialExpressionOneMinus_6 {class=MaterialExpressionOneMinus, inputs [None {from=MaterialExpressionPower_3}]}
  MaterialExpressionOneMinus_8 {class=MaterialExpressionOneMinus, inputs [None {from=MaterialExpressionDistance_0}]}
  MaterialExpressionOneMinus_9 {class=MaterialExpressionOneMinus, inputs [None {from=MaterialExpressionDotProduct_6}]}
  MaterialExpressionPanner_0 {
   class=MaterialExpressionPanner
   inputs [
    Coordinate {from=MaterialExpressionDivide_3}
    Time {from=MaterialExpressionMultiply_45}
    Speed {from=""}
   ]
   properties {speedX=1}
  }
  MaterialExpressionPower_1 {
   class=MaterialExpressionPower
   inputs [
    Base {from=MaterialExpressionSmoothStep_4}
    Exp {from=""}
   ]
   properties {constExponent=8}
  }
  MaterialExpressionPower_10 {
   class=MaterialExpressionPower
   inputs [
    Base {from=""}
    Exp {from=""}
   ]
   properties {constExponent=4}
  }
  MaterialExpressionPower_11 {
   class=MaterialExpressionPower
   inputs [
    Base {from=MaterialExpressionOneMinus_8}
    Exp {from=""}
   ]
   properties {constExponent=3}
  }
  MaterialExpressionPower_12 {
   class=MaterialExpressionPower
   inputs [
    Base {from=MaterialExpressionMultiply_58}
    Exp {from=MaterialExpressionScalarParameter_79}
   ]
  }
  MaterialExpressionPower_14 {
   class=MaterialExpressionPower
   inputs [
    Base {from=MaterialExpressionTextureSample_74, output=R}
    Exp {from=MaterialExpressionConstant_56}
   ]
  }
  MaterialExpressionPower_16 {
   class=MaterialExpressionPower
   inputs [
    Base {from=MaterialExpressionTextureSampleParameter2D_9, output=G}
    Exp {from=MaterialExpressionLinearInterpolate_59}
   ]
  }
  MaterialExpressionPower_17 {
   class=MaterialExpressionPower
   inputs [
    Base {from=MaterialExpressionSphereMask_1}
    Exp {from=""}
   ]
   properties {constExponent=4}
  }
  MaterialExpressionPower_18 {
   class=MaterialExpressionPower
   inputs [
    Base {from=MaterialExpressionNamedRerouteUsage_44}
    Exp {from=MaterialExpressionConstant_3}
   ]
  }
  MaterialExpressionPower_19 {
   class=MaterialExpressionPower
   inputs [
    Base {from=MaterialExpressionSaturate_4}
    Exp {from=""}
   ]
  }
  MaterialExpressionPower_2 {
   class=MaterialExpressionPower
   inputs [
    Base {from=MaterialExpressionOneMinus_12}
    Exp {from=""}
   ]
  }
  MaterialExpressionPower_22 {
   class=MaterialExpressionPower
   inputs [
    Base {from=MaterialExpressionNamedRerouteUsage_37}
    Exp {from=MaterialExpressionScalarParameter_209}
   ]
  }
  MaterialExpressionPower_3 {
   class=MaterialExpressionPower
   inputs [
    Base {from=MaterialExpressionOneMinus_5}
    Exp {from=MaterialExpressionScalarParameter_54}
   ]
  }
  MaterialExpressionPower_34 {
   class=MaterialExpressionPower
   inputs [
    Base {from=MaterialExpressionMultiply_78}
    Exp {from=MaterialExpressionLinearInterpolate_55}
   ]
  }
  MaterialExpressionPower_35 {
   class=MaterialExpressionPower
   inputs [
    Base {from=MaterialExpressionReroute_18}
    Exp {from=MaterialExpressionScalarParameter_193}
   ]
  }
  MaterialExpressionPower_37 {
   class=MaterialExpressionPower
   inputs [
    Base {from=MaterialExpressionNamedRerouteUsage_81}
    Exp {from=MaterialExpressionLinearInterpolate_31}
   ]
  }
  MaterialExpressionPower_4 {
   class=MaterialExpressionPower
   inputs [
    Base {from=MaterialExpressionMaterialFunctionCall_15, output=Result}
    Exp {from=""}
   ]
  }
  MaterialExpressionPower_6 {
   class=MaterialExpressionPower
   inputs [
    Base {from=MaterialExpressionReroute_13}
    Exp {from=MaterialExpressionScalarParameter_18}
   ]
  }
  MaterialExpressionPower_8 {
   class=MaterialExpressionPower
   inputs [
    Base {from=MaterialExpressionOneMinus_11}
    Exp {from=MaterialExpressionScalarParameter_64}
   ]
  }
  MaterialExpressionPower_9 {
   class=MaterialExpressionPower
   inputs [
    Base {from=MaterialExpressionTextureSampleParameter2D_15, output=R}
    Exp {from=MaterialExpressionLinearInterpolate_12}
   ]
  }
  MaterialExpressionPreSkinnedNormal_0 {class=MaterialExpressionPreSkinnedNormal, inputs [], properties {}}
  MaterialExpressionQualitySwitch_0 {
   class=MaterialExpressionQualitySwitch
   inputs [
    Default {from=MaterialExpressionMultiply_60, output=""}
    Low {from=MaterialExpressionMultiply_59, output=""}
    High {from=MaterialExpressionMultiply_60, output=""}
    Medium {from="", output=""}
    Epic {from="", output=""}
   ]
   properties {}
  }
  MaterialExpressionQualitySwitch_2 {
   class=MaterialExpressionQualitySwitch
   inputs [
    Default {from=MaterialExpressionReroute_70, output=""}
    Low {from=MaterialExpressionMultiply_62, output=""}
    High {from=MaterialExpressionReroute_70, output=""}
    Medium {from="", output=""}
    Epic {from="", output=""}
   ]
   properties {}
  }
  MaterialExpressionQualitySwitch_4 {
   class=MaterialExpressionQualitySwitch
   inputs [
    Default {from=MaterialExpressionMultiply_50, output=""}
    Low {from=MaterialExpressionMultiply_154, output=""}
    High {from=MaterialExpressionMultiply_50, output=""}
    Medium {from="", output=""}
    Epic {from="", output=""}
   ]
   properties {}
  }
  MaterialExpressionQualitySwitch_5 {
   class=MaterialExpressionQualitySwitch
   inputs [
    Default {from=MaterialExpressionMultiply_52, output=""}
    Low {from=MaterialExpressionSaturate_6, output=""}
    High {from=MaterialExpressionMultiply_52, output=""}
    Medium {from="", output=""}
    Epic {from="", output=""}
   ]
   properties {}
  }
  MaterialExpressionReroute_0 {class=MaterialExpressionReroute, inputs [None {from=MaterialExpressionAdd_4}]}
  MaterialExpressionReroute_1 {class=MaterialExpressionReroute, inputs [None {from=MaterialExpressionNamedRerouteUsage_121}]}
  MaterialExpressionReroute_10 {class=MaterialExpressionReroute, inputs [None {from=MaterialExpressionMultiply_3}]}
  MaterialExpressionReroute_11 {class=MaterialExpressionReroute, inputs [None {from=MaterialExpressionReroute_12}]}
  MaterialExpressionReroute_12 {class=MaterialExpressionReroute, inputs [None {from=MaterialExpressionAdd_48}]}
  MaterialExpressionReroute_13 {class=MaterialExpressionReroute, inputs [None {from=MaterialExpressionTextureSampleParameter2D_7, output=G}]}
  MaterialExpressionReroute_14 {class=MaterialExpressionReroute, inputs [None {from=MaterialExpressionMaterialFunctionCall_43, output=Result}]}
  MaterialExpressionReroute_15 {class=MaterialExpressionReroute, inputs [None {from=MaterialExpressionOneMinus_15}]}
  MaterialExpressionReroute_16 {class=MaterialExpressionReroute, inputs [None {from=MaterialExpressionSaturate_3}]}
  MaterialExpressionReroute_17 {class=MaterialExpressionReroute, inputs [None {from=MaterialExpressionPower_11}]}
  MaterialExpressionReroute_18 {class=MaterialExpressionReroute, inputs [None {from=MaterialExpressionReroute_17}]}
  MaterialExpressionReroute_19 {class=MaterialExpressionReroute, inputs [None {from=MaterialExpressionMultiply_61}]}
  MaterialExpressionReroute_2 {class=MaterialExpressionReroute, inputs [None {from=MaterialExpressionNamedRerouteUsage_122}]}
  MaterialExpressionReroute_20 {class=MaterialExpressionReroute, inputs [None {from=MaterialExpressionMultiply_108}]}
  MaterialExpressionReroute_21 {class=MaterialExpressionReroute, inputs [None {from=MaterialExpressionReroute_22}]}
  MaterialExpressionReroute_22 {class=MaterialExpressionReroute, inputs [None {from=MaterialExpressionAdd_1}]}
  MaterialExpressionReroute_23 {class=MaterialExpressionReroute, inputs [None {from=MaterialExpressionReroute_22}]}
  MaterialExpressionReroute_24 {class=MaterialExpressionReroute, inputs [None {from=MaterialExpressionReroute_4}]}
  MaterialExpressionReroute_25 {class=MaterialExpressionReroute, inputs [None {from=MaterialExpressionDivide_9}]}
  MaterialExpressionReroute_26 {class=MaterialExpressionReroute, inputs [None {from=MaterialExpressionReroute_11}]}
  MaterialExpressionReroute_27 {class=MaterialExpressionReroute, inputs [None {from=MaterialExpressionLinearInterpolate_26}]}
  MaterialExpressionReroute_28 {class=MaterialExpressionReroute, inputs [None {from=MaterialExpressionPreSkinnedNormal_0}]}
  MaterialExpressionReroute_29 {class=MaterialExpressionReroute, inputs [None {from=MaterialExpressionOneMinus_15}]}
  MaterialExpressionReroute_3 {class=MaterialExpressionReroute, inputs [None {from=MaterialExpressionReroute_48}]}
  MaterialExpressionReroute_4 {class=MaterialExpressionReroute, inputs [None {from=MaterialExpressionScalarParameter_10}]}
  MaterialExpressionReroute_48 {class=MaterialExpressionReroute, inputs [None {from=MaterialExpressionAdd_44}]}
  MaterialExpressionReroute_5 {class=MaterialExpressionReroute, inputs [None {from=MaterialExpressionTextureSampleParameter2D_15, output=R}]}
  MaterialExpressionReroute_6 {class=MaterialExpressionReroute, inputs [None {from=MaterialExpressionMultiply_11}]}
  MaterialExpressionReroute_63 {class=MaterialExpressionReroute, inputs [None {from=MaterialExpressionAdd_45}]}
  MaterialExpressionReroute_64 {class=MaterialExpressionReroute, inputs [None {from=MaterialExpressionReroute_23}]}
  MaterialExpressionReroute_66 {class=MaterialExpressionReroute, inputs [None {from=MaterialExpressionMultiply_4}]}
  MaterialExpressionReroute_7 {class=MaterialExpressionReroute, inputs [None {from=MaterialExpressionSmoothStep_2}]}
  MaterialExpressionReroute_70 {class=MaterialExpressionReroute, inputs [None {from=MaterialExpressionReroute_48}]}
  MaterialExpressionReroute_77 {class=MaterialExpressionReroute, inputs [None {from=MaterialExpressionReroute_80}]}
  MaterialExpressionReroute_78 {class=MaterialExpressionReroute, inputs [None {from=MaterialExpressionReroute_83}]}
  MaterialExpressionReroute_79 {class=MaterialExpressionReroute, inputs [None {from=MaterialExpressionReroute_81}]}
  MaterialExpressionReroute_8 {class=MaterialExpressionReroute, inputs [None {from=MaterialExpressionSaturate_7}]}
  MaterialExpressionReroute_80 {class=MaterialExpressionReroute, inputs [None {from=MaterialExpressionReroute_82}]}
  MaterialExpressionReroute_81 {class=MaterialExpressionReroute, inputs [None {from=MaterialExpressionNamedRerouteUsage_8}]}
  MaterialExpressionReroute_82 {class=MaterialExpressionReroute, inputs [None {from=MaterialExpressionReroute_78}]}
  MaterialExpressionReroute_83 {class=MaterialExpressionReroute, inputs [None {from=MaterialExpressionMultiply_24}]}
  MaterialExpressionReroute_9 {class=MaterialExpressionReroute, inputs [None {from=MaterialExpressionSaturate_7}]}
  MaterialExpressionSaturate_0 {class=MaterialExpressionSaturate, inputs [None {from=MaterialExpressionAdd_101}]}
  MaterialExpressionSaturate_10 {class=MaterialExpressionSaturate, inputs [None {from=MaterialExpressionAdd_5}]}
  MaterialExpressionSaturate_11 {class=MaterialExpressionSaturate, inputs [None {from=MaterialExpressionSphereMask_0}]}
  MaterialExpressionSaturate_15 {class=MaterialExpressionSaturate, inputs [None {from=MaterialExpressionAdd_54}]}
  MaterialExpressionSaturate_16 {class=MaterialExpressionSaturate, inputs [None {from=MaterialExpressionMultiply_22}]}
  MaterialExpressionSaturate_17 {class=MaterialExpressionSaturate, inputs [None {from=MaterialExpressionAdd_62}]}
  MaterialExpressionSaturate_2 {class=MaterialExpressionSaturate, inputs [None {from=MaterialExpressionMultiply_44}]}
  MaterialExpressionSaturate_3 {class=MaterialExpressionSaturate, inputs [None {from=MaterialExpressionLinearInterpolate_10}]}
  MaterialExpressionSaturate_4 {class=MaterialExpressionSaturate, inputs [None {from=MaterialExpressionSubtract_17}]}
  MaterialExpressionSaturate_5 {class=MaterialExpressionSaturate, inputs [None {from=MaterialExpressionMax_4}]}
  MaterialExpressionSaturate_6 {class=MaterialExpressionSaturate, inputs [None {from=MaterialExpressionAdd_39}]}
  MaterialExpressionSaturate_7 {class=MaterialExpressionSaturate, inputs [None {from=MaterialExpressionAdd_0}]}
  MaterialExpressionSaturate_73 {class=MaterialExpressionSaturate, inputs [None {from=MaterialExpressionMultiply_89}]}
  MaterialExpressionSaturate_79 {class=MaterialExpressionSaturate, inputs [None {from=MaterialExpressionAdd_67}]}
  MaterialExpressionSaturate_8 {class=MaterialExpressionSaturate, inputs [None {from=MaterialExpressionAdd_82}]}
  MaterialExpressionSaturate_9 {class=MaterialExpressionSaturate, inputs [None {from=MaterialExpressionDivide_16}]}
  MaterialExpressionScalarParameter_0 {class=MaterialExpressionScalarParameter, properties {defaultValue=1, parameterName=SpriteVar_Cube_Emissive_InnerGlow_Mult}}
  MaterialExpressionScalarParameter_1 {
   class=MaterialExpressionScalarParameter
   properties {defaultValue=0.3, desc=2, group=SpriteVariants, parameterName=SpriteVar_Cube_RuneSecondary}
  }
  MaterialExpressionScalarParameter_10 {
   class=MaterialExpressionScalarParameter
   properties {defaultValue=20, desc=20, group=SpriteVariants, parameterName=SpriteVar_Cube_RuneBig_Scale}
  }
  MaterialExpressionScalarParameter_101 {class=MaterialExpressionScalarParameter, properties {defaultValue=5, desc=75, group=SpriteVariants, parameterName=SpritesVar_Hotspot}}
  MaterialExpressionScalarParameter_11 {class=MaterialExpressionScalarParameter, properties {defaultValue=0.4, parameterName=SpriteVar_Cube_SphereMaskEdges_Hardness}}
  MaterialExpressionScalarParameter_12 {
   class=MaterialExpressionScalarParameter
   properties {defaultValue=0.6, desc=.29, group=Cube_OG, parameterName=SpriteVar_Cube_Roughness_Center}
  }
  MaterialExpressionScalarParameter_13 {
   class=MaterialExpressionScalarParameter
   properties {defaultValue=0.29, desc=.5, group=SpriteVariants, parameterName=SpriteVar_Cube_Roughness_Edge}
  }
  MaterialExpressionScalarParameter_133 {class=MaterialExpressionScalarParameter, properties {defaultValue=1, parameterName=Blend Tweak}}
  MaterialExpressionScalarParameter_134 {class=MaterialExpressionScalarParameter, properties {parameterName=Blend Tweak Desat}}
  MaterialExpressionScalarParameter_135 {class=MaterialExpressionScalarParameter, properties {parameterName=Cube Emissive Extra}}
  MaterialExpressionScalarParameter_14 {
   class=MaterialExpressionScalarParameter
   properties {defaultValue=0.06, desc=.1, group=SpriteVariants, parameterName=SpriteVar_Cube_RuneBig_OffsetX}
  }
  MaterialExpressionScalarParameter_142 {
   class=MaterialExpressionScalarParameter
   properties {defaultValue=1, group=Sprite_Cosmetics | Cube, parameterName=Derivatives Use Pre-Skinned Scale, sliderMax=1, sortPriority=30}
  }
  MaterialExpressionScalarParameter_143 {
   class=MaterialExpressionScalarParameter
   properties {defaultValue=1, group=Sprite_Cosmetics | Cube, parameterName=Bump Offset Use Pre-Skinned Scale, sliderMax=1, sortPriority=10}
  }
  MaterialExpressionScalarParameter_144 {
   class=MaterialExpressionScalarParameter
   properties {defaultValue=1, group=Sprite_Cosmetics | Cube, parameterName=Bump Offset Mult, sortPriority=20}
  }
  MaterialExpressionScalarParameter_145 {
   class=MaterialExpressionScalarParameter
   properties {defaultValue=2, group=Sprite_Cosmetics | Cube, parameterName=Derivatives Mult, sortPriority=40}
  }
  MaterialExpressionScalarParameter_15 {class=MaterialExpressionScalarParameter, properties {defaultValue=0.8, parameterName=SpriteVar_Cube_SphereMaskEdges_Radius}}
  MaterialExpressionScalarParameter_16 {
   class=MaterialExpressionScalarParameter
   properties {defaultValue=0.1, group=SpriteVariants, parameterName=SpriteVar_Cube_RuneBig_BumpOffset_Intensity}
  }
  MaterialExpressionScalarParameter_165 {
   class=MaterialExpressionScalarParameter
   properties {defaultValue=1, group=SpriteVariants, parameterName=SpriteVar_Cube_RunePatternAlpha_Big/Small}
  }
  MaterialExpressionScalarParameter_17 {class=MaterialExpressionScalarParameter, properties {defaultValue=0.25, parameterName=SpriteVar_Cube_Specular_Center}}
  MaterialExpressionScalarParameter_18 {class=MaterialExpressionScalarParameter, properties {defaultValue=20, group=Cube_OG, parameterName=SpriteVar_Cube_RuneBig_Power}}
  MaterialExpressionScalarParameter_19 {
   class=MaterialExpressionScalarParameter
   properties {defaultValue=0.06, desc=.05, group=SpriteVariants, parameterName=SpriteVar_Cube_RuneBig_OffsetY}
  }
  MaterialExpressionScalarParameter_192 {class=MaterialExpressionScalarParameter, properties {defaultValue=0.87, parameterName=gr334}}
  MaterialExpressionScalarParameter_193 {class=MaterialExpressionScalarParameter, properties {defaultValue=5, desc=20, group=SpriteVariants, parameterName=SpritesVar_Hotspot_gpow3}}
  MaterialExpressionScalarParameter_196 {class=MaterialExpressionScalarParameter, properties {defaultValue=0.25, group=Cube_OG, parameterName=BaseCubeBackground}}
  MaterialExpressionScalarParameter_2 {class=MaterialExpressionScalarParameter, properties {defaultValue=10, parameterName=SpriteVar_Cube_InnerGlow_Radius1}}
  MaterialExpressionScalarParameter_20 {
   class=MaterialExpressionScalarParameter
   properties {
    defaultValue=2
    desc="(4) this value defines how many smaller runes exist behind each big one"
    group=SpriteVariants
    parameterName="SpriteVar_Cube_RuneSmall_Scale(subgrid)"
   }
  }
  MaterialExpressionScalarParameter_205 {class=MaterialExpressionScalarParameter, properties {defaultValue=0.5, group=Cube_OG, parameterName=PulseSpeed_3}}
  MaterialExpressionScalarParameter_206 {class=MaterialExpressionScalarParameter, properties {defaultValue=2, parameterName=SpriteVar_Cube_G_7}}
  MaterialExpressionScalarParameter_207 {class=MaterialExpressionScalarParameter, properties {defaultValue=10, parameterName=SpriteVar_Cube_G_8}}
  MaterialExpressionScalarParameter_209 {
   class=MaterialExpressionScalarParameter
   properties {defaultValue=2.5, group=Cube_OG, parameterName=SpriteVar_Cube_Emissive_RuneScrolling_InnerGlow_gpow3_1}
  }
  MaterialExpressionScalarParameter_21 {class=MaterialExpressionScalarParameter, properties {defaultValue=0.2, parameterName=SpriteVar_Cube_Specular_Edge}}
  MaterialExpressionScalarParameter_22 {
   class=MaterialExpressionScalarParameter
   properties {defaultValue=0.1, group=SpriteVariants, parameterName=SpriteVar_Cube_RuneSmall_BumpOffset_Intensity}
  }
  MaterialExpressionScalarParameter_23 {
   class=MaterialExpressionScalarParameter
   properties {defaultValue=1, group=SpriteVariants, parameterName=SpriteVar_Cube_RuneSmall_BumpOffset_Intensity_Flat}
  }
  MaterialExpressionScalarParameter_24 {class=MaterialExpressionScalarParameter, properties {defaultValue=20, parameterName=SpriteVar_Cube_FacialMask_ZOff}}
  MaterialExpressionScalarParameter_29 {class=MaterialExpressionScalarParameter, properties {defaultValue=1, group=Cube_OG, parameterName=Blend-Tex-Sphere}}
  MaterialExpressionScalarParameter_3 {
   class=MaterialExpressionScalarParameter
   properties {defaultValue=0.5, group=SpriteVariants, parameterName=SpriteVar_Cube_MarterNoise_BumpOffset}
  }
  MaterialExpressionScalarParameter_33 {
   class=MaterialExpressionScalarParameter
   properties {defaultValue=80, group=Sprite_Base, parameterName=SpriteVar_Cube_FacialMask_Radius, sortPriority=100}
  }
  MaterialExpressionScalarParameter_34 {
   class=MaterialExpressionScalarParameter
   properties {defaultValue=0.2, group=Sprite_Base, parameterName=SpriteVar_Cube_FacialMask_Hardness, sortPriority=100}
  }
  MaterialExpressionScalarParameter_4 {class=MaterialExpressionScalarParameter, properties {defaultValue=1.3, parameterName=SpriteVar_Cube_Innerglow}}
  MaterialExpressionScalarParameter_41 {
   class=MaterialExpressionScalarParameter
   properties {defaultValue=10, group=SpriteVariants_CubeAdvanced, parameterName=SpriteVar_Cube_RuneScrolling_G_1}
  }
  MaterialExpressionScalarParameter_42 {class=MaterialExpressionScalarParameter, properties {defaultValue=0.4, parameterName=LowQ-InnerGlowWhileOn}}
  MaterialExpressionScalarParameter_43 {class=MaterialExpressionScalarParameter, properties {defaultValue=4, group=Cube_OG, parameterName=Mipbias_Shadow}}
  MaterialExpressionScalarParameter_46 {class=MaterialExpressionScalarParameter, properties {defaultValue=150, group=SpriteVariants, parameterName=SpriteVar_Cube_StarEmissiveRange}}
  MaterialExpressionScalarParameter_47 {
   class=MaterialExpressionScalarParameter
   properties {defaultValue=1.25, group=SpriteVariants_CubeAdvanced, parameterName=SpriteVar_Cube_RuneScrolling_G_3}
  }
  MaterialExpressionScalarParameter_48 {class=MaterialExpressionScalarParameter, properties {defaultValue=200, group=Cube_OG, parameterName=SpriteVar_Cube_StarFarIntensity}}
  MaterialExpressionScalarParameter_49 {class=MaterialExpressionScalarParameter, properties {defaultValue=400, group=Cube_OG, parameterName=StarEmissiveRange02}}
  MaterialExpressionScalarParameter_5 {class=MaterialExpressionScalarParameter, properties {defaultValue=3, parameterName=SpriteVar_Cube_InnerGlow_Radius}}
  MaterialExpressionScalarParameter_50 {class=MaterialExpressionScalarParameter, properties {defaultValue=1, group=Cube_OG, parameterName=SpriteVar_Cube_RuneScrolling_MinGrid}}
  MaterialExpressionScalarParameter_52 {class=MaterialExpressionScalarParameter, properties {defaultValue=10, group=Cube_Additional, parameterName=ZUnit}}
  MaterialExpressionScalarParameter_53 {class=MaterialExpressionScalarParameter, properties {defaultValue=0.6, desc=0.6, group=Cube_Additional, parameterName=ZRoomScale}}
  MaterialExpressionScalarParameter_54 {class=MaterialExpressionScalarParameter, properties {defaultValue=3, group=Cube_Additional, parameterName=ZPow}}
  MaterialExpressionScalarParameter_55 {class=MaterialExpressionScalarParameter, properties {defaultValue=-1, group=Cube_OG, parameterName=SpriteVar_Cube_RuneScrolling_MaxGrid}}
  MaterialExpressionScalarParameter_56 {
   class=MaterialExpressionScalarParameter
   properties {defaultValue=0.0135, group=Cube_OG, parameterName=SpriteVar_Cube_RuneScrolling_Mask_ScrollSpeed}
  }
  MaterialExpressionScalarParameter_57 {
   class=MaterialExpressionScalarParameter
   properties {defaultValue=-0.03, desc=-1, group=Cube_OG, parameterName=SpriteVar_Cube_RuneScrolling_RuneHeightMin}
  }
  MaterialExpressionScalarParameter_58 {
   class=MaterialExpressionScalarParameter
   properties {defaultValue=0.1, group=Cube_OG, parameterName=SpriteVar_Cube_RuneScrolling_RuneHeightMax}
  }
  MaterialExpressionScalarParameter_6 {class=MaterialExpressionScalarParameter, properties {defaultValue=5, group=SpriteVariants, parameterName=SpriteVar_Cube_Sparkle1_Scale}}
  MaterialExpressionScalarParameter_60 {
   class=MaterialExpressionScalarParameter
   properties {defaultValue=1, group=SpriteVariants_CubeAdvanced, parameterName=SpriteVar_Cube_RuneScrolling_G_6}
  }
  MaterialExpressionScalarParameter_61 {class=MaterialExpressionScalarParameter, properties {defaultValue=0.15, group=Cube_Additional, parameterName=MinGradientColor}}
  MaterialExpressionScalarParameter_62 {
   class=MaterialExpressionScalarParameter
   properties {defaultValue=5, group=SpriteVariants_CubeAdvanced, parameterName=SpriteVar_Cube_RuneScrolling_G_9}
  }
  MaterialExpressionScalarParameter_63 {class=MaterialExpressionScalarParameter, properties {defaultValue=0.8, group=Cube_Additional, parameterName=MinCornerColor}}
  MaterialExpressionScalarParameter_64 {class=MaterialExpressionScalarParameter, properties {defaultValue=0.5, group=Cube_Additional, parameterName=CornerPow}}
  MaterialExpressionScalarParameter_65 {
   class=MaterialExpressionScalarParameter
   properties {defaultValue=2, desc=2, group=SpriteVariants, parameterName=SpriteVar_Cube_RuneScrolling_Scale}
  }
  MaterialExpressionScalarParameter_66 {class=MaterialExpressionScalarParameter, properties {defaultValue=2, group=Cube_Additional, parameterName=OverallEmissive}}
  MaterialExpressionScalarParameter_69 {class=MaterialExpressionScalarParameter, properties {defaultValue=50, parameterName=a_radius}}
  MaterialExpressionScalarParameter_7 {class=MaterialExpressionScalarParameter, properties {defaultValue=1, group=SpriteVariants, parameterName=SpriteVar_Cube_Sparkle1_BumpOffset}}
  MaterialExpressionScalarParameter_70 {class=MaterialExpressionScalarParameter, properties {defaultValue=0.01, parameterName=RadiusMult}}
  MaterialExpressionScalarParameter_71 {class=MaterialExpressionScalarParameter, properties {defaultValue=75, parameterName=ActiveHotspot}}
  MaterialExpressionScalarParameter_72 {class=MaterialExpressionScalarParameter, properties {defaultValue=0.2, group=Cube_OG, parameterName=Metal}}
  MaterialExpressionScalarParameter_73 {class=MaterialExpressionScalarParameter, properties {defaultValue=1, group=Cube_OG, parameterName=MetalExp}}
  MaterialExpressionScalarParameter_74 {class=MaterialExpressionScalarParameter, properties {defaultValue=0.4, group=SpriteVariants, parameterName=SpriteVar_Cube_Roughness}}
  MaterialExpressionScalarParameter_75 {
   class=MaterialExpressionScalarParameter
   properties {defaultValue=0.5, desc=0.2, group=SpriteVariants, parameterName=SpriteVar_Cube_Metallic}
  }
  MaterialExpressionScalarParameter_77 {class=MaterialExpressionScalarParameter, properties {defaultValue=30, group=SpriteVariants, parameterName=SpriteVar_Cube_MasterNoise_Scale}}
  MaterialExpressionScalarParameter_78 {class=MaterialExpressionScalarParameter, properties {defaultValue=0.87, parameterName=gr334}}
  MaterialExpressionScalarParameter_79 {class=MaterialExpressionScalarParameter, properties {defaultValue=20, parameterName=gpow3}}
  MaterialExpressionScalarParameter_8 {class=MaterialExpressionScalarParameter, properties {defaultValue=0.6, group=SpriteVariants, parameterName=SpritesVar_Hotspot_Noise}}
  MaterialExpressionScalarParameter_80 {class=MaterialExpressionScalarParameter, properties {defaultValue=1, group=Cube_OG, parameterName=Blend-Tex-Sphere}}
  MaterialExpressionScalarParameter_9 {class=MaterialExpressionScalarParameter, properties {defaultValue=1, group=SpriteVariants, parameterName=SpriteVar_Cube_Sparkle2_BumpOffset}}
  MaterialExpressionSetMaterialAttributes_0 {
   class=MaterialExpressionSetMaterialAttributes
   inputs [
    MaterialAttributes {from=MaterialExpressionNamedRerouteUsage_47, output=""}
    BaseColor {from=MaterialExpressionNamedRerouteUsage_90, output=""}
    EmissiveColor {from=MaterialExpressionNamedRerouteUsage_108, output=""}
    Metallic {from=MaterialExpressionNamedRerouteUsage_118, output=""}
    Specular {from=MaterialExpressionNamedRerouteUsage_89, output=""}
    Roughness {from=MaterialExpressionNamedRerouteUsage_91, output=""}
    Normal {from="", output=""}
   ]
   properties {attributeSetTypes [
     #2
     #6
     #3
     #5
     #4
     #1
    ]}
  }
  MaterialExpressionShaderStageSwitch_0 {
   class=MaterialExpressionShaderStageSwitch
   inputs [
    PixelShader {from=MaterialExpressionVertexInterpolator_0, output=PS}
    VertexShader {from=MaterialExpressionReroute_28, output=""}
   ]
   properties {}
  }
  MaterialExpressionShaderStageSwitch_4 {
   class=MaterialExpressionShaderStageSwitch
   inputs [
    PixelShader {from=MaterialExpressionVertexInterpolator_4, output=PS}
    VertexShader {from=MaterialExpressionLocalPosition_5, output=XYZ}
   ]
   properties {}
  }
  MaterialExpressionSine_1 {class=MaterialExpressionSine, inputs [None {from=MaterialExpressionDivide_6}]}
  MaterialExpressionSmoothStep_0 {
   class=MaterialExpressionSmoothStep
   inputs [
    Min {from="", output=""}
    Max {from="", output=""}
    Value {from=MaterialExpressionMax_5, output=""}
   ]
   properties {constMax=0, constMin=0.6}
  }
  MaterialExpressionSmoothStep_1 {
   class=MaterialExpressionSmoothStep
   inputs [
    Min {from="", output=""}
    Max {from="", output=""}
    Value {from=MaterialExpressionMax_5, output=""}
   ]
   properties {constMax=0.6}
  }
  MaterialExpressionSmoothStep_2 {
   class=MaterialExpressionSmoothStep
   inputs [
    Min {from="", output=""}
    Max {from="", output=""}
    Value {from=MaterialExpressionNamedRerouteDeclaration_20, output=""}
   ]
   properties {constMax=0.7}
  }
  MaterialExpressionSmoothStep_3 {
   class=MaterialExpressionSmoothStep
   inputs [
    Min {from="", output=""}
    Max {from="", output=""}
    Value {from=MaterialExpressionLinearInterpolate_4, output=""}
   ]
   properties {constMin=0.2}
  }
  MaterialExpressionSmoothStep_4 {
   class=MaterialExpressionSmoothStep
   inputs [
    Min {from="", output=""}
    Max {from="", output=""}
    Value {from=MaterialExpressionNamedRerouteUsage_38, output=""}
   ]
   properties {constMax=0, constMin=0.2}
  }
  MaterialExpressionSmoothStep_8 {
   class=MaterialExpressionSmoothStep
   inputs [
    Min {from="", output=""}
    Max {from="", output=""}
    Value {from=MaterialExpressionNamedRerouteDeclaration_20, output=""}
   ]
   properties {constMax=0.2}
  }
  MaterialExpressionSphereMask_0 {
   class=MaterialExpressionSphereMask
   inputs [
    A {from=MaterialExpressionMultiply_34, output=""}
    B {from=MaterialExpressionConstant3Vector_3, output=""}
    Radius {from=MaterialExpressionScalarParameter_33, output=""}
    Hardness {from=MaterialExpressionScalarParameter_34, output=""}
   ]
   properties {attenuationRadius=13, hardnessPercent=15}
  }
  MaterialExpressionSphereMask_1 {
   class=MaterialExpressionSphereMask
   inputs [
    A {from=MaterialExpressionDivide_18, output=""}
    B {from=MaterialExpressionConstant3Vector_4, output=""}
    Radius {from=MaterialExpressionScalarParameter_15, output=""}
    Hardness {from=MaterialExpressionScalarParameter_11, output=""}
   ]
   properties {attenuationRadius=0.65, hardnessPercent=50}
  }
  MaterialExpressionSquareRoot_0 {class=MaterialExpressionSquareRoot, inputs [None {from=MaterialExpressionMax_6, output=""}], properties {}}
  MaterialExpressionSquareRoot_1 {class=MaterialExpressionSquareRoot, inputs [None {from=MaterialExpressionSubtract_16, output=""}], properties {}}
  MaterialExpressionSquareRoot_12 {class=MaterialExpressionSquareRoot, inputs [None {from=MaterialExpressionMax_2, output=""}], properties {}}
  MaterialExpressionStaticSwitchParameter_1 {
   class=MaterialExpressionStaticSwitchParameter
   inputs [
    True {from=MaterialExpressionSetMaterialAttributes_0, output=""}
    False {from=MaterialExpressionNamedRerouteUsage_47, output=""}
   ]
   properties {group=SpriteVariants, parameterName=Sprite_Variant_Cube, sortPriority=-1}
  }
  MaterialExpressionStep_1 {
   class=MaterialExpressionStep
   inputs [
    Y {from="", output=""}
    X {from=MaterialExpressionNamedRerouteDeclaration_28, output=""}
   ]
   properties {constY=0.6}
  }
  MaterialExpressionStep_2 {
   class=MaterialExpressionStep
   inputs [
    Y {from="", output=""}
    X {from=MaterialExpressionDivide_15, output=""}
   ]
   properties {constY=0.8}
  }
  MaterialExpressionStep_4 {
   class=MaterialExpressionStep
   inputs [
    Y {from="", output=""}
    X {from=MaterialExpressionDivide_5, output=""}
   ]
   properties {constY=0.8}
  }
  MaterialExpressionStep_6 {
   class=MaterialExpressionStep
   inputs [
    Y {from=MaterialExpressionSine_1, output=""}
    X {from=MaterialExpressionComponentMask_9, output=""}
   ]
   properties {}
  }
  MaterialExpressionSubtract_0 {
   class=MaterialExpressionSubtract
   inputs [
    A {from=MaterialExpressionReroute_25}
    B {from=MaterialExpressionMultiply_17}
   ]
  }
  MaterialExpressionSubtract_1 {
   class=MaterialExpressionSubtract
   inputs [
    A {from=MaterialExpressionTextureCoordinate_4}
    B {from=MaterialExpressionConstant2Vector_6}
   ]
  }
  MaterialExpressionSubtract_10 {
   class=MaterialExpressionSubtract
   inputs [
    A {from=MaterialExpressionMultiply_65}
    B {from=MaterialExpressionSubtract_11}
   ]
  }
  MaterialExpressionSubtract_11 {
   class=MaterialExpressionSubtract
   inputs [
    A {from=MaterialExpressionMultiply_92}
    B {from=MaterialExpressionMultiply_91}
   ]
  }
  MaterialExpressionSubtract_12 {
   class=MaterialExpressionSubtract
   inputs [
    A {from=MaterialExpressionCameraPositionWS_1}
    B {from=""}
   ]
  }
  MaterialExpressionSubtract_13 {
   class=MaterialExpressionSubtract
   inputs [
    A {from=MaterialExpressionObjectPositionWS_2}
    B {from=MaterialExpressionConvert_0}
   ]
  }
  MaterialExpressionSubtract_14 {
   class=MaterialExpressionSubtract
   inputs [
    A {from=MaterialExpressionCameraPositionWS_2}
    B {from=MaterialExpressionSubtract_13}
   ]
  }
  MaterialExpressionSubtract_15 {
   class=MaterialExpressionSubtract
   inputs [
    A {from=MaterialExpressionObjectPositionWS_1}
    B {from=MaterialExpressionObjectBounds_0}
   ]
  }
  MaterialExpressionSubtract_16 {
   class=MaterialExpressionSubtract
   inputs [
    A {from=MaterialExpressionMultiply_93}
    B {from=MaterialExpressionMultiply_96}
   ]
  }
  MaterialExpressionSubtract_17 {
   class=MaterialExpressionSubtract
   inputs [
    A {from=MaterialExpressionNamedRerouteUsage_43}
    B {from=MaterialExpressionPower_18}
   ]
  }
  MaterialExpressionSubtract_18 {
   class=MaterialExpressionSubtract
   inputs [
    A {from=MaterialExpressionNamedRerouteUsage_51}
    B {from=MaterialExpressionConvert_2}
   ]
  }
  MaterialExpressionSubtract_19 {
   class=MaterialExpressionSubtract
   inputs [
    A {from=MaterialExpressionNamedRerouteUsage_20}
    B {from=""}
   ]
   properties {constB=0.5}
  }
  MaterialExpressionSubtract_2 {
   class=MaterialExpressionSubtract
   inputs [
    A {from=MaterialExpressionConvert_3}
    B {from=MaterialExpressionMax_8}
   ]
  }
  MaterialExpressionSubtract_3 {
   class=MaterialExpressionSubtract
   inputs [
    A {from=MaterialExpressionConvert_3}
    B {from=MaterialExpressionMax_7}
   ]
  }
  MaterialExpressionSubtract_37 {
   class=MaterialExpressionSubtract
   inputs [
    A {from=MaterialExpressionMultiply_67}
    B {from=MaterialExpressionSubtract_38}
   ]
  }
  MaterialExpressionSubtract_38 {
   class=MaterialExpressionSubtract
   inputs [
    A {from=MaterialExpressionMultiply_71}
    B {from=MaterialExpressionMultiply_70}
   ]
  }
  MaterialExpressionSubtract_39 {
   class=MaterialExpressionSubtract
   inputs [
    A {from=MaterialExpressionCameraPositionWS_14}
    B {from=MaterialExpressionNamedRerouteUsage_158}
   ]
  }
  MaterialExpressionSubtract_40 {
   class=MaterialExpressionSubtract
   inputs [
    A {from=MaterialExpressionNamedRerouteUsage_36}
    B {from=MaterialExpressionNamedRerouteUsage_35}
   ]
  }
  MaterialExpressionSubtract_41 {
   class=MaterialExpressionSubtract
   inputs [
    A {from=MaterialExpressionNamedRerouteUsage_12}
    B {from=MaterialExpressionNamedRerouteUsage_149}
   ]
  }
  MaterialExpressionSubtract_5 {
   class=MaterialExpressionSubtract
   inputs [
    A {from=MaterialExpressionConvert_3}
    B {from=MaterialExpressionMax_9}
   ]
  }
  MaterialExpressionSubtract_6 {
   class=MaterialExpressionSubtract
   inputs [
    A {from=MaterialExpressionReroute_2}
    B {from=MaterialExpressionNamedRerouteUsage_123}
   ]
  }
  MaterialExpressionSubtract_7 {
   class=MaterialExpressionSubtract
   inputs [
    A {from=MaterialExpressionReroute_0}
    B {from=MaterialExpressionMultiply_0}
   ]
  }
  MaterialExpressionSubtract_8 {
   class=MaterialExpressionSubtract
   inputs [
    A {from=MaterialExpressionCameraPositionWS_0}
    B {from=MaterialExpressionObjectPositionWS_1}
   ]
  }
  MaterialExpressionSubtract_9 {
   class=MaterialExpressionSubtract
   inputs [
    A {from=MaterialExpressionMaterialFunctionCall_59, output=StabilizedUVs}
    B {from=MaterialExpressionLinearInterpolate_25}
   ]
  }
  MaterialExpressionTemporalResponsivenessOutput_0 {class=MaterialExpressionTemporalResponsivenessOutput, inputs [None {from=MaterialExpressionConstant_31, output=""}], properties {}}
  MaterialExpressionTextureCoordinate_0 {class=MaterialExpressionTextureCoordinate}
  MaterialExpressionTextureCoordinate_4 {class=MaterialExpressionTextureCoordinate}
  MaterialExpressionTextureObjectParameter_0 {
   class=MaterialExpressionTextureObjectParameter
   inputs [
    UVs {from="", output=""}
    Apply View MipBias {from="", output=""}
   ]
   properties {
    group=Cube_OG
    parameterName=CubeNoiseTexture
    samplerType=SAMPLERTYPE_Masks
    texture=/BRCosmetics/Animation/Props/Gadgets_PixieParts/Textures/TA/T_CubeNoise_Relic
   }
  }
  MaterialExpressionTextureSampleParameter2D_10 {
   class=MaterialExpressionTextureSampleParameter2D
   inputs [
    UVs {from=MaterialExpressionMaterialFunctionCall_45, output=UVs}
    "DDX(UVs)" {from=MaterialExpressionNamedRerouteUsage_198, output=""}
    "DDY(UVs)" {from=MaterialExpressionNamedRerouteUsage_199, output=""}
    Apply View MipBias {from="", output=""}
   ]
   properties {
    group=Cube_OG
    mipValueMode=TMVM_Derivative
    parameterName=RuneTex
    samplerSource=SSM_Wrap_WorldGroupSettings
    samplerType=SAMPLERTYPE_LinearColor
    texture=/Game/Athena/Prototype/Blueprints/Cube/Materials/DFHeightMask
   }
  }
  MaterialExpressionTextureSampleParameter2D_15 {
   class=MaterialExpressionTextureSampleParameter2D
   inputs [
    UVs {from=MaterialExpressionMaterialFunctionCall_57, output=Add}
    "DDX(UVs)" {from=MaterialExpressionNamedRerouteUsage_204, output=""}
    "DDY(UVs)" {from=MaterialExpressionNamedRerouteUsage_205, output=""}
    Apply View MipBias {from="", output=""}
   ]
   properties {
    group=Cube_OG
    mipValueMode=TMVM_Derivative
    parameterName=RuneTex
    samplerSource=SSM_Wrap_WorldGroupSettings
    samplerType=SAMPLERTYPE_LinearColor
    texture=&DFHeightMask
   }
  }
  MaterialExpressionTextureSampleParameter2D_16 {
   class=MaterialExpressionTextureSampleParameter2D
   inputs [
    UVs {from=MaterialExpressionMaterialFunctionCall_56, output=UVs}
    "DDX(UVs)" {from=MaterialExpressionNamedRerouteUsage_202, output=""}
    "DDY(UVs)" {from=MaterialExpressionNamedRerouteUsage_203, output=""}
    Apply View MipBias {from="", output=""}
   ]
   properties {
    group=Cube_OG
    mipValueMode=TMVM_Derivative
    parameterName=RuneTex
    samplerSource=SSM_Wrap_WorldGroupSettings
    samplerType=SAMPLERTYPE_LinearColor
    texture=&DFHeightMask
   }
  }
  MaterialExpressionTextureSampleParameter2D_17 {
   class=MaterialExpressionTextureSampleParameter2D
   inputs [
    UVs {from=MaterialExpressionPanner_0, output=""}
    Apply View MipBias {from="", output=""}
   ]
   properties {
    group=Cube_OG
    parameterName=RuneTex
    samplerSource=SSM_Wrap_WorldGroupSettings
    samplerType=SAMPLERTYPE_LinearColor
    texture=&DFHeightMask
   }
  }
  MaterialExpressionTextureSampleParameter2D_7 {
   class=MaterialExpressionTextureSampleParameter2D
   inputs [
    UVs {from=MaterialExpressionMaterialFunctionCall_41, output=Add}
    "DDX(UVs)" {from=MaterialExpressionNamedRerouteUsage_196, output=""}
    "DDY(UVs)" {from=MaterialExpressionNamedRerouteUsage_197, output=""}
    Apply View MipBias {from="", output=""}
   ]
   properties {
    group=Cube_OG
    mipValueMode=TMVM_Derivative
    parameterName=RuneTex
    samplerSource=SSM_Wrap_WorldGroupSettings
    samplerType=SAMPLERTYPE_LinearColor
    texture=&DFHeightMask
   }
  }
  MaterialExpressionTextureSampleParameter2D_8 {
   class=MaterialExpressionTextureSampleParameter2D
   inputs [
    UVs {from=MaterialExpressionReroute_14, output=""}
    "DDX(UVs)" {from=MaterialExpressionNamedRerouteUsage_194, output=""}
    "DDY(UVs)" {from=MaterialExpressionNamedRerouteUsage_195, output=""}
    Apply View MipBias {from="", output=""}
   ]
   properties {
    group=Cube_OG
    mipValueMode=TMVM_Derivative
    parameterName=RuneTex
    samplerSource=SSM_Wrap_WorldGroupSettings
    samplerType=SAMPLERTYPE_LinearColor
    texture=&DFHeightMask
   }
  }
  MaterialExpressionTextureSampleParameter2D_9 {
   class=MaterialExpressionTextureSampleParameter2D
   inputs [
    UVs {from=MaterialExpressionMaterialFunctionCall_47, output=Add}
    "DDX(UVs)" {from=MaterialExpressionNamedRerouteUsage_200, output=""}
    "DDY(UVs)" {from=MaterialExpressionNamedRerouteUsage_201, output=""}
    Apply View MipBias {from="", output=""}
   ]
   properties {
    group=Cube_OG
    mipValueMode=TMVM_Derivative
    parameterName=RuneTex
    samplerSource=SSM_Wrap_WorldGroupSettings
    samplerType=SAMPLERTYPE_LinearColor
    texture=&DFHeightMask
   }
  }
  MaterialExpressionTextureSample_37 {
   class=MaterialExpressionTextureSample
   inputs [
    UVs {from=MaterialExpressionMaterialFunctionCall_35, output=Add}
    Tex {from=MaterialExpressionNamedRerouteUsage_87}
    "DDX(UVs)" {from=MaterialExpressionNamedRerouteUsage_192}
    "DDY(UVs)" {from=MaterialExpressionNamedRerouteUsage_193}
    Apply View MipBias {from=""}
   ]
   properties {
    mipValueMode=TMVM_Derivative
    samplerSource=SSM_Wrap_WorldGroupSettings
    samplerType=SAMPLERTYPE_Masks
    texture=/Game/Global/Textures/Default/Noise/T_CubeNoise
   }
  }
  MaterialExpressionTextureSample_39 {
   class=MaterialExpressionTextureSample
   inputs [
    UVs {from=MaterialExpressionMaterialFunctionCall_37, output=Add}
    Tex {from=MaterialExpressionNamedRerouteUsage_87}
    "DDX(UVs)" {from=MaterialExpressionNamedRerouteUsage_192}
    "DDY(UVs)" {from=MaterialExpressionNamedRerouteUsage_193}
    Apply View MipBias {from=""}
   ]
   properties {
    mipValueMode=TMVM_Derivative
    samplerSource=SSM_Wrap_WorldGroupSettings
    samplerType=SAMPLERTYPE_Masks
    texture=&T_CubeNoise
   }
  }
  MaterialExpressionTextureSample_73 {
   class=MaterialExpressionTextureSample
   inputs [
    UVs {from=MaterialExpressionMaterialFunctionCall_21, output=UVs}
    Tex {from=MaterialExpressionNamedRerouteUsage_13}
    "DDX(UVs)" {from=MaterialExpressionNamedRerouteUsage_190}
    "DDY(UVs)" {from=MaterialExpressionNamedRerouteUsage_191}
    Apply View MipBias {from=""}
   ]
   properties {
    mipValueMode=TMVM_Derivative
    samplerSource=SSM_Wrap_WorldGroupSettings
    samplerType=SAMPLERTYPE_Masks
    texture=&T_CubeNoise
   }
  }
  MaterialExpressionTextureSample_74 {
   class=MaterialExpressionTextureSample
   inputs [
    UVs {from=MaterialExpressionMaterialFunctionCall_31, output=Add}
    Tex {from=""}
    "DDX(UVs)" {from=MaterialExpressionMultiply_152}
    "DDY(UVs)" {from=MaterialExpressionMultiply_155}
    Apply View MipBias {from=""}
   ]
   properties {mipValueMode=TMVM_Derivative, samplerType=SAMPLERTYPE_Masks, texture=&T_CubeNoise}
  }
  MaterialExpressionTime_0 {class=MaterialExpressionTime}
  MaterialExpressionTime_2 {class=MaterialExpressionTime}
  MaterialExpressionTime_9 {class=MaterialExpressionTime}
  MaterialExpressionTransform_0 {class=MaterialExpressionTransform, inputs [None {from=MaterialExpressionNamedRerouteUsage_4}], properties {transformType=TRANSFORM_Local}}
  MaterialExpressionTransform_2 {
   class=MaterialExpressionTransform
   inputs [None {from=MaterialExpressionMaterialFunctionCall_6, output=Result}]
   properties {transformSourceType=TRANSFORMSOURCE_World, transformType=TRANSFORM_Local}
  }
  MaterialExpressionVectorParameter_0 {class=MaterialExpressionVectorParameter, properties {parameterName=SpriteVar_Cube_SphereMaskEdgesSize, defaultValue=<rgba 1.1 1.4 1.2 1>}}
  MaterialExpressionVectorParameter_1 {
   class=MaterialExpressionVectorParameter
   properties {group=SpriteVariants, parameterName=SpriteVar_Cube_Emissive_Logo, defaultValue=<rgba 0.1546 0.0464 0.2319 1>}
  }
  MaterialExpressionVectorParameter_11 {
   class=MaterialExpressionVectorParameter
   properties {group=SpriteVariants, parameterName=SpriteVar_Cube_Emissive_GlowColor2, defaultValue=<rgba 0.0356 0.001 0.2 1>}
  }
  MaterialExpressionVectorParameter_2 {
   class=MaterialExpressionVectorParameter
   properties {group=SpriteVariants, parameterName=SpriteVar_Cube_Emissive_BlinkingLights, defaultValue=<rgba 0.4 0.0116 0.389 1>}
  }
  MaterialExpressionVectorParameter_25 {
   class=MaterialExpressionVectorParameter
   properties {group=SpriteVariants, parameterName=SpriteVar_Cube_Emissive_GlowColor3, defaultValue=<rgba 0.0211 0 0.05 1>}
  }
  MaterialExpressionVectorParameter_27 {
   class=MaterialExpressionVectorParameter
   properties {group=SpriteVariants, parameterName=SpriteVar_Cube_Emissive_GlowColor4, defaultValue=<rgba 0.0717 0 0.1 1>}
  }
  MaterialExpressionVectorParameter_29 {
   class=MaterialExpressionVectorParameter
   properties {group=SpriteVariants, parameterName=SpriteVar_Cube_Emissive_GlowColor1, defaultValue=<rgba 0.012 0.0011 0.08 1>}
  }
  MaterialExpressionVectorParameter_30 {class=MaterialExpressionVectorParameter, properties {group=Cube_OG, parameterName=BaseDiffuse, defaultValue=<rgba 0.011 0.0081 0.02 1>}}
  MaterialExpressionVectorParameter_31 {
   class=MaterialExpressionVectorParameter
   properties {group=SpriteVariants, parameterName=SpriteVar_Cube_Emissive_TOD_Morning, defaultValue=<rgba 1.5 1.2006 0.8739 1>}
  }
  MaterialExpressionVectorParameter_32 {
   class=MaterialExpressionVectorParameter
   properties {group=SpriteVariants, parameterName=SpriteVar_Cube_Emissive_TOD_Day, defaultValue=<rgba 1 1 1 1>}
  }
  MaterialExpressionVectorParameter_33 {
   class=MaterialExpressionVectorParameter
   properties {group=SpriteVariants, parameterName=SpriteVar_Cube_Emissive_TOD_Evening, defaultValue=<rgba 1.5 1.302 1.2652 1>}
  }
  MaterialExpressionVectorParameter_34 {
   class=MaterialExpressionVectorParameter
   properties {group=SpriteVariants, parameterName=SpriteVar_Cube_Emissive_TOD_Night, defaultValue=<rgba 1.38 1.828 2.3 1>}
  }
  MaterialExpressionVectorParameter_35 {
   class=MaterialExpressionVectorParameter
   properties {group=SpriteVariants, parameterName=SpriteVar_Cube_Emissive_SparklesColor, defaultValue=<rgba 0.9235 0.2 1 1>}
  }
  MaterialExpressionVectorParameter_36 {class=MaterialExpressionVectorParameter, properties {group=Cube_OG, parameterName=EdgeDiffuse, defaultValue=<rgba 0.0103 0.0087 0.02 1>}}
  MaterialExpressionVectorParameter_37 {class=MaterialExpressionVectorParameter, properties {group=Cube_OG, parameterName=offz, defaultValue=<rgba 0 0 0 1>}}
  MaterialExpressionVectorParameter_38 {class=MaterialExpressionVectorParameter, properties {desc=For zapping players, group=Cube_OG, parameterName=SphereOffset}}
  MaterialExpressionVectorParameter_39 {
   class=MaterialExpressionVectorParameter
   properties {group=SpriteVariants, parameterName=SpriteVar_Cube_Emissive_RuneScrolling, defaultValue=<rgba 1.7246 0.4565 3.5 1>}
  }
  MaterialExpressionVectorParameter_40 {class=MaterialExpressionVectorParameter, properties {group=Cube_Additional, parameterName=TilingUVW, defaultValue=<rgba 1 1 2 2>}}
  MaterialExpressionVertexInterpolator_0 {class=MaterialExpressionVertexInterpolator, inputs [VS {from=MaterialExpressionReroute_28, output=""}], properties {}}
  MaterialExpressionVertexInterpolator_4 {class=MaterialExpressionVertexInterpolator, inputs [VS {from=MaterialExpressionLocalPosition_5, output=XYZ}], properties {}}
  MaterialExpressionVertexNormalWS_7 {class=MaterialExpressionVertexNormalWS, inputs [], properties {}}
 }}
