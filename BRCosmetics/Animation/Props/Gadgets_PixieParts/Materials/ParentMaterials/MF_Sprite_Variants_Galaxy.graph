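# Material function for the Galaxy variant of Sprite cosmetic materials, gated by the Sprite_Variant_Galaxy static switch.
# Two hue-shifted nebula textures and a star layer use MF_TriplanarVirtualBox for scrolling depth, with pivot and bounds-based offset and tiling.
# Time-of-day brightness and desaturation, Fresnel glow, and variant, logo and blinking-light masks shape the emissive.
# Incoming material attributes are tinted or blended with galaxy surface parameters and re-emitted.
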MaterialFunction /BRCosmetics/Animation/Props/Gadgets_PixieParts/Materials/ParentMaterials/MF_Sprite_Variants_Galaxy
graph {nodes [
  Constant3Vector() {constant=<rgba 0.1502 0 1 1>}
  Constant4Vector() {constant=<rgba 5 5 10 1>}
  Constant4Vector() {constant=<rgba 0.6 0.75 1.5 1>}
  Constant()
  FunctionOutput(None=StaticSwitchParameter(True=NamedRerouteUsage(), False=NamedRerouteUsage()) {group=SpriteVariants, parameterName=Sprite_Variant_Galaxy}) {outputName=SetMaterialAttributes}
  n1 = GetMaterialAttributes(None=NamedRerouteDeclaration(None=FunctionInput() {inputName=Material, inputType=FunctionInput_MaterialAttributes, sortPriority=-1}) {name=IN_Material, nodeColor=<rgba 1 0 0.9609 1>}) {attributeGetTypes [
    #1
    #2
    #3
    #4
    #5
   ]}
  n2 = NamedRerouteDeclaration(None=ScalarParameter() {defaultValue=0.4, group=Galaxy, parameterName=UseIncomingMatParams}) {name=UseIncomingMatParams, nodeColor=<rgba 0.875 0 1 1>}
  LinearInterpolate(B=@n1.BaseColor, Alpha=@n2) {constB=0}
  n3 = VectorParameter() {
   group=SpriteVariants
   parameterName=SpriteVar_Galaxy_Color_DiffuseMask
   channelNames {a="0=Just Galaxy 1= Blend diffuse", b=DiffuseB Influence, g=DiffuseG Influence, r=DiffuseR Influence}
   defaultValue=<rgba 1 0 1 0.7>
  }
  n4 = NamedRerouteUsage()
  n5 = Multiply(A=@n4, B=DotProduct(A=@n4, B=@n3.RGB))
  n6 = Saturate(None=@n3."0=Just Galaxy 1= Blend diffuse")
  LinearInterpolate(A=NamedRerouteUsage(), B=LinearInterpolate(A=@n4, B=@n5, Alpha=@n6), Alpha=NamedRerouteUsage())
  n7 = Multiply(
   A=Fresnel(ExponentIn=ScalarParameter() {defaultValue=3.7, group=Galaxy, parameterName=EmissiveFresnelExponent, sortPriority=90}) {baseReflectFraction=0}
   B=ScalarParameter() {defaultValue=1, group=Galaxy, parameterName=EmissiveFresnelIntensity, sortPriority=92}
  )
  n8 = Add(A=NamedRerouteUsage())
  Multiply(
   A=@n7
   B=LinearInterpolate(B=Saturate(None=@n8), Alpha=ScalarParameter() {group=Galaxy, parameterName=UseEmissiveMaskOnFresnel, sortPriority=92}) {constA=1}
  )
  n9 = TextureObjectParameter() {
   group=Galaxy
   parameterName=Galaxy_01_Texture
   sortPriority=58
   texture=/Game/Characters/Player/Male/Medium/Bodies/M_Med_Soldier_04/Skins/Galaxy/Textures/T_FN_Nebula_style_blue
  }
  n10 = ScalarParameter() {defaultValue=80, group=Galaxy, parameterName=Galaxy_01_Depth, sortPriority=61}
  n11 = Multiply(A=ScalarParameter() {defaultValue=1, group=Galaxy, parameterName=Galaxy_01_Tile, sortPriority=62}, B=NamedRerouteUsage())
  n12 = ScalarParameter() {defaultValue=0.4, group=Galaxy, parameterName=Galaxy_01_BlendContrast, sortPriority=63}
  n13 = VectorParameter() {group=Galaxy, parameterName=Galaxy_01_Speed, sortPriority=64, defaultValue=<rgba 0.002 -0.0025 0 1>}
  n14 = ComponentMask(None=@n13.RGB) {g=true, r=true}
  n15 = Add(
   A=NamedRerouteDeclaration(None=VectorParameter() {group=Galaxy, parameterName=Galaxy_Offset, sortPriority=59, defaultValue=<rgba 0 0 0 1>}.RGB) {name=ActorOffset, nodeColor=<rgba 0.9688 1 0 1>}
   B=NamedRerouteUsage()
  )
  n16 = Multiply(
   A=MaterialFunctionCall(
    Hue Shift Percentage=ScalarParameter() {defaultValue=-0.03, group=Galaxy, parameterName=Galaxy_01_HueShift, sortPriority=64}
    Texture=MaterialFunctionCall!(Texture=@n9, Depth=@n10, Tile=@n11, BlendContrast=@n12, SpeedXY=@n14, ActorOffset=@n15, bUseLocalCameraVector, bUseDitherSingleSample) {materialFunction=/Game/Global/MaterialFunctions/MF_TriplanarVirtualBox}.Result
   ) {materialFunction=&HueShift}.Result
   B=ScalarParameter() {defaultValue=0.6, group=Galaxy, parameterName=Galaxy_01_Brightness, sortPriority=65}
  )
  n17 = ScalarParameter() {defaultValue=1.75, group=Galaxy, parameterName=Galaxy_01_Exp, sortPriority=66}
  NamedRerouteDeclaration(None=Desaturation(
    None=Power(Base=@n16, Exp=@n17)
    Fraction=ScalarParameter() {defaultValue=0.2, group=Galaxy, parameterName=Galaxy_01_Desaturation, sortPriority=67}
   )) {name=Galaxy_01, nodeColor=<rgba 1 0.2578 0 1>}
  n18 = VectorParameter() {group=SpriteVariants, parameterName=SpriteVar_Galaxy_LogoEmissive, defaultValue=<rgba 0.3321 0.3321 1.5 0.35>}
  n19 = Multiply(A=Multiply(A=NamedRerouteUsage(), B=@n18.RGB), B=OneMinus(None=NamedRerouteUsage()))
  n20 = NamedRerouteUsage()
  n21 = NamedRerouteUsage()
  n22 = LinearInterpolate(A=@n21, B=NamedRerouteUsage(), Alpha=NamedRerouteUsage())
  n23 = Power(Base=NamedRerouteUsage(), Exp=ScalarParameter() {defaultValue=1, parameterName=GalaxyVariantMaskPower})
  n24 = Add(A=Multiply(A=OneMinus(None=@n20), B=LinearInterpolate(A=@n21, B=@n22, Alpha=@n23)), B=NamedRerouteUsage())
  n25 = LinearInterpolate(A=Multiply(A=@n24, B=@n19), B=@n19, Alpha=@n18.A)
  n26 = NamedRerouteUsage()
  n27 = VectorParameter() {group=SpriteVariants, parameterName=SpriteVar_Galaxy_BlinkingLights, defaultValue=<rgba 0.4888 0.0957 1 1>}
  NamedRerouteDeclaration(None=Add(A=Add(A=@n24, B=@n25), B=Multiply(A=@n26, B=@n27.RGB))) {name=OUT_Emissive, nodeColor=<rgba 1 0 0.1875 1>}
  n28 = ScalarParameter() {defaultValue=1, parameterName=Emissive Blend Influence}
  n29 = Saturate(None=DotProduct(A=@n4, B=@n3.RGB))
  n30 = SetMaterialAttributes!(
   MaterialAttributes=NamedRerouteUsage()
   BaseColor=NamedRerouteUsage()
   Metallic=NamedRerouteUsage()
   Specular=NamedRerouteUsage()
   Roughness=NamedRerouteUsage()
   EmissiveColor=Multiply(A=LinearInterpolate(B=@n28, Alpha=@n29) {constA=1}, B=NamedRerouteUsage())
  ) {attributeSetTypes [
    #5
    #2
    #3
    #1
    #4
   ]}
  NamedRerouteDeclaration(None=@n30) {name=MaterialOUT, nodeColor=<rgba 0 1 0.625 1>}
  n31 = ScalarParameter() {group=Galaxy, parameterName=Galaxy_02_HueShift, sortPriority=74}
  n32 = TextureObjectParameter() {
   group=Galaxy
   parameterName=Galaxy_02_Texture
   sortPriority=70
   texture=/Game/Characters/Player/Male/Medium/Bodies/M_Med_Soldier_04/Skins/Galaxy/Textures/T_FN_Nebula_style2
  }
  n33 = VectorParameter() {group=Galaxy, parameterName=Galaxy_02_Speed, sortPriority=74, defaultValue=<rgba -0.002 -0.0065 0 1>}
  n34 = MaterialFunctionCall(
   Texture=@n32
   Depth=ScalarParameter() {defaultValue=35, group=Galaxy, parameterName=Galaxy_02_Depth, sortPriority=71}
   Tile=Multiply(A=ScalarParameter() {defaultValue=1, group=Galaxy, parameterName=Galaxy_02_Tile, sortPriority=72}, B=NamedRerouteUsage())
   BlendContrast=ScalarParameter() {defaultValue=1, group=Galaxy, parameterName=Galaxy_02_BlendContrast, sortPriority=73}
   SpeedXY=ComponentMask(None=@n33.RGB) {g=true, r=true}
   ActorOffset=Add(A=NamedRerouteUsage(), B=NamedRerouteUsage())
  ) {materialFunction=&MF_TriplanarVirtualBox}
  NamedRerouteDeclaration(None=Multiply(
    A=MaterialFunctionCall(Hue Shift Percentage=@n31, Texture=@n34.Result) {materialFunction=&HueShift}.Result
    B=ScalarParameter() {defaultValue=0.5, group=Galaxy, parameterName=Galaxy_02_Brightness, sortPriority=75}
   )) {name=Galaxy_02, nodeColor=<rgba 0.0781 0 1 1>}
  n35 = NamedRerouteUsage()
  n36 = NamedRerouteUsage()
  n37 = NamedRerouteDeclaration(None=Saturate(None=Add(A=@n35, B=@n36))) {name=Galaxies, nodeColor=<rgba 1 0 0.2578 1>}
  n38 = ChannelMaskParameter(None=@n37) {group=Galaxy, maskChannel=Blue, parameterName=StarMaskChannel, sortPriority=84, defaultValue=<rgba 0 0 1 0>}
  NamedRerouteDeclaration(None=Multiply(A=@n38.R, B=ScalarParameter() {defaultValue=3, group=Galaxy, parameterName=GalaxyStarMaskStrength, sortPriority=85}) {constB=2}) {name=StarMask, nodeColor=<rgba 0.9922 1 0 1>}
  n39 = LinearInterpolate(A=NamedRerouteUsage(), Alpha=NamedRerouteUsage()) {constB=0}
  n40 = Multiply(
   A=NamedRerouteUsage()
   B=MaterialFunctionCall(
    Value_Morning=ScalarParameter() {defaultValue=1, group=Galaxy, parameterName=Galaxy_Brightness_Morning, sortPriority=100}
    Value_Day=ScalarParameter() {defaultValue=1, group=Galaxy, parameterName=Galaxy_Brightness_Day, sortPriority=101}
    Value_Evening=ScalarParameter() {defaultValue=1, group=Galaxy, parameterName=Galaxy_Brightness_Evening, sortPriority=102}
    Value_Night=ScalarParameter() {defaultValue=1, group=Galaxy, parameterName=Galaxy_Brightness_Night, sortPriority=103}
   ) {materialFunction=&Scalar_Time_of_Day}.Result
  )
  n41 = MaterialFunctionCall(
   Value_Morning=ScalarParameter() {group=Galaxy, parameterName=Galaxy_Desat_Morning, sortPriority=110}
   Value_Day=ScalarParameter() {group=Galaxy, parameterName=Galaxy_Desat_Day, sortPriority=111}
   Value_Evening=ScalarParameter() {group=Galaxy, parameterName=Galaxy_Desat_Evening, sortPriority=112}
   Value_Night=ScalarParameter() {group=Galaxy, parameterName=Galaxy_Desat_Night, sortPriority=113}
  ) {materialFunction=&Scalar_Time_of_Day}
  NamedRerouteDeclaration(None=Add(A=Saturate(None=@n39), B=Desaturation(None=@n40, Fraction=@n41.Result))) {name=GalaxiesAndStars, nodeColor=<rgba 0 1 0.9062 1>}
  NamedRerouteDeclaration(None=LinearInterpolate(A=ScalarParameter() {defaultValue=0.2, group=Galaxy, parameterName=Galaxy_Roughness, sortPriority=52}, B=@n1.Roughness, Alpha=@n2) {constB=0}) {name=OUTRoughness, nodeColor=<rgba 0 0.4844 1 1>}
  NamedRerouteDeclaration(None=LinearInterpolate(A=ScalarParameter() {defaultValue=0.1, group=Galaxy, parameterName=Galaxy_Metallic, sortPriority=50}, B=@n1.Metallic, Alpha=@n2) {constB=0}) {name=OUTMetallic, nodeColor=<rgba 0 0.4844 1 1>}
  NamedRerouteDeclaration(None=LinearInterpolate(A=ScalarParameter() {defaultValue=0.5, group=Galaxy, parameterName=Galaxy_Spec, sortPriority=51}, B=@n1.Specular, Alpha=@n2) {constB=0}) {name=OUTSpec, nodeColor=<rgba 0 0.4844 1 1>}
  NamedRerouteDeclaration(None=@n1.EmissiveColor) {name=IN_Emissive, nodeColor=<rgba 0.4766 1 0 1>}
  NamedRerouteDeclaration(None=Multiply(A=@n5, B=@n6)) {name=OUTBaseColor, nodeColor=<rgba 1 0 0.6562 1>}
  NamedRerouteDeclaration(None=FunctionInput() {inputName=Mask_InnerRadialGlow, inputType=FunctionInput_Scalar}) {name=M_InnerRadialGlow, nodeColor=<rgba 1 0 0.7031 1>}
  NamedRerouteDeclaration(None=FunctionInput() {inputName=Mask Variant, inputType=FunctionInput_Scalar, sortPriority=1}) {name=M_VariantMask, nodeColor=<rgba 0 1 0.8828 1>}
  NamedRerouteDeclaration(None=FunctionInput() {inputName=Mask_SpriteFresnel, inputType=FunctionInput_Scalar, sortPriority=1}) {name=M_SpriteFresnel, nodeColor=<rgba 1 0.9844 0 1>}
  NamedRerouteDeclaration(None=FunctionInput() {inputName=Mask_Logo, inputType=FunctionInput_Scalar, sortPriority=1}) {name=M_Logo, nodeColor=<rgba 1 0.9844 0 1>}
  NamedRerouteDeclaration(None=FunctionInput() {inputName=WPO, inputType=FunctionInput_Scalar, sortPriority=2}) {name=IN_WPO, nodeColor=<rgba 0 0.9766 1 1>}
  NamedRerouteDeclaration(None=FunctionInput() {inputName=Mask_BlinkingLights, inputType=FunctionInput_Scalar, sortPriority=1}) {name=M_BlinkingLights, nodeColor=<rgba 1 0.9844 0 1>}
  NamedRerouteDeclaration(None=@n1.BaseColor) {name=IN_Basecolor, nodeColor=<rgba 0 1 0.2969 1>}
  n42 = TextureObjectParameter() {
   group=Galaxy
   parameterName=StarsTexture
   samplerType=SAMPLERTYPE_Masks
   sortPriority=80
   texture=/Game/Characters/Player/Male/Medium/Bodies/M_Med_Soldier_04/Skins/Galaxy/Textures/Celestial_T_Stars
  }
  n43 = MaterialFunctionCall(
   Texture=@n42
   Depth=ScalarParameter() {defaultValue=150, group=Galaxy, parameterName=Stars_Depth, sortPriority=81}
   Tile=Multiply(A=ScalarParameter() {defaultValue=2, group=Galaxy, parameterName=Stars_Tile, sortPriority=82}, B=NamedRerouteUsage())
   BlendContrast=ScalarParameter() {defaultValue=1, group=Galaxy, parameterName=Stars_BlendContrast, sortPriority=83}
   SpeedXY=ComponentMask(None=VectorParameter() {group=Galaxy, parameterName=Stars_Speed, sortPriority=84, defaultValue=<rgba 0.002 -0.03 0 1>}.RGB) {g=true, r=true}
   ActorOffset=Add(A=NamedRerouteUsage(), B=NamedRerouteUsage())
  ) {desc=3 TextureSampler, materialFunction=&MF_TriplanarVirtualBox}
  NamedRerouteDeclaration(None=Multiply(
    A=ComponentMask(None=@n43.Result) {b=true}
    B=ScalarParameter() {defaultValue=2, group=Galaxy, parameterName=Stars_Brightness, sortPriority=84}
   )) {name=Stars, nodeColor=<rgba 0 1 0.5078 1>}
  n44 = Bounds() {type=MEILB_PreSkinnedLocal}
  n45 = FunctionInput(Preview=@n44.Max) {bUsePreviewValueAsDefault=true, inputName=Pre-Skinned Bounds Max, sortPriority=120}
  n46 = FunctionInput(Preview=@n44.Min) {bUsePreviewValueAsDefault=true, inputName=Pre-Skinned Bounds Min, sortPriority=110}
  NamedRerouteDeclaration(None=Multiply(A=Subtract(A=@n45, B=@n46)) {constB=0.5}) {name="Pre-Skinned Bounds Half Extents (V3)", nodeColor=<rgba 1 1 1 1>}
  n47 = FunctionInput() {bUsePreviewValueAsDefault=true, inputName=Pre-Skinned Scale, sortPriority=130, previewValue=<1 1 1 1>}
  n48 = NamedRerouteDeclaration(None=@n47) {name="Pre-Skinned Scale (V3)", nodeColor=<rgba 1 0.5391 0 1>}
  n49 = Convert!() {convertOutputs [{type=Vector3, defaultValue=<rgba 1 1 1 1>}]}
  NamedRerouteDeclaration(None=Multiply(A=DotProduct(A=@n48, B=@n49)) {constB=0.3333}) {name="Pre-Skinned Scale (S)", nodeColor=<rgba 1 0.5391 0 1>}
  n50 = NamedRerouteDeclaration(None=@n7) {name=FresnelMask, nodeColor=<rgba 0 0.7891 1 1>}
  n51 = Multiply(
   A=@n37
   B=Multiply(A=@n50, B=VectorParameter() {group=Galaxy, parameterName=EmissiveColor, sortPriority=91, defaultValue=<rgba 15 15 15 1>}.RGB)
  )
  n52 = Multiply(A=Constant4Vector() {constant=<rgba 0.2 0.2 1 1>}.RGB, B=@n50)
  n53 = Multiply(A=NamedRerouteUsage(), B=ScalarParameter() {defaultValue=0.4, group=Galaxy, parameterName=Fresnel_FaceGlow, sortPriority=92})
  n54 = VectorParameter() {group=Galaxy, parameterName=EmissiveColorFace, sortPriority=91, defaultValue=<rgba 0.55 0.1 1 1>}
  NamedRerouteDeclaration(None=Max(A=LinearInterpolate(A=@n51, B=@n52) {constAlpha=0.6}, B=Multiply(A=@n53, B=@n54.RGB))) {name=FresnelColor, nodeColor=<rgba 0.8047 0 1 1>}
  n55 = Add(A=@n44.Min, B=@n44.Min)
  NamedRerouteDeclaration(None=FunctionInput(Preview=Multiply(A=@n55) {constB=0.5}) {bUsePreviewValueAsDefault=true, inputName=Pre-Skinned Bounds Center, sortPriority=130}) {name="Pre-Skinned Bounds Center (V3)", nodeColor=<rgba 1 0 0.2578 1>}
  n56 = NamedRerouteUsage()
  MaterialExpressionMultiply_47 = Multiply(A=NamedRerouteUsage(), B=ComponentMask(None=NamedRerouteUsage()) {b=true})
  n57 = Convert!(None=@MaterialExpressionMultiply_47) {
   convertInputs [{
     type=Scalar
     expressionInput {
      expression=&self:MaterialExpressionMultiply_47
      inputName=None
      mask=0
      maskA=0
      maskB=0
      maskG=0
      maskR=0
      outputIndex=0
     }
    }]
   convertOutputs [{type=Vector3}]
  }
  NamedRerouteDeclaration(None=TransformPosition(None=Subtract(A=@n56, B=@n57)) {transformType=TRANSFORMPOSSOURCE_World}) {name="Pivot WS (V3)", nodeColor=<rgba 0.6875 1 0 1>}
  n58 = Constant3Vector()
  NamedRerouteDeclaration(None=Subtract(A=NamedRerouteUsage(), B=TransformPosition(None=@n58) {transformType=TRANSFORMPOSSOURCE_World})) {name="Triplanar Actor Offset (V3)", nodeColor=<rgba 1 0 0.2109 1>}
  NamedRerouteDeclaration(None=Divide(B=NamedRerouteUsage())) {name="Triplanar Tiling Scale (S)", nodeColor=<rgba 0 0.5312 1 1>}
  VectorParameter() {group=Galaxy, parameterName=Galaxy_BaseColor, sortPriority=49, defaultValue=<rgba 0 0 0 1>}
 ]}
