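# Material function that samples a texture with dithered triplanar box projection, choosing among three supplied UV sets (X, Y, Z) by a direction that defaults to the camera vector.
# The vector's absolute Y and Z components pass through CheapContrast to form blend alphas.
# DitherTemporalAA noise, weighted toward transitions, is added before rounding, so each pixel picks one UV set and needs only one texture sample.
# Outputs the sampled RGB and chosen UVs.
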MaterialFunction /Game/Global/MaterialFunctions/MF_TriplanarVirtualBox_DitherSample
graph {nodes [
  FunctionOutput(None=TextureSample(UVs=NamedRerouteUsage(), Tex=NamedRerouteUsage()) {samplerSource=SSM_Wrap_WorldGroupSettings}.RGB)
  n1 = NamedRerouteUsage()
  n2 = MaterialFunctionCall(Alpha Threshold=Constant() {r=0.5}) {materialFunction=&DitherTemporalAA}
  n3 = Multiply(A=Multiply(A=OneMinus(None=@n1), B=@n1), B=@n2.Result)
  n4 = LinearInterpolate(A=NamedRerouteUsage(), B=NamedRerouteUsage(), Alpha=Round(None=Add(A=@n3, B=@n1)))
  n5 = NamedRerouteUsage()
  n6 = NamedRerouteUsage()
  n7 = Multiply(A=OneMinus(None=@n6), B=@n6)
  n8 = Round(None=Add(A=Multiply(A=@n7, B=@n2.Result), B=@n6))
  FunctionOutput(None=NamedRerouteDeclaration(None=LinearInterpolate(A=@n4, B=@n5, Alpha=@n8)) {name=UVW, nodeColor=<rgba 0.8047 0 1 1>}) {outputName=UVs}
  NamedRerouteDeclaration(None=FunctionInput(Preview=CameraVectorWS()) {bUsePreviewValueAsDefault=true, inputName=CameraVector, sortPriority=10}) {name=CameraVector, nodeColor=<rgba 0 1 0.1562 1>}
  NamedRerouteDeclaration(None=FunctionInput() {inputName=UVX, inputType=FunctionInput_Vector2, sortPriority=1}) {name=UVX, nodeColor=<rgba 0 0.6016 1 1>}
  NamedRerouteDeclaration(None=FunctionInput() {inputName=UVY, inputType=FunctionInput_Vector2, sortPriority=2}) {name=UVY, nodeColor=<rgba 0 0.6016 1 1>}
  NamedRerouteDeclaration(None=FunctionInput() {inputName=UVZ, inputType=FunctionInput_Vector2, sortPriority=3}) {name=UVZ, nodeColor=<rgba 0 0.6016 1 1>}
  n9 = FunctionInput() {bUsePreviewValueAsDefault=true, inputName=BlendContrast, inputType=FunctionInput_Scalar, sortPriority=3, previewValue=<0.5 0 0 1>}
  NamedRerouteDeclaration(None=@n9) {name=BlendContrast, nodeColor=<rgba 1 0 0.7266 1>}
  NamedRerouteDeclaration(None=FunctionInput(Preview=TextureObject() {texture=/UVEditor/Textures/UVEditorColorGrid_Color}) {bUsePreviewValueAsDefault=true, inputName=Texture, inputType=FunctionInput_Texture2D}) {name=Galaxy_01_Texture, nodeColor=<rgba 1 0.7031 0 1>}
  n10 = Abs(None=NamedRerouteUsage())
  n11 = NamedRerouteUsage()
  NamedRerouteDeclaration(None=MaterialFunctionCall(In=ComponentMask(None=@n10) {g=true}, Contrast=@n11) {materialFunction=&CheapContrast}.Result) {name=BlendAlphaY, nodeColor=<rgba 0 1 0.7422 1>}
  NamedRerouteDeclaration(None=MaterialFunctionCall(In=ComponentMask(None=@n10) {b=true}, Contrast=@n11) {materialFunction=&CheapContrast}.Result) {name=BlendAlphaZ, nodeColor=<rgba 0 0.0391 1 1>}
 ]}
