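# Material function that samples one texture with three supplied UV sets and blends them triplanar-style for a virtual box effect.
# The camera vector, overridable as an input, drives the blend in place of a surface normal.
# Its absolute Y and Z components pass through CheapContrast with a BlendContrast input to form alphas for two chained lerps of the X, Y and Z samples.
# Samples use the shared world-group wrap sampler.
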MaterialFunction /Game/Global/MaterialFunctions/MF_TriplanarVirtualBox_Samples
graph {nodes [
  n1 = LinearInterpolate(A=NamedRerouteUsage(), B=NamedRerouteUsage(), Alpha=NamedRerouteUsage())
  n2 = NamedRerouteUsage()
  n3 = NamedRerouteUsage()
  FunctionOutput(None=NamedRerouteDeclaration(None=LinearInterpolate(A=@n1, B=@n2, Alpha=@n3)) {name=result, nodeColor=<rgba 1 0 0.7734 1>})
  NamedRerouteDeclaration(None=FunctionInput(Preview=CameraVectorWS()) {bUsePreviewValueAsDefault=true, inputName=CameraVector, sortPriority=10}) {name=CameraVector, nodeColor=<rgba 0 1 0.1562 1>}
  NamedRerouteDeclaration(None=FunctionInput() {inputName=UVX, inputType=FunctionInput_Vector2, sortPriority=1}) {name=UVX, nodeColor=<rgba 0 0.6016 1 1>}
  NamedRerouteDeclaration(None=FunctionInput() {inputName=UVY, inputType=FunctionInput_Vector2, sortPriority=2}) {name=UVY, nodeColor=<rgba 0 0.6016 1 1>}
  NamedRerouteDeclaration(None=FunctionInput() {inputName=UVZ, inputType=FunctionInput_Vector2, sortPriority=3}) {name=UVZ, nodeColor=<rgba 0 0.6016 1 1>}
  n4 = FunctionInput() {bUsePreviewValueAsDefault=true, inputName=BlendContrast, inputType=FunctionInput_Scalar, sortPriority=3, previewValue=<0.5 0 0 1>}
  NamedRerouteDeclaration(None=@n4) {name=BlendContrast, nodeColor=<rgba 1 0 0.7266 1>}
  NamedRerouteDeclaration(None=FunctionInput(Preview=TextureObject() {texture=/UVEditor/Textures/UVEditorColorGrid_Color}) {bUsePreviewValueAsDefault=true, inputName=Texture, inputType=FunctionInput_Texture2D}) {name=Galaxy_01_Texture, nodeColor=<rgba 1 0.7031 0 1>}
  NamedRerouteDeclaration(None=TextureSample(UVs=NamedRerouteUsage(), Tex=NamedRerouteUsage()) {samplerSource=SSM_Wrap_WorldGroupSettings}.RGB) {name=SampleX, nodeColor=<rgba 1 0 0.6328 1>}
  NamedRerouteDeclaration(None=TextureSample(UVs=NamedRerouteUsage(), Tex=NamedRerouteUsage()) {samplerSource=SSM_Wrap_WorldGroupSettings}.RGB) {name=SampleY, nodeColor=<rgba 1 0 0.6328 1>}
  NamedRerouteDeclaration(None=TextureSample(UVs=NamedRerouteUsage(), Tex=NamedRerouteUsage()) {samplerSource=SSM_Wrap_WorldGroupSettings}.RGB) {name=SampleZ, nodeColor=<rgba 1 0 0.6328 1>}
  n5 = Abs(None=NamedRerouteUsage())
  n6 = NamedRerouteUsage()
  NamedRerouteDeclaration(None=MaterialFunctionCall(In=ComponentMask(None=@n5) {g=true}, Contrast=@n6) {materialFunction=&CheapContrast}.Result) {name=BlendAlphaY, nodeColor=<rgba 0 1 0.7422 1>}
  NamedRerouteDeclaration(None=MaterialFunctionCall(In=ComponentMask(None=@n5) {b=true}, Contrast=@n6) {materialFunction=&CheapContrast}.Result) {name=BlendAlphaZ, nodeColor=<rgba 0 0.0391 1 1>}
 ]}
