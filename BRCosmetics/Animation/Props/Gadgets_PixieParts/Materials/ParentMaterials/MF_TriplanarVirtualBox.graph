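# Material function that samples a texture as a triplanar virtual box, driven by depth, tile and camera vector.
# MF_TriplanarVirtualBox_UVs builds three axis UV sets and a camera vector from depth, tile, speed and actor offset, with optional local-space pivot.
# These feed a contrast-blended sampler, with a static switch choosing a dithered single-sample variant over the regular one.
# It outputs the sampled result, per-axis UVs and dither UVs.
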MaterialFunction /BRCosmetics/Animation/Props/Gadgets_PixieParts/Materials/ParentMaterials/MF_TriplanarVirtualBox
graph {nodes [
  n1 = FunctionInput(Preview=TextureObject() {texture=/UVEditor/Textures/UVEditorColorGrid_Color}) {bUsePreviewValueAsDefault=true, inputName=Texture, inputType=FunctionInput_Texture2D}
  n2 = FunctionInput() {bUsePreviewValueAsDefault=true, inputName=BlendContrast, inputType=FunctionInput_Scalar, sortPriority=3, previewValue=<0.5 0 0 1>}
  n3 = MaterialFunctionCall!(Texture=@n1, UVX=NamedRerouteUsage(), UVY=NamedRerouteUsage(), UVZ=NamedRerouteUsage(), BlendContrast=@n2, CameraVector=NamedRerouteUsage()) {materialFunction=/Game/Global/MaterialFunctions/MF_TriplanarVirtualBox_DitherSample}
  n4 = NamedRerouteUsage()
  n5 = NamedRerouteUsage()
  n6 = NamedRerouteUsage()
  n7 = NamedRerouteUsage()
  n8 = FunctionInput(Preview=StaticBool()) {bUsePreviewValueAsDefault=true, inputName=bUseDitherSingleSample, inputType=FunctionInput_StaticBool, sortPriority=101}
  FunctionOutput(None=StaticSwitch(
    True=@n3.Result
    False=MaterialFunctionCall!(Texture=@n1, UVX=@n4, UVY=@n5, UVZ=@n6, BlendContrast=@n2, CameraVector=@n7) {materialFunction=/Game/Global/MaterialFunctions/MF_TriplanarVirtualBox_Samples}.Result
    Value=@n8
   ))
  FunctionOutput(None=StaticSwitch(True=@n3.UVs, False=Constant(), Value=@n8)) {outputName=DitherUVs, sortPriority=10}
  FunctionOutput(None=NamedRerouteUsage()) {outputName=UVX, sortPriority=1}
  FunctionOutput(None=NamedRerouteUsage()) {outputName=UVY, sortPriority=2}
  FunctionOutput(None=NamedRerouteUsage()) {outputName=UVZ, sortPriority=3}
  n9 = FunctionInput() {bUsePreviewValueAsDefault=true, inputName=Depth, inputType=FunctionInput_Scalar, sortPriority=1, previewValue=<200 0 0 1>}
  n10 = FunctionInput() {bUsePreviewValueAsDefault=true, inputName=Tile, inputType=FunctionInput_Scalar, sortPriority=2, previewValue=<0.25 0 0 1>}
  n11 = FunctionInput() {
   bUsePreviewValueAsDefault=true
   description="only used if bUseLocalCaneraVector is ON. Default: (0, 0, 0)"
   inputName=LocalSpacePivot
   sortPriority=110
  }
  n12 = MaterialFunctionCall!(
   Depth=@n9
   Tile=@n10
   SpeedXY=FunctionInput() {bUsePreviewValueAsDefault=true, inputName=SpeedXY, inputType=FunctionInput_Vector2, sortPriority=5}
   ActorOffset=FunctionInput() {bUsePreviewValueAsDefault=true, inputName=ActorOffset, sortPriority=10}
   UseLocalCameraVector=FunctionInput(Preview=StaticBool()) {bUsePreviewValueAsDefault=true, inputName=bUseLocalCameraVector, inputType=FunctionInput_StaticBool, sortPriority=100}
   LocalSpacePivot=@n11
  ) {materialFunction=&/MF_TriplanarVirtualBox_UVs}
  NamedRerouteDeclaration(None=@n12.UVZ) {name=UVZ, nodeColor=<rgba 0.2422 0 1 1>}
  NamedRerouteDeclaration(None=@n12.CameraVector) {name=CameraVector, nodeColor=<rgba 0.0781 1 0 1>}
  NamedRerouteDeclaration(None=@n12.UVX) {name=UVX, nodeColor=<rgba 0 1 0.8828 1>}
  NamedRerouteDeclaration(None=@n12.UVY) {name=UVY, nodeColor=<rgba 1 0.5625 0 1>}
 ]}
