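# Material function that gives sprite cosmetics an optional gem look, overriding surface attributes when the Sprite_Variant_Gem static switch is on.
# Dithered triplanar triangle normals in pre-skinned space form facets, while a triplanar virtual box supplies interior normals for depth.
# Emissive fakes glass by blending skylight environment reflection and refraction samples, with an IOR and internal-reflection term.
# Input masks add tint, inner glow, fresnel and logo emission, modulated by time-of-day color.
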MaterialFunction /BRCosmetics/Animation/Props/Gadgets_PixieParts/Materials/ParentMaterials/MF_Sprite_Variants_Gem
graph {nodes [
  n1 = NamedRerouteUsage()
  n2 = SetMaterialAttributes!(
   MaterialAttributes=@n1
   EmissiveColor=NamedRerouteUsage()
   Normal=NamedRerouteUsage()
   BaseColor=NamedRerouteUsage()
   Metallic=ScalarParameter() {defaultValue=0.75, group=SpriteVariants, parameterName=SpriteVar_Gem_Metallic}
   Roughness=ScalarParameter() {defaultValue=0.1, group=SpriteVariants, parameterName=SpriteVar_Gem_Roughness}
   Specular=ScalarParameter() {defaultValue=1, group=SpriteVariants, parameterName=SpriteVar_Gem_Specular}
   WorldPositionOffset
  ) {attributeSetTypes [
    #1
    #2
    #3
    #4
    #5
    #6
    #7
   ]}
  FunctionOutput(None=StaticSwitchParameter(True=@n2, False=@n1) {group=SpriteVariants, parameterName=Sprite_Variant_Gem}) {outputName=SetMaterialAttributes}
  n3 = NamedRerouteUsage()
  n4 = Add(A=@n3, B=Subtract(A=@n3, B=NamedRerouteUsage()))
  n5 = VectorParameter() {group=Override, parameterName=SpriteVar_Gem_Emissive_InnerGLow, defaultValue=<rgba 1.4773 2 1.2928 0.1>}
  n6 = VectorParameter() {group=SpriteVariants, parameterName=SpriteVar_Gem_Emissive_DiffuseChannelMask_Tint, defaultValue=<rgba 5.5453 6.9539 7.1652 1>}
  n7 = SkyLightEnvMapSample(
   Direction=ReflectionVectorWS(CustomWorldNormal=NamedRerouteUsage())
   Roughness=ScalarParameter() {defaultValue=0.1, parameterName=ReflectionRoughness}
  )
  n8 = SkyLightEnvMapSample(
   Direction=MaterialFunctionCall!(
    Ray Direction=NamedRerouteUsage()
    Surface Normal=NamedRerouteUsage()
    Refractive Index Origin=NamedRerouteUsage()
    Refractive Index Target=Constant() {r=1}
   ) {materialFunction=/Engine/Functions/Engine_MaterialFunctions01/Vectors/Refract}.Vector
   Roughness=ScalarParameter() {defaultValue=0.2, group=SpriteVariants, parameterName=RefractionRoughness}
  )
  n9 = NamedRerouteUsage()
  n10 = Multiply(
   A=Multiply(A=LinearInterpolate(A=@n7, B=@n8, Alpha=@n9), B=VectorParameter() {parameterName=Tint, defaultValue=<rgba 0.1198 0.1391 0.1295 1>}.RGB)
   B=NamedRerouteUsage()
  )
  n11 = LinearInterpolate(A=@n10, B=Multiply(A=@n10, B=@n6.RGB), Alpha=NamedRerouteUsage())
  n12 = Multiply(A=@n11, B=Multiply(A=Saturate(None=@n4), B=@n5.RGB))
  Multiply(A=@n12, B=OneMinus(None=NamedRerouteUsage()))
  n13 = GetMaterialAttributes(None=FunctionInput() {inputName=MaterialAttributes, inputType=FunctionInput_MaterialAttributes, sortPriority=-1}) {attributeGetTypes [
    #1
    #2
    #3
    #4
    #5
   ]}
  NamedRerouteDeclaration(None=@n13.Normal) {name=OriginalNormal, nodeColor=<rgba 1 0.6094 0 1>}
  NamedRerouteDeclaration(None=@n13.BaseColor) {name=OriginalBaseColor, nodeColor=<rgba 1 0.1406 0 1>}
  NamedRerouteDeclaration(None=@n13.MaterialAttributes) {name=OriginalMA, nodeColor=<rgba 1 0 0.1015 1>}
  NamedRerouteDeclaration(None=@n13.Metallic) {name=OriginalMetallic, nodeColor=<rgba 0 0.8594 1 1>}
  NamedRerouteDeclaration(None=@n13.Roughness) {name=OriginalRoughness, nodeColor=<rgba 1 0 0.3984 1>}
  NamedRerouteDeclaration() {name=Variant Mask, nodeColor=<rgba 0.2422 1 0 1>}
  n14 = NamedRerouteUsage()
  n15 = ScalarParameter() {defaultValue=0.01, group=SpriteVariants, parameterName=InteriorDepth}
  n16 = Multiply(A=ScalarParameter() {defaultValue=3, group=SpriteVariants, parameterName=InteriorTiling}, B=NamedRerouteUsage())
  n17 = ScalarParameter() {defaultValue=1, group=SpriteVariants, parameterName=InteriorBlendContrast}
  n18 = NamedRerouteUsage()
  n19 = StaticBool() {value=true}
  n20 = TransformPosition(None=NamedRerouteUsage()) {transformSourceType=TRANSFORMPOSSOURCE_World}
  NamedRerouteDeclaration(None=Transform(None=MaterialFunctionCall!(
     Texture=@n14
     Depth=@n15
     Tile=@n16
     BlendContrast=@n17
     SpeedXY
     ActorOffset=@n18
     bUseLocalCameraVector=@n19
     bUseDitherSingleSample=@n19
     LocalSpacePivot=@n20
    ) {materialFunction=&/MF_TriplanarVirtualBox}.Result)) {name=InteriorNormal, nodeColor=<rgba 0.1016 1 0 1>}
  n21 = CameraVectorWS()
  n22 = ScalarParameter() {defaultValue=1.5, group=SpriteVariants, parameterName=IOR}
  NamedRerouteDeclaration(None=MaterialFunctionCall(
    Ray Direction=Multiply(A=@n21) {constB=-1}
    Surface Normal=NamedRerouteUsage()
    Refractive Index Origin=Constant() {r=1}
    Refractive Index Target=NamedRerouteDeclaration(None=@n22) {name=IOR, nodeColor=<rgba 0 1 0.9297 1>}
   ) {materialFunction=&Refract}.Vector) {name=RefractionVector, nodeColor=<rgba 1 0 0.4219 1>}
  n23 = Multiply(A=DotProduct(A=NamedRerouteUsage(), B=NamedRerouteUsage())) {constB=-1}
  n24 = Constant() {r=0.73}
  n25 = Constant() {r=0.78}
  NamedRerouteDeclaration(None=LinearInterpolate(Alpha=MaterialFunctionCall(Input=@n23, Input Low=@n24, Input High=@n25) {materialFunction=&RemapValue_0-1}.Result) {constA=0.1, constB=0.9}) {name=TIR, nodeColor=<rgba 0 0.7422 1 1>}
  n26 = TextureSample!(
   UVs=MaterialFunctionCall!(
    Texture Scale=ScalarParameter() {defaultValue=20, parameterName=Gem_NormalTile}
    Normal
    InPosition=NamedRerouteUsage()
    Triplanar Fade=ScalarParameter() {defaultValue=100, parameterName=Gem_Normal_TriplanarFade}
    Derivatives Scale
   ) {materialFunction=&/MF_TriplanarDithered_PreSkinned}.UVs
   Tex=NamedRerouteUsage()
   "DDX(UVs)"=NamedRerouteUsage()
   "DDY(UVs)"=NamedRerouteUsage()
   Apply View MipBias
  ) {
   mipValueMode=TMVM_Derivative
   samplerSource=SSM_Wrap_WorldGroupSettings
   samplerType=SAMPLERTYPE_Normal
   texture=/BRCosmetics/Animation/Props/Gadgets_PixieParts/Textures/TA/T_Triangles_N
  }
  n27 = MaterialFunctionCall(BaseNormal=NamedRerouteDeclaration(None=@n26.RGB) {name=GemNormals, nodeColor=<rgba 0 0.6953 1 1>}, AdditionalNormal=NamedRerouteUsage()) {materialFunction=&BlendAngleCorrectedNormals}
  NamedRerouteDeclaration(None=Transform(None=NamedRerouteDeclaration(None=@n27.Result) {name=TangentNormal, nodeColor=<rgba 1 0.6094 0 1>})) {name=WorldSurfaceNormal, nodeColor=<rgba 0 1 0.1328 1>}
  n28 = TextureObjectParameter() {
   group=SpriteVariants
   parameterName=Gem Normals
   samplerType=SAMPLERTYPE_Normal
   texture=&T_Triangles_N
  }
  NamedRerouteDeclaration(None=@n28) {name=NormalTexture, nodeColor=<rgba 1 0 0.75 1>}
  n29 = ScalarParameter() {defaultValue=5, group=SpriteVariants, parameterName=ExponentIn}
  n30 = NamedRerouteUsage()
  NamedRerouteDeclaration(None=OneMinus(None=Fresnel(ExponentIn=@n29, Normal=@n30) {baseReflectFraction=0})) {name=FresnelMask, nodeColor=<rgba 1 0.0469 0 1>}
  n31 = NamedRerouteUsage()
  n32 = LinearInterpolate(A=Power(Base=SmoothStep(Value=@n31) {constMax=0.8}) {constExponent=1.5}, B=NamedRerouteUsage(), Alpha=Constant() {r=0.5})
  n33 = VectorParameter() {group=SpriteVariants, parameterName=SpriteVar_Gem_Emissive_Fresnel, defaultValue=<rgba 0.5 0.4856 0.413 1>}
  n34 = Multiply(A=@n32, B=@n33.RGB)
  n35 = Power(Base=@n32) {constExponent=3}
  n36 = OneMinus(None=NamedRerouteUsage())
  n37 = OneMinus(None=Multiply(A=@n36, B=@n36))
  n38 = NamedRerouteUsage()
  n39 = VectorParameter() {group=Override, parameterName=SpriteVar_Gem_LogoEmissive, defaultValue=<rgba 3.713 3.9989 7 0.05>}
  n40 = Multiply(A=Multiply(A=Multiply(A=@n38, B=@n39.RGB), B=NamedRerouteUsage()), B=OneMinus(None=NamedRerouteUsage()))
  n41 = Add(A=@n11, B=@n12)
  n42 = Multiply(A=@n41, B=@n40)
  n43 = Add(
   A=Add(A=@n41, B=LinearInterpolate(A=@n42, B=@n40, Alpha=@n39.A))
   B=Multiply(A=Multiply(A=@n34, B=@n35), B=LinearInterpolate(Alpha=@n37) {constA=0.2})
  )
  n44 = Add(A=NamedRerouteUsage())
  n45 = VectorParameter() {group=SpriteVariants, parameterName=SpriteVar_Gem_Emissive_TOD_Morning, defaultValue=<rgba 1.4 0.558 0.4017 1>}
  n46 = VectorParameter() {group=SpriteVariants, parameterName=SpriteVar_Gem_Emissive_TOD_Day, defaultValue=<rgba 1 1 1 1>}
  n47 = VectorParameter() {group=SpriteVariants, parameterName=SpriteVar_Gem_Emissive_TOD_Evening, defaultValue=<rgba 1.5 0.538 0.5217 1>}
  n48 = VectorParameter() {group=SpriteVariants, parameterName=SpriteVar_Gem_Emissive_TOD_Night, defaultValue=<rgba 1.4238 1.6 1.4452 1>}
  NamedRerouteDeclaration(None=Multiply(
    A=Multiply(A=@n43, B=@n44)
    B=MaterialFunctionCall(Color_Morning=@n45.RGB, Color_Day=@n46.RGB, Color_Evening=@n47.RGB, Color_Night=@n48.RGB) {materialFunction=&Color_Time_of_Day}.Result
   )) {name=Emissive, nodeColor=<rgba 0 0.2734 1 1>}
  NamedRerouteDeclaration(None=FunctionInput() {inputName=Mask Variant, inputType=FunctionInput_Scalar, sortPriority=1}) {name=M_VariantMask, nodeColor=<rgba 0 1 0.8828 1>}
  NamedRerouteDeclaration(None=FunctionInput() {inputName=Mask_InnerRadialGlow, inputType=FunctionInput_Scalar, sortPriority=1}) {name=M_InnerRadialGlow, nodeColor=<rgba 1 0.9844 0 1>}
  NamedRerouteDeclaration(None=FunctionInput() {inputName=Mask_SpriteFresnel, inputType=FunctionInput_Scalar, sortPriority=1}) {name=M_SpriteFresnel, nodeColor=<rgba 1 0.9844 0 1>}
  NamedRerouteDeclaration(None=FunctionInput() {inputName=Mask_Logo, inputType=FunctionInput_Scalar, sortPriority=1}) {name=M_Logo, nodeColor=<rgba 1 0.9844 0 1>}
  NamedRerouteDeclaration(None=FunctionInput() {inputName=WPO, inputType=FunctionInput_Scalar, sortPriority=2}) {name=IN_WPO, nodeColor=<rgba 0 0.9766 1 1>}
  n49 = VectorParameter() {group=SpriteVariants, parameterName=SpriteVar_Gem_Basecolor, defaultValue=<rgba 0.4271 0.4271 0.4271 0>}
  n50 = Multiply(A=@n49.RGB, B=ScalarParameter() {defaultValue=0.2, parameterName=SpriteVar_Gem_Basecolor_DiffuseChannelMask_Multiply})
  n51 = NamedRerouteUsage()
  NamedRerouteDeclaration(None=LinearInterpolate(A=@n49.RGB, B=LinearInterpolate(A=@n49.RGB, B=@n50, Alpha=@n51), Alpha=NamedRerouteUsage())) {name=OUT_BaseColor, nodeColor=<rgba 0.0547 0 1 1>}
  n52 = CameraVectorWS()
  n53 = VertexNormalWS()
  NamedRerouteDeclaration(None=OneMinus(None=DotProduct(A=@n52, B=@n53))) {name=M_Fresnel, nodeColor=<rgba 1 0.9844 0 1>}
  NamedRerouteDeclaration(None=FunctionInput() {inputName=Mask_BlinkingLights, inputType=FunctionInput_Scalar, sortPriority=1}) {name=M_BlinkingLights, nodeColor=<rgba 1 0.9844 0 1>}
  n54 = Transform(None=NamedRerouteUsage()) {transformType=TRANSFORM_Camera}
  n55 = Constant3Vector() {constant=<rgba 1 1 1 0>}
  NamedRerouteDeclaration(None=LinearInterpolate(Alpha=DotProduct(A=@n54, B=@n55)) {constA=0.6}) {name=M_GemPattern, nodeColor=<rgba 0 1 0.4609 1>}
  NamedRerouteDeclaration(None=FunctionInput() {inputName=Mask_DiffuseChannelMask, inputType=FunctionInput_Scalar, sortPriority=1}) {name=M_DiffuseMask, nodeColor=<rgba 0.875 1 0 1>}
  n56 = Subtract(A=NamedRerouteUsage(), B=NamedRerouteUsage())
  n57 = NamedRerouteUsage()
  NamedRerouteDeclaration(None=ComponentMask(None=Divide(A=@n56, B=@n57)) {b=true}) {name=M_ZGradient, nodeColor=<rgba 0.1719 1 0 1>}
  n58 = LocalPosition() {includedOffsets=ExcludeOffsets, localOrigin=InstancePreSkinning}
  n59 = VertexInterpolator(VS=@n58.XYZ)
  NamedRerouteDeclaration(None=FunctionInput(Preview=ShaderStageSwitch(PixelShader=@n59.PS, VertexShader=@n58.XYZ)) {bUsePreviewValueAsDefault=true, inputName=Pre-Skinned Local Pos, sortPriority=100}) {name="Pre-Skinned Local Pos (V3)", nodeColor=<rgba 0 0.1797 1 1>}
  n60 = Bounds() {type=MEILB_PreSkinnedLocal}
  NamedRerouteDeclaration(None=FunctionInput(Preview=@n60.Min) {bUsePreviewValueAsDefault=true, inputName=Pre-Skinned Bounds Min, sortPriority=110}) {name="Pre-Skinned Bounds Min (V3)", nodeColor=<rgba 1 1 1 1>}
  NamedRerouteDeclaration(None=FunctionInput(Preview=@n60.Max) {bUsePreviewValueAsDefault=true, inputName=Pre-Skinned Bounds Max, sortPriority=120}) {name="Pre-Skinned Bounds Max (V3)", nodeColor=<rgba 1 1 1 1>}
  NamedRerouteDeclaration(None=Subtract(A=NamedRerouteUsage(), B=NamedRerouteUsage())) {name="Pre-Skinned Bounds Extents (V3)", nodeColor=<rgba 1 1 1 1>}
  NamedRerouteDeclaration(None=Multiply(A=NamedRerouteUsage()) {constB=0.5}) {name="Pre-Skinned Bounds Half Extents (V3)", nodeColor=<rgba 1 1 1 1>}
  n61 = Convert!() {convertOutputs [{type=Vector3, defaultValue=<rgba 1 1 1 1>}]}
  n62 = FunctionInput() {bUsePreviewValueAsDefault=true, inputName=Pre-Skinned Scale, sortPriority=130, previewValue=<1 1 1 1>}
  n63 = NamedRerouteDeclaration(None=@n62) {name="Pre-Skinned Scale (V3)", nodeColor=<rgba 1 0.5391 0 1>}
  NamedRerouteDeclaration(None=Multiply(A=DotProduct(A=@n61, B=@n63)) {constB=0.3333}) {name="Pre-Skinned Scale (S)", nodeColor=<rgba 1 0.5391 0 1>}
  n64 = Divide(B=NamedRerouteUsage())
  n65 = ScalarParameter() {defaultValue=1, group=Sprite_Cosmetics | Triplanar, parameterName=Derivatives Use Pre-Skinned Scale, sliderMax=1, sortPriority=30}
  n66 = Multiply(
   A=Multiply(
    A=LinearInterpolate(B=@n64, Alpha=@n65) {constA=1}
    B=ScalarParameter() {defaultValue=2, group=Sprite_Cosmetics | Triplanar, parameterName=Derivatives Mult, sortPriority=40}
   )
   B=TextureCoordinate()
  )
  NamedRerouteDeclaration(None=DDX(Value=@n66)) {name="Scaled UV DDX (V2)", nodeColor=<rgba 0.0547 1 0 1>}
  NamedRerouteDeclaration(None=DDY(Value=@n66)) {name="Scaled UV DDY (V2)", nodeColor=<rgba 0.0547 1 0 1>}
  n67 = Add(A=@n60.Min, B=@n60.Min)
  NamedRerouteDeclaration(None=FunctionInput(Preview=Multiply(A=@n67) {constB=0.5}) {bUsePreviewValueAsDefault=true, inputName=Pre-Skinned Bounds Center, sortPriority=130}) {name="Pre-Skinned Bounds Center (V3)", nodeColor=<rgba 1 0 0.2578 1>}
  n68 = NamedRerouteUsage()
  MaterialExpressionMultiply_47 = Multiply(A=NamedRerouteUsage(), B=ComponentMask(None=NamedRerouteUsage()) {b=true})
  n69 = Convert!(None=@MaterialExpressionMultiply_47) {
   convertInputs [{
     type=Scalar
     expressionInput {
      expression=&self:MaterialExpressionMultiply_47
      inputName=None
      mask=0
      maskA=0
      maskB=0
      maskG=0
      maskR=0
      outputIndex=0
     }
    }]
   convertOutputs [{type=Vector3}]
  }
  NamedRerouteDeclaration(None=TransformPosition(None=Subtract(A=@n68, B=@n69)) {transformType=TRANSFORMPOSSOURCE_World}) {name="Pivot WS (V3)", nodeColor=<rgba 0.6875 1 0 1>}
  n70 = Constant3Vector()
  NamedRerouteDeclaration(None=Subtract(A=NamedRerouteUsage(), B=TransformPosition(None=@n70) {transformType=TRANSFORMPOSSOURCE_World})) {name="Triplanar Actor Offset (V3)", nodeColor=<rgba 1 0 0.2109 1>}
  NamedRerouteDeclaration(None=Divide(B=NamedRerouteUsage())) {name="Triplanar Tiling Scale (S)", nodeColor=<rgba 0 0.5312 1 1>}
 ]}
