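# Material function that applies the Golden variant look to Sprite cosmetic materials, gated by the Sprite_Variant_Golden static switch.
# It rebuilds base color, normal, metallic, roughness, specular and emissive from incoming attributes and masks, using gold tinting, FuzzyShading and noise-based patina.
# Animated glitter, logo and fresnel masks feed time-of-day-driven emissive.
# Patina and glitter use dithered triplanar sampling in pre-skinned space, with explicit scale-adjusted UV derivatives.
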MaterialFunction /BRCosmetics/Animation/Props/Gadgets_PixieParts/Materials/ParentMaterials/MF_Sprite_Variants_Golden
graph {expressions {
  MaterialExpressionAdd_2 {
   class=MaterialExpressionAdd
   inputs [
    A {from=MaterialExpressionComponentMask_2}
    B {from=MaterialExpressionMultiply_10}
   ]
  }
  MaterialExpressionAdd_4 {
   class=MaterialExpressionAdd
   inputs [
    A {from=MaterialExpressionAdd_5}
    B {from=MaterialExpressionMultiply_17}
   ]
  }
  MaterialExpressionAdd_5 {
   class=MaterialExpressionAdd
   inputs [
    A {from=MaterialExpressionMultiply_23}
    B {from=MaterialExpressionMultiply_27}
   ]
  }
  MaterialExpressionAdd_6 {
   class=MaterialExpressionAdd
   inputs [
    A {from=MaterialExpressionMaterialFunctionCall_8, output=UVs}
    B {from=MaterialExpressionConvert_2}
   ]
  }
  MaterialExpressionAdd_7 {
   class=MaterialExpressionAdd
   inputs [
    A {from=MaterialExpressionNamedRerouteUsage_66}
    B {from=""}
   ]
  }
  MaterialExpressionBounds_1 {class=MaterialExpressionBounds, inputs [], properties {type=MEILB_PreSkinnedLocal}}
  MaterialExpressionCameraVectorWS_0 {class=MaterialExpressionCameraVectorWS, inputs [], properties {}}
  MaterialExpressionClamp_1 {
   class=MaterialExpressionClamp
   inputs [
    None {from="", output=""}
    Min {from="", output=""}
    Max {from="", output=""}
   ]
   properties {maxDefault=100}
  }
  MaterialExpressionComponentMask_1 {class=MaterialExpressionComponentMask, inputs [None {from=MaterialExpressionReroute_1}], properties {r=true}}
  MaterialExpressionComponentMask_16 {class=MaterialExpressionComponentMask, inputs [None {from=MaterialExpressionDivide_28}], properties {b=true}}
  MaterialExpressionComponentMask_2 {class=MaterialExpressionComponentMask, inputs [None {from=MaterialExpressionReroute_1}], properties {g=true}}
  MaterialExpressionConstant4Vector_1 {class=MaterialExpressionConstant4Vector, inputs [], properties {constant=<rgba 1 2 1 0.9>}}
  MaterialExpressionConstant_12 {class=MaterialExpressionConstant, properties {r=150}}
  MaterialExpressionConstant_5 {class=MaterialExpressionConstant, properties {r=3}}
  MaterialExpressionConstant_6 {class=MaterialExpressionConstant, properties {r=2}}
  MaterialExpressionConstant_7 {class=MaterialExpressionConstant, properties {r=0.2}}
  MaterialExpressionConstant_8 {class=MaterialExpressionConstant, properties {r=0.7}}
  MaterialExpressionConstant_9 {class=MaterialExpressionConstant, properties {r=1}}
  MaterialExpressionConvert_2 {
   class=MaterialExpressionConvert
   inputs [
    None {from=""}
    None {from=MaterialExpressionFrac_3}
   ]
   properties {
    nodeName=Make Vector2
    convertInputs [
     {type=Scalar, expressionInput {expression=null, inputName=None, mask=0, maskA=0, maskB=0, maskG=0, maskR=0, outputIndex=0}}
     {
      type=Scalar
      expressionInput {
       expression=&self:MaterialExpressionFrac_3
       inputName=None
       mask=0
       maskA=0
       maskB=0
       maskG=0
       maskR=0
       outputIndex=0
      }
     }
    ]
    convertOutputs [{type=Vector2}]
   }
  }
  MaterialExpressionConvert_4 {class=MaterialExpressionConvert, inputs [], properties {convertOutputs [{type=Vector3, defaultValue=<rgba 1 1 1 1>}]}}
  MaterialExpressionDDX_0 {class=MaterialExpressionDDX, inputs [Value {from=MaterialExpressionMultiply_38, output=""}], properties {}}
  MaterialExpressionDDY_1 {class=MaterialExpressionDDY, inputs [Value {from=MaterialExpressionMultiply_38, output=""}], properties {}}
  MaterialExpressionDivide_28 {
   class=MaterialExpressionDivide
   inputs [
    A {from=MaterialExpressionSubtract_41}
    B {from=MaterialExpressionNamedRerouteUsage_155}
   ]
  }
  MaterialExpressionDivide_4 {
   class=MaterialExpressionDivide
   inputs [
    A {from=MaterialExpressionPower_3}
    B {from=""}
   ]
  }
  MaterialExpressionDivide_6 {
   class=MaterialExpressionDivide
   inputs [
    A {from=MaterialExpressionTime_2}
    B {from=MaterialExpressionConstant_12}
   ]
   properties {constB=5}
  }
  MaterialExpressionDivide_8 {
   class=MaterialExpressionDivide
   inputs [
    A {from=""}
    B {from=MaterialExpressionNamedRerouteUsage_72}
   ]
  }
  MaterialExpressionDotProduct_0 {
   class=MaterialExpressionDotProduct
   inputs [
    A {from=MaterialExpressionCameraVectorWS_0, output=""}
    B {from=MaterialExpressionVertexNormalWS_0, output=""}
   ]
   properties {}
  }
  MaterialExpressionDotProduct_1 {
   class=MaterialExpressionDotProduct
   inputs [
    A {from=MaterialExpressionNamedRerouteUsage_26, output=""}
    B {from=MaterialExpressionVectorParameter_10, output=RGB}
   ]
   properties {}
  }
  MaterialExpressionDotProduct_3 {
   class=MaterialExpressionDotProduct
   inputs [
    A {from=MaterialExpressionConvert_4, output=""}
    B {from=MaterialExpressionNamedRerouteDeclaration_73, output=""}
   ]
   properties {}
  }
  MaterialExpressionFrac_1 {class=MaterialExpressionFrac, inputs [None {from=MaterialExpressionAdd_2, output=""}], properties {}}
  MaterialExpressionFrac_3 {class=MaterialExpressionFrac, inputs [None {from=MaterialExpressionDivide_6, output=""}], properties {}}
  MaterialExpressionFunctionInput_0 {
   class=MaterialExpressionFunctionInput
   inputs [Preview {from="", output=""}]
   properties {inputName=MaterialAttributes, inputType=FunctionInput_MaterialAttributes, sortPriority=-1}
  }
  MaterialExpressionFunctionInput_1 {
   class=MaterialExpressionFunctionInput
   inputs [Preview {from="", output=""}]
   properties {inputName=Mask_BlinkingLights, inputType=FunctionInput_Scalar, sortPriority=1}
  }
  MaterialExpressionFunctionInput_10 {
   class=MaterialExpressionFunctionInput
   inputs [Preview {from="", output=""}]
   properties {inputName=WPO, inputType=FunctionInput_Scalar, sortPriority=2}
  }
  MaterialExpressionFunctionInput_12 {
   class=MaterialExpressionFunctionInput
   inputs [Preview {from=MaterialExpressionShaderStageSwitch_2, output=""}]
   properties {bUsePreviewValueAsDefault=true, inputName=Pre-Skinned Local Pos, sortPriority=100}
  }
  MaterialExpressionFunctionInput_13 {
   class=MaterialExpressionFunctionInput
   inputs [Preview {from=MaterialExpressionBounds_1, output=Max}]
   properties {bUsePreviewValueAsDefault=true, inputName=Pre-Skinned Bounds Max, sortPriority=120}
  }
  MaterialExpressionFunctionInput_14 {
   class=MaterialExpressionFunctionInput
   inputs [Preview {from=MaterialExpressionBounds_1, output=Min}]
   properties {bUsePreviewValueAsDefault=true, inputName=Pre-Skinned Bounds Min, sortPriority=110}
  }
  MaterialExpressionFunctionInput_16 {
   class=MaterialExpressionFunctionInput
   inputs [Preview {from="", output=""}]
   properties {bUsePreviewValueAsDefault=true, inputName=Pre-Skinned Scale, sortPriority=130, previewValue=<1 1 1 1>}
  }
  MaterialExpressionFunctionInput_6 {
   class=MaterialExpressionFunctionInput
   inputs [Preview {from="", output=""}]
   properties {inputName=Mask Variant, inputType=FunctionInput_Scalar, sortPriority=1}
  }
  MaterialExpressionFunctionInput_7 {
   class=MaterialExpressionFunctionInput
   inputs [Preview {from="", output=""}]
   properties {inputName=Mask_InnerRadialGlow, inputType=FunctionInput_Scalar, sortPriority=1}
  }
  MaterialExpressionFunctionInput_8 {
   class=MaterialExpressionFunctionInput
   inputs [Preview {from="", output=""}]
   properties {inputName=Mask_Logo, inputType=FunctionInput_Scalar, sortPriority=1}
  }
  MaterialExpressionFunctionInput_9 {
   class=MaterialExpressionFunctionInput
   inputs [Preview {from="", output=""}]
   properties {inputName=Mask_SpriteFresnel, inputType=FunctionInput_Scalar, sortPriority=1}
  }
  MaterialExpressionFunctionOutput_1 {
   class=MaterialExpressionFunctionOutput
   inputs [None {from=MaterialExpressionStaticSwitchParameter_0, output=""}]
   properties {outputName=SetMaterialAttributes}
  }
  MaterialExpressionGetMaterialAttributes_1 {
   class=MaterialExpressionGetMaterialAttributes
   inputs [None {from=MaterialExpressionFunctionInput_0, output=""}]
   properties {attributeGetTypes [
     #1
     #2
     #3
     #4
     #5
     #6
    ]}
  }
  MaterialExpressionLinearInterpolate_10 {
   class=MaterialExpressionLinearInterpolate
   inputs [
    A {from=""}
    B {from=MaterialExpressionNamedRerouteUsage_37}
    Alpha {from=MaterialExpressionNamedRerouteUsage_34}
   ]
  }
  MaterialExpressionLinearInterpolate_11 {
   class=MaterialExpressionLinearInterpolate
   inputs [
    A {from=MaterialExpressionReroute_4}
    B {from=MaterialExpressionMultiply_16}
    Alpha {from=MaterialExpressionReroute_5}
   ]
  }
  MaterialExpressionLinearInterpolate_12 {
   class=MaterialExpressionLinearInterpolate
   inputs [
    A {from=MaterialExpressionMultiply_18}
    B {from=MaterialExpressionReroute_7}
    Alpha {from=MaterialExpressionNamedRerouteUsage_40}
   ]
  }
  MaterialExpressionLinearInterpolate_13 {
   class=MaterialExpressionLinearInterpolate
   inputs [
    A {from=MaterialExpressionScalarParameter_24}
    B {from=MaterialExpressionScalarParameter_18}
    Alpha {from=MaterialExpressionNamedRerouteUsage_41}
   ]
  }
  MaterialExpressionLinearInterpolate_14 {
   class=MaterialExpressionLinearInterpolate
   inputs [
    A {from=MaterialExpressionScalarParameter_25}
    B {from=MaterialExpressionScalarParameter_19}
    Alpha {from=MaterialExpressionNamedRerouteUsage_42}
   ]
  }
  MaterialExpressionLinearInterpolate_15 {
   class=MaterialExpressionLinearInterpolate
   inputs [
    A {from=MaterialExpressionScalarParameter_26}
    B {from=MaterialExpressionScalarParameter_20}
    Alpha {from=MaterialExpressionNamedRerouteUsage_43}
   ]
  }
  MaterialExpressionLinearInterpolate_16 {
   class=MaterialExpressionLinearInterpolate
   inputs [
    A {from=MaterialExpressionScalarParameter_28}
    B {from=""}
    Alpha {from=MaterialExpressionNamedRerouteUsage_64}
   ]
  }
  MaterialExpressionLinearInterpolate_17 {
   class=MaterialExpressionLinearInterpolate
   inputs [
    A {from=""}
    B {from=""}
    Alpha {from=MaterialExpressionOneMinus_7}
   ]
   properties {constA=0.2}
  }
  MaterialExpressionLinearInterpolate_18 {
   class=MaterialExpressionLinearInterpolate
   inputs [
    A {from=MaterialExpressionScalarParameter_37}
    B {from=MaterialExpressionScalarParameter_38}
    Alpha {from=MaterialExpressionNamedRerouteUsage_53}
   ]
  }
  MaterialExpressionLinearInterpolate_19 {
   class=MaterialExpressionLinearInterpolate
   inputs [
    A {from=""}
    B {from=MaterialExpressionDivide_8}
    Alpha {from=MaterialExpressionScalarParameter_39}
   ]
   properties {constA=1}
  }
  MaterialExpressionLinearInterpolate_7 {
   class=MaterialExpressionLinearInterpolate
   inputs [
    A {from=MaterialExpressionReroute_6}
    B {from=MaterialExpressionLinearInterpolate_11}
    Alpha {from=MaterialExpressionNamedRerouteUsage_25}
   ]
  }
  MaterialExpressionLinearInterpolate_8 {
   class=MaterialExpressionLinearInterpolate
   inputs [
    A {from=""}
    B {from=MaterialExpressionNamedRerouteUsage_32}
    Alpha {from=MaterialExpressionNamedRerouteUsage_30}
   ]
  }
  MaterialExpressionLinearInterpolate_9 {
   class=MaterialExpressionLinearInterpolate
   inputs [
    A {from=""}
    B {from=MaterialExpressionNamedRerouteUsage_31}
    Alpha {from=MaterialExpressionNamedRerouteUsage_33}
   ]
  }
  MaterialExpressionLocalPosition_0 {class=MaterialExpressionLocalPosition, inputs [], properties {includedOffsets=ExcludeOffsets, localOrigin=InstancePreSkinning}}
  MaterialExpressionMaterialFunctionCall_10 {
   class=MaterialExpressionMaterialFunctionCall
   inputs [
    Texture Scale {from=MaterialExpressionScalarParameter_23}
    Normal {from=""}
    InPosition {from=MaterialExpressionNamedRerouteUsage_80}
    Triplanar Fade {from=MaterialExpressionConstant_6}
    Derivatives Scale {from=""}
   ]
   properties {materialFunction=&/MF_TriplanarDithered_PreSkinned}
  }
  MaterialExpressionMaterialFunctionCall_11 {
   class=MaterialExpressionMaterialFunctionCall
   inputs [
    Normal {from=MaterialExpressionNamedRerouteUsage_13}
    Flatness {from=MaterialExpressionScalarParameter_27}
   ]
   properties {materialFunction=&FlattenNormal}
  }
  MaterialExpressionMaterialFunctionCall_12 {
   class=MaterialExpressionMaterialFunctionCall
   inputs [
    Color_Morning {from=MaterialExpressionVectorParameter_17, output=RGB}
    Color_Day {from=MaterialExpressionVectorParameter_18, output=RGB}
    Color_Evening {from=MaterialExpressionVectorParameter_19, output=RGB}
    Color_Night {from=MaterialExpressionVectorParameter_20, output=RGB}
   ]
   properties {materialFunction=&Color_Time_of_Day}
  }
  MaterialExpressionMaterialFunctionCall_13 {
   class=MaterialExpressionMaterialFunctionCall
   inputs [
    Value_Morning {from=MaterialExpressionScalarParameter_30}
    Value_Day {from=MaterialExpressionScalarParameter_29}
    Value_Evening {from=MaterialExpressionScalarParameter_32}
    Value_Night {from=MaterialExpressionScalarParameter_31}
   ]
   properties {materialFunction=&Scalar_Time_of_Day}
  }
  MaterialExpressionMaterialFunctionCall_14 {
   class=MaterialExpressionMaterialFunctionCall
   inputs [
    Value_Morning {from=MaterialExpressionScalarParameter_34}
    Value_Day {from=MaterialExpressionScalarParameter_33}
    Value_Evening {from=MaterialExpressionScalarParameter_36}
    Value_Night {from=MaterialExpressionScalarParameter_35}
   ]
   properties {materialFunction=&Scalar_Time_of_Day}
  }
  MaterialExpressionMaterialFunctionCall_3 {
   class=MaterialExpressionMaterialFunctionCall
   inputs [
    BaseColor {from=MaterialExpressionReroute_8}
    Normal {from=MaterialExpressionNamedRerouteUsage_16}
    CoreDarkness {from=MaterialExpressionConstant_8}
    Power {from=MaterialExpressionConstant_9}
    EdgeBrightness {from=MaterialExpressionConstant_7}
   ]
   properties {materialFunction=/Engine/Functions/Engine_MaterialFunctions01/Shading/FuzzyShading}
  }
  MaterialExpressionMaterialFunctionCall_8 {
   class=MaterialExpressionMaterialFunctionCall
   inputs [
    Texture Scale {from=MaterialExpressionScalarParameter_22}
    Normal {from=""}
    InPosition {from=MaterialExpressionNamedRerouteUsage_81}
    Triplanar Fade {from=MaterialExpressionConstant_5}
    Derivatives Scale {from=""}
   ]
   properties {materialFunction=&MF_TriplanarDithered_PreSkinned}
  }
  MaterialExpressionMultiply_10 {
   class=MaterialExpressionMultiply
   inputs [
    A {from=MaterialExpressionTime_1}
    B {from=""}
   ]
   properties {constB=0.4, desc=.2}
  }
  MaterialExpressionMultiply_11 {
   class=MaterialExpressionMultiply
   inputs [
    A {from=MaterialExpressionSaturate_1}
    B {from=MaterialExpressionPower_2}
   ]
  }
  MaterialExpressionMultiply_13 {
   class=MaterialExpressionMultiply
   inputs [
    A {from=MaterialExpressionMultiply_19}
    B {from=""}
   ]
   properties {constB=5}
  }
  MaterialExpressionMultiply_14 {
   class=MaterialExpressionMultiply
   inputs [
    A {from=MaterialExpressionNamedRerouteUsage_20}
    B {from=MaterialExpressionMultiply_26}
   ]
  }
  MaterialExpressionMultiply_15 {
   class=MaterialExpressionMultiply
   inputs [
    A {from=MaterialExpressionNamedRerouteUsage_24}
    B {from=MaterialExpressionVectorParameter_9, output=RGB}
   ]
  }
  MaterialExpressionMultiply_16 {
   class=MaterialExpressionMultiply
   inputs [
    A {from=MaterialExpressionReroute_4}
    B {from=MaterialExpressionDotProduct_1}
   ]
  }
  MaterialExpressionMultiply_17 {
   class=MaterialExpressionMultiply
   inputs [
    A {from=MaterialExpressionMultiply_14}
    B {from=MaterialExpressionOneMinus_0}
   ]
  }
  MaterialExpressionMultiply_18 {
   class=MaterialExpressionMultiply
   inputs [
    A {from=MaterialExpressionReroute_7}
    B {from=MaterialExpressionVectorParameter_16, output=RGB}
   ]
  }
  MaterialExpressionMultiply_19 {
   class=MaterialExpressionMultiply
   inputs [
    A {from=MaterialExpressionPower_1}
    B {from=MaterialExpressionSine_0}
   ]
  }
  MaterialExpressionMultiply_20 {
   class=MaterialExpressionMultiply
   inputs [
    A {from=MaterialExpressionMultiply_29}
    B {from=MaterialExpressionLinearInterpolate_17}
   ]
  }
  MaterialExpressionMultiply_21 {
   class=MaterialExpressionMultiply
   inputs [
    A {from=MaterialExpressionOneMinus_4}
    B {from=MaterialExpressionOneMinus_4}
   ]
  }
  MaterialExpressionMultiply_22 {
   class=MaterialExpressionMultiply
   inputs [
    A {from=MaterialExpressionAdd_4}
    B {from=MaterialExpressionMaterialFunctionCall_12, output=Result}
   ]
  }
  MaterialExpressionMultiply_23 {
   class=MaterialExpressionMultiply
   inputs [
    A {from=MaterialExpressionReroute_9}
    B {from=MaterialExpressionLinearInterpolate_16}
   ]
  }
  MaterialExpressionMultiply_26 {
   class=MaterialExpressionMultiply
   inputs [
    A {from=MaterialExpressionVectorParameter_8, output=RGB}
    B {from=MaterialExpressionMaterialFunctionCall_13, output=Result}
   ]
  }
  MaterialExpressionMultiply_27 {
   class=MaterialExpressionMultiply
   inputs [
    A {from=MaterialExpressionNamedRerouteUsage_22}
    B {from=MaterialExpressionMultiply_28}
   ]
  }
  MaterialExpressionMultiply_28 {
   class=MaterialExpressionMultiply
   inputs [
    A {from=MaterialExpressionVectorParameter_21, output=RGB}
    B {from=MaterialExpressionMaterialFunctionCall_14, output=Result}
   ]
  }
  MaterialExpressionMultiply_29 {
   class=MaterialExpressionMultiply
   inputs [
    A {from=MaterialExpressionNamedRerouteUsage_50}
    B {from=MaterialExpressionLinearInterpolate_18}
   ]
  }
  MaterialExpressionMultiply_30 {
   class=MaterialExpressionMultiply
   inputs [
    A {from=MaterialExpressionMultiply_22}
    B {from=MaterialExpressionAdd_7}
   ]
  }
  MaterialExpressionMultiply_35 {
   class=MaterialExpressionMultiply
   inputs [
    A {from=MaterialExpressionDotProduct_3}
    B {from=""}
   ]
   properties {constB=0.3333}
  }
  MaterialExpressionMultiply_37 {
   class=MaterialExpressionMultiply
   inputs [
    A {from=MaterialExpressionLinearInterpolate_19}
    B {from=MaterialExpressionScalarParameter_42}
   ]
  }
  MaterialExpressionMultiply_38 {
   class=MaterialExpressionMultiply
   inputs [
    A {from=MaterialExpressionMultiply_37}
    B {from=MaterialExpressionTextureCoordinate_1}
   ]
  }
  MaterialExpressionMultiply_8 {
   class=MaterialExpressionMultiply
   inputs [
    A {from=MaterialExpressionNamedRerouteUsage_18}
    B {from=MaterialExpressionVectorParameter_7, output=RGB}
   ]
  }
  MaterialExpressionNamedRerouteDeclaration_32 {
   class=MaterialExpressionNamedRerouteDeclaration
   inputs [None {from=MaterialExpressionGetMaterialAttributes_1, output=Normal}]
   properties {name=IN_Normal, nodeColor=<rgba 1 0.6094 0 1>}
  }
  MaterialExpressionNamedRerouteDeclaration_33 {
   class=MaterialExpressionNamedRerouteDeclaration
   inputs [None {from=MaterialExpressionGetMaterialAttributes_1, output=BaseColor}]
   properties {name=IN_BaseColor, nodeColor=<rgba 1 0.1406 0 1>}
  }
  MaterialExpressionNamedRerouteDeclaration_34 {
   class=MaterialExpressionNamedRerouteDeclaration
   inputs [None {from=MaterialExpressionGetMaterialAttributes_1, output=MaterialAttributes}]
   properties {name=IN_MA, nodeColor=<rgba 1 0 0.1015 1>}
  }
  MaterialExpressionNamedRerouteDeclaration_35 {
   class=MaterialExpressionNamedRerouteDeclaration
   inputs [None {from=MaterialExpressionGetMaterialAttributes_1, output=Metallic}]
   properties {name=IN_Metallic, nodeColor=<rgba 0 0.8594 1 1>}
  }
  MaterialExpressionNamedRerouteDeclaration_36 {
   class=MaterialExpressionNamedRerouteDeclaration
   inputs [None {from=MaterialExpressionGetMaterialAttributes_1, output=Roughness}]
   properties {name=IN_Roughness, nodeColor=<rgba 1 0 0.3984 1>}
  }
  MaterialExpressionNamedRerouteDeclaration_37 {
   class=MaterialExpressionNamedRerouteDeclaration
   inputs [None {from=MaterialExpressionMaterialFunctionCall_11, output=Result}]
   properties {name=OUT_Normal, nodeColor=<rgba 1 0 0.9999 1>}
  }
  MaterialExpressionNamedRerouteDeclaration_38 {
   class=MaterialExpressionNamedRerouteDeclaration
   inputs [None {from=MaterialExpressionMaterialFunctionCall_3, output=Result}]
   properties {name=BaseColor, nodeColor=<rgba 1 0.8672 0 1>}
  }
  MaterialExpressionNamedRerouteDeclaration_39 {
   class=MaterialExpressionNamedRerouteDeclaration
   inputs [None {from=MaterialExpressionMultiply_30}]
   properties {name=Emissive, nodeColor=<rgba 1 0.1875 0 1>}
  }
  MaterialExpressionNamedRerouteDeclaration_41 {
   class=MaterialExpressionNamedRerouteDeclaration
   inputs [None {from=MaterialExpressionFunctionInput_6}]
   properties {name=M_VariantMask, nodeColor=<rgba 0 1 0.8828 1>}
  }
  MaterialExpressionNamedRerouteDeclaration_42 {
   class=MaterialExpressionNamedRerouteDeclaration
   inputs [None {from=MaterialExpressionFunctionInput_7}]
   properties {name=M_InnerRadialGlow, nodeColor=<rgba 1 0.9844 0 1>}
  }
  MaterialExpressionNamedRerouteDeclaration_43 {
   class=MaterialExpressionNamedRerouteDeclaration
   inputs [None {from=MaterialExpressionFunctionInput_9}]
   properties {name=M_SpriteFresnel, nodeColor=<rgba 1 0.9844 0 1>}
  }
  MaterialExpressionNamedRerouteDeclaration_44 {
   class=MaterialExpressionNamedRerouteDeclaration
   inputs [None {from=MaterialExpressionFunctionInput_8}]
   properties {name=M_Logo, nodeColor=<rgba 1 0.9844 0 1>}
  }
  MaterialExpressionNamedRerouteDeclaration_45 {
   class=MaterialExpressionNamedRerouteDeclaration
   inputs [None {from=MaterialExpressionFunctionInput_10}]
   properties {name=IN_WPO, nodeColor=<rgba 0 0.9766 1 1>}
  }
  MaterialExpressionNamedRerouteDeclaration_46 {
   class=MaterialExpressionNamedRerouteDeclaration
   inputs [None {from=MaterialExpressionSaturate_2}]
   properties {name=M_Glitter, nodeColor=<rgba 0 1 0.0625 1>}
  }
  MaterialExpressionNamedRerouteDeclaration_48 {
   class=MaterialExpressionNamedRerouteDeclaration
   inputs [None {from=MaterialExpressionLinearInterpolate_13}]
   properties {name=OUT_Metallic, nodeColor=<rgba 1 0.3984 0 1>}
  }
  MaterialExpressionNamedRerouteDeclaration_49 {
   class=MaterialExpressionNamedRerouteDeclaration
   inputs [None {from=MaterialExpressionLinearInterpolate_14}]
   properties {name=OUT_Roughness, nodeColor=<rgba 0.0781 0 1 1>}
  }
  MaterialExpressionNamedRerouteDeclaration_50 {
   class=MaterialExpressionNamedRerouteDeclaration
   inputs [None {from=MaterialExpressionLinearInterpolate_15}]
   properties {name=OUT_Specular, nodeColor=<rgba 0.2422 1 0 1>}
  }
  MaterialExpressionNamedRerouteDeclaration_51 {
   class=MaterialExpressionNamedRerouteDeclaration
   inputs [None {from=MaterialExpressionGetMaterialAttributes_1, output=Specular}]
   properties {name=IN_Specular, nodeColor=<rgba 1 0 0.3984 1>}
  }
  MaterialExpressionNamedRerouteDeclaration_55 {
   class=MaterialExpressionNamedRerouteDeclaration
   inputs [None {from=MaterialExpressionTextureObject_1}]
   properties {name=T_PatinaPattern, nodeColor=<rgba 0 1 0.3438 1>}
  }
  MaterialExpressionNamedRerouteDeclaration_56 {
   class=MaterialExpressionNamedRerouteDeclaration
   inputs [None {from=MaterialExpressionTextureSample_1, output=R}]
   properties {name=M_Patina, nodeColor=<rgba 1 0.0234 0 1>}
  }
  MaterialExpressionNamedRerouteDeclaration_57 {
   class=MaterialExpressionNamedRerouteDeclaration
   inputs [None {from=MaterialExpressionDotProduct_1}]
   properties {name=M_BasecolorInfluence, nodeColor=<rgba 0.4297 0 1 1>}
  }
  MaterialExpressionNamedRerouteDeclaration_58 {
   class=MaterialExpressionNamedRerouteDeclaration
   inputs [None {from=MaterialExpressionSmoothStep_0}]
   properties {name=M_Fresnel, nodeColor=<rgba 1 0.9844 0 1>}
  }
  MaterialExpressionNamedRerouteDeclaration_59 {
   class=MaterialExpressionNamedRerouteDeclaration
   inputs [None {from=MaterialExpressionComponentMask_16}]
   properties {name=M_ZGradient, nodeColor=<rgba 0.1719 1 0 1>}
  }
  MaterialExpressionNamedRerouteDeclaration_61 {
   class=MaterialExpressionNamedRerouteDeclaration
   inputs [None {from=MaterialExpressionGetMaterialAttributes_1, output=EmissiveColor}]
   properties {name=IN_Emissive, nodeColor=<rgba 0 1 0.7891 1>}
  }
  MaterialExpressionNamedRerouteDeclaration_63 {
   class=MaterialExpressionNamedRerouteDeclaration
   inputs [None {from=MaterialExpressionFunctionInput_1}]
   properties {name=M_BlinkingLights, nodeColor=<rgba 1 0.9844 0 1>}
  }
  MaterialExpressionNamedRerouteDeclaration_64 {
   class=MaterialExpressionNamedRerouteDeclaration
   inputs [None {from=MaterialExpressionFunctionInput_12}]
   properties {name="Pre-Skinned Local Pos (V3)", nodeColor=<rgba 0 0.1797 1 1>}
  }
  MaterialExpressionNamedRerouteDeclaration_65 {
   class=MaterialExpressionNamedRerouteDeclaration
   inputs [None {from=MaterialExpressionFunctionInput_14}]
   properties {name="Pre-Skinned Bounds Min (V3)", nodeColor=<rgba 1 1 1 1>}
  }
  MaterialExpressionNamedRerouteDeclaration_66 {
   class=MaterialExpressionNamedRerouteDeclaration
   inputs [None {from=MaterialExpressionFunctionInput_13}]
   properties {name="Pre-Skinned Bounds Max (V3)", nodeColor=<rgba 1 1 1 1>}
  }
  MaterialExpressionNamedRerouteDeclaration_68 {
   class=MaterialExpressionNamedRerouteDeclaration
   inputs [None {from=MaterialExpressionSubtract_1}]
   properties {name="Pre-Skinned Bounds Extents (V3)", nodeColor=<rgba 1 1 1 1>}
  }
  MaterialExpressionNamedRerouteDeclaration_71 {
   class=MaterialExpressionNamedRerouteDeclaration
   inputs [None {from=MaterialExpressionMultiply_35}]
   properties {name="Pre-Skinned Scale (S)", nodeColor=<rgba 1 0.5391 0 1>}
  }
  MaterialExpressionNamedRerouteDeclaration_73 {
   class=MaterialExpressionNamedRerouteDeclaration
   inputs [None {from=MaterialExpressionFunctionInput_16}]
   properties {name="Pre-Skinned Scale (V3)", nodeColor=<rgba 1 0.5391 0 1>}
  }
  MaterialExpressionNamedRerouteDeclaration_74 {
   class=MaterialExpressionNamedRerouteDeclaration
   inputs [None {from=MaterialExpressionDDX_0}]
   properties {name="Scaled UV DDX (V2)", nodeColor=<rgba 0.0547 1 0 1>}
  }
  MaterialExpressionNamedRerouteDeclaration_75 {
   class=MaterialExpressionNamedRerouteDeclaration
   inputs [None {from=MaterialExpressionDDY_1}]
   properties {name="Scaled UV DDY (V2)", nodeColor=<rgba 0.0547 1 0 1>}
  }
  MaterialExpressionNamedRerouteUsage_10 {class=MaterialExpressionNamedRerouteUsage}
  MaterialExpressionNamedRerouteUsage_13 {class=MaterialExpressionNamedRerouteUsage}
  MaterialExpressionNamedRerouteUsage_153 {class=MaterialExpressionNamedRerouteUsage}
  MaterialExpressionNamedRerouteUsage_154 {class=MaterialExpressionNamedRerouteUsage}
  MaterialExpressionNamedRerouteUsage_155 {class=MaterialExpressionNamedRerouteUsage}
  MaterialExpressionNamedRerouteUsage_16 {class=MaterialExpressionNamedRerouteUsage}
  MaterialExpressionNamedRerouteUsage_17 {class=MaterialExpressionNamedRerouteUsage}
  MaterialExpressionNamedRerouteUsage_18 {class=MaterialExpressionNamedRerouteUsage}
  MaterialExpressionNamedRerouteUsage_19 {class=MaterialExpressionNamedRerouteUsage}
  MaterialExpressionNamedRerouteUsage_20 {class=MaterialExpressionNamedRerouteUsage}
  MaterialExpressionNamedRerouteUsage_22 {class=MaterialExpressionNamedRerouteUsage}
  MaterialExpressionNamedRerouteUsage_24 {class=MaterialExpressionNamedRerouteUsage}
  MaterialExpressionNamedRerouteUsage_25 {class=MaterialExpressionNamedRerouteUsage}
  MaterialExpressionNamedRerouteUsage_26 {class=MaterialExpressionNamedRerouteUsage}
  MaterialExpressionNamedRerouteUsage_27 {class=MaterialExpressionNamedRerouteUsage}
  MaterialExpressionNamedRerouteUsage_28 {class=MaterialExpressionNamedRerouteUsage}
  MaterialExpressionNamedRerouteUsage_29 {class=MaterialExpressionNamedRerouteUsage}
  MaterialExpressionNamedRerouteUsage_30 {class=MaterialExpressionNamedRerouteUsage}
  MaterialExpressionNamedRerouteUsage_31 {class=MaterialExpressionNamedRerouteUsage}
  MaterialExpressionNamedRerouteUsage_32 {class=MaterialExpressionNamedRerouteUsage}
  MaterialExpressionNamedRerouteUsage_33 {class=MaterialExpressionNamedRerouteUsage}
  MaterialExpressionNamedRerouteUsage_34 {class=MaterialExpressionNamedRerouteUsage}
  MaterialExpressionNamedRerouteUsage_37 {class=MaterialExpressionNamedRerouteUsage}
  MaterialExpressionNamedRerouteUsage_38 {class=MaterialExpressionNamedRerouteUsage}
  MaterialExpressionNamedRerouteUsage_39 {class=MaterialExpressionNamedRerouteUsage}
  MaterialExpressionNamedRerouteUsage_40 {class=MaterialExpressionNamedRerouteUsage}
  MaterialExpressionNamedRerouteUsage_41 {class=MaterialExpressionNamedRerouteUsage}
  MaterialExpressionNamedRerouteUsage_42 {class=MaterialExpressionNamedRerouteUsage}
  MaterialExpressionNamedRerouteUsage_43 {class=MaterialExpressionNamedRerouteUsage}
  MaterialExpressionNamedRerouteUsage_46 {class=MaterialExpressionNamedRerouteUsage}
  MaterialExpressionNamedRerouteUsage_50 {class=MaterialExpressionNamedRerouteUsage}
  MaterialExpressionNamedRerouteUsage_51 {class=MaterialExpressionNamedRerouteUsage}
  MaterialExpressionNamedRerouteUsage_53 {class=MaterialExpressionNamedRerouteUsage}
  MaterialExpressionNamedRerouteUsage_64 {class=MaterialExpressionNamedRerouteUsage}
  MaterialExpressionNamedRerouteUsage_66 {class=MaterialExpressionNamedRerouteUsage}
  MaterialExpressionNamedRerouteUsage_68 {class=MaterialExpressionNamedRerouteUsage}
  MaterialExpressionNamedRerouteUsage_69 {class=MaterialExpressionNamedRerouteUsage}
  MaterialExpressionNamedRerouteUsage_72 {class=MaterialExpressionNamedRerouteUsage}
  MaterialExpressionNamedRerouteUsage_75 {class=MaterialExpressionNamedRerouteUsage}
  MaterialExpressionNamedRerouteUsage_76 {class=MaterialExpressionNamedRerouteUsage}
  MaterialExpressionNamedRerouteUsage_80 {class=MaterialExpressionNamedRerouteUsage}
  MaterialExpressionNamedRerouteUsage_81 {class=MaterialExpressionNamedRerouteUsage}
  MaterialExpressionNamedRerouteUsage_82 {class=MaterialExpressionNamedRerouteUsage}
  MaterialExpressionNamedRerouteUsage_83 {class=MaterialExpressionNamedRerouteUsage}
  MaterialExpressionOneMinus_0 {class=MaterialExpressionOneMinus, inputs [None {from=MaterialExpressionNamedRerouteUsage_38}]}
  MaterialExpressionOneMinus_1 {class=MaterialExpressionOneMinus, inputs [None {from=MaterialExpressionFrac_1}]}
  MaterialExpressionOneMinus_4 {class=MaterialExpressionOneMinus, inputs [None {from=MaterialExpressionNamedRerouteUsage_51}]}
  MaterialExpressionOneMinus_7 {class=MaterialExpressionOneMinus, inputs [None {from=MaterialExpressionMultiply_21}]}
  MaterialExpressionPower_1 {
   class=MaterialExpressionPower
   inputs [
    Base {from=MaterialExpressionComponentMask_1}
    Exp {from=""}
   ]
   properties {constExponent=3}
  }
  MaterialExpressionPower_2 {
   class=MaterialExpressionPower
   inputs [
    Base {from=MaterialExpressionFrac_1}
    Exp {from=""}
   ]
   properties {constExponent=4}
  }
  MaterialExpressionPower_3 {
   class=MaterialExpressionPower
   inputs [
    Base {from=MaterialExpressionOneMinus_1}
    Exp {from=""}
   ]
   properties {constExponent=4}
  }
  MaterialExpressionReroute_1 {class=MaterialExpressionReroute, inputs [None {from=MaterialExpressionTextureSample_0, output=RGB}]}
  MaterialExpressionReroute_4 {class=MaterialExpressionReroute, inputs [None {from=MaterialExpressionReroute_6}]}
  MaterialExpressionReroute_5 {class=MaterialExpressionReroute, inputs [None {from=MaterialExpressionVectorParameter_10, output="0=Just Golden 1= Use diffuse"}]}
  MaterialExpressionReroute_6 {class=MaterialExpressionReroute, inputs [None {from=MaterialExpressionVectorParameter_3, output=RGB}]}
  MaterialExpressionReroute_7 {class=MaterialExpressionReroute, inputs [None {from=MaterialExpressionLinearInterpolate_7}]}
  MaterialExpressionReroute_8 {class=MaterialExpressionReroute, inputs [None {from=MaterialExpressionLinearInterpolate_12}]}
  MaterialExpressionReroute_9 {class=MaterialExpressionReroute, inputs [None {from=MaterialExpressionMultiply_20}]}
  MaterialExpressionSaturate_1 {class=MaterialExpressionSaturate, inputs [None {from=MaterialExpressionMultiply_13}]}
  MaterialExpressionSaturate_2 {class=MaterialExpressionSaturate, inputs [None {from=MaterialExpressionMultiply_11}]}
  MaterialExpressionScalarParameter_18 {class=MaterialExpressionScalarParameter, properties {defaultValue=1, group=SpriteVariants, parameterName=SpriteVar_Gold_Metallic}}
  MaterialExpressionScalarParameter_19 {class=MaterialExpressionScalarParameter, properties {defaultValue=0.23, group=SpriteVariants, parameterName=SpriteVar_Gold_Roughness}}
  MaterialExpressionScalarParameter_20 {class=MaterialExpressionScalarParameter, properties {defaultValue=0.3, group=SpriteVariants, parameterName=SpriteVar_Gold_Specular}}
  MaterialExpressionScalarParameter_22 {
   class=MaterialExpressionScalarParameter
   properties {defaultValue=37, group=SpriteVariants, parameterName=SpriteVar_Gold_Emissive_Sparks_Scale}
  }
  MaterialExpressionScalarParameter_23 {class=MaterialExpressionScalarParameter, properties {defaultValue=35, group=SpriteVariants, parameterName=SpriteVar_Gold_Patina_Scale}}
  MaterialExpressionScalarParameter_24 {class=MaterialExpressionScalarParameter, properties {defaultValue=0.8, group=SpriteVariants, parameterName=SpriteVar_Gold_Patina_Metallic}}
  MaterialExpressionScalarParameter_25 {class=MaterialExpressionScalarParameter, properties {defaultValue=0.28, group=SpriteVariants, parameterName=SpriteVar_Gold_Patina_Roughness}}
  MaterialExpressionScalarParameter_26 {class=MaterialExpressionScalarParameter, properties {defaultValue=0.3, group=SpriteVariants, parameterName=SpriteVar_Gold_Patina_Specular}}
  MaterialExpressionScalarParameter_27 {class=MaterialExpressionScalarParameter, properties {group=SpriteVariants, parameterName=SpriteVar_Gold_FlattenNormal}}
  MaterialExpressionScalarParameter_28 {
   class=MaterialExpressionScalarParameter
   properties {defaultValue=0.85, group=SpriteVariants, parameterName=SpriteVar_Golden_Emissive_Pattina}
  }
  MaterialExpressionScalarParameter_29 {
   class=MaterialExpressionScalarParameter
   properties {defaultValue=1, group=SpriteVariants, parameterName=SpriteVar_Golden_Emissive_TOD_Day_Logo}
  }
  MaterialExpressionScalarParameter_30 {
   class=MaterialExpressionScalarParameter
   properties {defaultValue=0.2, group=SpriteVariants, parameterName=SpriteVar_Golden_Emissive_TOD_Morning_Logo}
  }
  MaterialExpressionScalarParameter_31 {
   class=MaterialExpressionScalarParameter
   properties {defaultValue=0.2, group=SpriteVariants, parameterName=SpriteVar_Golden_Emissive_TOD_Night_Logo}
  }
  MaterialExpressionScalarParameter_32 {
   class=MaterialExpressionScalarParameter
   properties {defaultValue=0.2, group=SpriteVariants, parameterName=SpriteVar_Golden_Emissive_TOD_Evening_Logo}
  }
  MaterialExpressionScalarParameter_33 {
   class=MaterialExpressionScalarParameter
   properties {defaultValue=1, group=SpriteVariants, parameterName=SpriteVar_Golden_Emissive_TOD_Day_Glitter}
  }
  MaterialExpressionScalarParameter_34 {
   class=MaterialExpressionScalarParameter
   properties {defaultValue=0.3, group=SpriteVariants, parameterName=SpriteVar_Golden_Emissive_TOD_Morning_Glitter}
  }
  MaterialExpressionScalarParameter_35 {
   class=MaterialExpressionScalarParameter
   properties {defaultValue=0.4, group=SpriteVariants, parameterName=SpriteVar_Golden_Emissive_TOD_Night_Glitter}
  }
  MaterialExpressionScalarParameter_36 {
   class=MaterialExpressionScalarParameter
   properties {defaultValue=0.8, group=SpriteVariants, parameterName=SpriteVar_Golden_Emissive_TOD_Evening_Glitter}
  }
  MaterialExpressionScalarParameter_37 {
   class=MaterialExpressionScalarParameter
   properties {defaultValue=0.5, group=SpriteVariants, parameterName=SpriteVar_Golden_Emissive_Fresnel_Min}
  }
  MaterialExpressionScalarParameter_38 {
   class=MaterialExpressionScalarParameter
   properties {defaultValue=1, group=SpriteVariants, parameterName=SpriteVar_Golden_Emissive_Fresnel_Max}
  }
  MaterialExpressionScalarParameter_39 {
   class=MaterialExpressionScalarParameter
   properties {defaultValue=1, group=Sprite_Cosmetics | Triplanar, parameterName=Derivatives Use Pre-Skinned Scale, sliderMax=1, sortPriority=30}
  }
  MaterialExpressionScalarParameter_42 {
   class=MaterialExpressionScalarParameter
   properties {defaultValue=2, group=Sprite_Cosmetics | Triplanar, parameterName=Derivatives Mult, sortPriority=40}
  }
  MaterialExpressionSetMaterialAttributes_0 {
   class=MaterialExpressionSetMaterialAttributes
   inputs [
    MaterialAttributes {from=MaterialExpressionNamedRerouteUsage_10, output=""}
    EmissiveColor {from=MaterialExpressionNamedRerouteUsage_19, output=""}
    Normal {from=MaterialExpressionNamedRerouteUsage_46, output=""}
    BaseColor {from=MaterialExpressionNamedRerouteUsage_17, output=""}
    Metallic {from=MaterialExpressionNamedRerouteUsage_27, output=""}
    Roughness {from=MaterialExpressionNamedRerouteUsage_28, output=""}
    Specular {from=MaterialExpressionNamedRerouteUsage_29, output=""}
   ]
   properties {attributeSetTypes [
     #6
     #1
     #2
     #3
     #4
     #5
    ]}
  }
  MaterialExpressionShaderStageSwitch_2 {
   class=MaterialExpressionShaderStageSwitch
   inputs [
    PixelShader {from=MaterialExpressionVertexInterpolator_2, output=PS}
    VertexShader {from=MaterialExpressionLocalPosition_0, output=XYZ}
   ]
   properties {}
  }
  MaterialExpressionSine_0 {class=MaterialExpressionSine, inputs [None {from=MaterialExpressionDivide_4}]}
  MaterialExpressionSmoothStep_0 {
   class=MaterialExpressionSmoothStep
   inputs [
    Min {from="", output=""}
    Max {from="", output=""}
    Value {from=MaterialExpressionDotProduct_0, output=""}
   ]
   properties {constMax=0, constMin=0.8}
  }
  MaterialExpressionStaticSwitchParameter_0 {
   class=MaterialExpressionStaticSwitchParameter
   inputs [
    True {from=MaterialExpressionSetMaterialAttributes_0, output=""}
    False {from=MaterialExpressionNamedRerouteUsage_10, output=""}
   ]
   properties {group=SpriteVariants, parameterName=Sprite_Variant_Golden, sortPriority=-1}
  }
  MaterialExpressionSubtract_1 {
   class=MaterialExpressionSubtract
   inputs [
    A {from=MaterialExpressionNamedRerouteUsage_69}
    B {from=MaterialExpressionNamedRerouteUsage_68}
   ]
  }
  MaterialExpressionSubtract_41 {
   class=MaterialExpressionSubtract
   inputs [
    A {from=MaterialExpressionNamedRerouteUsage_153}
    B {from=MaterialExpressionNamedRerouteUsage_154}
   ]
  }
  MaterialExpressionTextureCoordinate_1 {class=MaterialExpressionTextureCoordinate}
  MaterialExpressionTextureObject_1 {
   class=MaterialExpressionTextureObject
   inputs []
   properties {samplerType=SAMPLERTYPE_Masks, texture=/Game/Global/Textures/Default/Noise/T_Metal_Noise}
  }
  MaterialExpressionTextureSample_0 {
   class=MaterialExpressionTextureSample
   inputs [
    UVs {from=MaterialExpressionMaterialFunctionCall_8, output=UVs}
    Tex {from=""}
    "DDX(UVs)" {from=MaterialExpressionNamedRerouteUsage_82}
    "DDY(UVs)" {from=MaterialExpressionNamedRerouteUsage_83}
    Apply View MipBias {from=""}
   ]
   properties {
    mipValueMode=TMVM_Derivative
    samplerType=SAMPLERTYPE_LinearColor
    texture=/BRCosmetics/Animation/Props/Gadgets_PixieParts/Textures/TA/T_SpriteVar_Gold
   }
  }
  MaterialExpressionTextureSample_1 {
   class=MaterialExpressionTextureSample
   inputs [
    UVs {from=MaterialExpressionMaterialFunctionCall_10, output=UVs}
    Tex {from=MaterialExpressionNamedRerouteUsage_39}
    "DDX(UVs)" {from=MaterialExpressionNamedRerouteUsage_75}
    "DDY(UVs)" {from=MaterialExpressionNamedRerouteUsage_76}
    Apply View MipBias {from=""}
   ]
   properties {mipValueMode=TMVM_Derivative}
  }
  MaterialExpressionTime_1 {class=MaterialExpressionTime}
  MaterialExpressionTime_2 {class=MaterialExpressionTime}
  MaterialExpressionVectorParameter_10 {
   class=MaterialExpressionVectorParameter
   properties {
    group=SpriteVariants
    parameterName=SpriteVar_Gold_Color_HueShift
    channelNames {a="0=Just Golden 1= Use diffuse", b=DiffuseB Influence, g=DiffuseG Influence, r=DiffuseR Influence}
    defaultValue=<rgba 1 2 1 0.55>
   }
  }
  MaterialExpressionVectorParameter_16 {
   class=MaterialExpressionVectorParameter
   properties {group=SpriteVariants, parameterName=SpriteVar_Gold_Color_PatinaMult, defaultValue=<rgba 0.8435 0.6704 0.5061 1>}
  }
  MaterialExpressionVectorParameter_17 {
   class=MaterialExpressionVectorParameter
   properties {group=SpriteVariants, parameterName=SpriteVar_Gold_Emissive_TOD_Morning, defaultValue=<rgba 0.829 0.6031 0.6385 1>}
  }
  MaterialExpressionVectorParameter_18 {
   class=MaterialExpressionVectorParameter
   properties {group=SpriteVariants, parameterName=SpriteVar_Gold_Emissive_TOD_Day, defaultValue=<rgba 0.15 0.15 0.15 1>}
  }
  MaterialExpressionVectorParameter_19 {
   class=MaterialExpressionVectorParameter
   properties {group=SpriteVariants, parameterName=SpriteVar_Gold_Emissive_TOD_Evening, defaultValue=<rgba 0.8308 0.6038 0.6376 1>}
  }
  MaterialExpressionVectorParameter_20 {
   class=MaterialExpressionVectorParameter
   properties {group=SpriteVariants, parameterName=SpriteVar_Gold_Emissive_TOD_Night, defaultValue=<rgba 1.2 0.838 0.5391 1>}
  }
  MaterialExpressionVectorParameter_21 {
   class=MaterialExpressionVectorParameter
   properties {group=SpriteVariants, parameterName=SpriteVar_Gold_Emissive_Sparks_Color, defaultValue=<rgba 30 19.4948 2.595 1>}
  }
  MaterialExpressionVectorParameter_3 {
   class=MaterialExpressionVectorParameter
   properties {group=SpriteVariants, parameterName=SpriteVar_Gold_Color1, defaultValue=<rgba 1.3 0.8665 0.169 1>}
  }
  MaterialExpressionVectorParameter_7 {
   class=MaterialExpressionVectorParameter
   properties {group=SpriteVariants, parameterName=SpriteVar_Gold_Emissive_Fresnel, defaultValue=<rgba 0 0 0 1>}
  }
  MaterialExpressionVectorParameter_8 {
   class=MaterialExpressionVectorParameter
   properties {group=SpriteVariants, parameterName=SpriteVar_Gold_Emissive_Logo, defaultValue=<rgba 1.7565 1.1999 0.0866 1>}
  }
  MaterialExpressionVectorParameter_9 {
   class=MaterialExpressionVectorParameter
   properties {group=SpriteVariants, parameterName=SpriteVar_Gold_InnerGlow, defaultValue=<rgba 1 0.505 0.1 1>}
  }
  MaterialExpressionVertexInterpolator_2 {class=MaterialExpressionVertexInterpolator, inputs [VS {from=MaterialExpressionLocalPosition_0, output=XYZ}], properties {}}
  MaterialExpressionVertexNormalWS_0 {class=MaterialExpressionVertexNormalWS, inputs [], properties {}}
 }}
