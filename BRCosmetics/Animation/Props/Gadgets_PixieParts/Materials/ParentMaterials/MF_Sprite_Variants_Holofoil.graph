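# Material function that applies an animated holofoil variant look to Sprite gadget materials, modifying incoming material attributes behind a static switch.
# Holo textures are driven by view-direction dot products, time scrolling, pre-skinned triplanar mapping and object-stabilized screen-space UVs.
# Input masks such as variant, logo, inner glow, fresnel and blinking lights steer tinted color blends.
# It also blends flattened detail normals and uses time-of-day color.
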MaterialFunction /BRCosmetics/Animation/Props/Gadgets_PixieParts/Materials/ParentMaterials/MF_Sprite_Variants_Holofoil
graph {expressions {
  MaterialExpressionAbs_0 {class=MaterialExpressionAbs, inputs [None {from=MaterialExpressionComponentMask_5, output=""}], properties {}}
  MaterialExpressionAdd_0 {
   class=MaterialExpressionAdd
   inputs [
    A {from=MaterialExpressionReroute_33}
    B {from=MaterialExpressionScalarParameter_5}
   ]
  }
  MaterialExpressionAdd_10 {
   class=MaterialExpressionAdd
   inputs [
    A {from=MaterialExpressionReroute_9}
    B {from=MaterialExpressionSubtract_0}
   ]
  }
  MaterialExpressionAdd_11 {
   class=MaterialExpressionAdd
   inputs [
    A {from=MaterialExpressionDivide_0}
    B {from=MaterialExpressionFrac_2}
   ]
  }
  MaterialExpressionAdd_12 {
   class=MaterialExpressionAdd
   inputs [
    A {from=MaterialExpressionOneMinus_7}
    B {from=MaterialExpressionScalarParameter_17}
   ]
  }
  MaterialExpressionAdd_13 {
   class=MaterialExpressionAdd
   inputs [
    A {from=MaterialExpressionOneMinus_0}
    B {from=""}
   ]
  }
  MaterialExpressionAdd_14 {
   class=MaterialExpressionAdd
   inputs [
    A {from=MaterialExpressionAppendVector_0}
    B {from=MaterialExpressionMultiply_12}
   ]
  }
  MaterialExpressionAdd_16 {
   class=MaterialExpressionAdd
   inputs [
    A {from=MaterialExpressionReroute_14}
    B {from=MaterialExpressionMultiply_18}
   ]
  }
  MaterialExpressionAdd_18 {
   class=MaterialExpressionAdd
   inputs [
    A {from=MaterialExpressionAdd_38}
    B {from=MaterialExpressionMultiply_16}
   ]
  }
  MaterialExpressionAdd_19 {
   class=MaterialExpressionAdd
   inputs [
    A {from=MaterialExpressionAdd_18}
    B {from=MaterialExpressionComponentMask_8}
   ]
  }
  MaterialExpressionAdd_2 {
   class=MaterialExpressionAdd
   inputs [
    A {from=MaterialExpressionReroute_16}
    B {from=MaterialExpressionMultiply_0}
   ]
  }
  MaterialExpressionAdd_20 {
   class=MaterialExpressionAdd
   inputs [
    A {from=MaterialExpressionReroute_17}
    B {from=MaterialExpressionMultiply_24}
   ]
  }
  MaterialExpressionAdd_21 {
   class=MaterialExpressionAdd
   inputs [
    A {from=MaterialExpressionAdd_16}
    B {from=MaterialExpressionComponentMask_13}
   ]
  }
  MaterialExpressionAdd_22 {
   class=MaterialExpressionAdd
   inputs [
    A {from=MaterialExpressionLinearInterpolate_20}
    B {from=MaterialExpressionScalarParameter_47}
   ]
  }
  MaterialExpressionAdd_23 {
   class=MaterialExpressionAdd
   inputs [
    A {from=MaterialExpressionAdd_4}
    B {from=MaterialExpressionAdd_26}
   ]
  }
  MaterialExpressionAdd_26 {
   class=MaterialExpressionAdd
   inputs [
    A {from=MaterialExpressionScalarParameter_36}
    B {from=MaterialExpressionMultiply_30}
   ]
  }
  MaterialExpressionAdd_27 {
   class=MaterialExpressionAdd
   inputs [
    A {from=MaterialExpressionRotator_0}
    B {from=MaterialExpressionMultiply_9}
   ]
  }
  MaterialExpressionAdd_28 {
   class=MaterialExpressionAdd
   inputs [
    A {from=MaterialExpressionMultiply_35}
    B {from=MaterialExpressionDivide_11}
   ]
  }
  MaterialExpressionAdd_29 {
   class=MaterialExpressionAdd
   inputs [
    A {from=MaterialExpressionReroute_18}
    B {from=MaterialExpressionAdd_20}
   ]
  }
  MaterialExpressionAdd_3 {
   class=MaterialExpressionAdd
   inputs [
    A {from=MaterialExpressionNamedRerouteUsage_73}
    B {from=MaterialExpressionScalarParameter_16}
   ]
  }
  MaterialExpressionAdd_32 {
   class=MaterialExpressionAdd
   inputs [
    A {from=MaterialExpressionReroute_35}
    B {from=MaterialExpressionMultiply_34}
   ]
  }
  MaterialExpressionAdd_33 {
   class=MaterialExpressionAdd
   inputs [
    A {from=MaterialExpressionReroute_32}
    B {from=MaterialExpressionMultiply_57}
   ]
  }
  MaterialExpressionAdd_35 {
   class=MaterialExpressionAdd
   inputs [
    A {from=MaterialExpressionMultiply_13}
    B {from=MaterialExpressionDivide_6}
   ]
  }
  MaterialExpressionAdd_37 {
   class=MaterialExpressionAdd
   inputs [
    A {from=MaterialExpressionReroute_39}
    B {from=MaterialExpressionLinearInterpolate_14}
   ]
  }
  MaterialExpressionAdd_38 {
   class=MaterialExpressionAdd
   inputs [
    A {from=MaterialExpressionLinearInterpolate_23}
    B {from=MaterialExpressionMultiply_53}
   ]
  }
  MaterialExpressionAdd_4 {
   class=MaterialExpressionAdd
   inputs [
    A {from=MaterialExpressionAdd_19}
    B {from=MaterialExpressionFrac_1}
   ]
  }
  MaterialExpressionAdd_42 {
   class=MaterialExpressionAdd
   inputs [
    A {from=MaterialExpressionDotProduct_9}
    B {from=MaterialExpressionScalarParameter_0}
   ]
   properties {constB=0.3}
  }
  MaterialExpressionAdd_43 {
   class=MaterialExpressionAdd
   inputs [
    A {from=MaterialExpressionNamedRerouteUsage_84}
    B {from=""}
   ]
   properties {constB=0.1}
  }
  MaterialExpressionAdd_5 {
   class=MaterialExpressionAdd
   inputs [
    A {from=MaterialExpressionReroute_37}
    B {from=MaterialExpressionMultiply_20}
   ]
  }
  MaterialExpressionAppendVector_0 {
   class=MaterialExpressionAppendVector
   inputs [
    A {from=MaterialExpressionAdd_35, output=""}
    B {from=MaterialExpressionConstant_1, output=""}
   ]
  }
  MaterialExpressionBounds_0 {class=MaterialExpressionBounds, inputs [], properties {type=MEILB_PreSkinnedLocal}}
  MaterialExpressionBounds_1 {class=MaterialExpressionBounds, inputs [], properties {type=MEILB_ObjectLocal}}
  MaterialExpressionBounds_11 {class=MaterialExpressionBounds, inputs [], properties {type=MEILB_PreSkinnedLocal}}
  MaterialExpressionCameraVectorWS_5 {class=MaterialExpressionCameraVectorWS, inputs [], properties {}}
  MaterialExpressionComponentMask_11 {class=MaterialExpressionComponentMask, inputs [None {from=MaterialExpressionReroute_0}], properties {r=true}}
  MaterialExpressionComponentMask_12 {class=MaterialExpressionComponentMask, inputs [None {from=MaterialExpressionReroute_0}], properties {g=true}}
  MaterialExpressionComponentMask_13 {class=MaterialExpressionComponentMask, inputs [None {from=MaterialExpressionMultiply_25}], properties {g=true, r=true}}
  MaterialExpressionComponentMask_17 {class=MaterialExpressionComponentMask, inputs [None {from=MaterialExpressionBounds_11, output=Extents}], properties {b=true}}
  MaterialExpressionComponentMask_2 {class=MaterialExpressionComponentMask, inputs [None {from=MaterialExpressionReroute_8}], properties {g=true, r=true}}
  MaterialExpressionComponentMask_3 {class=MaterialExpressionComponentMask, inputs [None {from=MaterialExpressionReroute_8}], properties {b=true, r=true}}
  MaterialExpressionComponentMask_5 {class=MaterialExpressionComponentMask, inputs [None {from=MaterialExpressionNamedRerouteUsage_32}], properties {g=true}}
  MaterialExpressionComponentMask_8 {class=MaterialExpressionComponentMask, inputs [None {from=MaterialExpressionMultiply_45}], properties {g=true, r=true}}
  MaterialExpressionComponentMask_9 {class=MaterialExpressionComponentMask, inputs [None {from=MaterialExpressionDivide_28}], properties {b=true}}
  MaterialExpressionConstant3Vector_0 {class=MaterialExpressionConstant3Vector, properties {constant=<rgba 0.2 0.2 1 0>}}
  MaterialExpressionConstant4Vector_13 {class=MaterialExpressionConstant4Vector, inputs [], properties {constant=<rgba -1 -1 1.5 1>}}
  MaterialExpressionConstant4Vector_6 {class=MaterialExpressionConstant4Vector, inputs [], properties {constant=<rgba 0.5816 0.8406 0.5336 0>}}
  MaterialExpressionConstant_1 {class=MaterialExpressionConstant}
  MaterialExpressionConstant_2 {class=MaterialExpressionConstant, properties {r=1}}
  MaterialExpressionConstant_4 {class=MaterialExpressionConstant, properties {r=0.6}}
  MaterialExpressionConstant_5 {class=MaterialExpressionConstant, properties {r=3}}
  MaterialExpressionConstant_6 {class=MaterialExpressionConstant}
  MaterialExpressionConstant_7 {class=MaterialExpressionConstant, properties {r=1}}
  MaterialExpressionConstant_8 {class=MaterialExpressionConstant}
  MaterialExpressionConstant_9 {class=MaterialExpressionConstant, properties {r=0.5}}
  MaterialExpressionConvert_1 {
   class=MaterialExpressionConvert
   inputs [None {from=MaterialExpressionNamedRerouteUsage_173}]
   properties {
    nodeName=Break Float3
    convertInputs [{
      type=Vector3
      expressionInput {
       expression=&self:MaterialExpressionNamedRerouteUsage_173
       inputName=None
       mask=0
       maskA=0
       maskB=0
       maskG=0
       maskR=0
       outputIndex=0
      }
     }]
    convertOutputs [{type=Scalar}, {type=Scalar}, {type=Scalar}]
   }
  }
  MaterialExpressionConvert_2 {
   class=MaterialExpressionConvert
   inputs [
    None {from=MaterialExpressionComponentMask_12}
    None {from=MaterialExpressionComponentMask_11}
   ]
   properties {
    nodeName=Make Vector2
    convertInputs [
     {
      type=Scalar
      expressionInput {
       expression=&self:MaterialExpressionComponentMask_12
       inputName=None
       mask=0
       maskA=0
       maskB=0
       maskG=0
       maskR=0
       outputIndex=0
      }
     }
     {
      type=Scalar
      expressionInput {
       expression=&self:MaterialExpressionComponentMask_11
       inputName=None
       mask=0
       maskA=0
       maskB=0
       maskG=0
       maskR=0
       outputIndex=0
      }
     }
    ]
    convertOutputs [{type=Vector2}]
   }
  }
  MaterialExpressionConvert_4 {class=MaterialExpressionConvert, inputs [], properties {convertOutputs [{type=Vector3, defaultValue=<rgba -1 -1 1 1>}]}}
  MaterialExpressionConvert_5 {class=MaterialExpressionConvert, inputs [], properties {convertOutputs [{type=Vector3, defaultValue=<rgba 1 1 1 1>}]}}
  MaterialExpressionDesaturation_0 {
   class=MaterialExpressionDesaturation
   inputs [
    None {from=MaterialExpressionMultiply_15, output=""}
    Fraction {from=MaterialExpressionConstant_2, output=""}
   ]
   properties {}
  }
  MaterialExpressionDesaturation_1 {
   class=MaterialExpressionDesaturation
   inputs [
    None {from=MaterialExpressionTextureSample_12, output=RGB}
    Fraction {from=MaterialExpressionScalarParameter_55, output=""}
   ]
   properties {}
  }
  MaterialExpressionDivide_0 {
   class=MaterialExpressionDivide
   inputs [
    A {from=MaterialExpressionNamedRerouteUsage_29}
    B {from=MaterialExpressionScalarParameter_28}
   ]
  }
  MaterialExpressionDivide_1 {
   class=MaterialExpressionDivide
   inputs [
    A {from=MaterialExpressionTime_1}
    B {from=""}
   ]
   properties {constB=-20}
  }
  MaterialExpressionDivide_11 {
   class=MaterialExpressionDivide
   inputs [
    A {from=MaterialExpressionScalarParameter_62}
    B {from=""}
   ]
  }
  MaterialExpressionDivide_14 {
   class=MaterialExpressionDivide
   inputs [
    A {from=MaterialExpressionDivide_9}
    B {from=""}
   ]
   properties {constB=10}
  }
  MaterialExpressionDivide_28 {
   class=MaterialExpressionDivide
   inputs [
    A {from=MaterialExpressionSubtract_41}
    B {from=MaterialExpressionNamedRerouteUsage_148}
   ]
  }
  MaterialExpressionDivide_5 {
   class=MaterialExpressionDivide
   inputs [
    A {from=MaterialExpressionAdd_10}
    B {from=MaterialExpressionAdd_13}
   ]
  }
  MaterialExpressionDivide_6 {
   class=MaterialExpressionDivide
   inputs [
    A {from=MaterialExpressionNamedRerouteUsage_70}
    B {from=""}
   ]
  }
  MaterialExpressionDivide_7 {
   class=MaterialExpressionDivide
   inputs [
    A {from=MaterialExpressionAdd_22}
    B {from=MaterialExpressionScalarParameter_50}
   ]
  }
  MaterialExpressionDivide_9 {
   class=MaterialExpressionDivide
   inputs [
    A {from=MaterialExpressionTime_2}
    B {from=""}
   ]
   properties {constB=10}
  }
  MaterialExpressionDotProduct_11 {
   class=MaterialExpressionDotProduct
   inputs [
    A {from=MaterialExpressionConvert_5, output=""}
    B {from=MaterialExpressionNamedRerouteDeclaration_38, output=""}
   ]
   properties {}
  }
  MaterialExpressionDotProduct_3 {
   class=MaterialExpressionDotProduct
   inputs [
    A {from=MaterialExpressionNamedRerouteUsage_40, output=""}
    B {from=MaterialExpressionVectorParameter_3, output=RGB}
   ]
   properties {}
  }
  MaterialExpressionDotProduct_5 {
   class=MaterialExpressionDotProduct
   inputs [
    A {from=MaterialExpressionNamedRerouteUsage_75, output=""}
    B {from=MaterialExpressionConstant4Vector_13, output=RGB}
   ]
   properties {}
  }
  MaterialExpressionDotProduct_6 {
   class=MaterialExpressionDotProduct
   inputs [
    A {from=MaterialExpressionCameraVectorWS_5, output=""}
    B {from=MaterialExpressionVertexNormalWS_7, output=""}
   ]
   properties {}
  }
  MaterialExpressionDotProduct_7 {
   class=MaterialExpressionDotProduct
   inputs [
    A {from=MaterialExpressionMaterialFunctionCall_6, output=Result}
    B {from=MaterialExpressionNormalize_3, output=""}
   ]
   properties {}
  }
  MaterialExpressionDotProduct_9 {
   class=MaterialExpressionDotProduct
   inputs [
    A {from=MaterialExpressionMax_3, output=""}
    B {from=MaterialExpressionNamedRerouteUsage_72, output=""}
   ]
   properties {}
  }
  MaterialExpressionFrac_1 {class=MaterialExpressionFrac, inputs [None {from=MaterialExpressionDivide_14, output=""}], properties {}}
  MaterialExpressionFrac_2 {class=MaterialExpressionFrac, inputs [None {from=MaterialExpressionDivide_1, output=""}], properties {}}
  MaterialExpressionFunctionInput_0 {
   class=MaterialExpressionFunctionInput
   inputs [Preview {from=MaterialExpressionObjectPositionWS_4, output=""}]
   properties {bUsePreviewValueAsDefault=true, inputName=Object Position WS, sortPriority=140}
  }
  MaterialExpressionFunctionInput_1 {
   class=MaterialExpressionFunctionInput
   inputs [Preview {from="", output=""}]
   properties {bUsePreviewValueAsDefault=true, inputName=Pre-Skinned Scale, sortPriority=130, previewValue=<1 1 1 1>}
  }
  MaterialExpressionFunctionInput_24 {
   class=MaterialExpressionFunctionInput
   inputs [Preview {from="", output=""}]
   properties {inputName=MaterialAttributes, inputType=FunctionInput_MaterialAttributes, sortPriority=-1}
  }
  MaterialExpressionFunctionInput_25 {
   class=MaterialExpressionFunctionInput
   inputs [Preview {from="", output=""}]
   properties {inputName=Mask Variant, inputType=FunctionInput_Scalar, sortPriority=1}
  }
  MaterialExpressionFunctionInput_26 {
   class=MaterialExpressionFunctionInput
   inputs [Preview {from="", output=""}]
   properties {inputName=Mask_InnerRadialGlow, inputType=FunctionInput_Scalar, sortPriority=1}
  }
  MaterialExpressionFunctionInput_27 {
   class=MaterialExpressionFunctionInput
   inputs [Preview {from="", output=""}]
   properties {inputName=Mask_Logo, inputType=FunctionInput_Scalar, sortPriority=1}
  }
  MaterialExpressionFunctionInput_28 {
   class=MaterialExpressionFunctionInput
   inputs [Preview {from="", output=""}]
   properties {inputName=Mask_SpriteFresnel, inputType=FunctionInput_Scalar, sortPriority=1}
  }
  MaterialExpressionFunctionInput_29 {
   class=MaterialExpressionFunctionInput
   inputs [Preview {from="", output=""}]
   properties {inputName=WPO, inputType=FunctionInput_Scalar, sortPriority=2}
  }
  MaterialExpressionFunctionInput_30 {
   class=MaterialExpressionFunctionInput
   inputs [Preview {from="", output=""}]
   properties {inputName=Mask_BlinkingLights, inputType=FunctionInput_Scalar, sortPriority=1}
  }
  MaterialExpressionFunctionInput_31 {
   class=MaterialExpressionFunctionInput
   inputs [Preview {from="", output=""}]
   properties {inputName=Mask_DiffuseChannelMask, inputType=FunctionInput_Scalar, sortPriority=1}
  }
  MaterialExpressionFunctionInput_7 {
   class=MaterialExpressionFunctionInput
   inputs [Preview {from=MaterialExpressionShaderStageSwitch_4, output=""}]
   properties {bUsePreviewValueAsDefault=true, inputName=Pre-Skinned Local Pos, sortPriority=100}
  }
  MaterialExpressionFunctionInput_8 {
   class=MaterialExpressionFunctionInput
   inputs [Preview {from=MaterialExpressionBounds_0, output=Max}]
   properties {bUsePreviewValueAsDefault=true, inputName=Pre-Skinned Bounds Max, sortPriority=120}
  }
  MaterialExpressionFunctionInput_9 {
   class=MaterialExpressionFunctionInput
   inputs [Preview {from=MaterialExpressionBounds_0, output=Min}]
   properties {bUsePreviewValueAsDefault=true, inputName=Pre-Skinned Bounds Min, sortPriority=110}
  }
  MaterialExpressionFunctionOutput_1 {
   class=MaterialExpressionFunctionOutput
   inputs [None {from=MaterialExpressionStaticSwitchParameter_3, output=""}]
   properties {outputName=SetMaterialAttributes}
  }
  MaterialExpressionGetMaterialAttributes_0 {
   class=MaterialExpressionGetMaterialAttributes
   inputs [None {from=MaterialExpressionFunctionInput_24, output=""}]
   properties {attributeGetTypes [
     #1
     #2
     #3
     #4
     #5
     #6
    ]}
  }
  MaterialExpressionIf_0 {
   class=MaterialExpressionIf
   inputs [
    A {from=MaterialExpressionBounds_1, output=Extents}
    B {from=MaterialExpressionMultiply_60, output=""}
    "A > B" {from=MaterialExpressionConstant_7, output=""}
    "A == B" {from="", output=""}
    "A < B" {from=MaterialExpressionConstant_8, output=""}
   ]
   properties {}
  }
  MaterialExpressionIf_1 {
   class=MaterialExpressionIf
   inputs [
    A {from=MaterialExpressionComponentMask_17, output=""}
    B {from="", output=""}
    "A > B" {from=MaterialExpressionConstant_7, output=""}
    "A == B" {from="", output=""}
    "A < B" {from=MaterialExpressionConstant_8, output=""}
   ]
   properties {constB=30}
  }
  MaterialExpressionLinearInterpolate_0 {
   class=MaterialExpressionLinearInterpolate
   inputs [
    A {from=MaterialExpressionScalarParameter_3}
    B {from=""}
    Alpha {from=MaterialExpressionNamedRerouteUsage_21}
   ]
  }
  MaterialExpressionLinearInterpolate_1 {
   class=MaterialExpressionLinearInterpolate
   inputs [
    A {from=MaterialExpressionMultiply_59}
    B {from=MaterialExpressionVectorParameter_20, output=RGB}
    Alpha {from=MaterialExpressionVectorParameter_20, output=RGB}
   ]
  }
  MaterialExpressionLinearInterpolate_10 {
   class=MaterialExpressionLinearInterpolate
   inputs [
    A {from=MaterialExpressionLinearInterpolate_25}
    B {from=MaterialExpressionMultiply_22}
    Alpha {from=MaterialExpressionMultiply_44}
   ]
  }
  MaterialExpressionLinearInterpolate_11 {
   class=MaterialExpressionLinearInterpolate
   inputs [
    A {from=MaterialExpressionLinearInterpolate_41}
    B {from=MaterialExpressionNamedRerouteUsage_17}
    Alpha {from=MaterialExpressionScalarParameter_10}
   ]
  }
  MaterialExpressionLinearInterpolate_12 {
   class=MaterialExpressionLinearInterpolate
   inputs [
    A {from=MaterialExpressionVectorParameter_15, output=RGB}
    B {from=MaterialExpressionVectorParameter_12, output=RGB}
    Alpha {from=MaterialExpressionNamedRerouteUsage_19}
   ]
  }
  MaterialExpressionLinearInterpolate_13 {
   class=MaterialExpressionLinearInterpolate
   inputs [
    A {from=MaterialExpressionReroute_7}
    B {from=MaterialExpressionMultiply_14}
    Alpha {from=MaterialExpressionNamedRerouteUsage_3}
   ]
  }
  MaterialExpressionLinearInterpolate_14 {
   class=MaterialExpressionLinearInterpolate
   inputs [
    A {from=MaterialExpressionReroute_38}
    B {from=MaterialExpressionReroute_30}
    Alpha {from=MaterialExpressionSaturate_0}
   ]
  }
  MaterialExpressionLinearInterpolate_15 {
   class=MaterialExpressionLinearInterpolate
   inputs [
    A {from=""}
    B {from=MaterialExpressionLinearInterpolate_16}
    Alpha {from=MaterialExpressionNamedRerouteUsage_37}
   ]
   properties {constA=1}
  }
  MaterialExpressionLinearInterpolate_16 {
   class=MaterialExpressionLinearInterpolate
   inputs [
    A {from=MaterialExpressionMultiply_28}
    B {from=MaterialExpressionVectorParameter_18, output=RGB}
    Alpha {from=MaterialExpressionVectorParameter_18, output=RGB}
   ]
  }
  MaterialExpressionLinearInterpolate_18 {
   class=MaterialExpressionLinearInterpolate
   inputs [
    A {from=MaterialExpressionNamedRerouteUsage_36}
    B {from=""}
    Alpha {from=MaterialExpressionNamedRerouteUsage_35}
   ]
   properties {constB=0}
  }
  MaterialExpressionLinearInterpolate_19 {
   class=MaterialExpressionLinearInterpolate
   inputs [
    A {from=MaterialExpressionMaterialFunctionCall_1, output=Result}
    B {from=""}
    Alpha {from=MaterialExpressionNamedRerouteUsage_74}
   ]
   properties {constB=0}
  }
  MaterialExpressionLinearInterpolate_20 {
   class=MaterialExpressionLinearInterpolate
   inputs [
    A {from=MaterialExpressionAdd_29}
    B {from=MaterialExpressionReroute_18}
    Alpha {from=MaterialExpressionNamedRerouteUsage_62}
   ]
   properties {constB=0}
  }
  MaterialExpressionLinearInterpolate_21 {
   class=MaterialExpressionLinearInterpolate
   inputs [
    A {from=MaterialExpressionNamedRerouteUsage_16}
    B {from=MaterialExpressionConstant_6}
    Alpha {from=MaterialExpressionReroute_41}
   ]
  }
  MaterialExpressionLinearInterpolate_23 {
   class=MaterialExpressionLinearInterpolate
   inputs [
    A {from=MaterialExpressionNamedRerouteUsage_57}
    B {from=MaterialExpressionReroute_31}
    Alpha {from=""}
   ]
   properties {constAlpha=0.85}
  }
  MaterialExpressionLinearInterpolate_24 {
   class=MaterialExpressionLinearInterpolate
   inputs [
    A {from=MaterialExpressionVectorParameter_22, output=RGB}
    B {from=MaterialExpressionVectorParameter_21, output=RGB}
    Alpha {from=MaterialExpressionPower_10}
   ]
  }
  MaterialExpressionLinearInterpolate_25 {
   class=MaterialExpressionLinearInterpolate
   inputs [
    A {from=MaterialExpressionLinearInterpolate_29}
    B {from=MaterialExpressionLinearInterpolate_42}
    Alpha {from=MaterialExpressionPower_9}
   ]
   properties {constB=0}
  }
  MaterialExpressionLinearInterpolate_26 {
   class=MaterialExpressionLinearInterpolate
   inputs [
    A {from=MaterialExpressionReroute_29}
    B {from=MaterialExpressionReroute_15}
    Alpha {from=MaterialExpressionLinearInterpolate_40}
   ]
  }
  MaterialExpressionLinearInterpolate_28 {
   class=MaterialExpressionLinearInterpolate
   inputs [
    A {from=MaterialExpressionNamedRerouteUsage_7}
    B {from=MaterialExpressionNamedRerouteUsage_6}
    Alpha {from=MaterialExpressionScalarParameter_35}
   ]
  }
  MaterialExpressionLinearInterpolate_29 {
   class=MaterialExpressionLinearInterpolate
   inputs [
    A {from=MaterialExpressionMultiply_10}
    B {from=MaterialExpressionVectorParameter_19, output=RGB}
    Alpha {from=MaterialExpressionReroute_21}
   ]
  }
  MaterialExpressionLinearInterpolate_31 {
   class=MaterialExpressionLinearInterpolate
   inputs [
    A {from=MaterialExpressionConstant3Vector_0}
    B {from=MaterialExpressionMultiply_58}
    Alpha {from=MaterialExpressionNamedRerouteUsage_104}
   ]
  }
  MaterialExpressionLinearInterpolate_33 {
   class=MaterialExpressionLinearInterpolate
   inputs [
    A {from=MaterialExpressionVectorParameter_24, output=RGB}
    B {from=MaterialExpressionVectorParameter_19, output=RGB}
    Alpha {from=MaterialExpressionReroute_21}
   ]
  }
  MaterialExpressionLinearInterpolate_35 {
   class=MaterialExpressionLinearInterpolate
   inputs [
    A {from=MaterialExpressionLinearInterpolate_43}
    B {from=MaterialExpressionPower_1}
    Alpha {from=""}
   ]
  }
  MaterialExpressionLinearInterpolate_36 {
   class=MaterialExpressionLinearInterpolate
   inputs [
    A {from=MaterialExpressionMultiply_36}
    B {from=MaterialExpressionVectorParameter_26, output=RGB}
    Alpha {from=MaterialExpressionVectorParameter_26, output=RGB}
   ]
  }
  MaterialExpressionLinearInterpolate_4 {
   class=MaterialExpressionLinearInterpolate
   inputs [
    A {from=MaterialExpressionTextureSample_13, output=R}
    B {from=MaterialExpressionTextureSample_14, output=G}
    Alpha {from=MaterialExpressionAbs_0}
   ]
  }
  MaterialExpressionLinearInterpolate_40 {
   class=MaterialExpressionLinearInterpolate
   inputs [
    A {from=""}
    B {from=""}
    Alpha {from=MaterialExpressionNamedRerouteUsage_83}
   ]
   properties {constA=0.3}
  }
  MaterialExpressionLinearInterpolate_41 {
   class=MaterialExpressionLinearInterpolate
   inputs [
    A {from=MaterialExpressionLinearInterpolate_12}
    B {from=MaterialExpressionVectorParameter_15, output=RGB}
    Alpha {from=MaterialExpressionMaterialFunctionCall_12, output=Result}
   ]
  }
  MaterialExpressionLinearInterpolate_42 {
   class=MaterialExpressionLinearInterpolate
   inputs [
    A {from=MaterialExpressionMultiply_61}
    B {from=MaterialExpressionLinearInterpolate_24}
    Alpha {from=MaterialExpressionPower_10}
   ]
  }
  MaterialExpressionLinearInterpolate_43 {
   class=MaterialExpressionLinearInterpolate
   inputs [
    A {from=MaterialExpressionNamedRerouteUsage_67}
    B {from=""}
    Alpha {from=MaterialExpressionMaterialFunctionCall_0, output=Result}
   ]
   properties {constB=0}
  }
  MaterialExpressionLinearInterpolate_9 {
   class=MaterialExpressionLinearInterpolate
   inputs [
    A {from=MaterialExpressionReroute_33}
    B {from=""}
    Alpha {from=MaterialExpressionMaterialFunctionCall_0, output=Result}
   ]
   properties {constB=0}
  }
  MaterialExpressionLocalPosition_5 {class=MaterialExpressionLocalPosition, inputs [], properties {includedOffsets=ExcludeOffsets, localOrigin=InstancePreSkinning}}
  MaterialExpressionMaterialFunctionCall_0 {
   class=MaterialExpressionMaterialFunctionCall
   inputs [
    In {from=MaterialExpressionAdd_0}
    Contrast {from=MaterialExpressionScalarParameter_6}
   ]
   properties {materialFunction=&CheapContrast}
  }
  MaterialExpressionMaterialFunctionCall_1 {
   class=MaterialExpressionMaterialFunctionCall
   inputs [
    In {from=MaterialExpressionAdd_12}
    Contrast {from=MaterialExpressionScalarParameter_18}
   ]
   properties {materialFunction=&CheapContrast}
  }
  MaterialExpressionMaterialFunctionCall_10 {
   class=MaterialExpressionMaterialFunctionCall
   inputs [
    BaseNormal {from=MaterialExpressionNamedRerouteUsage_11}
    AdditionalNormal {from=MaterialExpressionMaterialFunctionCall_11, output=Result}
   ]
   properties {materialFunction=&BlendAngleCorrectedNormals}
  }
  MaterialExpressionMaterialFunctionCall_11 {
   class=MaterialExpressionMaterialFunctionCall
   inputs [
    Normal {from=MaterialExpressionTextureSample_0, output=RGB}
    Flatness {from=MaterialExpressionLinearInterpolate_0}
   ]
   properties {materialFunction=&FlattenNormal}
  }
  MaterialExpressionMaterialFunctionCall_12 {
   class=MaterialExpressionMaterialFunctionCall
   inputs [
    In {from=MaterialExpressionAdd_43}
    Contrast {from=MaterialExpressionConstant_5}
   ]
   properties {materialFunction=&CheapContrast}
  }
  MaterialExpressionMaterialFunctionCall_13 {
   class=MaterialExpressionMaterialFunctionCall
   inputs [
    Color_Morning {from=MaterialExpressionVectorParameter_31, output=RGB}
    Color_Day {from=MaterialExpressionVectorParameter_32, output=RGB}
    Color_Evening {from=MaterialExpressionVectorParameter_33, output=RGB}
    Color_Night {from=MaterialExpressionVectorParameter_34, output=RGB}
   ]
   properties {materialFunction=&Color_Time_of_Day}
  }
  MaterialExpressionMaterialFunctionCall_20 {
   class=MaterialExpressionMaterialFunctionCall
   inputs [
    Texture Scale {from=MaterialExpressionScalarParameter_26}
    Normal {from=""}
    InPosition {from=MaterialExpressionNamedRerouteUsage_87}
    Triplanar Fade {from=MaterialExpressionScalarParameter_27}
    Derivatives Scale {from=""}
   ]
   properties {materialFunction=&/MF_TriplanarDithered_PreSkinned}
  }
  MaterialExpressionMaterialFunctionCall_4 {
   class=MaterialExpressionMaterialFunctionCall
   inputs [
    In {from=MaterialExpressionAdd_3}
    Contrast {from=MaterialExpressionScalarParameter_19}
   ]
   properties {materialFunction=&CheapContrast}
  }
  MaterialExpressionMaterialFunctionCall_59 {
   class=MaterialExpressionMaterialFunctionCall
   inputs [
    Position {from=MaterialExpressionNamedRerouteUsage_171}
    UVRadialScale {from=MaterialExpressionMultiply_55}
    ObjectRadius {from=MaterialExpressionNamedRerouteUsage_175}
   ]
   properties {materialFunction=&/MF_ObjectStabilizedScreenSpaceUVs}
  }
  MaterialExpressionMaterialFunctionCall_6 {
   class=MaterialExpressionMaterialFunctionCall
   inputs []
   properties {materialFunction=&CameraDirectionVector}
  }
  MaterialExpressionMax_0 {
   class=MaterialExpressionMax
   inputs [
    A {from=MaterialExpressionNamedRerouteUsage_9}
    B {from=MaterialExpressionNamedRerouteUsage_10}
   ]
  }
  MaterialExpressionMax_10 {
   class=MaterialExpressionMax
   inputs [
    A {from=MaterialExpressionConvert_1}
    B {from=MaterialExpressionConvert_1}
   ]
  }
  MaterialExpressionMax_11 {
   class=MaterialExpressionMax
   inputs [
    A {from=MaterialExpressionMax_10}
    B {from=MaterialExpressionConvert_1}
   ]
  }
  MaterialExpressionMax_3 {
   class=MaterialExpressionMax
   inputs [
    A {from=MaterialExpressionLinearInterpolate_9}
    B {from=MaterialExpressionPower_11}
   ]
   properties {constB=0}
  }
  MaterialExpressionMultiply_0 {
   class=MaterialExpressionMultiply
   inputs [
    A {from=MaterialExpressionReroute_16}
    B {from=MaterialExpressionMultiply_6}
   ]
  }
  MaterialExpressionMultiply_10 {
   class=MaterialExpressionMultiply
   inputs [
    A {from=MaterialExpressionLinearInterpolate_26}
    B {from=MaterialExpressionLinearInterpolate_33}
   ]
  }
  MaterialExpressionMultiply_11 {
   class=MaterialExpressionMultiply
   inputs [
    A {from=MaterialExpressionNamedRerouteUsage_13}
    B {from=MaterialExpressionScalarParameter_1}
   ]
  }
  MaterialExpressionMultiply_12 {
   class=MaterialExpressionMultiply
   inputs [
    A {from=MaterialExpressionNamedRerouteUsage_69}
    B {from=""}
   ]
   properties {constB=0.1}
  }
  MaterialExpressionMultiply_13 {
   class=MaterialExpressionMultiply
   inputs [
    A {from=MaterialExpressionMultiply_46}
    B {from=""}
   ]
   properties {constB=1.5}
  }
  MaterialExpressionMultiply_14 {
   class=MaterialExpressionMultiply
   inputs [
    A {from=MaterialExpressionReroute_7}
    B {from=MaterialExpressionVectorParameter_17, output=RGB}
   ]
  }
  MaterialExpressionMultiply_141 {
   class=MaterialExpressionMultiply
   inputs [
    A {from=MaterialExpressionNamedRerouteUsage_165}
    B {from=""}
   ]
   properties {constB=0.5}
  }
  MaterialExpressionMultiply_147 {
   class=MaterialExpressionMultiply
   inputs [
    A {from=MaterialExpressionDotProduct_11}
    B {from=""}
   ]
   properties {constB=0.3333}
  }
  MaterialExpressionMultiply_15 {
   class=MaterialExpressionMultiply
   inputs [
    A {from=MaterialExpressionMultiply_11}
    B {from=MaterialExpressionScalarParameter_4}
   ]
  }
  MaterialExpressionMultiply_16 {
   class=MaterialExpressionMultiply
   inputs [
    A {from=MaterialExpressionNamedRerouteUsage_25}
    B {from=MaterialExpressionScalarParameter_40}
   ]
  }
  MaterialExpressionMultiply_17 {
   class=MaterialExpressionMultiply
   inputs [
    A {from=MaterialExpressionLinearInterpolate_4}
    B {from=MaterialExpressionScalarParameter_34}
   ]
  }
  MaterialExpressionMultiply_18 {
   class=MaterialExpressionMultiply
   inputs [
    A {from=MaterialExpressionDotProduct_3}
    B {from=MaterialExpressionVectorParameter_3, output=RGBMultiplier}
   ]
  }
  MaterialExpressionMultiply_19 {
   class=MaterialExpressionMultiply
   inputs [
    A {from=MaterialExpressionReroute_37}
    B {from=MaterialExpressionPower_11}
   ]
   properties {constB=3}
  }
  MaterialExpressionMultiply_20 {
   class=MaterialExpressionMultiply
   inputs [
    A {from=MaterialExpressionMultiply_19}
    B {from=""}
   ]
   properties {constB=5}
  }
  MaterialExpressionMultiply_22 {
   class=MaterialExpressionMultiply
   inputs [
    A {from=MaterialExpressionMultiply_41}
    B {from=""}
   ]
   properties {constB=2}
  }
  MaterialExpressionMultiply_23 {
   class=MaterialExpressionMultiply
   inputs [
    A {from=MaterialExpressionReroute_27}
    B {from=MaterialExpressionLinearInterpolate_15}
   ]
  }
  MaterialExpressionMultiply_24 {
   class=MaterialExpressionMultiply
   inputs [
    A {from=MaterialExpressionNamedRerouteUsage_46}
    B {from=MaterialExpressionScalarParameter_37}
   ]
  }
  MaterialExpressionMultiply_25 {
   class=MaterialExpressionMultiply
   inputs [
    A {from=MaterialExpressionNamedRerouteUsage_47}
    B {from=MaterialExpressionScalarParameter_39}
   ]
  }
  MaterialExpressionMultiply_26 {
   class=MaterialExpressionMultiply
   inputs [
    A {from=MaterialExpressionDesaturation_1}
    B {from=MaterialExpressionVectorParameter_1, output=RGB}
   ]
  }
  MaterialExpressionMultiply_28 {
   class=MaterialExpressionMultiply
   inputs [
    A {from=MaterialExpressionReroute_27}
    B {from=MaterialExpressionVectorParameter_18, output=RGB}
   ]
  }
  MaterialExpressionMultiply_30 {
   class=MaterialExpressionMultiply
   inputs [
    A {from=MaterialExpressionSine_0}
    B {from=MaterialExpressionScalarParameter_38}
   ]
   properties {constB=0.2}
  }
  MaterialExpressionMultiply_31 {
   class=MaterialExpressionMultiply
   inputs [
    A {from=MaterialExpressionLinearInterpolate_35}
    B {from=MaterialExpressionDotProduct_5}
   ]
  }
  MaterialExpressionMultiply_34 {
   class=MaterialExpressionMultiply
   inputs [
    A {from=MaterialExpressionLinearInterpolate_36}
    B {from=MaterialExpressionNamedRerouteUsage_76}
   ]
  }
  MaterialExpressionMultiply_35 {
   class=MaterialExpressionMultiply
   inputs [
    A {from=MaterialExpressionSubtract_3}
    B {from=MaterialExpressionScalarParameter_62}
   ]
  }
  MaterialExpressionMultiply_36 {
   class=MaterialExpressionMultiply
   inputs [
    A {from=MaterialExpressionReroute_35}
    B {from=MaterialExpressionVectorParameter_26, output=RGB}
   ]
  }
  MaterialExpressionMultiply_37 {
   class=MaterialExpressionMultiply
   inputs [
    A {from=MaterialExpressionVectorParameter_21, output=RGB}
    B {from=MaterialExpressionPower_10}
   ]
  }
  MaterialExpressionMultiply_38 {
   class=MaterialExpressionMultiply
   inputs [
    A {from=MaterialExpressionAdd_19}
    B {from=MaterialExpressionScalarParameter_41}
   ]
  }
  MaterialExpressionMultiply_41 {
   class=MaterialExpressionMultiply
   inputs [
    A {from=MaterialExpressionNamedRerouteUsage_78}
    B {from=MaterialExpressionMultiply_37}
   ]
  }
  MaterialExpressionMultiply_44 {
   class=MaterialExpressionMultiply
   inputs [
    A {from=MaterialExpressionSmoothStep_1}
    B {from=""}
   ]
   properties {constB=0.8}
  }
  MaterialExpressionMultiply_45 {
   class=MaterialExpressionMultiply
   inputs [
    A {from=MaterialExpressionNamedRerouteUsage_53}
    B {from=MaterialExpressionScalarParameter_23}
   ]
  }
  MaterialExpressionMultiply_46 {
   class=MaterialExpressionMultiply
   inputs [
    A {from=MaterialExpressionDotProduct_7}
    B {from=""}
   ]
   properties {constB=-1}
  }
  MaterialExpressionMultiply_47 {
   class=MaterialExpressionMultiply
   inputs [
    A {from=MaterialExpressionMax_11}
    B {from=""}
   ]
   properties {constB=1.4115, desc=a bit arbitrary...}
  }
  MaterialExpressionMultiply_49 {
   class=MaterialExpressionMultiply
   inputs [
    A {from=MaterialExpressionAdd_2}
    B {from=MaterialExpressionMaterialFunctionCall_13, output=Result}
   ]
  }
  MaterialExpressionMultiply_53 {
   class=MaterialExpressionMultiply
   inputs [
    A {from=MaterialExpressionNamedRerouteUsage_65}
    B {from=MaterialExpressionScalarParameter_31}
   ]
  }
  MaterialExpressionMultiply_55 {
   class=MaterialExpressionMultiply
   inputs [
    A {from=MaterialExpressionNamedRerouteUsage_176}
    B {from=MaterialExpressionScalarParameter_25}
   ]
  }
  MaterialExpressionMultiply_57 {
   class=MaterialExpressionMultiply
   inputs [
    A {from=MaterialExpressionLinearInterpolate_1}
    B {from=MaterialExpressionNamedRerouteUsage_38}
   ]
  }
  MaterialExpressionMultiply_58 {
   class=MaterialExpressionMultiply
   inputs [
    A {from=MaterialExpressionConstant3Vector_0}
    B {from=MaterialExpressionConvert_4}
   ]
  }
  MaterialExpressionMultiply_59 {
   class=MaterialExpressionMultiply
   inputs [
    A {from=MaterialExpressionReroute_32}
    B {from=MaterialExpressionVectorParameter_20, output=RGB}
   ]
  }
  MaterialExpressionMultiply_6 {
   class=MaterialExpressionMultiply
   inputs [
    A {from=MaterialExpressionNamedRerouteUsage_4}
    B {from=MaterialExpressionVectorParameter_7, output=RGB}
   ]
  }
  MaterialExpressionMultiply_60 {
   class=MaterialExpressionMultiply
   inputs [
    A {from=MaterialExpressionBounds_11, output=Extents}
    B {from=""}
   ]
   properties {constB=2.5}
  }
  MaterialExpressionMultiply_61 {
   class=MaterialExpressionMultiply
   inputs [
    A {from=MaterialExpressionAdd_5}
    B {from=MaterialExpressionLinearInterpolate_24}
   ]
  }
  MaterialExpressionMultiply_62 {
   class=MaterialExpressionMultiply
   inputs [
    A {from=MaterialExpressionNamedRerouteDeclaration_43}
    B {from=MaterialExpressionNamedRerouteDeclaration_44}
   ]
  }
  MaterialExpressionMultiply_9 {
   class=MaterialExpressionMultiply
   inputs [
    A {from=MaterialExpressionConstant_4}
    B {from=MaterialExpressionMultiply_46}
   ]
  }
  MaterialExpressionNamedRerouteDeclaration_0 {
   class=MaterialExpressionNamedRerouteDeclaration
   inputs [None {from=MaterialExpressionMultiply_26}]
   properties {name=Old_Holofoil, nodeColor=<rgba 0 0.5547 1 1>}
  }
  MaterialExpressionNamedRerouteDeclaration_1 {
   class=MaterialExpressionNamedRerouteDeclaration
   inputs [None {from=MaterialExpressionGetMaterialAttributes_0, output=BaseColor}]
   properties {name=IN_BaseColor, nodeColor=<rgba 0.0781 0 1 1>}
  }
  MaterialExpressionNamedRerouteDeclaration_10 {
   class=MaterialExpressionNamedRerouteDeclaration
   inputs [None {from=MaterialExpressionFunctionInput_29}]
   properties {name=IN_WPO, nodeColor=<rgba 0 0.9766 1 1>}
  }
  MaterialExpressionNamedRerouteDeclaration_11 {
   class=MaterialExpressionNamedRerouteDeclaration
   inputs [None {from=MaterialExpressionGetMaterialAttributes_0, output=EmissiveColor}]
   properties {name=IN_Emissive, nodeColor=<rgba 0.9688 1 0 1>}
  }
  MaterialExpressionNamedRerouteDeclaration_12 {
   class=MaterialExpressionNamedRerouteDeclaration
   inputs [None {from=MaterialExpressionFunctionInput_7}]
   properties {name="Pre-Skinned Local Pos (V3)", nodeColor=<rgba 0 0.1797 1 1>}
  }
  MaterialExpressionNamedRerouteDeclaration_13 {
   class=MaterialExpressionNamedRerouteDeclaration
   inputs [None {from=MaterialExpressionLinearInterpolate_18}]
   properties {name=M_Logo+Variant, nodeColor=<rgba 0 1 0.1328 1>}
  }
  MaterialExpressionNamedRerouteDeclaration_14 {
   class=MaterialExpressionNamedRerouteDeclaration
   inputs [None {from=MaterialExpressionTextureSample_2, output=RGB}]
   properties {name=HoloColors, nodeColor=<rgba 0 0.1562 1 1>}
  }
  MaterialExpressionNamedRerouteDeclaration_15 {
   class=MaterialExpressionNamedRerouteDeclaration
   inputs [None {from=MaterialExpressionShaderStageSwitch_0}]
   properties {name=PreSkinnedNormal_VS-PS, nodeColor=<rgba 0.3587 0.1657 1 1>}
  }
  MaterialExpressionNamedRerouteDeclaration_16 {
   class=MaterialExpressionNamedRerouteDeclaration
   inputs [None {from=MaterialExpressionSaturate_1}]
   properties {name=IN_BaseColor+Emissive, nodeColor=<rgba 0 1 0.5078 1>}
  }
  MaterialExpressionNamedRerouteDeclaration_17 {
   class=MaterialExpressionNamedRerouteDeclaration
   inputs [None {from=MaterialExpressionTextureSample_10, output=RGB}]
   properties {name=HolofoilRGB, nodeColor=<rgba 0.5 1 0 1>}
  }
  MaterialExpressionNamedRerouteDeclaration_18 {
   class=MaterialExpressionNamedRerouteDeclaration
   inputs [None {from=MaterialExpressionFunctionInput_31}]
   properties {name=M_DiffuseMask, nodeColor=<rgba 0.875 1 0 1>}
  }
  MaterialExpressionNamedRerouteDeclaration_2 {
   class=MaterialExpressionNamedRerouteDeclaration
   inputs [None {from=MaterialExpressionGetMaterialAttributes_0, output=Metallic}]
   properties {name=In_Metallic, nodeColor=<rgba 0.2422 1 0 1>}
  }
  MaterialExpressionNamedRerouteDeclaration_20 {
   class=MaterialExpressionNamedRerouteDeclaration
   inputs [None {from=MaterialExpressionTextureObject_0}]
   properties {name=T_ThinFilm_RGB, nodeColor=<rgba 0 0.4141 1 1>}
  }
  MaterialExpressionNamedRerouteDeclaration_21 {
   class=MaterialExpressionNamedRerouteDeclaration
   inputs [None {from=MaterialExpressionMaterialFunctionCall_10, output=Result}]
   properties {name=OUT_Normals, nodeColor=<rgba 0.1953 1 0 1>}
  }
  MaterialExpressionNamedRerouteDeclaration_22 {
   class=MaterialExpressionNamedRerouteDeclaration
   inputs [None {from=MaterialExpressionLinearInterpolate_13}]
   properties {name=OUT_BaseColor, nodeColor=<rgba 0 0.2031 1 1>}
  }
  MaterialExpressionNamedRerouteDeclaration_23 {
   class=MaterialExpressionNamedRerouteDeclaration
   inputs [None {from=MaterialExpressionTextureSample_7, output=RGB}]
   properties {name=RGB_AberrationColors, nodeColor=<rgba 1 0.1172 0 1>}
  }
  MaterialExpressionNamedRerouteDeclaration_24 {
   class=MaterialExpressionNamedRerouteDeclaration
   inputs [None {from=MaterialExpressionFunctionInput_9}]
   properties {name="Pre-Skinned Bounds Min (V3)", nodeColor=<rgba 1 1 1 1>}
  }
  MaterialExpressionNamedRerouteDeclaration_25 {
   class=MaterialExpressionNamedRerouteDeclaration
   inputs [None {from=MaterialExpressionTextureObject_1}]
   properties {name=T_Glitter, nodeColor=<rgba 0.6406 0 1 1>}
  }
  MaterialExpressionNamedRerouteDeclaration_26 {
   class=MaterialExpressionNamedRerouteDeclaration
   inputs [None {from=MaterialExpressionTextureObject_2}]
   properties {name=T_ThinFilm, nodeColor=<rgba 0 0.4141 1 1>}
  }
  MaterialExpressionNamedRerouteDeclaration_27 {
   class=MaterialExpressionNamedRerouteDeclaration
   inputs [None {from=MaterialExpressionTextureObject_8}]
   properties {name=T_Noise, nodeColor=<rgba 0 0.0391 1 1>}
  }
  MaterialExpressionNamedRerouteDeclaration_28 {
   class=MaterialExpressionNamedRerouteDeclaration
   inputs [None {from=MaterialExpressionLinearInterpolate_14}]
   properties {name=OUT_Emissive, nodeColor=<rgba 0.8516 1 0 1>}
  }
  MaterialExpressionNamedRerouteDeclaration_29 {
   class=MaterialExpressionNamedRerouteDeclaration
   inputs [None {from=MaterialExpressionOneMinus_9}]
   properties {name=M_Fresnel, nodeColor=<rgba 1 0.9844 0 1>}
  }
  MaterialExpressionNamedRerouteDeclaration_3 {
   class=MaterialExpressionNamedRerouteDeclaration
   inputs [None {from=MaterialExpressionGetMaterialAttributes_0, output=Roughness}]
   properties {name=In_Roughness, nodeColor=<rgba 1 0 0.5625 1>}
  }
  MaterialExpressionNamedRerouteDeclaration_30 {
   class=MaterialExpressionNamedRerouteDeclaration
   inputs [None {from=MaterialExpressionLinearInterpolate_19}]
   properties {name=M_SpriteFresnel_Adjusted, nodeColor=<rgba 1 0 0.0703 1>}
  }
  MaterialExpressionNamedRerouteDeclaration_31 {
   class=MaterialExpressionNamedRerouteDeclaration
   inputs [None {from=MaterialExpressionGetMaterialAttributes_0, output=Normal}]
   properties {name=IN_Normals, nodeColor=<rgba 0 1 0.6016 1>}
  }
  MaterialExpressionNamedRerouteDeclaration_32 {
   class=MaterialExpressionNamedRerouteDeclaration
   inputs [None {from=MaterialExpressionFunctionInput_8}]
   properties {name="Pre-Skinned Bounds Max (V3)", nodeColor=<rgba 1 1 1 1>}
  }
  MaterialExpressionNamedRerouteDeclaration_33 {
   class=MaterialExpressionNamedRerouteDeclaration
   inputs [None {from=MaterialExpressionSubtract_40}]
   properties {name="Pre-Skinned Bounds Extents (V3)", nodeColor=<rgba 1 1 1 1>}
  }
  MaterialExpressionNamedRerouteDeclaration_34 {
   class=MaterialExpressionNamedRerouteDeclaration
   inputs [None {from=MaterialExpressionMultiply_141}]
   properties {name="Pre-Skinned Bounds Half Extents (V3)", nodeColor=<rgba 1 1 1 1>}
  }
  MaterialExpressionNamedRerouteDeclaration_35 {
   class=MaterialExpressionNamedRerouteDeclaration
   inputs [None {from=MaterialExpressionFunctionInput_0}]
   properties {name="Object Position WS (V3)", nodeColor=<rgba 1 0.9375 0 1>}
  }
  MaterialExpressionNamedRerouteDeclaration_36 {
   class=MaterialExpressionNamedRerouteDeclaration
   inputs [None {from=MaterialExpressionMultiply_147}]
   properties {name="Pre-Skinned Scale (S)", nodeColor=<rgba 1 0.5391 0 1>}
  }
  MaterialExpressionNamedRerouteDeclaration_37 {
   class=MaterialExpressionNamedRerouteDeclaration
   inputs [None {from=MaterialExpressionMultiply_47}]
   properties {name=Object Radius, nodeColor=<rgba 1 0.1406 0 1>}
  }
  MaterialExpressionNamedRerouteDeclaration_38 {
   class=MaterialExpressionNamedRerouteDeclaration
   inputs [None {from=MaterialExpressionFunctionInput_1}]
   properties {name="Pre-Skinned Scale (V3)", nodeColor=<rgba 1 0.5391 0 1>}
  }
  MaterialExpressionNamedRerouteDeclaration_4 {
   class=MaterialExpressionNamedRerouteDeclaration
   inputs [None {from=MaterialExpressionGetMaterialAttributes_0, output=Specular}]
   properties {name=IN_Specular, nodeColor=<rgba 1 0.4453 0 1>}
  }
  MaterialExpressionNamedRerouteDeclaration_43 {
   class=MaterialExpressionNamedRerouteDeclaration
   inputs [None {from=MaterialExpressionIf_1}]
   properties {name="Is Backpack Sprite (S)", nodeColor=<rgba 0.3594 1 0 1>}
  }
  MaterialExpressionNamedRerouteDeclaration_44 {
   class=MaterialExpressionNamedRerouteDeclaration
   inputs [None {from=MaterialExpressionIf_0}]
   properties {name="Is Mannequin (S)", nodeColor=<rgba 1 0.7734 0 1>}
  }
  MaterialExpressionNamedRerouteDeclaration_45 {
   class=MaterialExpressionNamedRerouteDeclaration
   inputs [None {from=MaterialExpressionComponentMask_9}]
   properties {name=M_ZGradient, nodeColor=<rgba 0.1719 1 0 1>}
  }
  MaterialExpressionNamedRerouteDeclaration_46 {
   class=MaterialExpressionNamedRerouteDeclaration
   inputs [None {from=MaterialExpressionFunctionInput_30}]
   properties {name=M_BlinkingLights, nodeColor=<rgba 1 0.9844 0 1>}
  }
  MaterialExpressionNamedRerouteDeclaration_47 {
   class=MaterialExpressionNamedRerouteDeclaration
   inputs [None {from=MaterialExpressionMultiply_62}]
   properties {name="Is Mannequin Backpack (S)", nodeColor=<rgba 0.4531 0 1 1>}
  }
  MaterialExpressionNamedRerouteDeclaration_5 {
   class=MaterialExpressionNamedRerouteDeclaration
   inputs [None {from=MaterialExpressionFunctionInput_25}]
   properties {name=M_VariantMask, nodeColor=<rgba 0 1 0.8828 1>}
  }
  MaterialExpressionNamedRerouteDeclaration_6 {
   class=MaterialExpressionNamedRerouteDeclaration
   inputs [None {from=MaterialExpressionGetMaterialAttributes_0, output=MaterialAttributes}]
   properties {name=IN_MA, nodeColor=<rgba 1 0 0.1015 1>}
  }
  MaterialExpressionNamedRerouteDeclaration_7 {
   class=MaterialExpressionNamedRerouteDeclaration
   inputs [None {from=MaterialExpressionFunctionInput_26}]
   properties {name=M_InnerRadialGlow, nodeColor=<rgba 1 0.9844 0 1>}
  }
  MaterialExpressionNamedRerouteDeclaration_8 {
   class=MaterialExpressionNamedRerouteDeclaration
   inputs [None {from=MaterialExpressionFunctionInput_28}]
   properties {name=M_SpriteFresnel, nodeColor=<rgba 1 0.9844 0 1>}
  }
  MaterialExpressionNamedRerouteDeclaration_9 {
   class=MaterialExpressionNamedRerouteDeclaration
   inputs [None {from=MaterialExpressionFunctionInput_27}]
   properties {name=M_Logo, nodeColor=<rgba 1 0.9844 0 1>}
  }
  MaterialExpressionNamedRerouteUsage_10 {class=MaterialExpressionNamedRerouteUsage}
  MaterialExpressionNamedRerouteUsage_104 {class=MaterialExpressionNamedRerouteUsage}
  MaterialExpressionNamedRerouteUsage_11 {class=MaterialExpressionNamedRerouteUsage}
  MaterialExpressionNamedRerouteUsage_13 {class=MaterialExpressionNamedRerouteUsage}
  MaterialExpressionNamedRerouteUsage_14 {class=MaterialExpressionNamedRerouteUsage}
  MaterialExpressionNamedRerouteUsage_148 {class=MaterialExpressionNamedRerouteUsage}
  MaterialExpressionNamedRerouteUsage_149 {class=MaterialExpressionNamedRerouteUsage}
  MaterialExpressionNamedRerouteUsage_15 {class=MaterialExpressionNamedRerouteUsage}
  MaterialExpressionNamedRerouteUsage_16 {class=MaterialExpressionNamedRerouteUsage}
  MaterialExpressionNamedRerouteUsage_165 {class=MaterialExpressionNamedRerouteUsage}
  MaterialExpressionNamedRerouteUsage_17 {class=MaterialExpressionNamedRerouteUsage}
  MaterialExpressionNamedRerouteUsage_171 {class=MaterialExpressionNamedRerouteUsage}
  MaterialExpressionNamedRerouteUsage_173 {class=MaterialExpressionNamedRerouteUsage}
  MaterialExpressionNamedRerouteUsage_175 {class=MaterialExpressionNamedRerouteUsage}
  MaterialExpressionNamedRerouteUsage_176 {class=MaterialExpressionNamedRerouteUsage}
  MaterialExpressionNamedRerouteUsage_18 {class=MaterialExpressionNamedRerouteUsage}
  MaterialExpressionNamedRerouteUsage_19 {class=MaterialExpressionNamedRerouteUsage}
  MaterialExpressionNamedRerouteUsage_20 {class=MaterialExpressionNamedRerouteUsage}
  MaterialExpressionNamedRerouteUsage_21 {class=MaterialExpressionNamedRerouteUsage}
  MaterialExpressionNamedRerouteUsage_23 {class=MaterialExpressionNamedRerouteUsage}
  MaterialExpressionNamedRerouteUsage_24 {class=MaterialExpressionNamedRerouteUsage}
  MaterialExpressionNamedRerouteUsage_25 {class=MaterialExpressionNamedRerouteUsage}
  MaterialExpressionNamedRerouteUsage_26 {class=MaterialExpressionNamedRerouteUsage}
  MaterialExpressionNamedRerouteUsage_27 {class=MaterialExpressionNamedRerouteUsage}
  MaterialExpressionNamedRerouteUsage_28 {class=MaterialExpressionNamedRerouteUsage}
  MaterialExpressionNamedRerouteUsage_29 {class=MaterialExpressionNamedRerouteUsage}
  MaterialExpressionNamedRerouteUsage_3 {class=MaterialExpressionNamedRerouteUsage}
  MaterialExpressionNamedRerouteUsage_30 {class=MaterialExpressionNamedRerouteUsage}
  MaterialExpressionNamedRerouteUsage_31 {class=MaterialExpressionNamedRerouteUsage}
  MaterialExpressionNamedRerouteUsage_32 {class=MaterialExpressionNamedRerouteUsage}
  MaterialExpressionNamedRerouteUsage_34 {class=MaterialExpressionNamedRerouteUsage}
  MaterialExpressionNamedRerouteUsage_35 {class=MaterialExpressionNamedRerouteUsage}
  MaterialExpressionNamedRerouteUsage_36 {class=MaterialExpressionNamedRerouteUsage}
  MaterialExpressionNamedRerouteUsage_37 {class=MaterialExpressionNamedRerouteUsage}
  MaterialExpressionNamedRerouteUsage_38 {class=MaterialExpressionNamedRerouteUsage}
  MaterialExpressionNamedRerouteUsage_4 {class=MaterialExpressionNamedRerouteUsage}
  MaterialExpressionNamedRerouteUsage_40 {class=MaterialExpressionNamedRerouteUsage}
  MaterialExpressionNamedRerouteUsage_46 {class=MaterialExpressionNamedRerouteUsage}
  MaterialExpressionNamedRerouteUsage_47 {class=MaterialExpressionNamedRerouteUsage}
  MaterialExpressionNamedRerouteUsage_50 {class=MaterialExpressionNamedRerouteUsage}
  MaterialExpressionNamedRerouteUsage_53 {class=MaterialExpressionNamedRerouteUsage}
  MaterialExpressionNamedRerouteUsage_56 {class=MaterialExpressionNamedRerouteUsage}
  MaterialExpressionNamedRerouteUsage_57 {class=MaterialExpressionNamedRerouteUsage}
  MaterialExpressionNamedRerouteUsage_59 {class=MaterialExpressionNamedRerouteUsage}
  MaterialExpressionNamedRerouteUsage_6 {class=MaterialExpressionNamedRerouteUsage}
  MaterialExpressionNamedRerouteUsage_62 {class=MaterialExpressionNamedRerouteUsage}
  MaterialExpressionNamedRerouteUsage_65 {class=MaterialExpressionNamedRerouteUsage}
  MaterialExpressionNamedRerouteUsage_67 {class=MaterialExpressionNamedRerouteUsage}
  MaterialExpressionNamedRerouteUsage_68 {class=MaterialExpressionNamedRerouteUsage}
  MaterialExpressionNamedRerouteUsage_69 {class=MaterialExpressionNamedRerouteUsage}
  MaterialExpressionNamedRerouteUsage_7 {class=MaterialExpressionNamedRerouteUsage}
  MaterialExpressionNamedRerouteUsage_70 {class=MaterialExpressionNamedRerouteUsage}
  MaterialExpressionNamedRerouteUsage_71 {class=MaterialExpressionNamedRerouteUsage}
  MaterialExpressionNamedRerouteUsage_72 {class=MaterialExpressionNamedRerouteUsage}
  MaterialExpressionNamedRerouteUsage_73 {class=MaterialExpressionNamedRerouteUsage}
  MaterialExpressionNamedRerouteUsage_74 {class=MaterialExpressionNamedRerouteUsage}
  MaterialExpressionNamedRerouteUsage_75 {class=MaterialExpressionNamedRerouteUsage}
  MaterialExpressionNamedRerouteUsage_76 {class=MaterialExpressionNamedRerouteUsage}
  MaterialExpressionNamedRerouteUsage_78 {class=MaterialExpressionNamedRerouteUsage}
  MaterialExpressionNamedRerouteUsage_79 {class=MaterialExpressionNamedRerouteUsage}
  MaterialExpressionNamedRerouteUsage_8 {class=MaterialExpressionNamedRerouteUsage}
  MaterialExpressionNamedRerouteUsage_81 {class=MaterialExpressionNamedRerouteUsage}
  MaterialExpressionNamedRerouteUsage_82 {class=MaterialExpressionNamedRerouteUsage}
  MaterialExpressionNamedRerouteUsage_83 {class=MaterialExpressionNamedRerouteUsage}
  MaterialExpressionNamedRerouteUsage_84 {class=MaterialExpressionNamedRerouteUsage}
  MaterialExpressionNamedRerouteUsage_85 {class=MaterialExpressionNamedRerouteUsage}
  MaterialExpressionNamedRerouteUsage_87 {class=MaterialExpressionNamedRerouteUsage}
  MaterialExpressionNamedRerouteUsage_9 {class=MaterialExpressionNamedRerouteUsage}
  MaterialExpressionNormalize_3 {class=MaterialExpressionNormalize, inputs [VectorInput {from=MaterialExpressionTransform_0, output=""}]}
  MaterialExpressionObjectPositionWS_4 {class=MaterialExpressionObjectPositionWS, inputs [], properties {}}
  MaterialExpressionOneMinus_0 {class=MaterialExpressionOneMinus, inputs [None {from=MaterialExpressionScalarParameter_34}]}
  MaterialExpressionOneMinus_7 {class=MaterialExpressionOneMinus, inputs [None {from=MaterialExpressionNamedRerouteUsage_56}]}
  MaterialExpressionOneMinus_9 {class=MaterialExpressionOneMinus, inputs [None {from=MaterialExpressionDotProduct_6}]}
  MaterialExpressionPower_1 {
   class=MaterialExpressionPower
   inputs [
    Base {from=MaterialExpressionSmoothStep_5}
    Exp {from=""}
   ]
   properties {constExponent=8}
  }
  MaterialExpressionPower_10 {
   class=MaterialExpressionPower
   inputs [
    Base {from=MaterialExpressionSmoothStep_7}
    Exp {from=""}
   ]
  }
  MaterialExpressionPower_11 {
   class=MaterialExpressionPower
   inputs [
    Base {from=MaterialExpressionAdd_0}
    Exp {from=MaterialExpressionScalarParameter_7}
   ]
   properties {constExponent=4}
  }
  MaterialExpressionPower_9 {
   class=MaterialExpressionPower
   inputs [
    Base {from=MaterialExpressionSaturate_4}
    Exp {from=""}
   ]
   properties {constExponent=4}
  }
  MaterialExpressionPreSkinnedNormal_0 {class=MaterialExpressionPreSkinnedNormal, inputs [], properties {}}
  MaterialExpressionReroute_0 {class=MaterialExpressionReroute, inputs [None {from=MaterialExpressionMaterialFunctionCall_59, output=StabilizedUVs}]}
  MaterialExpressionReroute_14 {class=MaterialExpressionReroute, inputs [None {from=MaterialExpressionDivide_5}]}
  MaterialExpressionReroute_15 {class=MaterialExpressionReroute, inputs [None {from=MaterialExpressionNamedRerouteUsage_71}]}
  MaterialExpressionReroute_16 {class=MaterialExpressionReroute, inputs [None {from=MaterialExpressionAdd_33}]}
  MaterialExpressionReroute_17 {class=MaterialExpressionReroute, inputs [None {from=MaterialExpressionAdd_28}]}
  MaterialExpressionReroute_18 {class=MaterialExpressionReroute, inputs [None {from=MaterialExpressionAdd_21}]}
  MaterialExpressionReroute_2 {class=MaterialExpressionReroute, inputs [None {from=MaterialExpressionAdd_32}]}
  MaterialExpressionReroute_21 {class=MaterialExpressionReroute, inputs [None {from=MaterialExpressionReroute_34}]}
  MaterialExpressionReroute_22 {class=MaterialExpressionReroute, inputs [None {from=MaterialExpressionMultiply_23}]}
  MaterialExpressionReroute_26 {class=MaterialExpressionReroute, inputs [None {from=MaterialExpressionLinearInterpolate_10}]}
  MaterialExpressionReroute_27 {class=MaterialExpressionReroute, inputs [None {from=MaterialExpressionLinearInterpolate_11}]}
  MaterialExpressionReroute_29 {class=MaterialExpressionReroute, inputs [None {from=MaterialExpressionNamedRerouteUsage_23}]}
  MaterialExpressionReroute_30 {class=MaterialExpressionReroute, inputs [None {from=MaterialExpressionMultiply_11}]}
  MaterialExpressionReroute_31 {class=MaterialExpressionReroute, inputs [None {from=MaterialExpressionAdd_27}]}
  MaterialExpressionReroute_32 {class=MaterialExpressionReroute, inputs [None {from=MaterialExpressionReroute_2}]}
  MaterialExpressionReroute_33 {class=MaterialExpressionReroute, inputs [None {from=MaterialExpressionMaterialFunctionCall_4, output=Result}]}
  MaterialExpressionReroute_34 {class=MaterialExpressionReroute, inputs [None {from=MaterialExpressionMultiply_31}]}
  MaterialExpressionReroute_35 {class=MaterialExpressionReroute, inputs [None {from=MaterialExpressionReroute_26}]}
  MaterialExpressionReroute_37 {class=MaterialExpressionReroute, inputs [None {from=MaterialExpressionLinearInterpolate_28}]}
  MaterialExpressionReroute_38 {class=MaterialExpressionReroute, inputs [None {from=MaterialExpressionMultiply_49}]}
  MaterialExpressionReroute_39 {class=MaterialExpressionReroute, inputs [None {from=MaterialExpressionMultiply_49}]}
  MaterialExpressionReroute_41 {class=MaterialExpressionReroute, inputs [None {from=MaterialExpressionSaturate_0}]}
  MaterialExpressionReroute_7 {class=MaterialExpressionReroute, inputs [None {from=MaterialExpressionReroute_22}]}
  MaterialExpressionReroute_8 {class=MaterialExpressionReroute, inputs [None {from=MaterialExpressionAdd_11}]}
  MaterialExpressionReroute_9 {class=MaterialExpressionReroute, inputs [None {from=MaterialExpressionNamedRerouteUsage_34}]}
  MaterialExpressionRotator_0 {
   class=MaterialExpressionRotator
   inputs [
    Coordinate {from=MaterialExpressionConvert_2, output=""}
    Time {from=MaterialExpressionScalarParameter_11, output=""}
   ]
   properties {}
  }
  MaterialExpressionSaturate_0 {class=MaterialExpressionSaturate, inputs [None {from=MaterialExpressionDesaturation_0}]}
  MaterialExpressionSaturate_1 {class=MaterialExpressionSaturate, inputs [None {from=MaterialExpressionMax_0}]}
  MaterialExpressionSaturate_4 {class=MaterialExpressionSaturate, inputs [None {from=MaterialExpressionAdd_42}]}
  MaterialExpressionScalarParameter_0 {
   class=MaterialExpressionScalarParameter
   properties {defaultValue=0.3, desc=0.25, group=SpriteVariants, parameterName=SpriteVar_Holo_Emissive_DFresnel_Offset}
  }
  MaterialExpressionScalarParameter_1 {class=MaterialExpressionScalarParameter, properties {parameterName=Holofoil Emissive Extra}}
  MaterialExpressionScalarParameter_10 {class=MaterialExpressionScalarParameter, properties {defaultValue=0.25, group=SpriteVariants, parameterName=SpriteVar_Holo_BaseColorIN}}
  MaterialExpressionScalarParameter_11 {class=MaterialExpressionScalarParameter, properties {defaultValue=1, group=SpriteVariants, parameterName=SpriteVar_Holo_ColorGrad_Rotate}}
  MaterialExpressionScalarParameter_16 {
   class=MaterialExpressionScalarParameter
   properties {defaultValue=0.3, desc=0.25, group=SpriteVariants, parameterName=SpriteVar_Holo_Emissive_DFresnel_Offset}
  }
  MaterialExpressionScalarParameter_17 {
   class=MaterialExpressionScalarParameter
   properties {defaultValue=-0.48, group=SpriteVariants, parameterName=SpriteVar_Holo_SpriteFresnel_Bias}
  }
  MaterialExpressionScalarParameter_18 {
   class=MaterialExpressionScalarParameter
   properties {defaultValue=1, group=SpriteVariants, parameterName=SpriteVar_Holo_SpriteFresnel_Contrast}
  }
  MaterialExpressionScalarParameter_19 {
   class=MaterialExpressionScalarParameter
   properties {defaultValue=0.1, group=SpriteVariants, parameterName=SpriteVar_Holo_Emissive_DFresnel_Contrast}
  }
  MaterialExpressionScalarParameter_23 {
   class=MaterialExpressionScalarParameter
   properties {defaultValue=0.2, group=SpriteVariants, parameterName=SpriteVar_Holo_ColorOffset_Normal}
  }
  MaterialExpressionScalarParameter_25 {class=MaterialExpressionScalarParameter, properties {defaultValue=3, group=SpriteVariants, parameterName=SpriteVar_Holo_ColorGrad_Scale}}
  MaterialExpressionScalarParameter_26 {
   class=MaterialExpressionScalarParameter
   properties {defaultValue=60, desc=50, group=SpriteVariants, parameterName=SpriteVar_Holo_Glitter_Tile}
  }
  MaterialExpressionScalarParameter_27 {
   class=MaterialExpressionScalarParameter
   properties {defaultValue=100, group=SpriteVariants, parameterName=SpriteVar_Holo_Glitter_TriplanarFade}
  }
  MaterialExpressionScalarParameter_28 {class=MaterialExpressionScalarParameter, properties {defaultValue=50, group=SpriteVariants, parameterName=SpriteVar_Holo_PerlinScale}}
  MaterialExpressionScalarParameter_3 {
   class=MaterialExpressionScalarParameter
   properties {defaultValue=0.7, group=SpriteVariants, parameterName=SpriteVar_Holo_Glitter_FlattenNormal}
  }
  MaterialExpressionScalarParameter_31 {
   class=MaterialExpressionScalarParameter
   properties {defaultValue=0.05, group=SpriteVariants, parameterName=SpriteVar_Holo_ColorOffset_Delay_Logo}
  }
  MaterialExpressionScalarParameter_32 {
   class=MaterialExpressionScalarParameter
   properties {defaultValue=0.8, group=SpriteVariants, parameterName=SpriteVar_Holo_Emissive_Core_BaseColorGradLerp}
  }
  MaterialExpressionScalarParameter_34 {class=MaterialExpressionScalarParameter, properties {defaultValue=0.5, group=SpriteVariants, parameterName=SpriteVar_Holo_Perlin_ColorBias}}
  MaterialExpressionScalarParameter_35 {
   class=MaterialExpressionScalarParameter
   properties {defaultValue=0.4, desc=0.4, group=SpriteVariants, parameterName=SpriteVar_Holo_Emissive_Dfresnel_BaseColorGradLerp}
  }
  MaterialExpressionScalarParameter_36 {class=MaterialExpressionScalarParameter, properties {defaultValue=0.3, parameterName=SpriteVar_Holofoil_SheenOffset}}
  MaterialExpressionScalarParameter_37 {class=MaterialExpressionScalarParameter, properties {defaultValue=0.6, group=SpriteVariants, parameterName=SpriteVar_Holo_Logo_ColorBias}}
  MaterialExpressionScalarParameter_38 {class=MaterialExpressionScalarParameter, properties {defaultValue=0.08, desc=0.045, parameterName=SpriteVar_Holofoil_SheenOffset_Sine}}
  MaterialExpressionScalarParameter_39 {class=MaterialExpressionScalarParameter, properties {defaultValue=0.2, group=SpriteVariants, parameterName=SpriteVar_Holo_Normal_ColorBias}}
  MaterialExpressionScalarParameter_4 {class=MaterialExpressionScalarParameter, properties {parameterName=Holofoil Emissive Extra B}}
  MaterialExpressionScalarParameter_40 {
   class=MaterialExpressionScalarParameter
   properties {defaultValue=0.4, group=SpriteVariants, parameterName=SpriteVar_Holo_ColorOffset_Delay_DiffuseMask}
  }
  MaterialExpressionScalarParameter_41 {class=MaterialExpressionScalarParameter, properties {defaultValue=3, parameterName=RGB_TIle}}
  MaterialExpressionScalarParameter_47 {
   class=MaterialExpressionScalarParameter
   properties {defaultValue=1.7, desc=2, group=SpriteVariants, parameterName=SpriteVar_Holo_ColorBias_old}
  }
  MaterialExpressionScalarParameter_5 {
   class=MaterialExpressionScalarParameter
   properties {defaultValue=-0.15, desc=-0.15, group=SpriteVariants, parameterName=SpriteVar_Holo_Emissive_DFresnelGap_Offset}
  }
  MaterialExpressionScalarParameter_50 {
   class=MaterialExpressionScalarParameter
   properties {defaultValue=2.5, desc=2.5 or 3, group=SpriteVariants, parameterName=SpriteVar_Holo_Color_Freq_old}
  }
  MaterialExpressionScalarParameter_55 {class=MaterialExpressionScalarParameter, properties {defaultValue=0.35, parameterName=SpriteVar_Holo_Desturation}}
  MaterialExpressionScalarParameter_59 {class=MaterialExpressionScalarParameter, properties {defaultValue=0.1, group=SpriteVariants, parameterName=SpriteVar_Holo_SpriteFresnel_Min}}
  MaterialExpressionScalarParameter_6 {
   class=MaterialExpressionScalarParameter
   properties {defaultValue=3, desc=2.5, group=SpriteVariants, parameterName=SpriteVar_Holo_Emissive_DFresnelGap_Contrast}
  }
  MaterialExpressionScalarParameter_60 {class=MaterialExpressionScalarParameter, properties {defaultValue=0.6, group=SpriteVariants, parameterName=SpriteVar_Holo_SpriteFresnel_Max}}
  MaterialExpressionScalarParameter_62 {
   class=MaterialExpressionScalarParameter
   properties {defaultValue=-1, group=SpriteVariants, parameterName=SpriteVar_Holo_SpriteFresnel_ColorBias}
  }
  MaterialExpressionScalarParameter_67 {class=MaterialExpressionScalarParameter, properties {defaultValue=0.4, group=SpriteVariants, parameterName=SpriteVar_Holo_Glitter_Roughness}}
  MaterialExpressionScalarParameter_68 {class=MaterialExpressionScalarParameter, properties {defaultValue=1, group=SpriteVariants, parameterName=SpriteVar_Holo_Glitter_Metallic}}
  MaterialExpressionScalarParameter_69 {class=MaterialExpressionScalarParameter, properties {group=SpriteVariants, parameterName=SpriteVar_Holo_Glitter_Specular}}
  MaterialExpressionScalarParameter_7 {
   class=MaterialExpressionScalarParameter
   properties {defaultValue=2, group=SpriteVariants, parameterName=SpriteVar_Holo_Emissive_DFresnelB_Power}
  }
  MaterialExpressionSetMaterialAttributes_0 {
   class=MaterialExpressionSetMaterialAttributes
   inputs [
    MaterialAttributes {from=MaterialExpressionNamedRerouteUsage_8, output=""}
    BaseColor {from=MaterialExpressionLinearInterpolate_21, output=""}
    EmissiveColor {from=MaterialExpressionNamedRerouteUsage_50, output=""}
    Metallic {from=MaterialExpressionScalarParameter_68, output=""}
    Specular {from=MaterialExpressionScalarParameter_69, output=""}
    Roughness {from=MaterialExpressionScalarParameter_67, output=""}
    Normal {from=MaterialExpressionNamedRerouteUsage_15, output=""}
   ]
   properties {attributeSetTypes [
     #2
     #6
     #3
     #5
     #4
     #1
    ]}
  }
  MaterialExpressionShaderStageSwitch_0 {
   class=MaterialExpressionShaderStageSwitch
   inputs [
    PixelShader {from=MaterialExpressionVertexInterpolator_0, output=PS}
    VertexShader {from=MaterialExpressionPreSkinnedNormal_0, output=""}
   ]
   properties {}
  }
  MaterialExpressionShaderStageSwitch_4 {
   class=MaterialExpressionShaderStageSwitch
   inputs [
    PixelShader {from=MaterialExpressionVertexInterpolator_4, output=PS}
    VertexShader {from=MaterialExpressionLocalPosition_5, output=XYZ}
   ]
   properties {}
  }
  MaterialExpressionSine_0 {class=MaterialExpressionSine, inputs [None {from=MaterialExpressionDivide_9}]}
  MaterialExpressionSmoothStep_0 {
   class=MaterialExpressionSmoothStep
   inputs [
    Min {from=MaterialExpressionScalarParameter_59, output=""}
    Max {from=MaterialExpressionScalarParameter_60, output=""}
    Value {from=MaterialExpressionNamedRerouteUsage_56, output=""}
   ]
   properties {}
  }
  MaterialExpressionSmoothStep_1 {
   class=MaterialExpressionSmoothStep
   inputs [
    Min {from="", output=""}
    Max {from="", output=""}
    Value {from=MaterialExpressionNamedRerouteUsage_18, output=""}
   ]
   properties {constMin=0.4}
  }
  MaterialExpressionSmoothStep_5 {
   class=MaterialExpressionSmoothStep
   inputs [
    Min {from="", output=""}
    Max {from="", output=""}
    Value {from=MaterialExpressionNamedRerouteUsage_24, output=""}
   ]
   properties {constMax=0, constMin=0.1}
  }
  MaterialExpressionSmoothStep_7 {
   class=MaterialExpressionSmoothStep
   inputs [
    Min {from="", output=""}
    Max {from="", output=""}
    Value {from=MaterialExpressionNamedRerouteUsage_82, output=""}
   ]
   properties {constMin=0.4}
  }
  MaterialExpressionStaticSwitchParameter_3 {
   class=MaterialExpressionStaticSwitchParameter
   inputs [
    True {from=MaterialExpressionSetMaterialAttributes_0, output=""}
    False {from=MaterialExpressionNamedRerouteUsage_8, output=""}
   ]
   properties {group=SpriteVariants, parameterName=Sprite_Variant_Holofoil, sortPriority=-1}
  }
  MaterialExpressionSubtract_0 {
   class=MaterialExpressionSubtract
   inputs [
    A {from=MaterialExpressionReroute_9}
    B {from=MaterialExpressionMultiply_17}
   ]
  }
  MaterialExpressionSubtract_3 {
   class=MaterialExpressionSubtract
   inputs [
    A {from=MaterialExpressionNamedRerouteUsage_59}
    B {from=MaterialExpressionConstant_9}
   ]
   properties {constB=0.5}
  }
  MaterialExpressionSubtract_40 {
   class=MaterialExpressionSubtract
   inputs [
    A {from=MaterialExpressionNamedRerouteUsage_26}
    B {from=MaterialExpressionNamedRerouteUsage_14}
   ]
  }
  MaterialExpressionSubtract_41 {
   class=MaterialExpressionSubtract
   inputs [
    A {from=MaterialExpressionNamedRerouteUsage_81}
    B {from=MaterialExpressionNamedRerouteUsage_149}
   ]
  }
  MaterialExpressionTextureObject_0 {
   class=MaterialExpressionTextureObject
   inputs []
   properties {texture=/BRCosmetics/Animation/Props/Gadgets_PixieParts/Textures/TA/T_SpriteVariants_HoloFoilGrad_RGB}
  }
  MaterialExpressionTextureObject_1 {
   class=MaterialExpressionTextureObject
   inputs []
   properties {samplerType=SAMPLERTYPE_Normal, texture=/Game/Weapons/WeaponSkins/Wraps/Textures/T_Wrap_metallicFlakes_Small}
  }
  MaterialExpressionTextureObject_2 {
   class=MaterialExpressionTextureObject
   inputs []
   properties {texture=/BRCosmetics/Animation/Props/Gadgets_PixieParts/Textures/TA/T_SpriteVariants_HoloFoilGrad}
  }
  MaterialExpressionTextureObject_7 {
   class=MaterialExpressionTextureObject
   inputs []
   properties {texture=/Game/Global/Textures/Default/Thinfilm/T_ThinFilm_Spectrum_DESATURATED-50}
  }
  MaterialExpressionTextureObject_8 {
   class=MaterialExpressionTextureObject
   inputs []
   properties {samplerType=SAMPLERTYPE_LinearColor, texture=&HighContrastNoise}
  }
  MaterialExpressionTextureSample_0 {
   class=MaterialExpressionTextureSample
   inputs [
    UVs {from=MaterialExpressionMaterialFunctionCall_20, output=UVs}
    Tex {from=MaterialExpressionNamedRerouteUsage_20}
    "DDX(UVs)" {from=MaterialExpressionMaterialFunctionCall_20, output=UVs}
    "DDY(UVs)" {from=MaterialExpressionMaterialFunctionCall_20, output=UVs}
    Apply View MipBias {from=""}
   ]
   properties {
    mipValueMode=TMVM_Derivative
    samplerSource=SSM_Wrap_WorldGroupSettings
    samplerType=SAMPLERTYPE_Normal
    texture=&T_Wrap_metallicFlakes_Small
   }
  }
  MaterialExpressionTextureSample_10 {
   class=MaterialExpressionTextureSample
   inputs [
    UVs {from=MaterialExpressionAdd_14}
    Tex {from=MaterialExpressionNamedRerouteUsage_68}
    Apply View MipBias {from=""}
   ]
  }
  MaterialExpressionTextureSample_12 {
   class=MaterialExpressionTextureSample
   inputs [
    UVs {from=MaterialExpressionDivide_7}
    Tex {from=MaterialExpressionNamedRerouteUsage_27}
    Apply View MipBias {from=""}
   ]
  }
  MaterialExpressionTextureSample_13 {
   class=MaterialExpressionTextureSample
   inputs [
    UVs {from=MaterialExpressionComponentMask_2}
    Tex {from=MaterialExpressionNamedRerouteUsage_28}
    Apply View MipBias {from=""}
   ]
  }
  MaterialExpressionTextureSample_14 {
   class=MaterialExpressionTextureSample
   inputs [
    UVs {from=MaterialExpressionComponentMask_3}
    Tex {from=MaterialExpressionNamedRerouteUsage_30}
    Apply View MipBias {from=""}
   ]
  }
  MaterialExpressionTextureSample_2 {
   class=MaterialExpressionTextureSample
   inputs [
    UVs {from=MaterialExpressionAdd_23}
    Tex {from=MaterialExpressionNamedRerouteUsage_31}
    Apply View MipBias {from=""}
   ]
  }
  MaterialExpressionTextureSample_7 {
   class=MaterialExpressionTextureSample
   inputs [
    UVs {from=MaterialExpressionMultiply_38}
    Tex {from=MaterialExpressionNamedRerouteUsage_79}
    Apply View MipBias {from=""}
   ]
  }
  MaterialExpressionTime_1 {class=MaterialExpressionTime}
  MaterialExpressionTime_2 {class=MaterialExpressionTime}
  MaterialExpressionTransform_0 {
   class=MaterialExpressionTransform
   inputs [None {from=MaterialExpressionLinearInterpolate_31}]
   properties {transformSourceType=TRANSFORMSOURCE_Local}
  }
  MaterialExpressionVectorParameter_1 {
   class=MaterialExpressionVectorParameter
   properties {group=SpriteVariants, parameterName=SpriteVar_Holo_ColorTint, defaultValue=<rgba 1 1 1 1>}
  }
  MaterialExpressionVectorParameter_12 {
   class=MaterialExpressionVectorParameter
   properties {
    group=SpriteVariants
    parameterName=SpriteVar_Holo_BaseColorTop
    channelNames {a="0=UseBaseColor 1=FlatColor", b="", g="", r=""}
    defaultValue=<rgba 0.0198 0.2203 0.1418 0.05>
   }
  }
  MaterialExpressionVectorParameter_15 {
   class=MaterialExpressionVectorParameter
   properties {
    group=SpriteVariants
    parameterName=SpriteVar_Holo_BaseColorBottoms
    channelNames {a="0=UseBaseColor 1=FlatColor", b="", g="", r=""}
    defaultValue=<rgba 0.1589 0.0739 0.4812 1>
   }
  }
  MaterialExpressionVectorParameter_17 {
   class=MaterialExpressionVectorParameter
   properties {group=SpriteVariants, parameterName=SpriteVar_Holo_DiffuseChannel_Tint, defaultValue=<rgba 1.1008 0.1635 1.2 1>}
  }
  MaterialExpressionVectorParameter_18 {
   class=MaterialExpressionVectorParameter
   properties {
    group=SpriteVariants
    parameterName=SpriteVar_Holo_BaseColor_Logo
    channelNames {a="0=Multiply 1=FlatColor", b="", g="", r=""}
    defaultValue=<rgba 2 1.294 0.2725 0.2232>
   }
  }
  MaterialExpressionVectorParameter_19 {
   class=MaterialExpressionVectorParameter
   properties {group=SpriteVariants, parameterName=SpriteVar_Holo_Emissive_Core_Tint_In, defaultValue=<rgba 0.1343 1 0.0696 1>}
  }
  MaterialExpressionVectorParameter_20 {
   class=MaterialExpressionVectorParameter
   properties {
    group=SpriteVariants
    parameterName=SpriteVar_Holo_Emissive_Logo
    channelNames {a="0=Multiply 1=FlatColor", b="", g="", r=""}
    defaultValue=<rgba 3.5 1.1058 2.2301 0.0522>
   }
  }
  MaterialExpressionVectorParameter_21 {
   class=MaterialExpressionVectorParameter
   properties {group=SpriteVariants, parameterName=SpriteVar_Holo_Emissive_DFresnelOut_Tint, defaultValue=<rgba 2.5 1.7036 0.37 1>}
  }
  MaterialExpressionVectorParameter_22 {
   class=MaterialExpressionVectorParameter
   properties {group=SpriteVariants, parameterName=SpriteVar_Holo_Emissive_DFresnel_In_Tint, defaultValue=<rgba 1.3 0.176 0.1168 1>}
  }
  MaterialExpressionVectorParameter_24 {
   class=MaterialExpressionVectorParameter
   properties {group=SpriteVariants, parameterName=SpriteVar_Holo_Emissive_Core_Tint_Out, defaultValue=<rgba 0.7685 0.2392 1 1>}
  }
  MaterialExpressionVectorParameter_26 {
   class=MaterialExpressionVectorParameter
   properties {
    group=SpriteVariants
    parameterName=SpriteVar_Holo_Emissive_DifM_Tint
    channelNames {a="0=Multiply 1=FlatColor", b="", g="", r=""}
    defaultValue=<rgba 2.5 1.4043 0.5 0.0522>
   }
  }
  MaterialExpressionVectorParameter_3 {
   class=MaterialExpressionVectorParameter
   properties {
    group=SpriteVariants
    parameterName=SpriteVar_Holo_DiffuseChannelBias
    channelNames {a=RGBMultiplier, b=DiffuseB Influence, g=DiffuseG Influence, r=DiffuseR Influence}
    defaultValue=<rgba 1 2 1 0.4>
   }
  }
  MaterialExpressionVectorParameter_31 {
   class=MaterialExpressionVectorParameter
   properties {group=SpriteVariants, parameterName=SpriteVar_Holofoil_Emissive_TOD_Morning, defaultValue=<rgba 1.3 0.9161 0.4974 1>}
  }
  MaterialExpressionVectorParameter_32 {
   class=MaterialExpressionVectorParameter
   properties {group=SpriteVariants, parameterName=SpriteVar_Holofoil_Emissive_TOD_Day, defaultValue=<rgba 1 1 1 1>}
  }
  MaterialExpressionVectorParameter_33 {
   class=MaterialExpressionVectorParameter
   properties {group=SpriteVariants, parameterName=SpriteVar_Holofoil_Emissive_TOD_Evening, defaultValue=<rgba 1.2 0.9205 0.6157 1>}
  }
  MaterialExpressionVectorParameter_34 {
   class=MaterialExpressionVectorParameter
   properties {group=SpriteVariants, parameterName=SpriteVar_Holofoil_Emissive_TOD_Night, defaultValue=<rgba 0.84 1.1127 1.4 1>}
  }
  MaterialExpressionVectorParameter_7 {
   class=MaterialExpressionVectorParameter
   properties {
    group=SpriteVariants
    parameterName=SpriteVar_Holo_Emissive_BlinkingLights
    channelNames {a="0=Multiply 1=FlatColor", b="", g="", r=""}
    defaultValue=<rgba 3 1.4087 0 0.0522>
   }
  }
  MaterialExpressionVertexInterpolator_0 {class=MaterialExpressionVertexInterpolator, inputs [VS {from=MaterialExpressionPreSkinnedNormal_0, output=""}], properties {}}
  MaterialExpressionVertexInterpolator_4 {class=MaterialExpressionVertexInterpolator, inputs [VS {from=MaterialExpressionLocalPosition_5, output=XYZ}], properties {}}
  MaterialExpressionVertexNormalWS_7 {class=MaterialExpressionVertexNormalWS, inputs [], properties {}}
 }}
