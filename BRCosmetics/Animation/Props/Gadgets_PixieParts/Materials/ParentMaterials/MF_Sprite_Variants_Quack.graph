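# Material function that layers sprite variant looks for a gadget cosmetic onto incoming material attributes, returned through a static switch.
# Input masks cover variant, inner radial glow, logo, sprite fresnel, blinking lights and diffuse channel, alongside a WPO input.
# View and light dot products, time-of-day values and time-driven fractional cycles feed the color blends.
# Dithered triplanar mapping on pre-skinned positions with bump offset and object-stabilized screen-space UVs supply texture coordinates.
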MaterialFunction /BRCosmetics/Animation/Props/Gadgets_PixieParts/Materials/ParentMaterials/MF_Sprite_Variants_Quack
graph {expressions {
  MaterialExpressionAbs_6 {class=MaterialExpressionAbs, inputs [None {from=MaterialExpressionDotProduct_27, output=""}], properties {}}
  MaterialExpressionAbs_7 {class=MaterialExpressionAbs, inputs [None {from=MaterialExpressionNamedRerouteUsage_143, output=""}], properties {}}
  MaterialExpressionAbs_8 {class=MaterialExpressionAbs, inputs [None {from=MaterialExpressionSubtract_42, output=""}], properties {}}
  MaterialExpressionAdd_30 {
   class=MaterialExpressionAdd
   inputs [
    A {from=MaterialExpressionOneMinus_19}
    B {from=MaterialExpressionAppendVector_10}
   ]
  }
  MaterialExpressionAdd_31 {
   class=MaterialExpressionAdd
   inputs [
    A {from=MaterialExpressionReroute_41}
    B {from=MaterialExpressionSubtract_33}
   ]
  }
  MaterialExpressionAdd_32 {
   class=MaterialExpressionAdd
   inputs [
    A {from=MaterialExpressionReroute_42}
    B {from=MaterialExpressionMultiply_104}
   ]
  }
  MaterialExpressionAdd_33 {
   class=MaterialExpressionAdd
   inputs [
    A {from=MaterialExpressionReroute_45}
    B {from=MaterialExpressionMultiply_118}
   ]
  }
  MaterialExpressionAdd_34 {
   class=MaterialExpressionAdd
   inputs [
    A {from=MaterialExpressionReroute_50}
    B {from=MaterialExpressionMultiply_124}
   ]
  }
  MaterialExpressionAdd_35 {
   class=MaterialExpressionAdd
   inputs [
    A {from=MaterialExpressionReroute_47}
    B {from=MaterialExpressionMultiply_119}
   ]
  }
  MaterialExpressionAdd_36 {
   class=MaterialExpressionAdd
   inputs [
    A {from=MaterialExpressionAdd_40}
    B {from=MaterialExpressionMultiply_126}
   ]
  }
  MaterialExpressionAdd_37 {
   class=MaterialExpressionAdd
   inputs [
    A {from=MaterialExpressionReroute_49}
    B {from=MaterialExpressionMultiply_139}
   ]
  }
  MaterialExpressionAdd_38 {
   class=MaterialExpressionAdd
   inputs [
    A {from=MaterialExpressionLinearInterpolate_52}
    B {from=MaterialExpressionScalarParameter_101}
   ]
  }
  MaterialExpressionAdd_39 {
   class=MaterialExpressionAdd
   inputs [
    A {from=MaterialExpressionReroute_53}
    B {from=MaterialExpressionSubtract_44}
   ]
  }
  MaterialExpressionAdd_40 {
   class=MaterialExpressionAdd
   inputs [
    A {from=MaterialExpressionMultiply_137}
    B {from=MaterialExpressionMultiply_138}
   ]
  }
  MaterialExpressionAdd_41 {
   class=MaterialExpressionAdd
   inputs [
    A {from=MaterialExpressionMultiply_126}
    B {from=MaterialExpressionAdd_42}
   ]
  }
  MaterialExpressionAdd_42 {
   class=MaterialExpressionAdd
   inputs [
    A {from=MaterialExpressionMultiply_142}
    B {from=MaterialExpressionScalarParameter_103}
   ]
  }
  MaterialExpressionAdd_43 {
   class=MaterialExpressionAdd
   inputs [
    A {from=MaterialExpressionDivide_28}
    B {from=MaterialExpressionFrac_11}
   ]
   properties {constB=0}
  }
  MaterialExpressionAdd_44 {
   class=MaterialExpressionAdd
   inputs [
    A {from=MaterialExpressionReroute_57}
    B {from=MaterialExpressionMultiply_145}
   ]
  }
  MaterialExpressionAdd_6 {
   class=MaterialExpressionAdd
   inputs [
    A {from=MaterialExpressionMultiply_40}
    B {from=MaterialExpressionMultiply_23}
   ]
  }
  MaterialExpressionAppendVector_10 {
   class=MaterialExpressionAppendVector
   inputs [
    A {from=MaterialExpressionConstant_51, output=""}
    B {from=MaterialExpressionConstant_62, output=""}
   ]
  }
  MaterialExpressionAppendVector_11 {
   class=MaterialExpressionAppendVector
   inputs [
    A {from=MaterialExpressionConstant2Vector_15, output=""}
    B {from=MaterialExpressionComponentMask_16, output=""}
   ]
  }
  MaterialExpressionAppendVector_12 {
   class=MaterialExpressionAppendVector
   inputs [
    A {from=MaterialExpressionConstant_52, output=""}
    B {from=MaterialExpressionDivide_24, output=""}
   ]
  }
  MaterialExpressionAppendVector_13 {
   class=MaterialExpressionAppendVector
   inputs [
    A {from=MaterialExpressionConstant_53, output=""}
    B {from=MaterialExpressionScalarParameter_88, output=""}
   ]
  }
  MaterialExpressionAppendVector_14 {
   class=MaterialExpressionAppendVector
   inputs [
    A {from=MaterialExpressionScalarParameter_87, output=""}
    B {from=MaterialExpressionScalarParameter_89, output=""}
   ]
  }
  MaterialExpressionAtmosphericLightVector_4 {class=MaterialExpressionAtmosphericLightVector, inputs [], properties {}}
  MaterialExpressionAtmosphericLightVector_5 {class=MaterialExpressionAtmosphericLightVector, inputs [], properties {}}
  MaterialExpressionBounds_0 {class=MaterialExpressionBounds, inputs [], properties {type=MEILB_PreSkinnedLocal}}
  MaterialExpressionCameraVectorWS_12 {class=MaterialExpressionCameraVectorWS, inputs [], properties {}}
  MaterialExpressionCameraVectorWS_13 {class=MaterialExpressionCameraVectorWS, inputs [], properties {}}
  MaterialExpressionCameraVectorWS_14 {class=MaterialExpressionCameraVectorWS, inputs [], properties {}}
  MaterialExpressionCameraVectorWS_15 {class=MaterialExpressionCameraVectorWS, inputs [], properties {}}
  MaterialExpressionCameraVectorWS_16 {class=MaterialExpressionCameraVectorWS, inputs [], properties {}}
  MaterialExpressionCameraVectorWS_17 {class=MaterialExpressionCameraVectorWS, inputs [], properties {}}
  MaterialExpressionComponentMask_11 {class=MaterialExpressionComponentMask, inputs [None {from=MaterialExpressionDivide_34}], properties {b=true}}
  MaterialExpressionComponentMask_14 {class=MaterialExpressionComponentMask, inputs [None {from=""}], properties {r=true}}
  MaterialExpressionComponentMask_16 {
   class=MaterialExpressionComponentMask
   inputs [None {from=MaterialExpressionPreSkinnedLocalBounds_2, output=Half Extents}]
   properties {b=true}
  }
  MaterialExpressionComponentMask_17 {class=MaterialExpressionComponentMask, inputs [None {from=MaterialExpressionTransformPosition_4}], properties {b=true}}
  MaterialExpressionComponentMask_18 {class=MaterialExpressionComponentMask, inputs [None {from=MaterialExpressionReroute_48}], properties {r=true}}
  MaterialExpressionComponentMask_19 {class=MaterialExpressionComponentMask, inputs [None {from=MaterialExpressionReroute_48}], properties {g=true}}
  MaterialExpressionComponentMask_20 {class=MaterialExpressionComponentMask, inputs [None {from=MaterialExpressionMaterialFunctionCall_33, output=Result}], properties {g=true}}
  MaterialExpressionComponentMask_21 {class=MaterialExpressionComponentMask, inputs [None {from=MaterialExpressionNamedRerouteUsage_140}], properties {r=true}}
  MaterialExpressionConstant2Vector_14 {class=MaterialExpressionConstant2Vector, properties {g=0.5, r=0.5}}
  MaterialExpressionConstant2Vector_15 {class=MaterialExpressionConstant2Vector}
  MaterialExpressionConstant2Vector_16 {class=MaterialExpressionConstant2Vector}
  MaterialExpressionConstant2Vector_17 {class=MaterialExpressionConstant2Vector, properties {g=-0.5, r=0.5}}
  MaterialExpressionConstant2Vector_18 {class=MaterialExpressionConstant2Vector, properties {g=0.3, r=0.4}}
  MaterialExpressionConstant2Vector_19 {class=MaterialExpressionConstant2Vector, properties {g=3, r=1}}
  MaterialExpressionConstant2Vector_20 {class=MaterialExpressionConstant2Vector, properties {r=1}}
  MaterialExpressionConstant3Vector_0 {class=MaterialExpressionConstant3Vector, properties {constant=<rgba 0 0 1 0>}}
  MaterialExpressionConstant3Vector_1 {class=MaterialExpressionConstant3Vector, properties {constant=<rgba 0 0 1 0>}}
  MaterialExpressionConstant4Vector_24 {class=MaterialExpressionConstant4Vector, inputs [], properties {constant=<rgba 0.6 0 -1 1>}}
  MaterialExpressionConstant4Vector_25 {class=MaterialExpressionConstant4Vector, inputs [], properties {constant=<rgba 0.7594 2 1.8597 1>}}
  MaterialExpressionConstant4Vector_26 {class=MaterialExpressionConstant4Vector, inputs [], properties {constant=<rgba 2.1605 0.2171 3.942 1>}}
  MaterialExpressionConstant4Vector_27 {class=MaterialExpressionConstant4Vector, inputs [], properties {constant=<rgba 0.0886 0.0351 0.2274 1>}}
  MaterialExpressionConstant4Vector_28 {class=MaterialExpressionConstant4Vector, inputs [], properties {constant=<rgba 0.2091 0.3314 0.3942 1>}}
  MaterialExpressionConstant4Vector_29 {class=MaterialExpressionConstant4Vector, inputs [], properties {constant=<rgba 0.4203 5 1.2964 0>}}
  MaterialExpressionConstant4Vector_30 {class=MaterialExpressionConstant4Vector, inputs [], properties {constant=<rgba -0.5 -1 -0.6 0>}}
  MaterialExpressionConstant4Vector_31 {class=MaterialExpressionConstant4Vector, inputs [], properties {constant=<rgba -0.5 -0.5 1 0>}}
  MaterialExpressionConstant4Vector_32 {class=MaterialExpressionConstant4Vector, inputs [], properties {constant=<rgba 0.245 5 1.4854 1>}}
  MaterialExpressionConstant4Vector_33 {class=MaterialExpressionConstant4Vector, inputs [], properties {constant=<rgba 0 5.6 7 1>}}
  MaterialExpressionConstant4Vector_34 {class=MaterialExpressionConstant4Vector, inputs [], properties {constant=<rgba 4 0.911 3.5434 1>}}
  MaterialExpressionConstant4Vector_35 {class=MaterialExpressionConstant4Vector, inputs [], properties {constant=<rgba 0.0275 0.1815 0.2435 1>}}
  MaterialExpressionConstantBiasScale_6 {class=MaterialExpressionConstantBiasScale, inputs [None {from=MaterialExpressionSine_1}]}
  MaterialExpressionConstantBiasScale_7 {
   class=MaterialExpressionConstantBiasScale
   inputs [None {from=MaterialExpressionScreenPosition_2, output=ViewportUV}]
   properties {bias=-0.5, scale=2}
  }
  MaterialExpressionConstantBiasScale_8 {class=MaterialExpressionConstantBiasScale, inputs [None {from=MaterialExpressionMultiply_107}]}
  MaterialExpressionConstant_24 {class=MaterialExpressionConstant}
  MaterialExpressionConstant_25 {class=MaterialExpressionConstant, properties {r=1}}
  MaterialExpressionConstant_28 {class=MaterialExpressionConstant, properties {r=1}}
  MaterialExpressionConstant_29 {class=MaterialExpressionConstant, properties {r=1}}
  MaterialExpressionConstant_30 {class=MaterialExpressionConstant, properties {r=0.5}}
  MaterialExpressionConstant_51 {class=MaterialExpressionConstant}
  MaterialExpressionConstant_52 {class=MaterialExpressionConstant, properties {r=1}}
  MaterialExpressionConstant_53 {class=MaterialExpressionConstant, properties {r=0.5}}
  MaterialExpressionConstant_54 {class=MaterialExpressionConstant, properties {r=10}}
  MaterialExpressionConstant_55 {class=MaterialExpressionConstant, properties {r=0.45}}
  MaterialExpressionConstant_56 {class=MaterialExpressionConstant, properties {r=10}}
  MaterialExpressionConstant_57 {class=MaterialExpressionConstant, properties {r=0.5}}
  MaterialExpressionConstant_58 {class=MaterialExpressionConstant, properties {r=2}}
  MaterialExpressionConstant_59 {class=MaterialExpressionConstant}
  MaterialExpressionConstant_60 {class=MaterialExpressionConstant, properties {r=0.5}}
  MaterialExpressionConstant_61 {class=MaterialExpressionConstant, properties {r=5}}
  MaterialExpressionConstant_62 {class=MaterialExpressionConstant, properties {r=-0.1}}
  MaterialExpressionConstant_63 {class=MaterialExpressionConstant, properties {r=6}}
  MaterialExpressionConstant_64 {class=MaterialExpressionConstant, properties {r=1}}
  MaterialExpressionConstant_65 {class=MaterialExpressionConstant, properties {r=1}}
  MaterialExpressionConstant_66 {class=MaterialExpressionConstant, properties {r=0.55}}
  MaterialExpressionConstant_67 {class=MaterialExpressionConstant}
  MaterialExpressionConstant_68 {class=MaterialExpressionConstant, properties {r=0.3}}
  MaterialExpressionConstant_69 {class=MaterialExpressionConstant, properties {r=-0.15}}
  MaterialExpressionConstant_70 {class=MaterialExpressionConstant, properties {r=0.05}}
  MaterialExpressionConstant_71 {class=MaterialExpressionConstant, properties {r=3}}
  MaterialExpressionConstant_72 {class=MaterialExpressionConstant, properties {r=256}}
  MaterialExpressionConstant_73 {class=MaterialExpressionConstant, properties {r=1000}}
  MaterialExpressionConvert_10 {
   class=MaterialExpressionConvert
   inputs [
    None {from=MaterialExpressionScalarParameter_85}
    None {from=MaterialExpressionScalarParameter_84}
   ]
   properties {
    nodeName=Make Vector2
    convertInputs [
     {
      type=Scalar
      defaultValue=<rgba 0.2 0 0 1>
      expressionInput {
       expression=&self:MaterialExpressionScalarParameter_85
       inputName=None
       mask=0
       maskA=0
       maskB=0
       maskG=0
       maskR=0
       outputIndex=0
      }
     }
     {
      type=Scalar
      defaultValue=<rgba 0.2 0 0 1>
      expressionInput {
       expression=&self:MaterialExpressionScalarParameter_84
       inputName=None
       mask=0
       maskA=0
       maskB=0
       maskG=0
       maskR=0
       outputIndex=0
      }
     }
    ]
    convertOutputs [{type=Vector2}]
   }
  }
  MaterialExpressionConvert_12 {
   class=MaterialExpressionConvert
   inputs [None {from=MaterialExpressionAbs_7}]
   properties {
    nodeName=Break Float3
    convertInputs [{
      type=Vector3
      expressionInput {
       expression=&self:MaterialExpressionAbs_7
       inputName=None
       mask=0
       maskA=0
       maskB=0
       maskG=0
       maskR=0
       outputIndex=0
      }
     }]
    convertOutputs [{type=Scalar}, {type=Scalar}, {type=Scalar}]
   }
  }
  MaterialExpressionConvert_2 {
   class=MaterialExpressionConvert
   inputs [None {from=MaterialExpressionNamedRerouteUsage_173}]
   properties {
    nodeName=Break Float3
    convertInputs [{
      type=Vector3
      expressionInput {
       expression=&self:MaterialExpressionNamedRerouteUsage_173
       inputName=None
       mask=0
       maskA=0
       maskB=0
       maskG=0
       maskR=0
       outputIndex=0
      }
     }]
    convertOutputs [{type=Scalar}, {type=Scalar}, {type=Scalar}]
   }
  }
  MaterialExpressionConvert_5 {class=MaterialExpressionConvert, inputs [], properties {convertOutputs [{type=Vector3, defaultValue=<rgba 1 1 1 1>}]}}
  MaterialExpressionConvert_9 {
   class=MaterialExpressionConvert
   inputs [
    None {from=MaterialExpressionScalarParameter_83}
    None {from=MaterialExpressionScalarParameter_82}
   ]
   properties {
    nodeName=Make Vector2
    convertInputs [
     {
      type=Scalar
      defaultValue=<rgba 0.2 0 0 1>
      expressionInput {
       expression=&self:MaterialExpressionScalarParameter_83
       inputName=None
       mask=0
       maskA=0
       maskB=0
       maskG=0
       maskR=0
       outputIndex=0
      }
     }
     {
      type=Scalar
      defaultValue=<rgba 0.2 0 0 1>
      expressionInput {
       expression=&self:MaterialExpressionScalarParameter_82
       inputName=None
       mask=0
       maskA=0
       maskB=0
       maskG=0
       maskR=0
       outputIndex=0
      }
     }
    ]
    convertOutputs [{type=Vector2}]
   }
  }
  MaterialExpressionDDX_2 {class=MaterialExpressionDDX, inputs [Value {from=MaterialExpressionMultiply_163, output=""}], properties {}}
  MaterialExpressionDDY_1 {class=MaterialExpressionDDY, inputs [Value {from=MaterialExpressionMultiply_163, output=""}], properties {}}
  MaterialExpressionDistance_2 {
   class=MaterialExpressionDistance
   inputs [
    A {from=MaterialExpressionDivide_27, output=""}
    B {from=MaterialExpressionConstant2Vector_16, output=""}
   ]
   properties {}
  }
  MaterialExpressionDivide_23 {
   class=MaterialExpressionDivide
   inputs [
    A {from=MaterialExpressionAdd_31}
    B {from=""}
   ]
  }
  MaterialExpressionDivide_24 {
   class=MaterialExpressionDivide
   inputs [
    A {from=MaterialExpressionMaterialFunctionCall_39, output=G}
    B {from=MaterialExpressionMaterialFunctionCall_39, output=G}
   ]
  }
  MaterialExpressionDivide_25 {
   class=MaterialExpressionDivide
   inputs [
    A {from=MaterialExpressionMultiply_116}
    B {from=MaterialExpressionComponentMask_17}
   ]
   properties {constA=100}
  }
  MaterialExpressionDivide_26 {
   class=MaterialExpressionDivide
   inputs [
    A {from=MaterialExpressionMultiply_115}
    B {from=MaterialExpressionConvert_10}
   ]
  }
  MaterialExpressionDivide_27 {
   class=MaterialExpressionDivide
   inputs [
    A {from=MaterialExpressionSubtract_36}
    B {from=MaterialExpressionAppendVector_14}
   ]
  }
  MaterialExpressionDivide_28 {
   class=MaterialExpressionDivide
   inputs [
    A {from=MaterialExpressionMultiply_120}
    B {from=MaterialExpressionConstant2Vector_19}
   ]
  }
  MaterialExpressionDivide_29 {
   class=MaterialExpressionDivide
   inputs [
    A {from=MaterialExpressionTime_7}
    B {from=MaterialExpressionScalarParameter_102}
   ]
  }
  MaterialExpressionDivide_30 {
   class=MaterialExpressionDivide
   inputs [
    A {from=MaterialExpressionAdd_39}
    B {from=""}
   ]
  }
  MaterialExpressionDivide_31 {
   class=MaterialExpressionDivide
   inputs [
    A {from=MaterialExpressionSubtract_45}
    B {from=MaterialExpressionConstant_73}
   ]
   properties {constB=5000}
  }
  MaterialExpressionDivide_32 {
   class=MaterialExpressionDivide
   inputs [
    A {from=""}
    B {from=MaterialExpressionNamedRerouteUsage_183}
   ]
  }
  MaterialExpressionDivide_33 {
   class=MaterialExpressionDivide
   inputs [
    A {from=MaterialExpressionTime_8}
    B {from=MaterialExpressionScalarParameter_104}
   ]
   properties {constB=10}
  }
  MaterialExpressionDivide_34 {
   class=MaterialExpressionDivide
   inputs [
    A {from=MaterialExpressionSubtract_46}
    B {from=MaterialExpressionNamedRerouteUsage_148}
   ]
  }
  MaterialExpressionDotProduct_0 {
   class=MaterialExpressionDotProduct
   inputs [
    A {from=MaterialExpressionConvert_5, output=""}
    B {from=MaterialExpressionNamedRerouteDeclaration_8, output=""}
   ]
   properties {}
  }
  MaterialExpressionDotProduct_21 {
   class=MaterialExpressionDotProduct
   inputs [
    A {from=MaterialExpressionCameraVectorWS_12, output=""}
    B {from=MaterialExpressionVertexNormalWS_8, output=""}
   ]
   properties {}
  }
  MaterialExpressionDotProduct_22 {
   class=MaterialExpressionDotProduct
   inputs [
    A {from=MaterialExpressionCameraVectorWS_14, output=""}
    B {from=MaterialExpressionTransform_17, output=""}
   ]
   properties {}
  }
  MaterialExpressionDotProduct_23 {
   class=MaterialExpressionDotProduct
   inputs [
    A {from=MaterialExpressionAtmosphericLightVector_4, output=""}
    B {from=MaterialExpressionMaterialFunctionCall_35, output=Result}
   ]
   properties {}
  }
  MaterialExpressionDotProduct_24 {
   class=MaterialExpressionDotProduct
   inputs [
    A {from=MaterialExpressionAtmosphericLightVector_5, output=""}
    B {from=MaterialExpressionPixelNormalWS_1, output=""}
   ]
   properties {}
  }
  MaterialExpressionDotProduct_25 {
   class=MaterialExpressionDotProduct
   inputs [
    A {from=MaterialExpressionNormalize_3, output=""}
    B {from=MaterialExpressionMultiply_101, output=""}
   ]
   properties {}
  }
  MaterialExpressionDotProduct_26 {
   class=MaterialExpressionDotProduct
   inputs [
    A {from=MaterialExpressionTransform_18, output=""}
    B {from=MaterialExpressionTransform_16, output=""}
   ]
   properties {}
  }
  MaterialExpressionDotProduct_27 {
   class=MaterialExpressionDotProduct
   inputs [
    A {from=MaterialExpressionTransform_19, output=""}
    B {from=MaterialExpressionConstant3Vector_1, output=""}
   ]
   properties {}
  }
  MaterialExpressionDotProduct_28 {
   class=MaterialExpressionDotProduct
   inputs [
    A {from=MaterialExpressionTransform_22, output=""}
    B {from=MaterialExpressionNormalize_4, output=""}
   ]
   properties {}
  }
  MaterialExpressionDotProduct_29 {
   class=MaterialExpressionDotProduct
   inputs [
    A {from=MaterialExpressionNamedRerouteUsage_111, output=""}
    B {from=MaterialExpressionVectorParameter_55, output=RGB}
   ]
   properties {}
  }
  MaterialExpressionDotProduct_30 {
   class=MaterialExpressionDotProduct
   inputs [
    A {from=MaterialExpressionCameraVectorWS_17, output=""}
    B {from=MaterialExpressionTransform_23, output=""}
   ]
   properties {}
  }
  MaterialExpressionFrac_10 {class=MaterialExpressionFrac, inputs [None {from=MaterialExpressionMultiply_143, output=""}], properties {}}
  MaterialExpressionFrac_11 {class=MaterialExpressionFrac, inputs [None {from=MaterialExpressionMultiply_144, output=""}], properties {}}
  MaterialExpressionFrac_8 {class=MaterialExpressionFrac, inputs [None {from=MaterialExpressionMultiply_125, output=""}], properties {}}
  MaterialExpressionFrac_9 {class=MaterialExpressionFrac, inputs [None {from=MaterialExpressionDivide_29, output=""}], properties {}}
  MaterialExpressionFunctionInput_0 {
   class=MaterialExpressionFunctionInput
   inputs [Preview {from=MaterialExpressionShaderStageSwitch_4, output=""}]
   properties {bUsePreviewValueAsDefault=true, inputName=Pre-Skinned Local Pos, sortPriority=100}
  }
  MaterialExpressionFunctionInput_1 {
   class=MaterialExpressionFunctionInput
   inputs [Preview {from=MaterialExpressionBounds_0, output=Max}]
   properties {bUsePreviewValueAsDefault=true, inputName=Pre-Skinned Bounds Max, sortPriority=120}
  }
  MaterialExpressionFunctionInput_11 {
   class=MaterialExpressionFunctionInput
   inputs [Preview {from="", output=""}]
   properties {inputName=Mask_Logo, inputType=FunctionInput_Scalar, sortPriority=1}
  }
  MaterialExpressionFunctionInput_12 {
   class=MaterialExpressionFunctionInput
   inputs [Preview {from="", output=""}]
   properties {inputName=Mask_SpriteFresnel, inputType=FunctionInput_Scalar, sortPriority=1}
  }
  MaterialExpressionFunctionInput_13 {
   class=MaterialExpressionFunctionInput
   inputs [Preview {from="", output=""}]
   properties {inputName=WPO, inputType=FunctionInput_Scalar, sortPriority=2}
  }
  MaterialExpressionFunctionInput_14 {
   class=MaterialExpressionFunctionInput
   inputs [Preview {from="", output=""}]
   properties {inputName=Mask_BlinkingLights, inputType=FunctionInput_Scalar, sortPriority=1}
  }
  MaterialExpressionFunctionInput_15 {
   class=MaterialExpressionFunctionInput
   inputs [Preview {from="", output=""}]
   properties {inputName=Mask_DiffuseChannelMask, inputType=FunctionInput_Scalar, sortPriority=1}
  }
  MaterialExpressionFunctionInput_2 {
   class=MaterialExpressionFunctionInput
   inputs [Preview {from=MaterialExpressionBounds_0, output=Min}]
   properties {bUsePreviewValueAsDefault=true, inputName=Pre-Skinned Bounds Min, sortPriority=110}
  }
  MaterialExpressionFunctionInput_3 {
   class=MaterialExpressionFunctionInput
   inputs [Preview {from="", output=""}]
   properties {inputName=MaterialAttributes, inputType=FunctionInput_MaterialAttributes, sortPriority=-1}
  }
  MaterialExpressionFunctionInput_4 {
   class=MaterialExpressionFunctionInput
   inputs [Preview {from="", output=""}]
   properties {inputName=Mask Variant, inputType=FunctionInput_Scalar, sortPriority=1}
  }
  MaterialExpressionFunctionInput_5 {
   class=MaterialExpressionFunctionInput
   inputs [Preview {from="", output=""}]
   properties {inputName=Mask_InnerRadialGlow, inputType=FunctionInput_Scalar, sortPriority=1}
  }
  MaterialExpressionFunctionInput_6 {
   class=MaterialExpressionFunctionInput
   inputs [Preview {from=MaterialExpressionObjectPositionWS_4, output=""}]
   properties {bUsePreviewValueAsDefault=true, inputName=Object Position WS, sortPriority=140}
  }
  MaterialExpressionFunctionInput_7 {
   class=MaterialExpressionFunctionInput
   inputs [Preview {from="", output=""}]
   properties {bUsePreviewValueAsDefault=true, inputName=Pre-Skinned Scale, sortPriority=130, previewValue=<1 1 1 1>}
  }
  MaterialExpressionFunctionOutput_0 {
   class=MaterialExpressionFunctionOutput
   inputs [None {from=MaterialExpressionStaticSwitchParameter_4, output=""}]
   properties {outputName=SetMaterialAttributes}
  }
  MaterialExpressionGetMaterialAttributes_0 {
   class=MaterialExpressionGetMaterialAttributes
   inputs [None {from=MaterialExpressionFunctionInput_3, output=""}]
   properties {attributeGetTypes [
     #1
     #2
     #3
     #4
     #5
    ]}
  }
  MaterialExpressionLinearInterpolate_12 {
   class=MaterialExpressionLinearInterpolate
   inputs [
    A {from=MaterialExpressionConstant3Vector_0}
    B {from=MaterialExpressionTextureSample_5, output=RGB}
    Alpha {from=MaterialExpressionScalarParameter_16}
   ]
  }
  MaterialExpressionLinearInterpolate_16 {
   class=MaterialExpressionLinearInterpolate
   inputs [
    A {from=MaterialExpressionConstant_24}
    B {from=MaterialExpressionReroute_21}
    Alpha {from=MaterialExpressionScalarParameter_24}
   ]
  }
  MaterialExpressionLinearInterpolate_48 {
   class=MaterialExpressionLinearInterpolate
   inputs [
    A {from=MaterialExpressionVectorParameter_39, output=RGB}
    B {from=MaterialExpressionVectorParameter_40, output=RGB}
    Alpha {from=MaterialExpressionNamedRerouteUsage_108}
   ]
  }
  MaterialExpressionLinearInterpolate_49 {
   class=MaterialExpressionLinearInterpolate
   inputs [
    A {from=MaterialExpressionVectorParameter_47, output=RGB}
    B {from=MaterialExpressionVectorParameter_48, output=RGB}
    Alpha {from=MaterialExpressionNamedRerouteUsage_99}
   ]
  }
  MaterialExpressionLinearInterpolate_50 {
   class=MaterialExpressionLinearInterpolate
   inputs [
    A {from=MaterialExpressionMultiply_96}
    B {from=MaterialExpressionVectorParameter_38, output=RGB}
    Alpha {from=MaterialExpressionVectorParameter_38, output=RGB}
   ]
  }
  MaterialExpressionLinearInterpolate_51 {
   class=MaterialExpressionLinearInterpolate
   inputs [
    A {from=MaterialExpressionAppendVector_13}
    B {from=MaterialExpressionConstant2Vector_14}
    Alpha {from=MaterialExpressionSaturate_20}
   ]
  }
  MaterialExpressionLinearInterpolate_52 {
   class=MaterialExpressionLinearInterpolate
   inputs [
    A {from=MaterialExpressionScalarParameter_92}
    B {from=MaterialExpressionScalarParameter_90}
    Alpha {from=MaterialExpressionPower_27}
   ]
   properties {constB=0}
  }
  MaterialExpressionLinearInterpolate_53 {
   class=MaterialExpressionLinearInterpolate
   inputs [
    A {from=""}
    B {from=MaterialExpressionQualitySwitch_2}
    Alpha {from=MaterialExpressionSmoothStep_21}
   ]
   properties {constA=1, constB=0}
  }
  MaterialExpressionLinearInterpolate_54 {
   class=MaterialExpressionLinearInterpolate
   inputs [
    A {from=MaterialExpressionVectorParameter_50, output=RGB}
    B {from=MaterialExpressionVectorParameter_49, output=RGB}
    Alpha {from=MaterialExpressionNamedRerouteUsage_107}
   ]
  }
  MaterialExpressionLinearInterpolate_55 {
   class=MaterialExpressionLinearInterpolate
   inputs [
    A {from=MaterialExpressionReroute_56}
    B {from=MaterialExpressionMultiply_132}
    Alpha {from=MaterialExpressionSmoothStep_24}
   ]
  }
  MaterialExpressionLinearInterpolate_56 {
   class=MaterialExpressionLinearInterpolate
   inputs [
    A {from=MaterialExpressionOneMinus_25}
    B {from=MaterialExpressionComponentMask_19}
    Alpha {from=MaterialExpressionScalarParameter_97}
   ]
  }
  MaterialExpressionLinearInterpolate_57 {
   class=MaterialExpressionLinearInterpolate
   inputs [
    A {from=MaterialExpressionMultiply_128}
    B {from=MaterialExpressionMultiply_127}
    Alpha {from=MaterialExpressionVectorParameter_56, output=A}
   ]
  }
  MaterialExpressionLinearInterpolate_58 {
   class=MaterialExpressionLinearInterpolate
   inputs [
    A {from=MaterialExpressionNamedRerouteUsage_122}
    B {from=MaterialExpressionDotProduct_28}
    Alpha {from=MaterialExpressionScalarParameter_98}
   ]
  }
  MaterialExpressionLinearInterpolate_59 {
   class=MaterialExpressionLinearInterpolate
   inputs [
    A {from=MaterialExpressionConstant_70}
    B {from=""}
    Alpha {from=MaterialExpressionNamedRerouteUsage_124}
   ]
   properties {constA=0.5, constB=1.1}
  }
  MaterialExpressionLinearInterpolate_60 {
   class=MaterialExpressionLinearInterpolate
   inputs [
    A {from=MaterialExpressionTextureSample_0, output=R}
    B {from=""}
    Alpha {from=MaterialExpressionSmoothStep_25}
   ]
   properties {constB=0}
  }
  MaterialExpressionLinearInterpolate_61 {
   class=MaterialExpressionLinearInterpolate
   inputs [
    A {from=MaterialExpressionMultiply_123}
    B {from=MaterialExpressionScalarParameter_99}
    Alpha {from=MaterialExpressionNamedRerouteUsage_117}
   ]
  }
  MaterialExpressionLinearInterpolate_62 {
   class=MaterialExpressionLinearInterpolate
   inputs [
    A {from=MaterialExpressionDotProduct_29}
    B {from=MaterialExpressionMultiply_131}
    Alpha {from=MaterialExpressionSmoothStep_26}
   ]
  }
  MaterialExpressionLinearInterpolate_63 {
   class=MaterialExpressionLinearInterpolate
   inputs [
    A {from=MaterialExpressionMultiply_133}
    B {from=""}
    Alpha {from=MaterialExpressionNamedRerouteUsage_120}
   ]
   properties {constA=1}
  }
  MaterialExpressionLinearInterpolate_64 {
   class=MaterialExpressionLinearInterpolate
   inputs [
    A {from=MaterialExpressionMultiply_135}
    B {from=MaterialExpressionMultiply_134}
    Alpha {from=MaterialExpressionPower_30}
   ]
  }
  MaterialExpressionLinearInterpolate_65 {
   class=MaterialExpressionLinearInterpolate
   inputs [
    A {from=MaterialExpressionAdd_33}
    B {from=MaterialExpressionMultiply_118}
    Alpha {from=MaterialExpressionMultiply_136}
   ]
  }
  MaterialExpressionLinearInterpolate_66 {
   class=MaterialExpressionLinearInterpolate
   inputs [
    A {from=""}
    B {from=MaterialExpressionConstant_71}
    Alpha {from=MaterialExpressionNamedRerouteUsage_126}
   ]
   properties {constA=1, constB=5}
  }
  MaterialExpressionLinearInterpolate_67 {
   class=MaterialExpressionLinearInterpolate
   inputs [
    A {from=MaterialExpressionVectorParameter_56, output=RGB}
    B {from=MaterialExpressionVectorParameter_53, output=RGB}
    Alpha {from=MaterialExpressionPower_32}
   ]
  }
  MaterialExpressionLinearInterpolate_68 {
   class=MaterialExpressionLinearInterpolate
   inputs [
    A {from=MaterialExpressionAdd_32}
    B {from=MaterialExpressionLinearInterpolate_70}
    Alpha {from=MaterialExpressionLinearInterpolate_69}
   ]
  }
  MaterialExpressionLinearInterpolate_69 {
   class=MaterialExpressionLinearInterpolate
   inputs [
    A {from=""}
    B {from=MaterialExpressionVectorParameter_46, output=A}
    Alpha {from=MaterialExpressionNamedRerouteUsage_97}
   ]
  }
  MaterialExpressionLinearInterpolate_70 {
   class=MaterialExpressionLinearInterpolate
   inputs [
    A {from=MaterialExpressionMultiply_104}
    B {from=MaterialExpressionMultiply_103}
    Alpha {from=MaterialExpressionVectorParameter_46, output=A}
   ]
  }
  MaterialExpressionLinearInterpolate_71 {
   class=MaterialExpressionLinearInterpolate
   inputs [
    A {from=""}
    B {from=MaterialExpressionDivide_32}
    Alpha {from=MaterialExpressionScalarParameter_142}
   ]
   properties {constA=1}
  }
  MaterialExpressionLinearInterpolate_72 {
   class=MaterialExpressionLinearInterpolate
   inputs [
    A {from=""}
    B {from=MaterialExpressionNamedRerouteUsage_210}
    Alpha {from=MaterialExpressionScalarParameter_143}
   ]
   properties {constA=1}
  }
  MaterialExpressionLinearInterpolate_73 {
   class=MaterialExpressionLinearInterpolate
   inputs [
    A {from=MaterialExpressionConstant_59}
    B {from=MaterialExpressionConstant4Vector_34, output=RGB}
    Alpha {from=MaterialExpressionMultiply_146}
   ]
   properties {constA=-10, constB=2}
  }
  MaterialExpressionLocalPosition_5 {class=MaterialExpressionLocalPosition, inputs [], properties {includedOffsets=ExcludeOffsets, localOrigin=InstancePreSkinning}}
  MaterialExpressionMaterialFunctionCall_15 {
   class=MaterialExpressionMaterialFunctionCall
   inputs [
    BaseNormal {from=MaterialExpressionReroute_29}
    AdditionalNormal {from=MaterialExpressionNamedRerouteUsage_58}
   ]
   properties {materialFunction=&BlendAngleCorrectedNormals}
  }
  MaterialExpressionMaterialFunctionCall_2 {
   class=MaterialExpressionMaterialFunctionCall
   inputs [
    Value_Morning {from=MaterialExpressionConstant_25}
    Value_Day {from=MaterialExpressionConstant_28}
    Value_Evening {from=MaterialExpressionConstant_29}
    Value_Night {from=MaterialExpressionConstant_30}
   ]
   properties {materialFunction=&Scalar_Time_of_Day}
  }
  MaterialExpressionMaterialFunctionCall_33 {
   class=MaterialExpressionMaterialFunctionCall
   inputs [Vector {from=MaterialExpressionReflectionVectorWS_2}]
   properties {materialFunction=/Engine/Functions/Engine_MaterialFunctions01/Coordinates/LongLatToUV}
  }
  MaterialExpressionMaterialFunctionCall_34 {
   class=MaterialExpressionMaterialFunctionCall
   inputs [
    Color_Morning {from=MaterialExpressionVectorParameter_41, output=RGB}
    Color_Day {from=MaterialExpressionVectorParameter_42, output=RGB}
    Color_Evening {from=MaterialExpressionVectorParameter_43, output=RGB}
    Color_Night {from=MaterialExpressionVectorParameter_44, output=RGB}
   ]
   properties {materialFunction=&Color_Time_of_Day}
  }
  MaterialExpressionMaterialFunctionCall_35 {
   class=MaterialExpressionMaterialFunctionCall
   inputs []
   properties {materialFunction=&CameraDirectionVector}
  }
  MaterialExpressionMaterialFunctionCall_36 {
   class=MaterialExpressionMaterialFunctionCall
   inputs []
   properties {materialFunction=&CameraDirectionVector}
  }
  MaterialExpressionMaterialFunctionCall_37 {
   class=MaterialExpressionMaterialFunctionCall
   inputs [
    Position {from=MaterialExpressionTransformPosition_3}
    UVRadialScale {from=""}
   ]
   properties {materialFunction=/Game/Weapons/WeaponSkins/MaterialFunction/ObjectStabilizedScreenSpaceUVs}
  }
  MaterialExpressionMaterialFunctionCall_38 {
   class=MaterialExpressionMaterialFunctionCall
   inputs []
   properties {materialFunction=&CameraDirectionVector}
  }
  MaterialExpressionMaterialFunctionCall_39 {
   class=MaterialExpressionMaterialFunctionCall
   inputs [Float2 {from=MaterialExpressionViewSize_1}]
   properties {materialFunction=&BreakOutFloat2Components}
  }
  MaterialExpressionMaterialFunctionCall_40 {
   class=MaterialExpressionMaterialFunctionCall
   inputs [
    World Position {from=MaterialExpressionNamedRerouteUsage_141}
    Buffer UV Pixel Offset {from=""}
   ]
   properties {materialFunction=&TransformToClipSpace}
  }
  MaterialExpressionMaterialFunctionCall_41 {
   class=MaterialExpressionMaterialFunctionCall
   inputs [
    UVs {from=MaterialExpressionSubtract_37}
    Texture Scale {from=MaterialExpressionDivide_25}
   ]
   properties {materialFunction=&ScaleUVsByCenter}
  }
  MaterialExpressionMaterialFunctionCall_42 {
   class=MaterialExpressionMaterialFunctionCall
   inputs []
   properties {materialFunction=/Engine/Functions/Engine_MaterialFunctions02/Utility/FOV}
  }
  MaterialExpressionMaterialFunctionCall_43 {
   class=MaterialExpressionMaterialFunctionCall
   inputs []
   properties {materialFunction=&ObjectScale}
  }
  MaterialExpressionMaterialFunctionCall_44 {
   class=MaterialExpressionMaterialFunctionCall
   inputs [
    UVs {from=MaterialExpressionMaterialFunctionCall_45, output=UVs}
    TriplanarMask {from=MaterialExpressionMaterialFunctionCall_45, output=UVs}
    CameraVector {from=MaterialExpressionCameraVectorWS_15}
    Bump Offset {from=MaterialExpressionMultiply_152}
   ]
   properties {materialFunction=&/MF_TriplanarBumpOffset}
  }
  MaterialExpressionMaterialFunctionCall_45 {
   class=MaterialExpressionMaterialFunctionCall
   inputs [
    Texture Scale {from=MaterialExpressionScalarParameter_91}
    Normal {from=MaterialExpressionNamedRerouteUsage_144}
    InPosition {from=MaterialExpressionNamedRerouteUsage_164}
    Triplanar Fade {from=MaterialExpressionConstant_54}
    Derivatives Scale {from=""}
   ]
   properties {materialFunction=&/MF_TriplanarDithered_PreSkinned}
  }
  MaterialExpressionMaterialFunctionCall_46 {
   class=MaterialExpressionMaterialFunctionCall
   inputs [
    Position {from=""}
    UVRadialScale {from=""}
   ]
   properties {materialFunction=&ObjectStabilizedScreenSpaceUVs}
  }
  MaterialExpressionMaterialFunctionCall_47 {
   class=MaterialExpressionMaterialFunctionCall
   inputs [
    Texture Scale {from=MaterialExpressionScalarParameter_100}
    Normal {from=MaterialExpressionNamedRerouteUsage_145}
    InPosition {from=MaterialExpressionNamedRerouteUsage_166}
    Triplanar Fade {from=MaterialExpressionScalarParameter_26}
    Derivatives Scale {from=""}
   ]
   properties {materialFunction=&MF_TriplanarDithered_PreSkinned}
  }
  MaterialExpressionMaterialFunctionCall_48 {
   class=MaterialExpressionMaterialFunctionCall
   inputs [
    UVs {from=MaterialExpressionMaterialFunctionCall_47, output=UVs}
    TriplanarMask {from=MaterialExpressionMaterialFunctionCall_47, output=UVs}
    CameraVector {from=MaterialExpressionCameraVectorWS_16}
    Bump Offset {from=MaterialExpressionMultiply_153}
   ]
   properties {materialFunction=&MF_TriplanarBumpOffset}
  }
  MaterialExpressionMaterialFunctionCall_50 {
   class=MaterialExpressionMaterialFunctionCall
   inputs [
    Position {from=MaterialExpressionNamedRerouteUsage_171}
    UVRadialScale {from=MaterialExpressionNamedRerouteUsage_176}
    ObjectRadius {from=MaterialExpressionNamedRerouteUsage_175}
   ]
   properties {materialFunction=&/MF_ObjectStabilizedScreenSpaceUVs}
  }
  MaterialExpressionMax_10 {
   class=MaterialExpressionMax
   inputs [
    A {from=MaterialExpressionConvert_2}
    B {from=MaterialExpressionConvert_2}
   ]
  }
  MaterialExpressionMax_11 {
   class=MaterialExpressionMax
   inputs [
    A {from=MaterialExpressionMax_10}
    B {from=MaterialExpressionConvert_2}
   ]
  }
  MaterialExpressionMax_18 {
   class=MaterialExpressionMax
   inputs [
    A {from=MaterialExpressionDotProduct_25}
    B {from=""}
   ]
   properties {constB=0}
  }
  MaterialExpressionMax_19 {
   class=MaterialExpressionMax
   inputs [
    A {from=MaterialExpressionConvert_12}
    B {from=MaterialExpressionConvert_12}
   ]
  }
  MaterialExpressionMax_20 {
   class=MaterialExpressionMax
   inputs [
    A {from=MaterialExpressionConvert_12}
    B {from=MaterialExpressionConvert_12}
   ]
  }
  MaterialExpressionMax_21 {
   class=MaterialExpressionMax
   inputs [
    A {from=MaterialExpressionConvert_12}
    B {from=MaterialExpressionConvert_12}
   ]
  }
  MaterialExpressionMax_22 {
   class=MaterialExpressionMax
   inputs [
    A {from=MaterialExpressionSubtract_38}
    B {from=MaterialExpressionSubtract_39}
   ]
  }
  MaterialExpressionMax_23 {
   class=MaterialExpressionMax
   inputs [
    A {from=MaterialExpressionMax_22}
    B {from=MaterialExpressionSubtract_41}
   ]
  }
  MaterialExpressionMax_24 {
   class=MaterialExpressionMax
   inputs [
    A {from=MaterialExpressionSmoothStep_22}
    B {from=MaterialExpressionSmoothStep_23}
   ]
  }
  MaterialExpressionMax_25 {
   class=MaterialExpressionMax
   inputs [
    A {from=MaterialExpressionFrac_8}
    B {from=MaterialExpressionFrac_10}
   ]
  }
  MaterialExpressionMultiply_0 {
   class=MaterialExpressionMultiply
   inputs [
    A {from=MaterialExpressionMax_11}
    B {from=""}
   ]
   properties {constB=1.4115, desc=a bit arbitrary...}
  }
  MaterialExpressionMultiply_100 {
   class=MaterialExpressionMultiply
   inputs [
    A {from=MaterialExpressionMultiply_99}
    B {from=MaterialExpressionPixelNormalWS_1}
   ]
  }
  MaterialExpressionMultiply_101 {
   class=MaterialExpressionMultiply
   inputs [
    A {from=MaterialExpressionMaterialFunctionCall_36, output=Result}
    B {from=""}
   ]
   properties {constB=-1}
  }
  MaterialExpressionMultiply_102 {
   class=MaterialExpressionMultiply
   inputs [
    A {from=MaterialExpressionPower_23}
    B {from=MaterialExpressionScalarParameter_74}
   ]
   properties {constB=0.4}
  }
  MaterialExpressionMultiply_103 {
   class=MaterialExpressionMultiply
   inputs [
    A {from=MaterialExpressionNamedRerouteUsage_97}
    B {from=MaterialExpressionVectorParameter_46, output=RGB}
   ]
  }
  MaterialExpressionMultiply_104 {
   class=MaterialExpressionMultiply
   inputs [
    A {from=MaterialExpressionReroute_42}
    B {from=MaterialExpressionMultiply_103}
   ]
  }
  MaterialExpressionMultiply_105 {
   class=MaterialExpressionMultiply
   inputs [
    A {from=MaterialExpressionLinearInterpolate_49}
    B {from=MaterialExpressionNamedRerouteUsage_101}
   ]
  }
  MaterialExpressionMultiply_106 {
   class=MaterialExpressionMultiply
   inputs [
    A {from=MaterialExpressionAbs_6}
    B {from=""}
   ]
   properties {constB=2}
  }
  MaterialExpressionMultiply_107 {
   class=MaterialExpressionMultiply
   inputs [
    A {from=MaterialExpressionConstantBiasScale_7}
    B {from=MaterialExpressionAppendVector_12}
   ]
  }
  MaterialExpressionMultiply_108 {
   class=MaterialExpressionMultiply
   inputs [
    A {from=MaterialExpressionMaterialFunctionCall_40, output=Clip Space XY}
    B {from=MaterialExpressionConstant2Vector_17}
   ]
   properties {constA=1}
  }
  MaterialExpressionMultiply_11 {
   class=MaterialExpressionMultiply
   inputs [
    A {from=MaterialExpressionScalarParameter_4}
    B {from=MaterialExpressionMaterialFunctionCall_47, output=UVs}
   ]
  }
  MaterialExpressionMultiply_113 {
   class=MaterialExpressionMultiply
   inputs [
    A {from=MaterialExpressionAppendVector_12}
    B {from=MaterialExpressionMultiply_108}
   ]
  }
  MaterialExpressionMultiply_114 {
   class=MaterialExpressionMultiply
   inputs [
    A {from=MaterialExpressionMaterialFunctionCall_41, output=UVs}
    B {from=MaterialExpressionConstant2Vector_18}
   ]
  }
  MaterialExpressionMultiply_115 {
   class=MaterialExpressionMultiply
   inputs [
    A {from=MaterialExpressionMultiply_114}
    B {from=MaterialExpressionMaterialFunctionCall_42, output="FOV "}
   ]
  }
  MaterialExpressionMultiply_116 {
   class=MaterialExpressionMultiply
   inputs [
    A {from=MaterialExpressionMultiply_150}
    B {from=MaterialExpressionScalarParameter_79}
   ]
  }
  MaterialExpressionMultiply_117 {
   class=MaterialExpressionMultiply
   inputs [
    A {from=MaterialExpressionAdd_6}
    B {from=MaterialExpressionVectorParameter_52, output=RGB}
   ]
  }
  MaterialExpressionMultiply_118 {
   class=MaterialExpressionMultiply
   inputs [
    A {from=MaterialExpressionReroute_45}
    B {from=MaterialExpressionMultiply_105}
   ]
  }
  MaterialExpressionMultiply_119 {
   class=MaterialExpressionMultiply
   inputs [
    A {from=MaterialExpressionReroute_47}
    B {from=MaterialExpressionLinearInterpolate_48}
   ]
  }
  MaterialExpressionMultiply_120 {
   class=MaterialExpressionMultiply
   inputs [
    A {from=MaterialExpressionAdd_30}
    B {from=MaterialExpressionConstant_61}
   ]
  }
  MaterialExpressionMultiply_121 {
   class=MaterialExpressionMultiply
   inputs [
    A {from=MaterialExpressionMultiply_97}
    B {from=MaterialExpressionMultiply_149}
   ]
  }
  MaterialExpressionMultiply_122 {
   class=MaterialExpressionMultiply
   inputs [
    A {from=MaterialExpressionNamedRerouteUsage_112}
    B {from=MaterialExpressionVectorParameter_51, output=RGB}
   ]
  }
  MaterialExpressionMultiply_123 {
   class=MaterialExpressionMultiply
   inputs [
    A {from=MaterialExpressionNamedRerouteUsage_113}
    B {from=MaterialExpressionScalarParameter_69}
   ]
  }
  MaterialExpressionMultiply_124 {
   class=MaterialExpressionMultiply
   inputs [
    A {from=MaterialExpressionReroute_50}
    B {from=MaterialExpressionMultiply_117}
   ]
  }
  MaterialExpressionMultiply_125 {
   class=MaterialExpressionMultiply
   inputs [
    A {from=MaterialExpressionAdd_36}
    B {from=MaterialExpressionScalarParameter_93}
   ]
   properties {constB=2}
  }
  MaterialExpressionMultiply_126 {
   class=MaterialExpressionMultiply
   inputs [
    A {from=MaterialExpressionTime_6}
    B {from=MaterialExpressionScalarParameter_96}
   ]
  }
  MaterialExpressionMultiply_127 {
   class=MaterialExpressionMultiply
   inputs [
    A {from=MaterialExpressionNamedRerouteUsage_114}
    B {from=MaterialExpressionLinearInterpolate_67}
   ]
  }
  MaterialExpressionMultiply_128 {
   class=MaterialExpressionMultiply
   inputs [
    A {from=MaterialExpressionReroute_49}
    B {from=MaterialExpressionMultiply_127}
   ]
  }
  MaterialExpressionMultiply_129 {
   class=MaterialExpressionMultiply
   inputs [
    A {from=MaterialExpressionPower_29}
    B {from=MaterialExpressionLinearInterpolate_59}
   ]
  }
  MaterialExpressionMultiply_130 {
   class=MaterialExpressionMultiply
   inputs [
    A {from=MaterialExpressionComponentMask_20}
    B {from=""}
   ]
   properties {constB=2}
  }
  MaterialExpressionMultiply_131 {
   class=MaterialExpressionMultiply
   inputs [
    A {from=MaterialExpressionDotProduct_29}
    B {from=MaterialExpressionConstant_55}
   ]
  }
  MaterialExpressionMultiply_132 {
   class=MaterialExpressionMultiply
   inputs [
    A {from=MaterialExpressionMultiply_122}
    B {from=MaterialExpressionLinearInterpolate_63}
   ]
  }
  MaterialExpressionMultiply_133 {
   class=MaterialExpressionMultiply
   inputs [
    A {from=MaterialExpressionReroute_51}
    B {from=MaterialExpressionReroute_51}
   ]
  }
  MaterialExpressionMultiply_134 {
   class=MaterialExpressionMultiply
   inputs [
    A {from=MaterialExpressionSmoothStep_27}
    B {from=MaterialExpressionConstant_68}
   ]
  }
  MaterialExpressionMultiply_135 {
   class=MaterialExpressionMultiply
   inputs [
    A {from=MaterialExpressionSmoothStep_27}
    B {from=MaterialExpressionConstant_69}
   ]
  }
  MaterialExpressionMultiply_136 {
   class=MaterialExpressionMultiply
   inputs [
    A {from=MaterialExpressionNamedRerouteUsage_101}
    B {from=MaterialExpressionConstant_57}
   ]
  }
  MaterialExpressionMultiply_137 {
   class=MaterialExpressionMultiply
   inputs [
    A {from=MaterialExpressionDotProduct_30}
    B {from=MaterialExpressionConstant_65}
   ]
  }
  MaterialExpressionMultiply_138 {
   class=MaterialExpressionMultiply
   inputs [
    A {from=MaterialExpressionPower_31}
    B {from=MaterialExpressionConstant_63}
   ]
  }
  MaterialExpressionMultiply_139 {
   class=MaterialExpressionMultiply
   inputs [
    A {from=MaterialExpressionLinearInterpolate_57}
    B {from=MaterialExpressionLinearInterpolate_66}
   ]
  }
  MaterialExpressionMultiply_140 {
   class=MaterialExpressionMultiply
   inputs [
    A {from=MaterialExpressionNamedRerouteUsage_114}
    B {from=""}
   ]
   properties {constB=10}
  }
  MaterialExpressionMultiply_141 {
   class=MaterialExpressionMultiply
   inputs [
    A {from=MaterialExpressionNamedRerouteUsage_165}
    B {from=""}
   ]
   properties {constB=0.5}
  }
  MaterialExpressionMultiply_142 {
   class=MaterialExpressionMultiply
   inputs [
    A {from=MaterialExpressionDotProduct_28}
    B {from=""}
   ]
  }
  MaterialExpressionMultiply_143 {
   class=MaterialExpressionMultiply
   inputs [
    A {from=MaterialExpressionScalarParameter_93}
    B {from=MaterialExpressionAdd_41}
   ]
  }
  MaterialExpressionMultiply_144 {
   class=MaterialExpressionMultiply
   inputs [
    A {from=MaterialExpressionDivide_33}
    B {from=MaterialExpressionConstant2Vector_20}
   ]
  }
  MaterialExpressionMultiply_145 {
   class=MaterialExpressionMultiply
   inputs [
    A {from=MaterialExpressionReroute_57}
    B {from=MaterialExpressionLinearInterpolate_73}
   ]
  }
  MaterialExpressionMultiply_146 {
   class=MaterialExpressionMultiply
   inputs [
    A {from=MaterialExpressionNamedRerouteUsage_129}
    B {from=MaterialExpressionMultiply_148}
   ]
  }
  MaterialExpressionMultiply_147 {
   class=MaterialExpressionMultiply
   inputs [
    A {from=MaterialExpressionDotProduct_0}
    B {from=""}
   ]
   properties {constB=0.3333}
  }
  MaterialExpressionMultiply_148 {
   class=MaterialExpressionMultiply
   inputs [
    A {from=MaterialExpressionOneMinus_26}
    B {from=MaterialExpressionSmoothStep_29}
   ]
  }
  MaterialExpressionMultiply_149 {
   class=MaterialExpressionMultiply
   inputs [
    A {from=MaterialExpressionNamedRerouteUsage_110}
    B {from=""}
   ]
   properties {constB=0.7}
  }
  MaterialExpressionMultiply_150 {
   class=MaterialExpressionMultiply
   inputs [
    A {from=MaterialExpressionMaterialFunctionCall_43, output=Scale X}
    B {from=MaterialExpressionComponentMask_21}
   ]
  }
  MaterialExpressionMultiply_152 {
   class=MaterialExpressionMultiply
   inputs [
    A {from=MaterialExpressionLinearInterpolate_52}
    B {from=MaterialExpressionNamedRerouteUsage_155}
   ]
  }
  MaterialExpressionMultiply_153 {
   class=MaterialExpressionMultiply
   inputs [
    A {from=MaterialExpressionAdd_38}
    B {from=MaterialExpressionNamedRerouteUsage_156}
   ]
  }
  MaterialExpressionMultiply_161 {
   class=MaterialExpressionMultiply
   inputs [
    A {from=MaterialExpressionLinearInterpolate_72}
    B {from=MaterialExpressionScalarParameter_144}
   ]
  }
  MaterialExpressionMultiply_162 {
   class=MaterialExpressionMultiply
   inputs [
    A {from=MaterialExpressionLinearInterpolate_71}
    B {from=MaterialExpressionScalarParameter_145}
   ]
  }
  MaterialExpressionMultiply_163 {
   class=MaterialExpressionMultiply
   inputs [
    A {from=MaterialExpressionMultiply_162}
    B {from=MaterialExpressionTextureCoordinate_1}
   ]
  }
  MaterialExpressionMultiply_23 {
   class=MaterialExpressionMultiply
   inputs [
    A {from=MaterialExpressionNamedRerouteUsage_21}
    B {from=""}
   ]
  }
  MaterialExpressionMultiply_40 {
   class=MaterialExpressionMultiply
   inputs [
    A {from=MaterialExpressionNamedRerouteUsage_20}
    B {from=MaterialExpressionMultiply_45}
   ]
  }
  MaterialExpressionMultiply_42 {
   class=MaterialExpressionMultiply
   inputs [
    A {from=MaterialExpressionScalarParameter_3}
    B {from=MaterialExpressionMaterialFunctionCall_2, output=Result}
   ]
  }
  MaterialExpressionMultiply_45 {
   class=MaterialExpressionMultiply
   inputs [
    A {from=MaterialExpressionSaturate_7}
    B {from=MaterialExpressionMultiply_42}
   ]
  }
  MaterialExpressionMultiply_95 {
   class=MaterialExpressionMultiply
   inputs [
    A {from=MaterialExpressionLinearInterpolate_55}
    B {from=MaterialExpressionMaterialFunctionCall_34, output=Result}
   ]
  }
  MaterialExpressionMultiply_96 {
   class=MaterialExpressionMultiply
   inputs [
    A {from=MaterialExpressionNamedRerouteUsage_96}
    B {from=MaterialExpressionVectorParameter_38, output=RGB}
   ]
  }
  MaterialExpressionMultiply_97 {
   class=MaterialExpressionMultiply
   inputs [
    A {from=MaterialExpressionSubtract_31}
    B {from=MaterialExpressionScalarParameter_71}
   ]
  }
  MaterialExpressionMultiply_98 {
   class=MaterialExpressionMultiply
   inputs [
    A {from=MaterialExpressionSmoothStep_20}
    B {from=MaterialExpressionOneMinus_20}
   ]
  }
  MaterialExpressionMultiply_99 {
   class=MaterialExpressionMultiply
   inputs [
    A {from=MaterialExpressionDotProduct_24}
    B {from=""}
   ]
   properties {constB=2}
  }
  MaterialExpressionNamedRerouteDeclaration_1 {
   class=MaterialExpressionNamedRerouteDeclaration
   inputs [None {from=MaterialExpressionFunctionInput_2}]
   properties {name="Pre-Skinned Bounds Min (V3)", nodeColor=<rgba 1 1 1 1>}
  }
  MaterialExpressionNamedRerouteDeclaration_10 {
   class=MaterialExpressionNamedRerouteDeclaration
   inputs [None {from=MaterialExpressionDDY_1}]
   properties {name="Scaled UV DDY (V2)", nodeColor=<rgba 0.0547 1 0 1>}
  }
  MaterialExpressionNamedRerouteDeclaration_100 {
   class=MaterialExpressionNamedRerouteDeclaration
   inputs [None {from=MaterialExpressionReroute_31}]
   properties {name=IN_Normal, nodeColor=<rgba 1 0.6094 0 1>}
  }
  MaterialExpressionNamedRerouteDeclaration_101 {
   class=MaterialExpressionNamedRerouteDeclaration
   inputs [None {from=MaterialExpressionLinearInterpolate_16}]
   properties {name=IN_BaseColor, nodeColor=<rgba 1 0.1406 0 1>}
  }
  MaterialExpressionNamedRerouteDeclaration_102 {
   class=MaterialExpressionNamedRerouteDeclaration
   inputs [None {from=MaterialExpressionGetMaterialAttributes_0, output=MaterialAttributes}]
   properties {name=IN_MA, nodeColor=<rgba 1 0 0.1015 1>}
  }
  MaterialExpressionNamedRerouteDeclaration_103 {
   class=MaterialExpressionNamedRerouteDeclaration
   inputs [None {from=MaterialExpressionGetMaterialAttributes_0, output=Metallic}]
   properties {name=IN_Metallic, nodeColor=<rgba 0 0.8594 1 1>}
  }
  MaterialExpressionNamedRerouteDeclaration_104 {
   class=MaterialExpressionNamedRerouteDeclaration
   inputs [None {from=MaterialExpressionGetMaterialAttributes_0, output=Roughness}]
   properties {name=IN_Roughness, nodeColor=<rgba 1 0 0.3984 1>}
  }
  MaterialExpressionNamedRerouteDeclaration_105 {
   class=MaterialExpressionNamedRerouteDeclaration
   inputs [None {from=MaterialExpressionGetMaterialAttributes_0, output=EmissiveColor}]
   properties {name=IN_Emissive, nodeColor=<rgba 0 1 0.7422 1>}
  }
  MaterialExpressionNamedRerouteDeclaration_106 {
   class=MaterialExpressionNamedRerouteDeclaration
   inputs [None {from=MaterialExpressionFunctionInput_4}]
   properties {name=M_VariantMask, nodeColor=<rgba 0 1 0.8828 1>}
  }
  MaterialExpressionNamedRerouteDeclaration_107 {
   class=MaterialExpressionNamedRerouteDeclaration
   inputs [None {from=MaterialExpressionShaderStageSwitch_7}]
   properties {name=PreSkinnedNormal_VS-PS, nodeColor=<rgba 0.3587 0.1657 1 1>}
  }
  MaterialExpressionNamedRerouteDeclaration_2 {
   class=MaterialExpressionNamedRerouteDeclaration
   inputs [None {from=MaterialExpressionFunctionInput_1}]
   properties {name="Pre-Skinned Bounds Max (V3)", nodeColor=<rgba 1 1 1 1>}
  }
  MaterialExpressionNamedRerouteDeclaration_3 {
   class=MaterialExpressionNamedRerouteDeclaration
   inputs [None {from=MaterialExpressionSubtract_40}]
   properties {name="Pre-Skinned Bounds Extents (V3)", nodeColor=<rgba 1 1 1 1>}
  }
  MaterialExpressionNamedRerouteDeclaration_36 {
   class=MaterialExpressionNamedRerouteDeclaration
   inputs [None {from=MaterialExpressionLinearInterpolate_12}]
   properties {name=GemNormals, nodeColor=<rgba 0 0.6953 1 1>}
  }
  MaterialExpressionNamedRerouteDeclaration_4 {
   class=MaterialExpressionNamedRerouteDeclaration
   inputs [None {from=MaterialExpressionMultiply_141}]
   properties {name="Pre-Skinned Bounds Half Extents (V3)", nodeColor=<rgba 1 1 1 1>}
  }
  MaterialExpressionNamedRerouteDeclaration_5 {
   class=MaterialExpressionNamedRerouteDeclaration
   inputs [None {from=MaterialExpressionFunctionInput_6}]
   properties {name="Object Position WS (V3)", nodeColor=<rgba 1 0.9375 0 1>}
  }
  MaterialExpressionNamedRerouteDeclaration_6 {
   class=MaterialExpressionNamedRerouteDeclaration
   inputs [None {from=MaterialExpressionMultiply_147}]
   properties {name="Pre-Skinned Scale (S)", nodeColor=<rgba 1 0.5391 0 1>}
  }
  MaterialExpressionNamedRerouteDeclaration_67 {
   class=MaterialExpressionNamedRerouteDeclaration
   inputs [None {from=MaterialExpressionMultiply_161}]
   properties {name="Bump Offset Mult (S)", nodeColor=<rgba 0.1484 0 1 1>}
  }
  MaterialExpressionNamedRerouteDeclaration_7 {
   class=MaterialExpressionNamedRerouteDeclaration
   inputs [None {from=MaterialExpressionMultiply_0}]
   properties {name=Object Radius, nodeColor=<rgba 1 0.1406 0 1>}
  }
  MaterialExpressionNamedRerouteDeclaration_72 {
   class=MaterialExpressionNamedRerouteDeclaration
   inputs [None {from=MaterialExpressionMultiply_95}]
   properties {name=OUT_Emissive, nodeColor=<rgba 0 0.2734 1 1>}
  }
  MaterialExpressionNamedRerouteDeclaration_73 {
   class=MaterialExpressionNamedRerouteDeclaration
   inputs [None {from=MaterialExpressionOneMinus_16}]
   properties {name=M_Fresnel, nodeColor=<rgba 1 0.9844 0 1>}
  }
  MaterialExpressionNamedRerouteDeclaration_74 {
   class=MaterialExpressionNamedRerouteDeclaration
   inputs [None {from=MaterialExpressionComponentMask_11}]
   properties {name=M_ZGradient, nodeColor=<rgba 0.1719 1 0 1>}
  }
  MaterialExpressionNamedRerouteDeclaration_75 {
   class=MaterialExpressionNamedRerouteDeclaration
   inputs [None {from=MaterialExpressionLinearInterpolate_50}]
   properties {name=OUT_BaseColor, nodeColor=<rgba 0.5234 1 0 1>}
  }
  MaterialExpressionNamedRerouteDeclaration_78 {
   class=MaterialExpressionNamedRerouteDeclaration
   inputs [None {from=MaterialExpressionMultiply_98}]
   properties {name=M_OuterGlow_DoubleFresnel, nodeColor=<rgba 0 0.1797 1 1>}
  }
  MaterialExpressionNamedRerouteDeclaration_79 {
   class=MaterialExpressionNamedRerouteDeclaration
   inputs [None {from=MaterialExpressionMultiply_102}]
   properties {name=M_InnerScatter, nodeColor=<rgba 1 0 0.1406 1>}
  }
  MaterialExpressionNamedRerouteDeclaration_8 {
   class=MaterialExpressionNamedRerouteDeclaration
   inputs [None {from=MaterialExpressionFunctionInput_7}]
   properties {name="Pre-Skinned Scale (V3)", nodeColor=<rgba 1 0.5391 0 1>}
  }
  MaterialExpressionNamedRerouteDeclaration_80 {
   class=MaterialExpressionNamedRerouteDeclaration
   inputs [None {from=MaterialExpressionSaturate_18}]
   properties {name=M_SunMask, nodeColor=<rgba 0.8047 0 1 1>}
  }
  MaterialExpressionNamedRerouteDeclaration_81 {
   class=MaterialExpressionNamedRerouteDeclaration
   inputs [None {from=MaterialExpressionPower_24}]
   properties {name=M_Highlight, nodeColor=<rgba 1 0.1641 0 1>}
  }
  MaterialExpressionNamedRerouteDeclaration_82 {
   class=MaterialExpressionNamedRerouteDeclaration
   inputs [None {from=MaterialExpressionPower_22}]
   properties {name=M_OuterGlow_Fresnel, nodeColor=<rgba 0.9922 1 0 1>}
  }
  MaterialExpressionNamedRerouteDeclaration_83 {
   class=MaterialExpressionNamedRerouteDeclaration
   inputs [None {from=MaterialExpressionPower_26}]
   properties {name=M_InnerGlow, nodeColor=<rgba 0.1719 0 1 1>}
  }
  MaterialExpressionNamedRerouteDeclaration_84 {
   class=MaterialExpressionNamedRerouteDeclaration
   inputs [None {from=MaterialExpressionPower_25}]
   properties {name=M_Quack_InnerGlow, nodeColor=<rgba 1 0.2812 0 1>}
  }
  MaterialExpressionNamedRerouteDeclaration_85 {
   class=MaterialExpressionNamedRerouteDeclaration
   inputs [None {from=MaterialExpressionMax_23}]
   properties {name=M_TriplanarSeams, nodeColor=<rgba 0.4531 0 1 1>}
  }
  MaterialExpressionNamedRerouteDeclaration_86 {
   class=MaterialExpressionNamedRerouteDeclaration
   inputs [None {from=MaterialExpressionFunctionInput_0}]
   properties {name="Pre-Skinned Local Pos (V3)", nodeColor=<rgba 0 0.1797 1 1>}
  }
  MaterialExpressionNamedRerouteDeclaration_87 {
   class=MaterialExpressionNamedRerouteDeclaration
   inputs [None {from=MaterialExpressionSaturate_23}]
   properties {name=M_InnerNoise, nodeColor=<rgba 1 0.4688 0 1>}
  }
  MaterialExpressionNamedRerouteDeclaration_88 {
   class=MaterialExpressionNamedRerouteDeclaration
   inputs [None {from=MaterialExpressionLinearInterpolate_60}]
   properties {name=Cubemap_Noise, nodeColor=<rgba 1 0 0.2812 1>}
  }
  MaterialExpressionNamedRerouteDeclaration_89 {
   class=MaterialExpressionNamedRerouteDeclaration
   inputs [None {from=MaterialExpressionSaturate_22}]
   properties {name=M_Sheen, nodeColor=<rgba 0 0.4375 1 1>}
  }
  MaterialExpressionNamedRerouteDeclaration_9 {
   class=MaterialExpressionNamedRerouteDeclaration
   inputs [None {from=MaterialExpressionDDX_2}]
   properties {name="Scaled UV DDX (V2)", nodeColor=<rgba 0.0547 1 0 1>}
  }
  MaterialExpressionNamedRerouteDeclaration_90 {
   class=MaterialExpressionNamedRerouteDeclaration
   inputs [None {from=MaterialExpressionLinearInterpolate_62}]
   properties {name=M_BasecolorChannel, nodeColor=<rgba 0.4297 0 1 1>}
  }
  MaterialExpressionNamedRerouteDeclaration_91 {
   class=MaterialExpressionNamedRerouteDeclaration
   inputs [None {from=MaterialExpressionFrac_9}]
   properties {name=Time_Sheen, nodeColor=<rgba 1 0.375 0 1>}
  }
  MaterialExpressionNamedRerouteDeclaration_92 {
   class=MaterialExpressionNamedRerouteDeclaration
   inputs [None {from=MaterialExpressionSaturate_25}]
   properties {name=M_PixelDepht, nodeColor=<rgba 1 0 0.8672 1>}
  }
  MaterialExpressionNamedRerouteDeclaration_94 {
   class=MaterialExpressionNamedRerouteDeclaration
   inputs [None {from=MaterialExpressionFunctionInput_5}]
   properties {name=M_InnerRadialGlow, nodeColor=<rgba 1 0.9844 0 1>}
  }
  MaterialExpressionNamedRerouteDeclaration_95 {
   class=MaterialExpressionNamedRerouteDeclaration
   inputs [None {from=MaterialExpressionFunctionInput_12}]
   properties {name=M_SpriteFresnel, nodeColor=<rgba 1 0.9844 0 1>}
  }
  MaterialExpressionNamedRerouteDeclaration_96 {
   class=MaterialExpressionNamedRerouteDeclaration
   inputs [None {from=MaterialExpressionFunctionInput_11}]
   properties {name=M_Logo, nodeColor=<rgba 1 0.9844 0 1>}
  }
  MaterialExpressionNamedRerouteDeclaration_97 {
   class=MaterialExpressionNamedRerouteDeclaration
   inputs [None {from=MaterialExpressionFunctionInput_13}]
   properties {name=IN_WPO, nodeColor=<rgba 0 0.9766 1 1>}
  }
  MaterialExpressionNamedRerouteDeclaration_98 {
   class=MaterialExpressionNamedRerouteDeclaration
   inputs [None {from=MaterialExpressionFunctionInput_14}]
   properties {name=M_BlinkingLights, nodeColor=<rgba 1 0.9844 0 1>}
  }
  MaterialExpressionNamedRerouteDeclaration_99 {
   class=MaterialExpressionNamedRerouteDeclaration
   inputs [None {from=MaterialExpressionFunctionInput_15}]
   properties {name=M_DiffuseMask, nodeColor=<rgba 0.875 1 0 1>}
  }
  MaterialExpressionNamedRerouteUsage_0 {class=MaterialExpressionNamedRerouteUsage}
  MaterialExpressionNamedRerouteUsage_101 {class=MaterialExpressionNamedRerouteUsage}
  MaterialExpressionNamedRerouteUsage_102 {class=MaterialExpressionNamedRerouteUsage}
  MaterialExpressionNamedRerouteUsage_104 {class=MaterialExpressionNamedRerouteUsage}
  MaterialExpressionNamedRerouteUsage_106 {class=MaterialExpressionNamedRerouteUsage}
  MaterialExpressionNamedRerouteUsage_107 {class=MaterialExpressionNamedRerouteUsage}
  MaterialExpressionNamedRerouteUsage_108 {class=MaterialExpressionNamedRerouteUsage}
  MaterialExpressionNamedRerouteUsage_109 {class=MaterialExpressionNamedRerouteUsage}
  MaterialExpressionNamedRerouteUsage_11 {class=MaterialExpressionNamedRerouteUsage}
  MaterialExpressionNamedRerouteUsage_110 {class=MaterialExpressionNamedRerouteUsage}
  MaterialExpressionNamedRerouteUsage_111 {class=MaterialExpressionNamedRerouteUsage}
  MaterialExpressionNamedRerouteUsage_112 {class=MaterialExpressionNamedRerouteUsage}
  MaterialExpressionNamedRerouteUsage_113 {class=MaterialExpressionNamedRerouteUsage}
  MaterialExpressionNamedRerouteUsage_114 {class=MaterialExpressionNamedRerouteUsage}
  MaterialExpressionNamedRerouteUsage_115 {class=MaterialExpressionNamedRerouteUsage}
  MaterialExpressionNamedRerouteUsage_116 {class=MaterialExpressionNamedRerouteUsage}
  MaterialExpressionNamedRerouteUsage_117 {class=MaterialExpressionNamedRerouteUsage}
  MaterialExpressionNamedRerouteUsage_118 {class=MaterialExpressionNamedRerouteUsage}
  MaterialExpressionNamedRerouteUsage_120 {class=MaterialExpressionNamedRerouteUsage}
  MaterialExpressionNamedRerouteUsage_122 {class=MaterialExpressionNamedRerouteUsage}
  MaterialExpressionNamedRerouteUsage_123 {class=MaterialExpressionNamedRerouteUsage}
  MaterialExpressionNamedRerouteUsage_124 {class=MaterialExpressionNamedRerouteUsage}
  MaterialExpressionNamedRerouteUsage_125 {class=MaterialExpressionNamedRerouteUsage}
  MaterialExpressionNamedRerouteUsage_126 {class=MaterialExpressionNamedRerouteUsage}
  MaterialExpressionNamedRerouteUsage_127 {class=MaterialExpressionNamedRerouteUsage}
  MaterialExpressionNamedRerouteUsage_128 {class=MaterialExpressionNamedRerouteUsage}
  MaterialExpressionNamedRerouteUsage_129 {class=MaterialExpressionNamedRerouteUsage}
  MaterialExpressionNamedRerouteUsage_130 {class=MaterialExpressionNamedRerouteUsage}
  MaterialExpressionNamedRerouteUsage_131 {class=MaterialExpressionNamedRerouteUsage}
  MaterialExpressionNamedRerouteUsage_132 {class=MaterialExpressionNamedRerouteUsage}
  MaterialExpressionNamedRerouteUsage_133 {class=MaterialExpressionNamedRerouteUsage}
  MaterialExpressionNamedRerouteUsage_137 {class=MaterialExpressionNamedRerouteUsage}
  MaterialExpressionNamedRerouteUsage_140 {class=MaterialExpressionNamedRerouteUsage}
  MaterialExpressionNamedRerouteUsage_141 {class=MaterialExpressionNamedRerouteUsage}
  MaterialExpressionNamedRerouteUsage_143 {class=MaterialExpressionNamedRerouteUsage}
  MaterialExpressionNamedRerouteUsage_144 {class=MaterialExpressionNamedRerouteUsage}
  MaterialExpressionNamedRerouteUsage_145 {class=MaterialExpressionNamedRerouteUsage}
  MaterialExpressionNamedRerouteUsage_148 {class=MaterialExpressionNamedRerouteUsage}
  MaterialExpressionNamedRerouteUsage_149 {class=MaterialExpressionNamedRerouteUsage}
  MaterialExpressionNamedRerouteUsage_153 {class=MaterialExpressionNamedRerouteUsage}
  MaterialExpressionNamedRerouteUsage_154 {class=MaterialExpressionNamedRerouteUsage}
  MaterialExpressionNamedRerouteUsage_155 {class=MaterialExpressionNamedRerouteUsage}
  MaterialExpressionNamedRerouteUsage_156 {class=MaterialExpressionNamedRerouteUsage}
  MaterialExpressionNamedRerouteUsage_157 {class=MaterialExpressionNamedRerouteUsage}
  MaterialExpressionNamedRerouteUsage_158 {class=MaterialExpressionNamedRerouteUsage}
  MaterialExpressionNamedRerouteUsage_159 {class=MaterialExpressionNamedRerouteUsage}
  MaterialExpressionNamedRerouteUsage_160 {class=MaterialExpressionNamedRerouteUsage}
  MaterialExpressionNamedRerouteUsage_161 {class=MaterialExpressionNamedRerouteUsage}
  MaterialExpressionNamedRerouteUsage_162 {class=MaterialExpressionNamedRerouteUsage}
  MaterialExpressionNamedRerouteUsage_164 {class=MaterialExpressionNamedRerouteUsage}
  MaterialExpressionNamedRerouteUsage_165 {class=MaterialExpressionNamedRerouteUsage}
  MaterialExpressionNamedRerouteUsage_166 {class=MaterialExpressionNamedRerouteUsage}
  MaterialExpressionNamedRerouteUsage_171 {class=MaterialExpressionNamedRerouteUsage}
  MaterialExpressionNamedRerouteUsage_173 {class=MaterialExpressionNamedRerouteUsage}
  MaterialExpressionNamedRerouteUsage_175 {class=MaterialExpressionNamedRerouteUsage}
  MaterialExpressionNamedRerouteUsage_176 {class=MaterialExpressionNamedRerouteUsage}
  MaterialExpressionNamedRerouteUsage_183 {class=MaterialExpressionNamedRerouteUsage}
  MaterialExpressionNamedRerouteUsage_20 {class=MaterialExpressionNamedRerouteUsage}
  MaterialExpressionNamedRerouteUsage_21 {class=MaterialExpressionNamedRerouteUsage}
  MaterialExpressionNamedRerouteUsage_210 {class=MaterialExpressionNamedRerouteUsage}
  MaterialExpressionNamedRerouteUsage_37 {class=MaterialExpressionNamedRerouteUsage}
  MaterialExpressionNamedRerouteUsage_58 {class=MaterialExpressionNamedRerouteUsage}
  MaterialExpressionNamedRerouteUsage_90 {class=MaterialExpressionNamedRerouteUsage}
  MaterialExpressionNamedRerouteUsage_91 {class=MaterialExpressionNamedRerouteUsage}
  MaterialExpressionNamedRerouteUsage_93 {class=MaterialExpressionNamedRerouteUsage}
  MaterialExpressionNamedRerouteUsage_96 {class=MaterialExpressionNamedRerouteUsage}
  MaterialExpressionNamedRerouteUsage_97 {class=MaterialExpressionNamedRerouteUsage}
  MaterialExpressionNamedRerouteUsage_98 {class=MaterialExpressionNamedRerouteUsage}
  MaterialExpressionNamedRerouteUsage_99 {class=MaterialExpressionNamedRerouteUsage}
  MaterialExpressionNormalize_3 {class=MaterialExpressionNormalize, inputs [VectorInput {from=MaterialExpressionSubtract_34, output=""}]}
  MaterialExpressionNormalize_4 {class=MaterialExpressionNormalize, inputs [VectorInput {from=MaterialExpressionTransform_25, output=""}]}
  MaterialExpressionObjectPositionWS_4 {class=MaterialExpressionObjectPositionWS, inputs [], properties {}}
  MaterialExpressionOneMinus_16 {class=MaterialExpressionOneMinus, inputs [None {from=MaterialExpressionDotProduct_21}]}
  MaterialExpressionOneMinus_19 {class=MaterialExpressionOneMinus, inputs [None {from=MaterialExpressionMaterialFunctionCall_33, output=Result}]}
  MaterialExpressionOneMinus_20 {class=MaterialExpressionOneMinus, inputs [None {from=MaterialExpressionConstantBiasScale_6}]}
  MaterialExpressionOneMinus_21 {class=MaterialExpressionOneMinus, inputs [None {from=MaterialExpressionDotProduct_22}]}
  MaterialExpressionOneMinus_22 {class=MaterialExpressionOneMinus, inputs [None {from=MaterialExpressionSaturate_21}]}
  MaterialExpressionOneMinus_23 {class=MaterialExpressionOneMinus, inputs [None {from=MaterialExpressionScalarParameter_94}]}
  MaterialExpressionOneMinus_24 {class=MaterialExpressionOneMinus, inputs [None {from=MaterialExpressionScalarParameter_95}]}
  MaterialExpressionOneMinus_25 {class=MaterialExpressionOneMinus, inputs [None {from=MaterialExpressionComponentMask_18}]}
  MaterialExpressionOneMinus_26 {class=MaterialExpressionOneMinus, inputs [None {from=MaterialExpressionNamedRerouteUsage_133}]}
  MaterialExpressionPixelDepth_2 {class=MaterialExpressionPixelDepth, inputs [], properties {}}
  MaterialExpressionPixelNormalWS_1 {class=MaterialExpressionPixelNormalWS, inputs [], properties {}}
  MaterialExpressionPower_22 {
   class=MaterialExpressionPower
   inputs [
    Base {from=MaterialExpressionOneMinus_21}
    Exp {from=MaterialExpressionScalarParameter_72}
   ]
  }
  MaterialExpressionPower_23 {
   class=MaterialExpressionPower
   inputs [
    Base {from=MaterialExpressionMax_18}
    Exp {from=MaterialExpressionScalarParameter_73}
   ]
  }
  MaterialExpressionPower_24 {
   class=MaterialExpressionPower
   inputs [
    Base {from=MaterialExpressionSaturate_19}
    Exp {from=MaterialExpressionScalarParameter_78}
   ]
  }
  MaterialExpressionPower_25 {
   class=MaterialExpressionPower
   inputs [
    Base {from=MaterialExpressionOneMinus_22}
    Exp {from=""}
   ]
   properties {constExponent=6}
  }
  MaterialExpressionPower_26 {
   class=MaterialExpressionPower
   inputs [
    Base {from=MaterialExpressionSphereMask_2}
    Exp {from=MaterialExpressionScalarParameter_86}
   ]
  }
  MaterialExpressionPower_27 {
   class=MaterialExpressionPower
   inputs [
    Base {from=MaterialExpressionTextureSample_10, output=R}
    Exp {from=""}
   ]
   properties {constExponent=3}
  }
  MaterialExpressionPower_28 {
   class=MaterialExpressionPower
   inputs [
    Base {from=MaterialExpressionAbs_6}
    Exp {from=""}
   ]
   properties {constExponent=8}
  }
  MaterialExpressionPower_29 {
   class=MaterialExpressionPower
   inputs [
    Base {from=MaterialExpressionMax_24}
    Exp {from=""}
   ]
   properties {constExponent=3}
  }
  MaterialExpressionPower_30 {
   class=MaterialExpressionPower
   inputs [
    Base {from=MaterialExpressionReroute_52}
    Exp {from=""}
   ]
   properties {constExponent=8}
  }
  MaterialExpressionPower_31 {
   class=MaterialExpressionPower
   inputs [
    Base {from=MaterialExpressionNamedRerouteUsage_125}
    Exp {from=MaterialExpressionConstant_64}
   ]
  }
  MaterialExpressionPower_32 {
   class=MaterialExpressionPower
   inputs [
    Base {from=MaterialExpressionSaturate_26}
    Exp {from=""}
   ]
   properties {constExponent=5}
  }
  MaterialExpressionPreSkinnedLocalBounds_2 {class=MaterialExpressionPreSkinnedLocalBounds, inputs [], properties {}}
  MaterialExpressionPreSkinnedNormal_3 {class=MaterialExpressionPreSkinnedNormal, inputs [], properties {}}
  MaterialExpressionQualitySwitch_2 {
   class=MaterialExpressionQualitySwitch
   inputs [
    Default {from=MaterialExpressionSubtract_43, output=""}
    Low {from=MaterialExpressionReroute_59, output=""}
    High {from="", output=""}
    Medium {from=MaterialExpressionReroute_59, output=""}
    Epic {from="", output=""}
   ]
   properties {}
  }
  MaterialExpressionReflectionVectorWS_2 {
   class=MaterialExpressionReflectionVectorWS
   inputs [CustomWorldNormal {from=MaterialExpressionSubtract_32, output=""}]
   properties {bNormalizeCustomWorldNormal=true}
  }
  MaterialExpressionReroute_21 {class=MaterialExpressionReroute, inputs [None {from=MaterialExpressionGetMaterialAttributes_0, output=BaseColor}]}
  MaterialExpressionReroute_29 {class=MaterialExpressionReroute, inputs [None {from=MaterialExpressionGetMaterialAttributes_0, output=Normal}]}
  MaterialExpressionReroute_30 {class=MaterialExpressionReroute, inputs [None {from=MaterialExpressionMaterialFunctionCall_15, output=Result}]}
  MaterialExpressionReroute_31 {class=MaterialExpressionReroute, inputs [None {from=MaterialExpressionReroute_30}]}
  MaterialExpressionReroute_40 {class=MaterialExpressionReroute, inputs [None {from=MaterialExpressionNamedRerouteDeclaration_82}]}
  MaterialExpressionReroute_41 {class=MaterialExpressionReroute, inputs [None {from=""}]}
  MaterialExpressionReroute_42 {class=MaterialExpressionReroute, inputs [None {from=MaterialExpressionReroute_44}]}
  MaterialExpressionReroute_43 {class=MaterialExpressionReroute, inputs [None {from=MaterialExpressionLinearInterpolate_68}]}
  MaterialExpressionReroute_44 {class=MaterialExpressionReroute, inputs [None {from=MaterialExpressionLinearInterpolate_65}]}
  MaterialExpressionReroute_45 {class=MaterialExpressionReroute, inputs [None {from=MaterialExpressionAdd_34}]}
  MaterialExpressionReroute_47 {class=MaterialExpressionReroute, inputs [None {from=MaterialExpressionLinearInterpolate_54}]}
  MaterialExpressionReroute_48 {class=MaterialExpressionReroute, inputs [None {from=MaterialExpressionMaterialFunctionCall_46, output=StabilizedUVs}]}
  MaterialExpressionReroute_49 {class=MaterialExpressionReroute, inputs [None {from=MaterialExpressionReroute_43}]}
  MaterialExpressionReroute_50 {class=MaterialExpressionReroute, inputs [None {from=MaterialExpressionAdd_35}]}
  MaterialExpressionReroute_51 {class=MaterialExpressionReroute, inputs [None {from=MaterialExpressionNamedRerouteUsage_112}]}
  MaterialExpressionReroute_52 {class=MaterialExpressionReroute, inputs [None {from=MaterialExpressionTextureSample_9, output=R}]}
  MaterialExpressionReroute_53 {class=MaterialExpressionReroute, inputs [None {from=MaterialExpressionPower_25}]}
  MaterialExpressionReroute_54 {class=MaterialExpressionReroute, inputs [None {from=MaterialExpressionMax_25}]}
  MaterialExpressionReroute_55 {class=MaterialExpressionReroute, inputs [None {from=MaterialExpressionReroute_43}]}
  MaterialExpressionReroute_56 {class=MaterialExpressionReroute, inputs [None {from=MaterialExpressionReroute_55}]}
  MaterialExpressionReroute_57 {class=MaterialExpressionReroute, inputs [None {from=MaterialExpressionAdd_34}]}
  MaterialExpressionReroute_58 {class=MaterialExpressionReroute, inputs [None {from=MaterialExpressionReroute_52}]}
  MaterialExpressionReroute_59 {class=MaterialExpressionReroute, inputs [None {from=MaterialExpressionReroute_58}]}
  MaterialExpressionReroute_60 {class=MaterialExpressionReroute, inputs [None {from=MaterialExpressionPreSkinnedNormal_3}]}
  MaterialExpressionSaturate_18 {class=MaterialExpressionSaturate, inputs [None {from=MaterialExpressionDotProduct_23}]}
  MaterialExpressionSaturate_19 {class=MaterialExpressionSaturate, inputs [None {from=MaterialExpressionSubtract_35}]}
  MaterialExpressionSaturate_20 {class=MaterialExpressionSaturate, inputs [None {from=MaterialExpressionPower_28}]}
  MaterialExpressionSaturate_21 {class=MaterialExpressionSaturate, inputs [None {from=MaterialExpressionDistance_2}]}
  MaterialExpressionSaturate_22 {class=MaterialExpressionSaturate, inputs [None {from=MaterialExpressionMultiply_129}]}
  MaterialExpressionSaturate_23 {class=MaterialExpressionSaturate, inputs [None {from=MaterialExpressionLinearInterpolate_53}]}
  MaterialExpressionSaturate_24 {class=MaterialExpressionSaturate, inputs [None {from=MaterialExpressionDivide_30}]}
  MaterialExpressionSaturate_25 {class=MaterialExpressionSaturate, inputs [None {from=MaterialExpressionDivide_31}]}
  MaterialExpressionSaturate_26 {class=MaterialExpressionSaturate, inputs [None {from=MaterialExpressionMultiply_140}]}
  MaterialExpressionSaturate_7 {class=MaterialExpressionSaturate, inputs [None {from=MaterialExpressionNamedRerouteUsage_37}]}
  MaterialExpressionScalarParameter_100 {class=MaterialExpressionScalarParameter, properties {defaultValue=10, group=SpriteVariants, parameterName=SpriteVar_Q_InnerBubbles_Scale}}
  MaterialExpressionScalarParameter_101 {
   class=MaterialExpressionScalarParameter
   properties {defaultValue=-0.3, group=SpriteVariants, parameterName=SpriteVar_Q_InnerBubbles_BumpOffset}
  }
  MaterialExpressionScalarParameter_102 {class=MaterialExpressionScalarParameter, properties {defaultValue=2.3, group=SpriteVariants, parameterName=Spritevar_Q_Sheen_TimeFreq}}
  MaterialExpressionScalarParameter_103 {class=MaterialExpressionScalarParameter, properties {defaultValue=2.2, parameterName=Param_19}}
  MaterialExpressionScalarParameter_104 {class=MaterialExpressionScalarParameter, properties {defaultValue=16, parameterName=SpriteVar_Q_Noise_Speed}}
  MaterialExpressionScalarParameter_142 {
   class=MaterialExpressionScalarParameter
   properties {defaultValue=1, group=Sprite_Cosmetics | Triplanar, parameterName=Derivatives Use Pre-Skinned Scale, sliderMax=1, sortPriority=30}
  }
  MaterialExpressionScalarParameter_143 {
   class=MaterialExpressionScalarParameter
   properties {defaultValue=1, group=Sprite_Cosmetics | Cube, parameterName=Bump Offset Use Pre-Skinned Scale, sliderMax=1, sortPriority=10}
  }
  MaterialExpressionScalarParameter_144 {
   class=MaterialExpressionScalarParameter
   properties {defaultValue=1, group=Sprite_Cosmetics | Cube, parameterName=Bump Offset Mult, sortPriority=20}
  }
  MaterialExpressionScalarParameter_145 {
   class=MaterialExpressionScalarParameter
   properties {defaultValue=2, group=Sprite_Cosmetics | Triplanar, parameterName=Derivatives Mult, sortPriority=40}
  }
  MaterialExpressionScalarParameter_16 {class=MaterialExpressionScalarParameter, properties {defaultValue=1, parameterName=Quack Gem Normals Influence}}
  MaterialExpressionScalarParameter_24 {class=MaterialExpressionScalarParameter, properties {defaultValue=0.8, parameterName=Quack Gem_BCInfluence}}
  MaterialExpressionScalarParameter_26 {class=MaterialExpressionScalarParameter, properties {defaultValue=2, parameterName=Quack Gem_Normal_TriplanarFade}}
  MaterialExpressionScalarParameter_3 {class=MaterialExpressionScalarParameter, properties {parameterName=Elemental Tex Influence}}
  MaterialExpressionScalarParameter_4 {class=MaterialExpressionScalarParameter, properties {defaultValue=0.37, parameterName=Quack Gem_NormalTile}}
  MaterialExpressionScalarParameter_68 {class=MaterialExpressionScalarParameter, properties {defaultValue=0.4, group=SpriteVariants, parameterName=SpriteVar_Q_Metallic}}
  MaterialExpressionScalarParameter_69 {class=MaterialExpressionScalarParameter, properties {defaultValue=0.5, group=SpriteVariants, parameterName=SpriteVar_Q_Roughness}}
  MaterialExpressionScalarParameter_70 {class=MaterialExpressionScalarParameter, properties {defaultValue=1, group=SpriteVariants, parameterName=SpriteVar_Q_Specular}}
  MaterialExpressionScalarParameter_71 {class=MaterialExpressionScalarParameter, properties {defaultValue=0.2, parameterName=SpriteVar_Q_NRefraction}}
  MaterialExpressionScalarParameter_72 {
   class=MaterialExpressionScalarParameter
   properties {
    defaultValue=1.5
    desc=originally 3 before last minute request for adding a sined fresnel
    group=SpriteVariants
    parameterName=SpriteVar_Q_OutGlowFresnelPow
    sortPriority=0
   }
  }
  MaterialExpressionScalarParameter_73 {
   class=MaterialExpressionScalarParameter
   properties {defaultValue=2, group=Sprite_Base, parameterName=SpriteVar_Q_InnerScatteringPower, sortPriority=0}
  }
  MaterialExpressionScalarParameter_74 {
   class=MaterialExpressionScalarParameter
   properties {defaultValue=0.7, group=Sprite_Base, parameterName=SpriteVar_Q_InnerScatteringScale, sortPriority=0}
  }
  MaterialExpressionScalarParameter_77 {class=MaterialExpressionScalarParameter, properties {defaultValue=0.2, group=Sprite_Base, parameterName=SpriteVar_Q_Emissive_HighLight_Bias}}
  MaterialExpressionScalarParameter_78 {class=MaterialExpressionScalarParameter, properties {defaultValue=16, group=Sprite_Base, parameterName=SpriteVar_Q_Emissive_HighLight_Pow}}
  MaterialExpressionScalarParameter_79 {class=MaterialExpressionScalarParameter, properties {defaultValue=6, group=Sprite_Base, parameterName=Screen Space Scale, sortPriority=30}}
  MaterialExpressionScalarParameter_80 {class=MaterialExpressionScalarParameter, properties {defaultValue=10, group=Sprite_Base, parameterName=InnerGlow_Radius, sortPriority=0}}
  MaterialExpressionScalarParameter_81 {class=MaterialExpressionScalarParameter, properties {defaultValue=-0.5, group=Sprite_Base, parameterName=InnerGlow_Hardness, sortPriority=0}}
  MaterialExpressionScalarParameter_82 {class=MaterialExpressionScalarParameter, properties {defaultValue=0.1, group=Sprite_Base, parameterName=InnerGlow_PosY, sortPriority=0}}
  MaterialExpressionScalarParameter_83 {class=MaterialExpressionScalarParameter, properties {defaultValue=0.2, group=Sprite_Base, parameterName=InnerGlow_PosX, sortPriority=0}}
  MaterialExpressionScalarParameter_84 {class=MaterialExpressionScalarParameter, properties {defaultValue=1, group=Sprite_Base, parameterName=InnerGlow_ScaleY, sortPriority=0}}
  MaterialExpressionScalarParameter_85 {class=MaterialExpressionScalarParameter, properties {defaultValue=1, group=Sprite_Base, parameterName=InnerGlow_ScaleX, sortPriority=0}}
  MaterialExpressionScalarParameter_86 {class=MaterialExpressionScalarParameter, properties {defaultValue=1, group=Sprite_Base, parameterName=InnerGlow_Power, sortPriority=0}}
  MaterialExpressionScalarParameter_87 {class=MaterialExpressionScalarParameter, properties {defaultValue=0.5, group=SpriteVariants, parameterName=SpriteVar_Q_InnerGlow_ScaleX}}
  MaterialExpressionScalarParameter_88 {class=MaterialExpressionScalarParameter, properties {defaultValue=0.6, group=SpriteVariants, parameterName=SpriteVar_Q_InnerGlow_Height}}
  MaterialExpressionScalarParameter_89 {class=MaterialExpressionScalarParameter, properties {defaultValue=0.6, group=SpriteVariants, parameterName=SpriteVar_Q_InnerGlow_ScaleY}}
  MaterialExpressionScalarParameter_90 {
   class=MaterialExpressionScalarParameter
   properties {defaultValue=-0.15, group=SpriteVariants, parameterName=SpriteVar_Q_InnerDetails_BumpOffset2}
  }
  MaterialExpressionScalarParameter_91 {class=MaterialExpressionScalarParameter, properties {defaultValue=60, group=SpriteVariants, parameterName=SpriteVar_Q_InnerDetails_Scale}}
  MaterialExpressionScalarParameter_92 {
   class=MaterialExpressionScalarParameter
   properties {defaultValue=-0.02, group=SpriteVariants, parameterName=SpriteVar_Q_InnerDetails_BumpOffset1}
  }
  MaterialExpressionScalarParameter_93 {class=MaterialExpressionScalarParameter, properties {defaultValue=0.17, group=SpriteVariants, parameterName=Spritevar_Q_Sheen_Freq}}
  MaterialExpressionScalarParameter_94 {class=MaterialExpressionScalarParameter, properties {defaultValue=0.2, group=SpriteVariants, parameterName=Spritevar_Q_Sheen_Grad1}}
  MaterialExpressionScalarParameter_95 {class=MaterialExpressionScalarParameter, properties {defaultValue=0.05, group=SpriteVariants, parameterName=Spritevar_Q_Sheen_Grad2}}
  MaterialExpressionScalarParameter_96 {class=MaterialExpressionScalarParameter, properties {defaultValue=-2.3, group=SpriteVariants, parameterName=Spritevar_Q_Sheen_Speed}}
  MaterialExpressionScalarParameter_97 {class=MaterialExpressionScalarParameter, properties {defaultValue=0.5, group=SpriteVariants, parameterName=Spritevar_Q_Sheen_Angle}}
  MaterialExpressionScalarParameter_98 {class=MaterialExpressionScalarParameter, properties {defaultValue=0.5, group=SpriteVariants, parameterName=Spritevar_Q_Sheen_Cam/Fresnel}}
  MaterialExpressionScalarParameter_99 {class=MaterialExpressionScalarParameter, properties {defaultValue=1, group=SpriteVariants, parameterName=SpriteVar_Q_Roughness_Cracks}}
  MaterialExpressionScreenPosition_2 {class=MaterialExpressionScreenPosition, inputs [], properties {}}
  MaterialExpressionSetMaterialAttributes_4 {
   class=MaterialExpressionSetMaterialAttributes
   inputs [
    MaterialAttributes {from=MaterialExpressionNamedRerouteUsage_90, output=""}
    EmissiveColor {from=MaterialExpressionNamedRerouteUsage_91, output=""}
    Normal {from="", output=""}
    BaseColor {from=MaterialExpressionNamedRerouteUsage_93, output=""}
    Metallic {from=MaterialExpressionScalarParameter_68, output=""}
    Roughness {from=MaterialExpressionLinearInterpolate_61, output=""}
    Specular {from=MaterialExpressionScalarParameter_70, output=""}
    WorldPositionOffset {from="", output=""}
   ]
   properties {attributeSetTypes [
     #1
     #2
     #3
     #4
     #5
     #6
     #7
    ]}
  }
  MaterialExpressionShaderStageSwitch_4 {
   class=MaterialExpressionShaderStageSwitch
   inputs [
    PixelShader {from=MaterialExpressionVertexInterpolator_4, output=PS}
    VertexShader {from=MaterialExpressionLocalPosition_5, output=XYZ}
   ]
   properties {}
  }
  MaterialExpressionShaderStageSwitch_7 {
   class=MaterialExpressionShaderStageSwitch
   inputs [
    PixelShader {from=MaterialExpressionVertexInterpolator_7, output=PS}
    VertexShader {from=MaterialExpressionReroute_60, output=""}
   ]
   properties {}
  }
  MaterialExpressionSine_1 {class=MaterialExpressionSine, inputs [None {from=MaterialExpressionReroute_40}]}
  MaterialExpressionSmoothStep_20 {
   class=MaterialExpressionSmoothStep
   inputs [
    Min {from="", output=""}
    Max {from="", output=""}
    Value {from=MaterialExpressionReroute_40, output=""}
   ]
   properties {constMax=0.1}
  }
  MaterialExpressionSmoothStep_21 {
   class=MaterialExpressionSmoothStep
   inputs [
    Min {from="", output=""}
    Max {from="", output=""}
    Value {from=MaterialExpressionNamedRerouteUsage_104, output=""}
   ]
   properties {constMax=0.5}
  }
  MaterialExpressionSmoothStep_22 {
   class=MaterialExpressionSmoothStep
   inputs [
    Min {from=MaterialExpressionOneMinus_23, output=""}
    Max {from="", output=""}
    Value {from=MaterialExpressionReroute_54, output=""}
   ]
   properties {}
  }
  MaterialExpressionSmoothStep_23 {
   class=MaterialExpressionSmoothStep
   inputs [
    Min {from=MaterialExpressionOneMinus_24, output=""}
    Max {from="", output=""}
    Value {from=MaterialExpressionReroute_54, output=""}
   ]
   properties {constMin=1}
  }
  MaterialExpressionSmoothStep_24 {
   class=MaterialExpressionSmoothStep
   inputs [
    Min {from=MaterialExpressionConstant_67, output=""}
    Max {from=MaterialExpressionConstant_66, output=""}
    Value {from=MaterialExpressionNamedRerouteUsage_128, output=""}
   ]
   properties {constMax=0.1}
  }
  MaterialExpressionSmoothStep_25 {
   class=MaterialExpressionSmoothStep
   inputs [
    Min {from="", output=""}
    Max {from="", output=""}
    Value {from=MaterialExpressionAbs_8, output=""}
   ]
   properties {constMax=0.9, constMin=0.7}
  }
  MaterialExpressionSmoothStep_26 {
   class=MaterialExpressionSmoothStep
   inputs [
    Min {from="", output=""}
    Max {from="", output=""}
    Value {from=MaterialExpressionNamedRerouteUsage_118, output=""}
   ]
   properties {constMin=0.5}
  }
  MaterialExpressionSmoothStep_27 {
   class=MaterialExpressionSmoothStep
   inputs [
    Min {from="", output=""}
    Max {from="", output=""}
    Value {from=MaterialExpressionTextureSample_11, output=R}
   ]
   properties {constMin=0.5}
  }
  MaterialExpressionSmoothStep_28 {
   class=MaterialExpressionSmoothStep
   inputs [
    Min {from="", output=""}
    Max {from="", output=""}
    Value {from=MaterialExpressionNamedRerouteUsage_130, output=""}
   ]
   properties {constMax=0.8, constMin=0.3}
  }
  MaterialExpressionSmoothStep_29 {
   class=MaterialExpressionSmoothStep
   inputs [
    Min {from="", output=""}
    Max {from="", output=""}
    Value {from=MaterialExpressionNamedRerouteUsage_130, output=""}
   ]
   properties {constMax=0.3, constMin=0.8}
  }
  MaterialExpressionSphereMask_2 {
   class=MaterialExpressionSphereMask
   inputs [
    A {from=MaterialExpressionDivide_26, output=""}
    B {from=MaterialExpressionConvert_9, output=""}
    Radius {from=MaterialExpressionScalarParameter_80, output=""}
    Hardness {from=MaterialExpressionScalarParameter_81, output=""}
   ]
   properties {attenuationRadius=12, hardnessPercent=55}
  }
  MaterialExpressionStaticSwitchParameter_4 {
   class=MaterialExpressionStaticSwitchParameter
   inputs [
    True {from=MaterialExpressionSetMaterialAttributes_4, output=""}
    False {from=MaterialExpressionNamedRerouteUsage_90, output=""}
   ]
   properties {group=SpriteVariants, parameterName=Sprite_Variant_Quack}
  }
  MaterialExpressionStep_3 {
   class=MaterialExpressionStep
   inputs [
    Y {from="", output=""}
    X {from=MaterialExpressionDotProduct_28, output=""}
   ]
   properties {constY=0.9}
  }
  MaterialExpressionSubtract_31 {
   class=MaterialExpressionSubtract
   inputs [
    A {from=MaterialExpressionTransform_20}
    B {from=MaterialExpressionConstant_60}
   ]
  }
  MaterialExpressionSubtract_32 {
   class=MaterialExpressionSubtract
   inputs [
    A {from=MaterialExpressionCameraVectorWS_13}
    B {from=MaterialExpressionMultiply_121}
   ]
  }
  MaterialExpressionSubtract_33 {
   class=MaterialExpressionSubtract
   inputs [
    A {from=MaterialExpressionReroute_41}
    B {from=MaterialExpressionComponentMask_14}
   ]
  }
  MaterialExpressionSubtract_34 {
   class=MaterialExpressionSubtract
   inputs [
    A {from=MaterialExpressionMultiply_100}
    B {from=MaterialExpressionAtmosphericLightVector_5}
   ]
  }
  MaterialExpressionSubtract_35 {
   class=MaterialExpressionSubtract
   inputs [
    A {from=MaterialExpressionDotProduct_26}
    B {from=MaterialExpressionScalarParameter_77}
   ]
  }
  MaterialExpressionSubtract_36 {
   class=MaterialExpressionSubtract
   inputs [
    A {from=MaterialExpressionMaterialFunctionCall_50, output=StabilizedUVs}
    B {from=MaterialExpressionLinearInterpolate_51}
   ]
  }
  MaterialExpressionSubtract_37 {
   class=MaterialExpressionSubtract
   inputs [
    A {from=MaterialExpressionConstantBiasScale_8}
    B {from=MaterialExpressionMultiply_113}
   ]
  }
  MaterialExpressionSubtract_38 {
   class=MaterialExpressionSubtract
   inputs [
    A {from=MaterialExpressionConvert_12}
    B {from=MaterialExpressionMax_19}
   ]
  }
  MaterialExpressionSubtract_39 {
   class=MaterialExpressionSubtract
   inputs [
    A {from=MaterialExpressionConvert_12}
    B {from=MaterialExpressionMax_20}
   ]
  }
  MaterialExpressionSubtract_40 {
   class=MaterialExpressionSubtract
   inputs [
    A {from=MaterialExpressionNamedRerouteUsage_0}
    B {from=MaterialExpressionNamedRerouteUsage_11}
   ]
  }
  MaterialExpressionSubtract_41 {
   class=MaterialExpressionSubtract
   inputs [
    A {from=MaterialExpressionConvert_12}
    B {from=MaterialExpressionMax_21}
   ]
  }
  MaterialExpressionSubtract_42 {
   class=MaterialExpressionSubtract
   inputs [
    A {from=MaterialExpressionMultiply_130}
    B {from=""}
   ]
  }
  MaterialExpressionSubtract_43 {
   class=MaterialExpressionSubtract
   inputs [
    A {from=MaterialExpressionReroute_58}
    B {from=MaterialExpressionLinearInterpolate_64}
   ]
  }
  MaterialExpressionSubtract_44 {
   class=MaterialExpressionSubtract
   inputs [
    A {from=MaterialExpressionReroute_53}
    B {from=MaterialExpressionNamedRerouteUsage_106}
   ]
  }
  MaterialExpressionSubtract_45 {
   class=MaterialExpressionSubtract
   inputs [
    A {from=MaterialExpressionPixelDepth_2, output=R}
    B {from=MaterialExpressionConstant_72}
   ]
   properties {constB=1000}
  }
  MaterialExpressionSubtract_46 {
   class=MaterialExpressionSubtract
   inputs [
    A {from=MaterialExpressionNamedRerouteUsage_137}
    B {from=MaterialExpressionNamedRerouteUsage_149}
   ]
  }
  MaterialExpressionTextureCoordinate_1 {class=MaterialExpressionTextureCoordinate}
  MaterialExpressionTextureObjectParameter_0 {
   class=MaterialExpressionTextureObjectParameter
   inputs [
    UVs {from="", output=""}
    Apply View MipBias {from="", output=""}
   ]
   properties {
    group=SpriteVariants
    parameterName=Quack Gem Normals
    samplerType=SAMPLERTYPE_Normal
    texture=/BRCosmetics/Animation/Props/Gadgets_PixieParts/Textures/TA/T_QuackCrystal2_N
   }
  }
  MaterialExpressionTextureObject_6 {
   class=MaterialExpressionTextureObject
   inputs []
   properties {samplerType=SAMPLERTYPE_LinearGrayscale, texture=/Game/Global/Textures/Default/Noise/BentNoise_H}
  }
  MaterialExpressionTextureObject_7 {
   class=MaterialExpressionTextureObject
   inputs []
   properties {samplerType=SAMPLERTYPE_Grayscale, texture=/Game/Global/Textures/Default/Noise/LowResBlurredNoiseNormalized}
  }
  MaterialExpressionTextureObject_8 {class=MaterialExpressionTextureObject, inputs [], properties {texture=/Game/Global/Textures/Default/Noise/T_CellNoise_01}}
  MaterialExpressionTextureSample_0 {
   class=MaterialExpressionTextureSample
   inputs [
    UVs {from=MaterialExpressionAdd_43}
    Tex {from=MaterialExpressionTextureObject_7}
    "DDX(UVs)" {from=MaterialExpressionNamedRerouteUsage_153}
    "DDY(UVs)" {from=MaterialExpressionNamedRerouteUsage_154}
    Apply View MipBias {from=""}
   ]
   properties {mipValueMode=TMVM_Derivative, samplerSource=SSM_Wrap_WorldGroupSettings}
  }
  MaterialExpressionTextureSample_10 {
   class=MaterialExpressionTextureSample
   inputs [
    UVs {from=MaterialExpressionMaterialFunctionCall_45, output=UVs}
    Tex {from=MaterialExpressionTextureObject_6}
    "DDX(UVs)" {from=MaterialExpressionNamedRerouteUsage_159}
    "DDY(UVs)" {from=MaterialExpressionNamedRerouteUsage_160}
    Apply View MipBias {from=""}
   ]
   properties {mipValueMode=TMVM_Derivative}
  }
  MaterialExpressionTextureSample_11 {
   class=MaterialExpressionTextureSample
   inputs [
    UVs {from=MaterialExpressionMaterialFunctionCall_48, output=Add}
    Tex {from=MaterialExpressionTextureObject_8}
    "DDX(UVs)" {from=MaterialExpressionNamedRerouteUsage_161}
    "DDY(UVs)" {from=MaterialExpressionNamedRerouteUsage_162}
    Apply View MipBias {from=""}
   ]
   properties {mipValueMode=TMVM_Derivative}
  }
  MaterialExpressionTextureSample_5 {
   class=MaterialExpressionTextureSample
   inputs [
    UVs {from=MaterialExpressionMultiply_11}
    Tex {from=MaterialExpressionTextureObjectParameter_0}
    "DDX(UVs)" {from=MaterialExpressionMaterialFunctionCall_47, output=DDX}
    "DDY(UVs)" {from=MaterialExpressionMaterialFunctionCall_47, output=DDX}
    Apply View MipBias {from=""}
   ]
   properties {mipValueMode=TMVM_Derivative, samplerType=SAMPLERTYPE_Normal}
  }
  MaterialExpressionTextureSample_9 {
   class=MaterialExpressionTextureSample
   inputs [
    UVs {from=MaterialExpressionMaterialFunctionCall_44, output=Add}
    Tex {from=MaterialExpressionTextureObject_6}
    "DDX(UVs)" {from=MaterialExpressionNamedRerouteUsage_157}
    "DDY(UVs)" {from=MaterialExpressionNamedRerouteUsage_158}
    Apply View MipBias {from=""}
   ]
   properties {mipValueMode=TMVM_Derivative}
  }
  MaterialExpressionTime_6 {class=MaterialExpressionTime}
  MaterialExpressionTime_7 {class=MaterialExpressionTime}
  MaterialExpressionTime_8 {class=MaterialExpressionTime}
  MaterialExpressionTransformPosition_3 {
   class=MaterialExpressionTransformPosition
   inputs [None {from=MaterialExpressionAppendVector_11, output=""}]
   properties {transformType=TRANSFORMPOSSOURCE_World}
  }
  MaterialExpressionTransformPosition_4 {
   class=MaterialExpressionTransformPosition
   inputs [None {from=MaterialExpressionNamedRerouteUsage_141, output=""}]
   properties {transformSourceType=TRANSFORMPOSSOURCE_World, transformType=TRANSFORMPOSSOURCE_View}
  }
  MaterialExpressionTransform_16 {
   class=MaterialExpressionTransform
   inputs [None {from=MaterialExpressionVectorParameter_45, output=RGB}]
   properties {transformSourceType=TRANSFORMSOURCE_Camera}
  }
  MaterialExpressionTransform_17 {class=MaterialExpressionTransform, inputs [None {from=MaterialExpressionNamedRerouteUsage_98}]}
  MaterialExpressionTransform_18 {class=MaterialExpressionTransform, inputs [None {from=MaterialExpressionNamedRerouteUsage_102}]}
  MaterialExpressionTransform_19 {
   class=MaterialExpressionTransform
   inputs [None {from=MaterialExpressionMaterialFunctionCall_38, output=Result}]
   properties {transformSourceType=TRANSFORMSOURCE_World, transformType=TRANSFORM_Local}
  }
  MaterialExpressionTransform_20 {class=MaterialExpressionTransform, inputs [None {from=MaterialExpressionNamedRerouteUsage_109}]}
  MaterialExpressionTransform_22 {class=MaterialExpressionTransform, inputs [None {from=MaterialExpressionNamedRerouteUsage_115}]}
  MaterialExpressionTransform_23 {class=MaterialExpressionTransform, inputs [None {from=MaterialExpressionNamedRerouteUsage_123}]}
  MaterialExpressionTransform_25 {
   class=MaterialExpressionTransform
   inputs [None {from=MaterialExpressionVectorParameter_54, output=RGB}]
   properties {transformSourceType=TRANSFORMSOURCE_Camera}
  }
  MaterialExpressionVectorParameter_38 {
   class=MaterialExpressionVectorParameter
   properties {group=SpriteVariants, parameterName=Spritevar_Q_Basecolor_Tint, defaultValue=<rgba 0.3177 0.5313 1 0.05>}
  }
  MaterialExpressionVectorParameter_39 {
   class=MaterialExpressionVectorParameter
   properties {group=SpriteVariants, parameterName=SpriteVar_Q_Cubemap_Color1, defaultValue=<rgba 0.3884 0.3172 0.829 1>}
  }
  MaterialExpressionVectorParameter_40 {
   class=MaterialExpressionVectorParameter
   properties {group=SpriteVariants, parameterName=SpriteVar_Q_Cubemap_Color2, defaultValue=<rgba 0 4.0001 10 1>}
  }
  MaterialExpressionVectorParameter_41 {
   class=MaterialExpressionVectorParameter
   properties {group=SpriteVariants, parameterName=SpriteVar_Q_Emissive_TOD_Morning, defaultValue=<rgba 1.3 0.9161 0.4974 1>}
  }
  MaterialExpressionVectorParameter_42 {
   class=MaterialExpressionVectorParameter
   properties {group=SpriteVariants, parameterName=SpriteVar_Q_Emissive_TOD_Day, defaultValue=<rgba 1 1 1 1>}
  }
  MaterialExpressionVectorParameter_43 {
   class=MaterialExpressionVectorParameter
   properties {group=SpriteVariants, parameterName=SpriteVar_Q_Emissive_TOD_Evening, defaultValue=<rgba 1.2 0.9205 0.6157 1>}
  }
  MaterialExpressionVectorParameter_44 {
   class=MaterialExpressionVectorParameter
   properties {group=SpriteVariants, parameterName=SpriteVar_Q_Emissive_TOD_Night, defaultValue=<rgba 0.84 1.1127 1.4 1>}
  }
  MaterialExpressionVectorParameter_45 {
   class=MaterialExpressionVectorParameter
   properties {
    group=Sprite_Base
    parameterName=SpriteVar_Q_Emissive_HighlightLight_CameraVector
    channelNames {a="", b=depht, g=up, r=sides}
    defaultValue=<rgba 0.6 0 -1 1>
   }
  }
  MaterialExpressionVectorParameter_46 {
   class=MaterialExpressionVectorParameter
   properties {group=SpriteVariants, parameterName=SpriteVar_Q_Sprite_Emissive_RimLight_Color, defaultValue=<rgba 1.7217 0 4.5 1.2>}
  }
  MaterialExpressionVectorParameter_47 {
   class=MaterialExpressionVectorParameter
   properties {
    group=SpriteVariants
    parameterName=SpriteVar_Q_OutGlowFresnelColor1
    sortPriority=0
    channelNames {a="0= lerp 1=add", b="", g="", r=""}
    defaultValue=<rgba 10 4.1739 0 1>
   }
  }
  MaterialExpressionVectorParameter_48 {
   class=MaterialExpressionVectorParameter
   properties {
    group=SpriteVariants
    parameterName=SpriteVar_Q_OutGlowFresnelColor2
    sortPriority=0
    channelNames {a="0= lerp 1=add", b="", g="", r=""}
    defaultValue=<rgba 4.7 0.0725 5 1>
   }
  }
  MaterialExpressionVectorParameter_49 {
   class=MaterialExpressionVectorParameter
   properties {group=SpriteVariants, parameterName=SpriteVar_Q_Emissive_Inner Noise_Color2, sortPriority=0, defaultValue=<rgba 0.0574 0.0124 0.13 1>}
  }
  MaterialExpressionVectorParameter_50 {
   class=MaterialExpressionVectorParameter
   properties {group=SpriteVariants, parameterName=SpriteVar_Q_Emissive_Inner Noise_Color1, sortPriority=0, defaultValue=<rgba 0.0452 0.1131 0.4 1>}
  }
  MaterialExpressionVectorParameter_51 {
   class=MaterialExpressionVectorParameter
   properties {group=SpriteVariants, parameterName=Spritevar_Q_EmissiveCrackGemTint, defaultValue=<rgba 0 5.6 7 1>}
  }
  MaterialExpressionVectorParameter_52 {
   class=MaterialExpressionVectorParameter
   properties {group=SpriteVariants, parameterName=SpriteVar_Q_Emissive_InnerGlow_Color, sortPriority=0, defaultValue=<rgba 0.9043 3.7427 8 1>}
  }
  MaterialExpressionVectorParameter_53 {
   class=MaterialExpressionVectorParameter
   properties {group=SpriteVariants, parameterName=SpriteVar_Q_Sprite_Emissive_Sheen_Color2, defaultValue=<rgba 0.1565 1.5 0.2733 2>}
  }
  MaterialExpressionVectorParameter_54 {class=MaterialExpressionVectorParameter, properties {parameterName=SpriteVar_Q_Sheen_CamVectorDir, defaultValue=<rgba 0 0 1 0>}}
  MaterialExpressionVectorParameter_55 {
   class=MaterialExpressionVectorParameter
   properties {group=SpriteVariants, parameterName=Spritevar_Q_EmissiveCrackGem_DiffuseMask, defaultValue=<rgba 1.5 0 1 1>}
  }
  MaterialExpressionVectorParameter_56 {
   class=MaterialExpressionVectorParameter
   properties {group=SpriteVariants, parameterName=SpriteVar_Q_Sprite_Emissive_Sheen_Color1, defaultValue=<rgba 1.9221 0.2087 2 0.8>}
  }
  MaterialExpressionVertexInterpolator_4 {class=MaterialExpressionVertexInterpolator, inputs [VS {from=MaterialExpressionLocalPosition_5, output=XYZ}], properties {}}
  MaterialExpressionVertexInterpolator_7 {class=MaterialExpressionVertexInterpolator, inputs [VS {from=MaterialExpressionReroute_60, output=""}], properties {}}
  MaterialExpressionVertexNormalWS_10 {class=MaterialExpressionVertexNormalWS, inputs [], properties {}}
  MaterialExpressionVertexNormalWS_11 {class=MaterialExpressionVertexNormalWS, inputs [], properties {}}
  MaterialExpressionVertexNormalWS_8 {class=MaterialExpressionVertexNormalWS, inputs [], properties {}}
  MaterialExpressionVertexNormalWS_9 {class=MaterialExpressionVertexNormalWS, inputs [], properties {}}
  MaterialExpressionViewSize_1 {class=MaterialExpressionViewSize, inputs [], properties {}}
 }}
