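# Material function that layers the Sprite look variants onto incoming material attributes for the PixieParts gadget props.
# Golden, Candy, Galaxy, Gem, Holofoil, Cube and Quack sub-functions run in series, sharing masks, WPO and pre-skinned position, bounds and scale inputs.
# A diffuse mask comes from a parameter-weighted channel dot product with optional inversion.
# The default bounds center adds Min to Min, so it equals Min, not the midpoint.
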MaterialFunction /BRCosmetics/Animation/Props/Gadgets_PixieParts/Materials/ParentMaterials/MF_Sprite_Variants
graph {nodes [
  n1 = MaterialFunctionCall!(
   MaterialAttributes=MaterialFunctionCall!(
    MaterialAttributes=NamedRerouteUsage()
    Mask_InnerRadialGlow=NamedRerouteUsage()
    Mask_Logo=NamedRerouteUsage()
    Mask_SpriteFresnel=NamedRerouteUsage()
    Mask Variant=NamedRerouteUsage()
    Mask_BlinkingLights=NamedRerouteUsage()
    WPO=NamedRerouteUsage()
    Pre-Skinned Local Pos=NamedRerouteUsage()
    Pre-Skinned Bounds Min=NamedRerouteUsage()
    Pre-Skinned Bounds Max=NamedRerouteUsage()
    Pre-Skinned Scale=NamedRerouteUsage()
   ) {materialFunction=&/MF_Sprite_Variants_Golden}.SetMaterialAttributes
   Mask_InnerRadialGlow=NamedRerouteUsage()
   Mask_Logo=NamedRerouteUsage()
   Mask_SpriteFresnel=NamedRerouteUsage()
   Mask Variant=NamedRerouteUsage()
   Mask_BlinkingLights=NamedRerouteUsage()
   Mask_DiffuseChannelMask=NamedRerouteUsage()
   WPO=NamedRerouteUsage()
   Pre-Skinned Local Pos=NamedRerouteUsage()
   Pre-Skinned Bounds Min=NamedRerouteUsage()
   Pre-Skinned Bounds Max=NamedRerouteUsage()
   Pre-Skinned Scale=NamedRerouteUsage()
  ) {materialFunction=&/MF_Sprite_Variants_Candy}
  n2 = NamedRerouteUsage()
  n3 = NamedRerouteUsage()
  n4 = NamedRerouteUsage()
  n5 = NamedRerouteUsage()
  n6 = NamedRerouteUsage()
  n7 = NamedRerouteUsage()
  n8 = NamedRerouteUsage()
  n9 = NamedRerouteUsage()
  n10 = NamedRerouteUsage()
  n11 = NamedRerouteUsage()
  n12 = MaterialFunctionCall!(
   MaterialAttributes=MaterialFunctionCall!(
    MaterialAttributes=MaterialFunctionCall!(
     Material=@n1.SetMaterialAttributes
     Mask_InnerRadialGlow=@n2
     Mask_Logo=@n3
     Mask_SpriteFresnel=@n4
     Mask Variant=@n5
     Mask_BlinkingLights=@n6
     WPO=@n7
     Pre-Skinned Bounds Min=@n8
     Pre-Skinned Bounds Max=@n9
     Pre-Skinned Bounds Center=@n10
     Pre-Skinned Scale=@n11
    ) {materialFunction=&/MF_Sprite_Variants_Galaxy}.SetMaterialAttributes
    Mask_InnerRadialGlow=NamedRerouteUsage()
    Mask_Logo=NamedRerouteUsage()
    Mask_SpriteFresnel=NamedRerouteUsage()
    Mask Variant=NamedRerouteUsage()
    Mask_BlinkingLights=NamedRerouteUsage()
    Mask_DiffuseChannelMask=NamedRerouteUsage()
    WPO=NamedRerouteUsage()
    Pre-Skinned Local Pos=NamedRerouteUsage()
    Pre-Skinned Bounds Min=NamedRerouteUsage()
    Pre-Skinned Bounds Max=NamedRerouteUsage()
    Pre-Skinned Bounds Center=NamedRerouteUsage()
    Pre-Skinned Scale=NamedRerouteUsage()
   ) {materialFunction=&/MF_Sprite_Variants_Gem}.SetMaterialAttributes
   Mask_InnerRadialGlow=NamedRerouteUsage()
   Mask_Logo=NamedRerouteUsage()
   Mask_SpriteFresnel=NamedRerouteUsage()
   Mask Variant=NamedRerouteUsage()
   Mask_BlinkingLights=NamedRerouteUsage()
   Mask_DiffuseChannelMask=NamedRerouteUsage()
   WPO=NamedRerouteUsage()
   Pre-Skinned Local Pos=NamedRerouteUsage()
   Pre-Skinned Bounds Min=NamedRerouteUsage()
   Pre-Skinned Bounds Max=NamedRerouteUsage()
   Pre-Skinned Scale=NamedRerouteUsage()
   Object Position WS=NamedRerouteUsage()
  ) {materialFunction=&/MF_Sprite_Variants_Holofoil}
  n13 = NamedRerouteUsage()
  n14 = NamedRerouteUsage()
  n15 = NamedRerouteUsage()
  n16 = NamedRerouteUsage()
  n17 = NamedRerouteUsage()
  n18 = NamedRerouteUsage()
  n19 = NamedRerouteUsage()
  n20 = NamedRerouteUsage()
  n21 = NamedRerouteUsage()
  n22 = NamedRerouteUsage()
  n23 = NamedRerouteUsage()
  n24 = NamedRerouteUsage()
  FunctionOutput(None=MaterialFunctionCall!(
    MaterialAttributes=MaterialFunctionCall!(
     MaterialAttributes=@n12.SetMaterialAttributes
     Mask_InnerRadialGlow=@n13
     Mask_Logo=@n14
     Mask_SpriteFresnel=@n15
     Mask Variant=@n16
     Mask_BlinkingLights=@n17
     Mask_DiffuseChannelMask=@n18
     WPO=@n19
     Pre-Skinned Local Pos=@n20
     Pre-Skinned Bounds Min=@n21
     Pre-Skinned Bounds Max=@n22
     Pre-Skinned Scale=@n23
     Object Position WS=@n24
    ) {materialFunction=&/MF_Sprite_Variants_Cube}.SetMaterialAttributes
    Mask_InnerRadialGlow=NamedRerouteUsage()
    Mask_Logo=NamedRerouteUsage()
    Mask_SpriteFresnel=NamedRerouteUsage()
    Mask Variant=NamedRerouteUsage()
    Mask_BlinkingLights=NamedRerouteUsage()
    Mask_DiffuseChannelMask=NamedRerouteUsage()
    WPO=NamedRerouteUsage()
    Pre-Skinned Local Pos=NamedRerouteUsage()
    Pre-Skinned Bounds Min=NamedRerouteUsage()
    Pre-Skinned Bounds Max=NamedRerouteUsage()
    Pre-Skinned Scale=NamedRerouteUsage()
    Object Position WS=NamedRerouteUsage()
   ) {materialFunction=&/MF_Sprite_Variants_Quack}.SetMaterialAttributes)
  NamedRerouteDeclaration(None=FunctionInput() {inputName=Mask Variant, inputType=FunctionInput_Scalar, sortPriority=1}) {name=M_VariantMask, nodeColor=<rgba 0 1 0.8828 1>}
  n25 = GetMaterialAttributes(None=NamedRerouteDeclaration(None=FunctionInput() {inputType=FunctionInput_MaterialAttributes}) {name=MA_In, nodeColor=<rgba 1 0.5625 0 1>}) {attributeGetTypes [
    #1
    #2
    #3
    #4
    #5
    #6
   ]}
  NamedRerouteDeclaration(None=@n25.Roughness) {name=In_Roughness, nodeColor=<rgba 1 0 0.5625 1>}
  NamedRerouteDeclaration(None=@n25.Specular) {name=IN_Specular, nodeColor=<rgba 1 0.4453 0 1>}
  NamedRerouteDeclaration(None=@n25.EmissiveColor) {name=IN_Emissive, nodeColor=<rgba 0.9688 1 0 1>}
  n26 = VectorParameter() {
   group=SpriteVariants
   parameterName=SpriteVar_DiffuseChannelMask
   sortPriority=0
   channelNames {a="0=FlipMask 1=NormalMask", b=DiffuseB Influence, g=DiffuseG Influence, r=DiffuseR Influence}
   defaultValue=<rgba 0 1 1 1>
  }
  n27 = DotProduct(A=NamedRerouteUsage(), B=@n26.RGB)
  n28 = OneMinus(None=@n27)
  NamedRerouteDeclaration(None=Saturate(None=LinearInterpolate(A=@n28, B=@n27, Alpha=@n26."0=FlipMask 1=NormalMask"))) {name=M_DiffuseMask, nodeColor=<rgba 0 1 0.5781 1>}
  NamedRerouteDeclaration(None=FunctionInput(Preview=ObjectPositionWS()) {bUsePreviewValueAsDefault=true, inputName=Object Position WS, sortPriority=150}) {name="Object Position WS (V3)", nodeColor=<rgba 1 0.9375 0 1>}
  n29 = FunctionInput() {bUsePreviewValueAsDefault=true, inputName=Pre-Skinned Scale, sortPriority=140, previewValue=<1 1 1 1>}
  NamedRerouteDeclaration(None=@n29) {name="Pre-Skinned Scale (V3)", nodeColor=<rgba 1 0.5391 0 1>}
  NamedRerouteDeclaration(None=FunctionInput() {inputName=Mask_InnerRadialGlow, inputType=FunctionInput_Scalar, sortPriority=1}) {name=M_InnerRadialGlow, nodeColor=<rgba 1 0.9844 0 1>}
  n30 = Bounds() {type=MEILB_PreSkinnedLocal}
  NamedRerouteDeclaration(None=FunctionInput(Preview=@n30.Min) {bUsePreviewValueAsDefault=true, inputName=Pre-Skinned Bounds Min, sortPriority=110}) {name="Pre-Skinned Bounds Min (V3)", nodeColor=<rgba 1 1 1 1>}
  NamedRerouteDeclaration(None=FunctionInput() {inputName=Mask_SpriteFresnel, inputType=FunctionInput_Scalar, sortPriority=1}) {name=M_SpriteFresnel, nodeColor=<rgba 1 0.9844 0 1>}
  NamedRerouteDeclaration(None=@n25.Normal) {name=IN_Normals, nodeColor=<rgba 0 1 0.6016 1>}
  NamedRerouteDeclaration(None=FunctionInput(Preview=@n30.Max) {bUsePreviewValueAsDefault=true, inputName=Pre-Skinned Bounds Max, sortPriority=120}) {name="Pre-Skinned Bounds Max (V3)", nodeColor=<rgba 1 1 1 1>}
  NamedRerouteDeclaration(None=FunctionInput() {inputName=Mask_Logo, inputType=FunctionInput_Scalar, sortPriority=1}) {name=M_Logo, nodeColor=<rgba 1 0.9844 0 1>}
  n31 = Add(A=@n30.Min, B=@n30.Min)
  NamedRerouteDeclaration(None=FunctionInput(Preview=Multiply(A=@n31) {constB=0.5}) {bUsePreviewValueAsDefault=true, inputName=Pre-Skinned Bounds Center, sortPriority=130}) {name="Pre-Skinned Bounds Center (V3)", nodeColor=<rgba 1 0 0.2578 1>}
  NamedRerouteDeclaration(None=FunctionInput() {inputName=WPO, inputType=FunctionInput_Scalar, sortPriority=2}) {name=IN_WPO, nodeColor=<rgba 0 0.9766 1 1>}
  n32 = LocalPosition() {includedOffsets=ExcludeOffsets, localOrigin=InstancePreSkinning}
  n33 = VertexInterpolator(VS=@n32.XYZ)
  NamedRerouteDeclaration(None=FunctionInput(Preview=ShaderStageSwitch(PixelShader=@n33.PS, VertexShader=@n32.XYZ)) {bUsePreviewValueAsDefault=true, inputName=Pre-Skinned Local Pos, sortPriority=100}) {name="Pre-Skinned Local Pos (V3)", nodeColor=<rgba 0 0.1797 1 1>}
  NamedRerouteDeclaration(None=FunctionInput() {inputName=Mask_BlinkingLights, inputType=FunctionInput_Scalar, sortPriority=1}) {name=M_BlinkingLights, nodeColor=<rgba 1 0.9844 0 1>}
  NamedRerouteDeclaration(None=@n25.MaterialAttributes) {name=IN_MA, nodeColor=<rgba 1 0 0.1015 1>}
  NamedRerouteDeclaration(None=@n25.BaseColor) {name=IN_BaseColor, nodeColor=<rgba 0.0781 0 1 1>}
  NamedRerouteDeclaration(None=@n25.Metallic) {name=In_Metallic, nodeColor=<rgba 0.2422 1 0 1>}
 ]}
