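# Material function that builds the pre-FX look of the Sprite gadget prop, outputting material attributes and pre-skinned local position.
# Base inputs flow through jelly, glitch, blinking-star, hat, WPO-mask, dither and glasses functions into a variant stage.
# A flipbook face is blended over the result using eye, mouth and brow alphas, with eyes masked by pre-skinned local X.
# HasPoseChanged, set by AnimBP, drives temporal responsiveness on face pixels.
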MaterialFunction /BRCosmetics/Animation/Props/Gadgets_PixieParts/Materials/ParentMaterials/MF_Sprite_PreFX
graph {nodes [
  n1 = MaterialFunctionCall() {materialFunction=&/MF_Sprite_BaseInputs}
  n2 = MaterialFunctionCall!(PBR_BaseInputs=@n1.MA, FXMask=NamedRerouteDeclaration(None=@n1.FX Mask) {name=FX Mask, nodeColor=<rgba 1 0 0.8203 1>}) {materialFunction=&/MF_Sprite_DefaultJelly}
  n3 = MaterialFunctionCall!(
   MA=MaterialFunctionCall!(MA=@n2.MA) {materialFunction=&/MF_Sprite_Glitch}.Result
   Pre-Skinned Scale=@n2.Pre-Skinned Local Scale
  ) {materialFunction=&/MF_Sprite_BlinkingStarsEmissive}
  n4 = NamedRerouteUsage()
  n5 = MaterialFunctionCall!(MA=@n3.MA, FXMask=@n4) {materialFunction=&/MF_Sprite_PeanutSpinningHat}
  n6 = MaterialFunctionCall!(MA=MaterialFunctionCall!(MA=MaterialFunctionCall!(MA=@n5.MA, FXMask=@n4) {materialFunction=&/MF_Sprite_SoftWPOMask}.MA) {materialFunction=&/MF_Sprite_DitherFresnel}.MA) {materialFunction=&/MF_Sprite_DuckGlassesOpacity}
  n7 = NamedRerouteUsage()
  n8 = GetMaterialAttributes(None=MaterialFunctionCall!(
    In=@n6.MA
    Mask_InnerRadialGlow=NamedRerouteUsage()
    Mask_Logo=NamedRerouteUsage()
    Mask_SpriteFresnel=NamedRerouteUsage()
    Mask Variant=ComponentMask(None=@n7) {r=true}
    Mask_BlinkingLights=NamedRerouteUsage()
    WPO=NamedRerouteUsage()
    Pre-Skinned Local Pos=NamedRerouteUsage()
    Pre-Skinned Bounds Min=NamedRerouteUsage()
    Pre-Skinned Bounds Max=NamedRerouteUsage()
    Pre-Skinned Bounds Center=NamedRerouteUsage()
    Pre-Skinned Scale=NamedRerouteUsage()
    Object Position WS=NamedRerouteUsage()
   ) {materialFunction=&/MF_Sprite_Variants}.Result) {attributeGetTypes [#1, #2]}
  n9 = SetMaterialAttributes!(
   MaterialAttributes=@n8.MaterialAttributes
   EmissiveColor=Multiply(A=@n8.EmissiveColor, B=@n6."Glass Emissive multiplier (after variant)")
   BaseColor=Multiply(A=@n8.BaseColor, B=@n6."Glass Emissive multiplier (after variant)")
  ) {attributeSetTypes [#1, #2]}
  n10 = MaterialFunctionCall!(Mouth UV, Eye UV, Brow UV, Material=@n9) {materialFunction=/Game/Global/Materials/Character_Parent_2023/FlipbookFace_Parent_2023/MaterialFunctions/MF_FlipbookFace}
  n11 = SetMaterialAttributes!(
   MaterialAttributes=@n10.Material
   ShadingModel=ShadingModel()
   BaseColor=@n10.FaceColor
   EmissiveColor=Multiply(A=@n10.FaceColor, B=ScalarParameter() {defaultValue=0.45, group=Flipbook 01 Global, parameterName=FB_Eye Emissive Intensity})
   Metallic=ScalarParameter() {defaultValue=1, group=Flipbook 01 Global, parameterName=FB_Metallic}
   Specular=ScalarParameter() {group=Flipbook 01 Global, parameterName=FB_Specular}
   Roughness=ScalarParameter() {defaultValue=1, group=Flipbook 01 Global, parameterName=FB_Roughness}
   Normal=Constant3Vector() {constant=<rgba 0 0 1 0>}
  ) {attributeSetTypes [
    #3
    #2
    #1
    #4
    #5
    #6
    #7
   ]}
  n12 = LocalPosition() {localOrigin=InstancePreSkinning}
  n13 = NamedRerouteDeclaration(None=ShaderStageSwitch(PixelShader=VertexInterpolator(VS=@n12.XYZ).PS, VertexShader=@n12.XYZ)) {name=PreSkinnedPos_VS-PS, nodeColor=<rgba 0 0.4844 1 1>}
  n14 = Step(X=Abs(None=ComponentMask(None=@n13) {r=true})) {constX=0.1, constY=0.1}
  n15 = Max(A=Max(A=Multiply(A=@n10.EyesAlpha, B=@n14), B=@n10.MouthAlpha), B=@n10.BrowsAlpha)
  FunctionOutput(None=BlendMaterialAttributes(A=@n9, B=@n11, Alpha=@n15) {vertexAttributeBlendType=UseA})
  FunctionOutput(None=@n13) {outputName=PreskinnedLocalPos, sortPriority=10}
  NamedRerouteDeclaration(None=@n2.Mask_InnerRadialGlow) {name=Mask_InnerRadialGlow, nodeColor=<rgba 0 1 0.5781 1>}
  NamedRerouteDeclaration(None=@n2.Mask_Logo) {name=Mask_Logo, nodeColor=<rgba 0 1 0.5781 1>}
  NamedRerouteDeclaration(None=@n2.Pre-Skinned Local Scale) {name="Pre-Skinned Local Scale (V3)", nodeColor=<rgba 1 0.5391 0 1>}
  NamedRerouteDeclaration(None=@n2.Mask_SpriteFresnel) {name=Mask_SpriteFresnel, nodeColor=<rgba 0 1 0.5781 1>}
  NamedRerouteDeclaration(None=@n2.Sprite WPO) {name=Sprite WPO, nodeColor=<rgba 0 1 0.5781 1>}
  NamedRerouteDeclaration(None=@n3.Mask_BlinkingEmissive) {name=Mask_BlinkingEmissive, nodeColor=<rgba 1 0 0.8203 1>}
  NamedRerouteDeclaration(None=@n2.Pre-Skinned Local Bounds Center) {name="Pre-Skinned Bounds Center (V3)", nodeColor=<rgba 1 0 0.2578 1>}
  NamedRerouteDeclaration(None=@n2.Pre-Skinned Local Pos) {name="Pre-Skinned Local Pos (V3)", nodeColor=<rgba 0 0.1797 1 1>}
  NamedRerouteDeclaration(None=@n2.Pre-Skinned Local Bounds Min) {name="Pre-Skinned Local Bounds Min (V3)", nodeColor=<rgba 1 1 1 1>}
  NamedRerouteDeclaration(None=@n2.Pre-Skinned Local Bounds Max) {name="Pre-Skinned Local Bounds Max (V3)", nodeColor=<rgba 1 1 1 1>}
  NamedRerouteDeclaration(None=@n2.Object Position WS) {name="Object Position WS (V3)", nodeColor=<rgba 0 0.4 1 1>}
  n16 = ScalarParameter() {desc="Set by base head AnimBP if \"IsFlipbookFace\" is true", parameterName=HasPoseChanged}
  TemporalResponsivenessOutput!(None=LinearInterpolate(Alpha=Multiply(A=@n15, B=@n16)) {constA=0.4})
 ]}
