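# Parent material for skeletal-mesh sprites among the Pixie Parts gadget props.
# It chains sprite PreFX, character brightness balance, shared 2023 character PostFX and sprite flatten functions through material attributes.
# An Enable Debug static switch swaps in an override that blanks surface inputs and drives only emissive.
# Debug subgraphs build facing-angle gradients and color min, max or inside-box masks from pre-skinned local position against adjustable bounds.
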Material /BRCosmetics/Animation/Props/Gadgets_PixieParts/Materials/ParentMaterials/M_Sprite_Parent
graph {
 properties {
  bAutomaticallySetUsageInEditor=false
  bUseMaterialAttributes=true
  bUsedWithSkeletalMesh=true
  previewMesh=/BRCosmetics/Accessories/FORT_Backpacks/Backpack_M_MED_ColdTrophy/Sprites/M_MED_ColdTrophy_Air_Pack
  usedShadingModels=Default Lit
 }
 nodes [
  NamedRerouteDeclaration(None=Convert!() {convertOutputs [{type=Vector3}]}) {name="Black (V3)", nodeColor=<rgba 1 0 0.8906 1>}
  NamedRerouteDeclaration(None=Convert() {convertOutputs [{type=Vector3, defaultValue=<rgba 1 1 1 1>}]}) {name="White (V3)", nodeColor=<rgba 0 1 0.7891 1>}
  NamedRerouteDeclaration(None=Convert() {convertOutputs [{type=Vector3, defaultValue=<rgba 1 0 0 1>}]}) {name="Red (V3)", nodeColor=<rgba 1 0.4688 0 1>}
  n1 = MaterialFunctionCall() {materialFunction=&/MF_Sprite_PreFX}
  n2 = MaterialFunctionCall!(Material=@n1.Result) {materialFunction=/Game/Packages/Fortress_SharedMaterials/Base_Material_Functions/CharacterBrightnessandScale_Balance}
  n3 = Constant3Vector() {constant=<rgba 1 1 0 0>}
  n4 = Fresnel() {baseReflectFraction=0, exponent=1}
  n5 = GetMaterialAttributes(None=MaterialFunctionCall!(Material=MaterialFunctionCall(Material=@n2.Result, M=@n3, Pre Skinned Local Position=@n1.PreskinnedLocalPos, Distance Blend=@n2.Result, Fresnel=@n4) {materialFunction=&MF_CharacterParent_2023_PostFX}.Result) {materialFunction=&/MF_Sprite_Flatten}.Result) {attributeGetTypes [#1]}
  NamedRerouteDeclaration(None=Desaturation(None=@n5.Emissive Color, Fraction=Constant() {r=1})) {name="Emission In (V3)", nodeColor=<rgba 0.1484 1 0 1>}
  n6 = CameraVectorWS()
  n7 = Multiply(A=DotProduct(A=VertexNormalWS(), B=@n6)) {constB=2}
  n8 = Saturate(None=Multiply(A=@n7, B=@n7))
  n9 = TextureCoordinate()
  n10 = Saturate(None=LinearInterpolate(
    A=ScalarParameter() {defaultValue=-1.5, group=__DEBUG__, parameterName=V Min, sortPriority=200}
    B=ScalarParameter() {defaultValue=4, group=__DEBUG__, parameterName=V Max, sortPriority=210}
    Alpha=ComponentMask(None=@n9) {g=true}
   ) {constA=-1.5, constB=4})
  n11 = Constant3Vector() {constant=<rgba 0 0 1 0>}
  n12 = Abs(None=DotProduct(A=Transform(None=@n11) {transformSourceType=TRANSFORMSOURCE_Local}, B=@n6))
  NamedRerouteDeclaration(None=LinearInterpolate(A=LinearInterpolate(A=@n8, Alpha=@n10), Alpha=Power(Base=@n12) {constExponent=5})) {name="Debug Emission Out (V3)", nodeColor=<rgba 1 0 0.4688 1>}
  n13 = Constant()
  n14 = Constant() {r=1}
  n15 = SetMaterialAttributes!(
   MaterialAttributes
   Base Color=@n13
   Metallic=@n13
   Specular=@n13
   Roughness=@n13
   World Position Offset=@n13
   Opacity=@n14
   Opacity Mask=@n14
   Normal=Convert() {convertOutputs [{type=Vector3, defaultValue=<rgba 0 0 1 1>}]}
   Shading Model=ShadingModel()
   Emissive Color=NamedRerouteUsage()
  ) {attributeSetTypes [
    #2
    #3
    #4
    #5
    #6
    #7
    #8
    #9
    #10
    #1
   ]}
  StaticSwitchParameter(True=@n15, False=@n5.MaterialAttributes) {group=__DEBUG__, parameterName=Enable Debug, sortPriority=0}
  n16 = ScalarParameter() {
   defaultValue=2
   desc="\r\n0 : Sprite Min Bounds\r\n1 : Sprite Max Bounds\r\n2 : Sprite Mask\r\n"
   group=__DEBUG__
   parameterName=Sprite Bounds Min Max Mode
   sliderMax=2
   sortPriority=10
  }
  n17 = VectorParameter() {group=__DEBUG__, parameterName=Min Preskinned Pos, sortPriority=20, channelNames {a=n/a, b="", g="", r=""}, defaultValue=<rgba -1 -1 -1 0>}
  n18 = VertexInterpolator(VS=LocalPosition() {localOrigin=InstancePreSkinning}.XYZ)
  n19 = Step(Y=@n17.RGB, X=@n18.PS)
  MaterialExpressionLinearInterpolate_1 = LinearInterpolate(A=NamedRerouteUsage(), B=NamedRerouteUsage(), Alpha=OneMinus(None=@n19))
  n20 = VectorParameter() {group=__DEBUG__, parameterName=Max Preskinned Pos, sortPriority=30, channelNames {a=n/a, b="", g="", r=""}, defaultValue=<rgba 1 1 1 0>}
  n21 = Step(Y=@n20.RGB, X=@n18.PS)
  MaterialExpressionLinearInterpolate_0 = LinearInterpolate(A=NamedRerouteUsage(), B=NamedRerouteUsage(), Alpha=@n21)
  MaterialExpressionMultiply_2 = Multiply(A=@n19, B=OneMinus(None=@n21))
  n22 = Convert!(None=@MaterialExpressionMultiply_2) {
   nodeName=Break Float3
   convertInputs [{
     type=Vector3
     expressionInput {
      expression=&self:MaterialExpressionMultiply_2
      inputName=None
      mask=0
      maskA=0
      maskB=0
      maskG=0
      maskR=0
      outputIndex=0
     }
    }]
   convertOutputs [{type=Scalar}, {type=Scalar}, {type=Scalar}]
  }
  MaterialExpressionLinearInterpolate_2 = LinearInterpolate(A=NamedRerouteUsage(), B=NamedRerouteUsage(), Alpha=Min(A=@n22, B=Min(A=@n22, B=@n22)))
  Switch!(
   SwitchValue=Clamp(None=Round(None=@n16)) {maxDefault=2}
   Default
   min=@MaterialExpressionLinearInterpolate_1
   max=@MaterialExpressionLinearInterpolate_0
   mask=@MaterialExpressionLinearInterpolate_2
  ) {inputs [
    {
     inputName=min
     input {
      expression=&self:MaterialExpressionLinearInterpolate_1
      inputName=None
      mask=0
      maskA=0
      maskB=0
      maskG=0
      maskR=0
      outputIndex=0
     }
    }
    {
     inputName=max
     input {
      expression=&self:MaterialExpressionLinearInterpolate_0
      inputName=None
      mask=0
      maskA=0
      maskB=0
      maskG=0
      maskR=0
      outputIndex=0
     }
    }
    {
     inputName=mask
     input {
      expression=&self:MaterialExpressionLinearInterpolate_2
      inputName=None
      mask=0
      maskA=0
      maskB=0
      maskG=0
      maskR=0
      outputIndex=0
     }
    }
   ]}
 ]
}
